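# Level Sequence for animating the RubySnapper_M skin's menu animation on a spawnable FortPlayerMannequin, about 324 frames long.
# A skeletal animation track sits alongside biped and legacy face Control Rig tracks, and the Head component carries a face bone tweaker rig.
# The biped rig holds dense per-frame keys on both FK and IK leg controls, consistent with animation baked onto the rig for editing.
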LevelSequence /BRCosmetics/Animation/Game/MainPlayer/Menu/BR/CustomSkins/_s39/RubySnapper_M/RS_anim/RS_LS
bindings [
 Head {
  id=#1
  parent_id=#2
  possessed_class=SkeletalMeshComponentBudgeted
  tracks [{
    type=MovieSceneControlRigParameterTrack
    display_name=Fortnite_Face_BoneTweaker_CtrlRig
    sections [{type=MovieSceneControlRigParameterSection}]
   }]
 }
 FortPlayerMannequin {
  id=#2
  possessed_class=FortPlayerMannequin
  tracks [
   {type=MovieSceneSpawnTrack, sections [{type=MovieSceneSpawnSection, channels [None {type=MovieSceneScriptingBoolChannel, default=true}]}]}
   {
    type=MovieScene3DTransformTrack
    sections [{
      type=MovieScene3DTransformSection
      channels [
       Location.X {type=MovieSceneScriptingDoubleChannel, default=0}
       Location.Y {type=MovieSceneScriptingDoubleChannel, default=0}
       Location.Z {type=MovieSceneScriptingDoubleChannel, default=0}
       Rotation.X {type=MovieSceneScriptingDoubleChannel, default=0}
       Rotation.Y {type=MovieSceneScriptingDoubleChannel, default=0}
       Rotation.Z {type=MovieSceneScriptingDoubleChannel, default=0}
       Scale.X {type=MovieSceneScriptingDoubleChannel, default=1}
       Scale.Y {type=MovieSceneScriptingDoubleChannel, default=1}
       Scale.Z {type=MovieSceneScriptingDoubleChannel, default=1}
      ]
     }]
   }
   {type=MovieSceneSkeletalAnimationTrack, sections [{type=MovieSceneSkeletalAnimationSection, end_frame=324, start_frame=0}]}
   {
    type=MovieSceneControlRigParameterTrack
    display_name=Fortnite_Legacy_Face_CtrlRig
    sections [{
      type=MovieSceneControlRigParameterSection
      channels [
       Flipbook {type=MovieSceneScriptingBoolChannel, default=true, const [false,10]}
       Face_Ctrl_Master.Location.X {type=MovieSceneScriptingFloatChannel, default=-21.3302, const [-21.3302,325]}
       Face_Ctrl_Master.Location.Y {type=MovieSceneScriptingFloatChannel, default=0, const [0,325]}
       Face_Ctrl_Master.Location.Z {type=MovieSceneScriptingFloatChannel, default=0, const [0,325]}
       Face_Ctrl_Master.Rotation.X {type=MovieSceneScriptingFloatChannel, default=0, const [0,325]}
       Face_Ctrl_Master.Rotation.Y {type=MovieSceneScriptingFloatChannel, default=0, const [0,325]}
       Face_Ctrl_Master.Rotation.Z {type=MovieSceneScriptingFloatChannel, default=0, const [0,325]}
       Face_Ctrl_Master.Scale.X {type=MovieSceneScriptingFloatChannel, default=1, const [1,325]}
       Face_Ctrl_Master.Scale.Y {type=MovieSceneScriptingFloatChannel, default=1, const [1,325]}
       Face_Ctrl_Master.Scale.Z {type=MovieSceneScriptingFloatChannel, default=1, const [1,325]}
       gui_follow_head {type=MovieSceneScriptingFloatChannel, default=3.4028234663852886e+38, const [1,10]}
      ]
     }]
   }
   {
    type=MovieSceneControlRigParameterTrack
    display_name=Fortnite_Biped_CtrlRig
    sections [{
      type=MovieSceneControlRigParameterSection
      channels [
       leg_l_stretch_switch {type=MovieSceneScriptingBoolChannel, default=false, const [false,325]}
       leg_r_stretch_switch {type=MovieSceneScriptingBoolChannel, default=false, const [false,325]}
       arm_l_fk_local {type=MovieSceneScriptingBoolChannel, default=true, const [true,325]}
       Auto Interpolate_3 {type=MovieSceneScriptingBoolChannel, default=false, const [false,325]}
       Auto Interpolate {type=MovieSceneScriptingBoolChannel, default=false, const [false,325]}
       arm_l_stretch_switch {type=MovieSceneScriptingBoolChannel, default=false, const [false,325]}
       Keep Aligned {type=MovieSceneScriptingBoolChannel, default=false, const [false,325]}
       arm_r_fk_local {type=MovieSceneScriptingBoolChannel, default=true, const [true,325]}
       Auto Interpolate_4 {type=MovieSceneScriptingBoolChannel, default=false, const [false,325]}
       Auto Interpolate_2 {type=MovieSceneScriptingBoolChannel, default=false, const [false,325]}
       arm_r_stretch_switch {type=MovieSceneScriptingBoolChannel, default=false, const [false,325]}
       Keep Aligned_2 {type=MovieSceneScriptingBoolChannel, default=false, const [false,325]}
       head_local {type=MovieSceneScriptingBoolChannel, default=true, const [true,325]}
       Stretch {type=MovieSceneScriptingBoolChannel, default=false, const [false,325]}
       head_stretch {type=MovieSceneScriptingBoolChannel, default=false, const [false,325]}
       spine_stretch_switch {type=MovieSceneScriptingBoolChannel, default=false, const [false,325]}
       arm_l_fk_ik_switch {type=MovieSceneScriptingBoolChannel, default=false, const [false,325]}
       leg_l_fk_ik_switch {type=MovieSceneScriptingBoolChannel, default=true, const [true,325]}
       arm_r_fk_ik_switch {type=MovieSceneScriptingBoolChannel, default=false, const [false,325]}
       leg_r_fk_ik_switch {type=MovieSceneScriptingBoolChannel, default=true, const [true,325]}
       ShowBodyControls {type=MovieSceneScriptingBoolChannel, default=true, const [true,325]}
       spine_fk_ik_switch {type=MovieSceneScriptingBoolChannel, default=false, const [false,325]}
       neck_fk_ik_switch {type=MovieSceneScriptingBoolChannel, default=false, const [false,325]}
       space {type=MovieSceneScriptingByteChannel, default=1, const [1,325]}
       global_ctrl.Location.X {type=MovieSceneScriptingFloatChannel, default=0, const [0,325]}
       global_ctrl.Location.Y {type=MovieSceneScriptingFloatChannel, default=0, const [0,325]}
       global_ctrl.Location.Z {type=MovieSceneScriptingFloatChannel, default=0, const [0,325]}
       global_ctrl.Rotation.X {type=MovieSceneScriptingFloatChannel, default=0, const [0,325]}
       global_ctrl.Rotation.Y {type=MovieSceneScriptingFloatChannel, default=0, const [0,325]}
       global_ctrl.Rotation.Z {type=MovieSceneScriptingFloatChannel, default=0, const [0,325]}
       root_ctrl.Location.X {type=MovieSceneScriptingFloatChannel, default=0, const [0,325]}
       root_ctrl.Location.Y {type=MovieSceneScriptingFloatChannel, default=0, const [0,325]}
       root_ctrl.Location.Z {type=MovieSceneScriptingFloatChannel, default=0, const [0,325]}
       root_ctrl.Rotation.X {type=MovieSceneScriptingFloatChannel, default=0, const [0,325]}
       root_ctrl.Rotation.Y {type=MovieSceneScriptingFloatChannel, default=0, const [0,325]}
       root_ctrl.Rotation.Z {type=MovieSceneScriptingFloatChannel, default=0, const [0,325]}
       root_ctrl.Scale.X {type=MovieSceneScriptingFloatChannel, default=1, const [1,325]}
       root_ctrl.Scale.Y {type=MovieSceneScriptingFloatChannel, default=1, const [1,325]}
       root_ctrl.Scale.Z {type=MovieSceneScriptingFloatChannel, default=1, const [1,325]}
       body_offset_ctrl.Location.X {type=MovieSceneScriptingFloatChannel, default=0, const [0,325]}
       body_offset_ctrl.Location.Y {type=MovieSceneScriptingFloatChannel, default=0, const [0,325]}
       body_offset_ctrl.Location.Z {type=MovieSceneScriptingFloatChannel, default=0, const [0,325]}
       body_offset_ctrl.Rotation.X {type=MovieSceneScriptingFloatChannel, default=0, const [0,325]}
       body_offset_ctrl.Rotation.Y {type=MovieSceneScriptingFloatChannel, default=0, const [0,325]}
       body_offset_ctrl.Rotation.Z {type=MovieSceneScriptingFloatChannel, default=0, const [0,325]}
       body_offset_ctrl.Scale.X {type=MovieSceneScriptingFloatChannel, default=1, const [1,325]}
       body_offset_ctrl.Scale.Y {type=MovieSceneScriptingFloatChannel, default=1, const [1,325]}
       body_offset_ctrl.Scale.Z {type=MovieSceneScriptingFloatChannel, default=1, const [1,325]}
       thigh_l_fk_ctrl.Location.X {type=MovieSceneScriptingFloatChannel, default=0, thinned_from=325, keys [[0,0],[324,0]]}
       thigh_l_fk_ctrl.Location.Y {type=MovieSceneScriptingFloatChannel, default=0, thinned_from=325, keys [[0,0],[324,0]]}
       thigh_l_fk_ctrl.Location.Z {type=MovieSceneScriptingFloatChannel, default=0, thinned_from=325, keys [[0,0],[324,0]]}
       thigh_l_fk_ctrl.Rotation.X {
        type=MovieSceneScriptingFloatChannel
        default=1.216
        thinned_from=325
        keys [[0,42.9532],[29,43.0731],[33,43.1787],[34,43.0877],[35,42.8442],[36,42.1817],[37,41.4251],[38,40.9688],[39,40.7115],[44,40.2533],[84,40.3442],[87,40.2248],[88,40.2769],[89,40.4399],[90,40.7365],[91,41.2558],[92,41.9082],[93,42.4484],[94,42.8034],[95,43.067],[97,43.3384],[187,42.9441],[188,43.0346],[189,43.2708],[190,43.7594],[191,44.3963],[192,44.9377],[193,45.3024],[194,45.5703],[199,46.0507],[243,46.0225],[245,45.8417],[246,45.5956],[247,45.1494],[248,44.257],[249,43.5436],[252,43.1048],[324,42.954]]
       }
       thigh_l_fk_ctrl.Rotation.Y {
        type=MovieSceneScriptingFloatChannel
        default=-1.2128
        thinned_from=325
        keys [[0,-2.8599],[28,-2.888],[30,-2.9014],[31,-2.9135],[32,-2.9417],[35,-3.0673],[38,-3.1432],[39,-3.1612],[42,-3.1775],[83,-3.1536],[85,-3.1233],[86,-3.0923],[87,-3.0495],[91,-2.84],[94,-2.7124],[96,-2.6448],[97,-2.6179],[98,-2.5974],[99,-2.5858],[101,-2.5838],[108,-2.6308],[177,-2.6041],[182,-2.6295],[183,-2.6284],[184,-2.6108],[185,-2.5786],[186,-2.5339],[187,-2.471],[189,-2.3152],[190,-2.2554],[191,-2.2035],[192,-2.147],[193,-2.0971],[194,-2.0699],[195,-2.0642],[196,-2.0674],[202,-2.1102],[243,-2.0852],[245,-2.0811],[246,-2.0859],[248,-2.1103],[249,-2.1585],[252,-2.4053],[255,-2.6237],[256,-2.668],[259,-2.7389],[261,-2.7732],[263,-2.795],[277,-2.8412],[282,-2.8462],[283,-2.8527],[324,-2.8606]]
       }
       thigh_l_fk_ctrl.Rotation.Z {
        type=MovieSceneScriptingFloatChannel
        default=-0.1664
        thinned_from=325
        keys [[0,3.9491],[1,3.9491],[3,3.9122],[5,3.8536],[8,3.7314],[14,3.4454],[16,3.3625],[22,3.1528],[24,3.0972],[25,3.0774],[26,3.0649],[29,3.0685],[33,3.1382],[34,3.1491],[35,3.1386],[36,3.0637],[37,2.9753],[38,2.934],[39,2.9196],[42,2.9331],[44,2.9596],[45,2.9842],[47,3.0535],[53,3.3018],[55,3.3669],[64,3.5312],[67,3.5719],[68,3.5719],[69,3.5953],[70,3.5953],[71,3.6162],[72,3.6162],[73,3.6343],[74,3.6343],[75,3.6495],[77,3.6495],[78,3.6667],[84,3.671],[85,3.6514],[86,3.618],[88,3.5353],[89,3.5038],[90,3.4878],[91,3.5037],[92,3.5395],[93,3.5576],[94,3.5475],[95,3.5247],[96,3.4903],[97,3.4463],[100,3.2781],[102,3.192],[105,3.0866],[107,3.0278],[109,2.9853],[110,2.9853],[111,2.9692],[115,2.9692],[116,2.9859],[117,2.9859],[118,3.0017],[119,3.0017],[120,3.0227],[121,3.0227],[122,3.049],[127,3.1373],[129,3.1907],[137,3.4961],[139,3.5551],[150,3.7325],[152,3.7325],[153,3.7683],[154,3.7683],[155,3.7892],[156,3.7892],[157,3.8073],[158,3.8073],[159,3.8227],[161,3.8227],[162,3.8397],[165,3.8397],[166,3.8521],[169,3.8521],[171,3.8115],[173,3.7451],[176,3.6092],[181,3.3606],[184,3.2368],[186,3.1675],[187,3.1398],[188,3.1256],[189,3.1384],[190,3.1965],[191,3.2831],[192,3.3583],[193,3.4098],[194,3.4509],[197,3.5357],[201,3.6056],[209,3.7217],[212,3.7873],[214,3.8567],[220,4.1137],[222,4.1788],[233,4.3587],[234,4.3587],[235,4.3832],[236,4.3832],[237,4.4054],[238,4.4054],[239,4.4248],[240,4.4248],[241,4.4416],[243,4.4416],[244,4.453],[245,4.4393],[246,4.4062],[247,4.3418],[248,4.2069],[249,4.0989],[252,4.0173],[257,3.8394],[267,3.3835],[274,3.1368],[276,3.0825],[277,3.0632],[278,3.0632],[279,3.0468],[282,3.0468],[283,3.0589],[284,3.0589],[285,3.0729],[286,3.0729],[287,3.0921],[288,3.0921],[289,3.1167],[294,3.201],[296,3.2456],[299,3.3416],[306,3.6073],[315,3.8369],[319,3.9104],[320,3.9237],[321,3.9237],[322,3.9424],[324,3.9424]]
       }
       thigh_l_fk_ctrl.Scale.X {type=MovieSceneScriptingFloatChannel, default=1, const [1,325]}
       thigh_l_fk_ctrl.Scale.Y {type=MovieSceneScriptingFloatChannel, default=1, const [1,325]}
       thigh_l_fk_ctrl.Scale.Z {type=MovieSceneScriptingFloatChannel, default=1, const [1,325]}
       calf_l_fk_ctrl.Location.X {
        type=MovieSceneScriptingFloatChannel
        default=0
        thinned_from=325
        keys [[0,-0.6326],[2,-0.6339],[3,-0.6363],[4,-0.6404],[5,-0.6468],[6,-0.6559],[7,-0.6682],[8,-0.6843],[9,-0.7054],[10,-0.7318],[11,-0.762],[14,-0.8602],[15,-0.8905],[16,-0.9171],[17,-0.9385],[19,-0.9718],[21,-0.9981],[23,-1.0173],[25,-1.029],[27,-1.0329],[31,-1.0313],[35,-1.0203],[40,-0.9944],[43,-0.9712],[44,-0.9597],[45,-0.9439],[46,-0.9247],[47,-0.9026],[52,-0.7768],[53,-0.7536],[54,-0.7328],[55,-0.7151],[56,-0.7012],[58,-0.6799],[60,-0.6625],[62,-0.6487],[64,-0.6381],[66,-0.6303],[69,-0.6232],[75,-0.6203],[78,-0.6228],[80,-0.6277],[82,-0.6363],[83,-0.6424],[84,-0.65],[85,-0.6596],[86,-0.6716],[87,-0.6857],[88,-0.7017],[90,-0.738],[97,-0.8807],[99,-0.917],[105,-1.0068],[107,-1.0325],[108,-1.0428],[109,-1.0508],[110,-1.0559],[111,-1.0577],[113,-1.0554],[116,-1.0442],[120,-1.0179],[127,-0.9533],[129,-0.9298],[132,-0.8854],[137,-0.806],[139,-0.7791],[143,-0.7381],[147,-0.7028],[150,-0.6802],[154,-0.6558],[157,-0.6422],[160,-0.6336],[164,-0.6289],[168,-0.6283],[169,-0.6294],[170,-0.6326],[171,-0.6378],[172,-0.6447],[173,-0.6531],[175,-0.6738],[183,-0.7778],[191,-0.8651],[194,-0.888],[199,-0.9073],[205,-0.9178],[211,-0.9189],[213,-0.9151],[215,-0.9048],[217,-0.8893],[220,-0.8588],[224,-0.8105],[226,-0.7811],[234,-0.6446],[236,-0.6134],[237,-0.5992],[238,-0.5863],[239,-0.5748],[240,-0.565],[241,-0.5568],[242,-0.5499],[244,-0.5397],[245,-0.5362],[246,-0.534],[247,-0.5331],[248,-0.5334],[249,-0.5362],[250,-0.5417],[252,-0.5561],[254,-0.5757],[255,-0.5844],[256,-0.5912],[260,-0.6139],[262,-0.6275],[264,-0.6438],[266,-0.6637],[268,-0.6878],[269,-0.7016],[270,-0.7207],[271,-0.7476],[272,-0.7801],[275,-0.8908],[276,-0.925],[277,-0.9543],[278,-0.9766],[279,-0.9899],[282,-1.0096],[286,-1.0246],[295,-1.031],[298,-1.0266],[301,-1.0137],[303,-1.0001],[305,-0.982],[307,-0.9594],[308,-0.9463],[309,-0.9307],[310,-0.9118],[311,-0.8901],[313,-0.841],[317,-0.7365],[318,-0.7125],[319,-0.6906],[320,-0.6713],[321,-0.6553],[322,-0.6431],[323,-0.6353],[324,-0.6326]]
       }
       calf_l_fk_ctrl.Location.Y {
        type=MovieSceneScriptingFloatChannel
        default=0
        thinned_from=325
        keys [[0,0.0509],[3,0.0512],[5,0.0521],[6,0.0528],[7,0.0538],[8,0.0551],[9,0.0568],[10,0.0589],[12,0.064],[14,0.0693],[15,0.0717],[16,0.0738],[17,0.0756],[19,0.0782],[21,0.0804],[23,0.0819],[25,0.0829],[27,0.0832],[32,0.0829],[37,0.0814],[41,0.0795],[43,0.0782],[44,0.0773],[45,0.076],[46,0.0745],[48,0.0707],[52,0.0625],[54,0.059],[55,0.0576],[56,0.0565],[59,0.054],[62,0.0522],[66,0.0507],[72,0.0499],[78,0.0501],[81,0.0508],[83,0.0517],[85,0.0531],[87,0.0552],[89,0.0579],[97,0.0709],[99,0.0738],[106,0.0822],[108,0.084],[109,0.0846],[110,0.085],[112,0.0851],[115,0.0845],[119,0.0826],[127,0.0768],[129,0.0749],[137,0.0649],[139,0.0627],[146,0.0572],[150,0.0548],[151,0.0482],[152,0.0477],[153,0.0532],[158,0.0514],[168,0.0506],[170,0.0509],[172,0.0519],[175,0.0542],[183,0.0626],[191,0.0697],[194,0.0715],[201,0.0735],[211,0.074],[213,0.0737],[215,0.0729],[218,0.0708],[223,0.0663],[226,0.0629],[235,0.0506],[238,0.0472],[240,0.0455],[242,0.0443],[243,0.0383],[244,0.0434],[246,0.043],[248,0.0429],[249,0.0432],[252,0.0448],[255,0.047],[262,0.0505],[265,0.0526],[268,0.0554],[269,0.0565],[270,0.058],[271,0.0602],[272,0.0628],[275,0.0717],[276,0.0745],[277,0.0768],[278,0.0786],[279,0.0797],[282,0.0813],[286,0.0825],[296,0.083],[299,0.0824],[302,0.0811],[305,0.0791],[308,0.0762],[309,0.0749],[310,0.0734],[312,0.0698],[317,0.0593],[319,0.0556],[320,0.054],[321,0.0528],[322,0.0518],[323,0.0511],[324,0.0509]]
       }
       calf_l_fk_ctrl.Location.Z {
        type=MovieSceneScriptingFloatChannel
        default=0
        thinned_from=325
        keys [[0,0.0004],[150,0.0004],[151,-0.0058],[152,-0.0059],[153,0.0004],[242,0.0003],[243,-0.0066],[244,0.0003],[324,0.0004]]
       }
       calf_l_fk_ctrl.Rotation.X {
        type=MovieSceneScriptingFloatChannel
        default=0.0543
        thinned_from=325
        keys [[0,0.0033],[150,0.0033],[151,0.0025],[152,0.0025],[153,0.0033],[242,0.0033],[243,-0.0053],[244,0.0033],[324,0.0033]]
       }
       calf_l_fk_ctrl.Rotation.Y {
        type=MovieSceneScriptingFloatChannel
        default=1.2238
        thinned_from=325
        keys [[0,0.0047],[150,0.0047],[151,0.0137],[152,0.0137],[153,0.0047],[242,0.0047],[243,0.0146],[244,0.0047],[324,0.0047]]
       }
       calf_l_fk_ctrl.Rotation.Z {
        type=MovieSceneScriptingFloatChannel
        default=0.3732
        thinned_from=325
        keys [[0,-4.9606],[150,-4.9607],[152,-4.9518],[242,-4.9605],[243,-4.9526],[324,-4.9606]]
       }
       calf_l_fk_ctrl.Scale.X {type=MovieSceneScriptingFloatChannel, default=1, const [1,325]}
       calf_l_fk_ctrl.Scale.Y {type=MovieSceneScriptingFloatChannel, default=1, const [1,325]}
       calf_l_fk_ctrl.Scale.Z {type=MovieSceneScriptingFloatChannel, default=1, const [1,325]}
       foot_l_fk_ctrl.Location.X {
        type=MovieSceneScriptingFloatChannel
        default=0
        thinned_from=325
        keys [[0,-0.6707],[2,-0.6721],[3,-0.6747],[4,-0.6792],[5,-0.6862],[6,-0.6962],[7,-0.7096],[8,-0.7272],[9,-0.7503],[10,-0.7792],[11,-0.8122],[14,-0.9196],[15,-0.9527],[16,-0.9818],[17,-1.0052],[19,-1.0416],[21,-1.0704],[23,-1.0913],[25,-1.1041],[27,-1.1084],[31,-1.1067],[35,-1.0946],[40,-1.0663],[43,-1.041],[44,-1.0284],[45,-1.0111],[46,-0.9901],[47,-0.966],[52,-0.8283],[53,-0.803],[54,-0.7802],[55,-0.7609],[56,-0.7457],[58,-0.7224],[60,-0.7034],[62,-0.6883],[64,-0.6767],[66,-0.6682],[69,-0.6604],[73,-0.6571],[77,-0.6586],[79,-0.6622],[81,-0.6694],[82,-0.6747],[83,-0.6815],[84,-0.6898],[85,-0.7003],[86,-0.7134],[87,-0.7288],[88,-0.7463],[90,-0.786],[97,-0.942],[99,-0.9816],[105,-1.0799],[107,-1.108],[108,-1.1193],[109,-1.128],[110,-1.1336],[111,-1.1355],[113,-1.133],[116,-1.1208],[119,-1.1003],[127,-1.0213],[129,-0.9957],[132,-0.9472],[136,-0.8768],[138,-0.8449],[139,-0.8309],[143,-0.786],[147,-0.7475],[151,-0.7154],[154,-0.6961],[157,-0.6812],[160,-0.6718],[164,-0.6667],[168,-0.666],[169,-0.6672],[170,-0.6707],[171,-0.6763],[172,-0.6839],[173,-0.6931],[175,-0.7157],[178,-0.7569],[182,-0.816],[183,-0.8295],[191,-0.925],[194,-0.95],[199,-0.9711],[205,-0.9825],[211,-0.9838],[213,-0.9796],[215,-0.9684],[217,-0.9514],[220,-0.9181],[223,-0.8793],[225,-0.8497],[228,-0.7973],[234,-0.6838],[236,-0.6497],[237,-0.6342],[238,-0.6201],[239,-0.6075],[240,-0.5968],[241,-0.5878],[242,-0.5802],[243,-0.5741],[244,-0.5691],[245,-0.5653],[246,-0.5629],[247,-0.5618],[248,-0.5622],[249,-0.5653],[250,-0.5713],[251,-0.5786],[252,-0.587],[254,-0.6085],[255,-0.618],[256,-0.6255],[260,-0.6503],[262,-0.6651],[264,-0.683],[266,-0.7047],[268,-0.731],[269,-0.7462],[270,-0.767],[271,-0.7964],[272,-0.832],[275,-0.953],[276,-0.9904],[277,-1.0225],[278,-1.0469],[279,-1.0614],[282,-1.083],[285,-1.0964],[291,-1.1063],[296,-1.1057],[299,-1.0979],[301,-1.0875],[303,-1.0725],[305,-1.0528],[307,-1.0281],[308,-1.0137],[309,-0.9967],[310,-0.976],[311,-0.9523],[313,-0.8986],[317,-0.7843],[318,-0.7581],[319,-0.7341],[320,-0.713],[321,-0.6955],[322,-0.6822],[323,-0.6737],[324,-0.6707]]
       }
       foot_l_fk_ctrl.Location.Y {
        type=MovieSceneScriptingFloatChannel
        default=0.0398
        thinned_from=325
        keys [[0,-0.0042],[9,-0.0047],[24,-0.0068],[46,-0.0061],[83,-0.0042],[111,-0.007],[172,-0.0042],[218,-0.0058],[249,-0.0035],[270,-0.0048],[278,-0.0065],[309,-0.0062],[323,-0.0042],[324,-0.0042]]
       }
       foot_l_fk_ctrl.Location.Z {type=MovieSceneScriptingFloatChannel, default=-0.8885, thinned_from=325, keys [[0,0.0005],[324,0.0005]]}
       foot_l_fk_ctrl.Rotation.X {
        type=MovieSceneScriptingFloatChannel
        default=-1.1647
        thinned_from=325
        keys [[0,0.3492],[31,0.3578],[32,0.3588],[34,0.3626],[35,0.3655],[37,0.3754],[38,0.3788],[39,0.3808],[43,0.3831],[85,0.3752],[87,0.3729],[89,0.368],[90,0.3645],[93,0.3505],[94,0.3472],[96,0.3429],[97,0.3417],[98,0.3413],[109,0.3475],[170,0.3389],[188,0.3483],[189,0.3479],[192,0.3433],[194,0.3426],[204,0.347],[211,0.3468],[243,0.3389],[245,0.3398],[246,0.3411],[247,0.3435],[248,0.3482],[249,0.3519],[251,0.353],[252,0.3526],[254,0.3497],[257,0.3481],[294,0.3557],[324,0.3492]]
       }
       foot_l_fk_ctrl.Rotation.Y {
        type=MovieSceneScriptingFloatChannel
        default=0.0792
        thinned_from=325
        keys [[0,2.7789],[25,2.8876],[31,2.8906],[32,2.9017],[33,2.9217],[34,2.9479],[35,2.9849],[37,3.1108],[38,3.1549],[39,3.1802],[42,3.2052],[46,3.2041],[85,3.1076],[86,3.0976],[87,3.0804],[88,3.0549],[89,3.0204],[90,2.9768],[91,2.9194],[92,2.8547],[93,2.7992],[94,2.7588],[95,2.7274],[96,2.7043],[97,2.6898],[98,2.6854],[107,2.7575],[129,2.7348],[169,2.6514],[182,2.7263],[187,2.7714],[188,2.7737],[189,2.7687],[190,2.7506],[191,2.7245],[192,2.7048],[193,2.6963],[194,2.6933],[203,2.7428],[211,2.7416],[212,2.7321],[242,2.6393],[244,2.6414],[245,2.6497],[246,2.6672],[247,2.6986],[248,2.7598],[249,2.8087],[250,2.8202],[251,2.8232],[252,2.8194],[254,2.7845],[256,2.7691],[259,2.7688],[278,2.8662],[294,2.8645],[324,2.7789]]
       }
       foot_l_fk_ctrl.Rotation.Z {
        type=MovieSceneScriptingFloatChannel
        default=-0.2074
        thinned_from=325
        keys [[0,6.635],[28,6.7209],[31,6.69],[35,6.5537],[37,6.5693],[83,6.4827],[85,6.5123],[87,6.5918],[90,6.7322],[99,6.9552],[187,6.9053],[188,6.8921],[189,6.8644],[190,6.8123],[192,6.6846],[197,6.4772],[201,6.3918],[244,6.2891],[246,6.3097],[247,6.3362],[249,6.4379],[252,6.5096],[254,6.6119],[256,6.6677],[294,6.7267],[324,6.635]]
       }
       foot_l_fk_ctrl.Scale.X {type=MovieSceneScriptingFloatChannel, default=1, const [1,325]}
       foot_l_fk_ctrl.Scale.Y {type=MovieSceneScriptingFloatChannel, default=1, const [1,325]}
       foot_l_fk_ctrl.Scale.Z {type=MovieSceneScriptingFloatChannel, default=1, const [1,325]}
       ball_l_fk_ctrl.Location.X {type=MovieSceneScriptingFloatChannel, default=0, const [0,325]}
       ball_l_fk_ctrl.Location.Y {type=MovieSceneScriptingFloatChannel, default=0, const [0,325]}
       ball_l_fk_ctrl.Location.Z {type=MovieSceneScriptingFloatChannel, default=0, const [0,325]}
       ball_l_fk_ctrl.Rotation.X {type=MovieSceneScriptingFloatChannel, default=0, const [0,325]}
       ball_l_fk_ctrl.Rotation.Y {type=MovieSceneScriptingFloatChannel, default=0, const [0,325]}
       ball_l_fk_ctrl.Rotation.Z {type=MovieSceneScriptingFloatChannel, default=0, const [0,325]}
       ball_l_fk_ctrl.Scale.X {type=MovieSceneScriptingFloatChannel, default=1, const [1,325]}
       ball_l_fk_ctrl.Scale.Y {type=MovieSceneScriptingFloatChannel, default=1, const [1,325]}
       ball_l_fk_ctrl.Scale.Z {type=MovieSceneScriptingFloatChannel, default=1, const [1,325]}
       foot_l_ik_ctrl.Location.X {
        type=MovieSceneScriptingFloatChannel
        default=0
        thinned_from=325
        keys [[0,-2.2125],[16,-2.2136],[17,-2.2451],[22,-2.182],[23,-2.2114],[27,-2.2144],[38,-2.2115],[39,-2.1767],[44,-2.2165],[46,-2.2432],[47,-2.213],[67,-2.2128],[68,-2.2065],[69,-2.2124],[70,-2.2067],[71,-2.2126],[72,-2.2077],[73,-2.2126],[74,-2.2084],[75,-2.2127],[77,-2.2062],[78,-2.2126],[82,-2.2074],[83,-2.2129],[100,-2.2138],[101,-2.2449],[106,-2.182],[107,-2.2135],[110,-2.2151],[115,-2.2092],[116,-2.2131],[117,-2.2102],[118,-2.2127],[119,-2.2086],[120,-2.2121],[121,-2.2071],[122,-2.2124],[123,-2.181],[128,-2.216],[130,-2.2388],[131,-2.2133],[151,-2.2129],[152,-2.2078],[153,-2.2125],[154,-2.2082],[155,-2.2127],[156,-2.2087],[157,-2.2127],[158,-2.2092],[159,-2.2125],[161,-2.2073],[162,-2.2126],[165,-2.2077],[166,-2.2126],[233,-2.2125],[234,-2.2093],[235,-2.2125],[236,-2.2093],[237,-2.2127],[238,-2.2099],[239,-2.2127],[240,-2.2099],[241,-2.2128],[268,-2.2136],[269,-2.2449],[274,-2.1821],[275,-2.2114],[279,-2.2133],[282,-2.2019],[283,-2.2135],[284,-2.2071],[285,-2.2131],[286,-2.2063],[287,-2.2125],[288,-2.2041],[289,-2.2112],[290,-2.2113],[291,-2.181],[296,-2.2158],[298,-2.2375],[299,-2.2125],[320,-2.2126],[321,-2.2087],[322,-2.2127],[324,-2.2101]]
       }
       foot_l_ik_ctrl.Location.Y {
        type=MovieSceneScriptingFloatChannel
        default=0
        thinned_from=325
        keys [[0,-2.3767],[1,-2.383],[2,-2.3771],[14,-2.3764],[16,-2.3797],[17,-2.4715],[22,-2.2886],[23,-2.3735],[26,-2.3769],[27,-2.3831],[32,-2.3761],[33,-2.3803],[38,-2.374],[39,-2.2866],[43,-2.3618],[44,-2.3866],[46,-2.4524],[47,-2.378],[67,-2.3774],[68,-2.3614],[69,-2.3765],[70,-2.3624],[71,-2.3768],[72,-2.3646],[73,-2.3769],[74,-2.3667],[75,-2.3771],[77,-2.3613],[78,-2.3769],[80,-2.3662],[83,-2.3603],[84,-2.3769],[100,-2.3802],[101,-2.471],[106,-2.2891],[107,-2.3794],[108,-2.3763],[109,-2.3788],[110,-2.3923],[111,-2.377],[115,-2.3647],[116,-2.3782],[117,-2.3679],[118,-2.3772],[119,-2.363],[120,-2.3754],[121,-2.3582],[122,-2.3764],[123,-2.2864],[127,-2.3617],[128,-2.3867],[130,-2.4527],[131,-2.3787],[151,-2.3777],[152,-2.3608],[153,-2.3767],[154,-2.3625],[155,-2.3772],[156,-2.3641],[157,-2.3772],[158,-2.3662],[159,-2.3767],[161,-2.36],[162,-2.377],[165,-2.3631],[166,-2.3768],[168,-2.3759],[169,-2.3813],[170,-2.3763],[182,-2.3779],[183,-2.3807],[233,-2.377],[234,-2.3588],[235,-2.3771],[236,-2.3609],[237,-2.3768],[238,-2.3625],[239,-2.3769],[240,-2.3646],[241,-2.3768],[242,-2.3666],[243,-2.377],[266,-2.3776],[267,-2.375],[268,-2.3797],[269,-2.4708],[274,-2.2891],[275,-2.3735],[277,-2.3767],[278,-2.3949],[279,-2.3788],[280,-2.3793],[281,-2.3714],[282,-2.3689],[283,-2.3794],[284,-2.3684],[285,-2.3783],[286,-2.3655],[287,-2.3767],[288,-2.3595],[289,-2.3729],[290,-2.3733],[291,-2.2858],[295,-2.3616],[296,-2.3862],[298,-2.4493],[299,-2.3766],[300,-2.3814],[320,-2.3768],[321,-2.3617],[322,-2.3772],[324,-2.3673]]
       }
       foot_l_ik_ctrl.Location.Z {
        type=MovieSceneScriptingFloatChannel
        default=0
        thinned_from=325
        keys [[0,0],[16,0],[17,0.0008],[22,-0.0006],[23,0],[38,0],[39,-0.0002],[46,0.0002],[47,0],[82,-0.0001],[83,-0.0004],[84,0],[100,0],[101,0.0006],[106,-0.0004],[107,0],[109,0],[110,0.0002],[111,0],[115,-0.0001],[116,0],[117,-0.0001],[118,0],[119,-0.0001],[120,0],[121,-0.0002],[122,0],[123,-0.0004],[129,0.0002],[130,0.0005],[131,0],[151,0],[152,-0.0002],[153,0],[154,-0.0001],[155,0],[156,-0.0001],[157,0],[161,-0.0002],[166,0],[168,-0.0002],[169,0],[185,-0.0001],[186,0.0002],[208,0],[210,0.0001],[211,-0.0001],[233,0],[234,-0.0004],[235,0],[236,-0.0003],[237,0],[238,-0.0003],[239,0],[240,-0.0002],[241,0],[242,-0.0002],[243,0],[268,0],[269,0.0008],[274,-0.0006],[275,0],[277,0],[278,0.0003],[279,0],[282,0.0003],[283,0],[290,0],[291,-0.0006],[293,-0.0003],[294,-0.0003],[298,0.0006],[299,0],[320,0],[321,-0.0002],[322,0],[324,-0.0001]]
       }
       foot_l_ik_ctrl.Rotation.X {
        type=MovieSceneScriptingFloatChannel
        default=0
        thinned_from=325
        keys [[0,-0.0001],[1,0.0032],[2,0],[3,0.0004],[4,-0.0002],[6,-0.0003],[9,0.0004],[10,0],[12,0.0004],[13,0.0004],[14,-0.0003],[15,0.0012],[16,0.0014],[17,0.0557],[18,0.0347],[19,0.0127],[20,-0.0096],[21,-0.0314],[22,-0.052],[23,-0.0017],[24,-0.0014],[25,-0.0002],[26,0],[27,0.0031],[28,0.0031],[29,0.0006],[30,0.0006],[31,0.0003],[32,-0.0004],[33,0.0017],[34,0.0002],[35,0],[36,0.0003],[37,-0.0006],[38,-0.0018],[39,-0.0556],[40,-0.0455],[41,-0.0344],[42,-0.0224],[43,-0.0094],[44,0.006],[45,0.0251],[46,0.047],[47,0.0007],[48,0.0007],[49,-0.0004],[50,0],[52,-0.0001],[53,0.0003],[54,0.0003],[55,-0.0001],[56,0],[57,0.0004],[58,-0.0002],[59,-0.0003],[60,0.0001],[61,-0.0002],[63,-0.0002],[64,0.0001],[66,-0.0001],[67,0.0002],[68,-0.0086],[69,-0.0003],[70,-0.0082],[71,0],[72,-0.007],[73,0],[74,-0.005],[75,0.0001],[76,-0.005],[77,-0.0089],[78,0],[80,-0.0062],[81,-0.0074],[82,-0.0082],[83,-0.0051],[84,0],[85,0.0002],[87,-0.0002],[88,0],[89,-0.0003],[90,0],[91,0],[92,0.0004],[94,-0.0002],[95,0.0001],[96,-0.0003],[98,0.0004],[99,0.0012],[100,0.0014],[101,0.0556],[102,0.0347],[103,0.0127],[104,-0.0096],[105,-0.0314],[106,-0.0519],[107,0.0012],[108,-0.0002],[109,0.0009],[110,0.0064],[111,0.0001],[112,-0.0002],[114,-0.0047],[115,-0.006],[116,0.0007],[117,-0.0042],[118,0.0001],[119,-0.0067],[120,-0.001],[121,-0.0091],[122,-0.0001],[123,-0.0536],[124,-0.0437],[125,-0.033],[126,-0.0214],[127,-0.009],[128,0.0058],[129,0.0239],[130,0.0449],[131,0.0007],[133,0.0004],[134,-0.0001],[137,-0.0001],[138,0.0002],[139,-0.0005],[140,0],[142,0.0002],[143,-0.0002],[144,0.0001],[146,0.0002],[148,-0.0002],[149,0],[150,0.0005],[151,0.0003],[152,-0.008],[153,-0.0001],[154,-0.0073],[155,0.0002],[156,-0.0065],[157,0.0002],[158,-0.0055],[159,-0.0001],[160,-0.0048],[161,-0.0086],[162,0],[163,-0.0031],[164,-0.0058],[165,-0.0074],[166,-0.0001],[167,-0.0005],[168,-0.0005],[169,0.0022],[170,-0.0003],[171,0],[173,-0.0004],[175,0.0001],[176,-0.0003],[177,0.0006],[179,0],[180,0],[181,0.0008],[182,0.0004],[183,0.0019],[184,0],[185,0],[186,-0.0006],[188,-0.0002],[191,-0.0004],[192,0],[193,-0.0001],[194,0],[195,-0.0003],[196,0],[197,-0.0001],[198,0.0002],[201,-0.0001],[202,0.0002],[204,-0.0001],[205,0.0001],[206,0.0001],[208,-0.0002],[211,0.006],[212,0.0001],[221,0.0001],[222,0],[233,0],[234,-0.0077],[235,-0.0001],[236,-0.0071],[237,0],[238,-0.0062],[239,0],[240,-0.0055],[241,0],[242,-0.0046],[243,0],[255,0.0001],[256,-0.0001],[257,0.0001],[258,-0.0002],[263,0.0002],[264,-0.0003],[266,0.0004],[267,-0.0008],[268,0.0014],[269,0.0557],[270,0.0347],[271,0.0127],[272,-0.0096],[273,-0.0314],[274,-0.052],[275,-0.0017],[276,-0.0014],[277,-0.0002],[278,0.0069],[279,0.001],[280,-0.0004],[281,-0.0062],[282,-0.009],[283,0.0012],[284,-0.0061],[285,0.0006],[286,-0.0074],[287,0],[288,-0.0105],[289,-0.0021],[290,-0.0017],[291,-0.0536],[292,-0.0437],[293,-0.0296],[294,-0.0145],[295,-0.009],[296,0.0055],[297,0.0228],[298,0.0425],[299,0],[300,0.0023],[301,0.0001],[302,0.0001],[303,-0.0004],[305,0.0006],[306,0],[307,0.0002],[308,0],[309,0.0002],[310,-0.0001],[312,-0.0002],[313,0.0003],[314,0.0003],[315,0.0001],[317,0.0001],[318,-0.0003],[319,0.0001],[320,-0.0001],[321,-0.0072],[322,0.0001],[323,-0.0033],[324,-0.0046]]
       }
       foot_l_ik_ctrl.Rotation.Y {
        type=MovieSceneScriptingFloatChannel
        default=0
        thinned_from=325
        keys [[0,-0.0001],[1,0.0031],[2,0.0001],[3,0.0004],[4,-0.0002],[6,-0.0003],[9,0.0004],[10,0],[12,0.0003],[13,0.0003],[14,-0.0003],[15,0.0012],[16,0.0014],[17,0.0529],[18,0.033],[19,0.0121],[20,-0.0091],[21,-0.0298],[22,-0.0494],[23,-0.0016],[24,-0.0014],[25,-0.0001],[26,0],[27,0.0033],[28,0.0033],[29,0.0006],[30,0.0005],[31,0.0003],[32,-0.0004],[33,0.0016],[34,0.0001],[35,0.0001],[36,0.0003],[37,-0.0005],[38,-0.0016],[39,-0.0487],[40,-0.0396],[41,-0.0298],[42,-0.0193],[43,-0.0081],[44,0.0052],[45,0.0216],[46,0.0404],[47,0.0006],[48,0.0006],[49,-0.0004],[50,0],[52,-0.0001],[53,0.0002],[54,0.0002],[55,-0.0001],[56,0],[57,0.0004],[58,-0.0002],[59,-0.0002],[60,0.0001],[61,-0.0002],[63,-0.0002],[64,0],[66,-0.0001],[67,0.0002],[68,-0.0077],[69,-0.0002],[70,-0.0071],[71,0],[72,-0.0061],[73,0],[74,-0.0043],[75,0.0001],[76,-0.0044],[77,-0.0076],[78,0],[80,-0.0051],[81,-0.006],[83,-0.0104],[84,0],[85,0.0001],[87,-0.0002],[88,0],[89,-0.0002],[91,0],[92,0.0003],[94,-0.0002],[95,0.0001],[96,-0.0003],[98,0.0004],[99,0.0011],[100,0.0014],[101,0.053],[102,0.033],[103,0.0121],[104,-0.0091],[105,-0.0298],[106,-0.0493],[107,0.0012],[108,-0.0002],[109,0.0009],[110,0.0087],[111,0.0001],[112,0],[113,-0.0022],[114,-0.0047],[115,-0.0065],[116,0.0006],[117,-0.0047],[118,0.0001],[119,-0.0072],[120,-0.0009],[121,-0.0096],[122,-0.0001],[123,-0.0509],[124,-0.0415],[125,-0.0313],[126,-0.0204],[127,-0.0086],[128,0.0055],[129,0.0228],[130,0.0427],[131,0.0007],[133,0.0004],[134,-0.0001],[137,-0.0001],[138,0.0002],[139,-0.0005],[140,0],[142,0.0002],[143,-0.0002],[144,0.0001],[146,0.0002],[148,-0.0002],[149,0],[150,0.0005],[151,0.0003],[152,-0.0084],[153,0],[154,-0.0075],[155,0.0002],[156,-0.0067],[157,0.0002],[158,-0.0055],[159,-0.0001],[160,-0.0048],[161,-0.0088],[162,0.0001],[164,-0.0055],[165,-0.007],[166,-0.0001],[167,-0.0005],[168,-0.0005],[169,0.0023],[170,-0.0003],[171,0],[173,-0.0004],[175,0.0001],[176,-0.0003],[177,0.0006],[179,0],[180,0],[181,0.0007],[182,0.0004],[183,0.0021],[184,-0.0001],[185,-0.0003],[186,0.0009],[188,0.0002],[191,0.0003],[192,0],[194,0],[195,0.0002],[196,0],[197,0.0001],[198,-0.0001],[201,0.0001],[202,-0.0001],[204,0.0001],[206,-0.0001],[208,0.0002],[209,0.0022],[211,0.0068],[212,-0.0001],[220,0],[221,-0.0001],[222,0.0001],[223,-0.0001],[228,0.0001],[233,0.0001],[234,-0.0102],[235,0.0001],[236,-0.0088],[237,-0.0001],[238,-0.008],[239,-0.0001],[240,-0.0067],[241,-0.0001],[242,-0.0057],[243,0.0001],[244,0.0001],[245,-0.0002],[246,-0.0002],[248,0.0002],[250,-0.0001],[251,0.0003],[252,-0.0002],[253,0.0001],[254,0.0001],[255,0.0003],[256,-0.0001],[257,0.0001],[258,-0.0002],[261,0.0002],[263,0.0002],[264,-0.0003],[266,0.0004],[267,-0.0008],[268,0.0013],[269,0.0529],[270,0.033],[271,0.0121],[272,-0.0091],[273,-0.0298],[274,-0.0493],[275,-0.0016],[276,-0.0013],[277,-0.0001],[278,0.0105],[279,0.001],[280,0.0019],[281,-0.0011],[282,-0.0016],[283,0.0012],[284,-0.0036],[285,0.0006],[286,-0.005],[287,0],[288,-0.0079],[289,-0.0019],[290,-0.0017],[291,-0.0509],[292,-0.0415],[293,-0.0301],[294,-0.0177],[295,-0.0086],[296,0.0052],[297,0.0217],[298,0.0404],[299,0],[300,0.0022],[301,0.0001],[302,0.0001],[303,-0.0004],[305,0.0006],[306,0],[307,0.0002],[308,0],[309,0.0002],[310,-0.0001],[312,-0.0002],[313,0.0003],[317,0.0001],[318,-0.0002],[319,0.0001],[320,-0.0001],[321,-0.0082],[322,0.0001],[323,-0.0037],[324,-0.0051]]
       }
       foot_l_ik_ctrl.Rotation.Z {
        type=MovieSceneScriptingFloatChannel
        default=0
        thinned_from=325
        keys [[0,-62.9662],[16,-62.966],[22,-63.0339],[38,-62.9689],[100,-62.9687],[106,-63.0339],[122,-62.9645],[268,-62.9665],[274,-63.0338],[290,-62.9654],[324,-62.9661]]
       }
       foot_l_ik_ctrl.Scale.X {type=MovieSceneScriptingFloatChannel, default=1, const [1,325]}
       foot_l_ik_ctrl.Scale.Y {type=MovieSceneScriptingFloatChannel, default=1, const [1,325]}
       foot_l_ik_ctrl.Scale.Z {type=MovieSceneScriptingFloatChannel, default=1, const [1,325]}
       ball_l_ik_ctrl.Location.X {
        type=MovieSceneScriptingFloatChannel
        default=0
        thinned_from=325
        keys [[0,0],[12,0],[13,0.0001],[65,0],[66,0.0001],[136,0],[137,0.0002],[141,-0.0001],[145,0],[146,-0.0001],[227,0],[228,-0.0001],[230,0.0001],[262,0],[263,0.0003],[264,0],[301,0],[302,-0.0004],[303,0],[316,0],[317,-0.0001],[324,0]]
       }
       ball_l_ik_ctrl.Location.Y {type=MovieSceneScriptingFloatChannel, default=0, thinned_from=325, keys [[0,0],[188,0],[189,-0.0001],[324,0]]}
       ball_l_ik_ctrl.Location.Z {type=MovieSceneScriptingFloatChannel, default=0, thinned_from=325, keys [[0,0],[324,0]]}
       ball_l_ik_ctrl.Rotation.X {
        type=MovieSceneScriptingFloatChannel
        default=0
        thinned_from=325
        keys [[0,0],[12,0],[13,-0.0001],[136,0],[137,-0.0001],[301,0],[302,0.0002],[324,0]]
       }
       ball_l_ik_ctrl.Rotation.Y {
        type=MovieSceneScriptingFloatChannel
        default=0
        thinned_from=325
        keys [[0,0],[12,0],[13,-0.0001],[136,0],[137,-0.0001],[301,0],[302,0.0002],[324,0]]
       }
       ball_l_ik_ctrl.Rotation.Z {
        type=MovieSceneScriptingFloatChannel
        default=0
        thinned_from=325
        keys [[0,0],[12,0],[13,-0.0009],[14,0],[65,0],[66,-0.0004],[67,0],[136,0],[137,-0.0009],[138,0],[140,0],[141,0.0004],[142,0],[145,0],[146,0.0005],[147,0],[188,0],[189,-0.0003],[190,0],[262,0],[263,-0.001],[264,0],[301,0],[302,0.0018],[303,0],[316,0],[317,0.0007],[318,0],[324,0]]
       }
       ball_l_ik_ctrl.Scale.X {type=MovieSceneScriptingFloatChannel, default=1, const [1,325]}
       ball_l_ik_ctrl.Scale.Y {type=MovieSceneScriptingFloatChannel, default=1, const [1,325]}
       ball_l_ik_ctrl.Scale.Z {type=MovieSceneScriptingFloatChannel, default=1, const [1,325]}
       foot_bk1_l_ctrl.Location.X {type=MovieSceneScriptingFloatChannel, default=0, const [0,325]}
       foot_bk1_l_ctrl.Location.Y {type=MovieSceneScriptingFloatChannel, default=0, const [0,325]}
       foot_bk1_l_ctrl.Location.Z {type=MovieSceneScriptingFloatChannel, default=0, const [0,325]}
       foot_bk1_l_ctrl.Rotation.X {type=MovieSceneScriptingFloatChannel, default=0, const [0,325]}
       foot_bk1_l_ctrl.Rotation.Y {type=MovieSceneScriptingFloatChannel, default=0, const [0,325]}
       foot_bk1_l_ctrl.Rotation.Z {type=MovieSceneScriptingFloatChannel, default=0, const [0,325]}
       foot_bk1_l_ctrl.Scale.X {type=MovieSceneScriptingFloatChannel, default=1, const [1,325]}
       foot_bk1_l_ctrl.Scale.Y {type=MovieSceneScriptingFloatChannel, default=1, const [1,325]}
       foot_bk1_l_ctrl.Scale.Z {type=MovieSceneScriptingFloatChannel, default=1, const [1,325]}
       foot_roll_l_ctrl.Location.X {type=MovieSceneScriptingFloatChannel, default=0, const [0,325]}
       foot_roll_l_ctrl.Location.Y {type=MovieSceneScriptingFloatChannel, default=0, const [0,325]}
       foot_roll_l_ctrl.Location.Z {type=MovieSceneScriptingFloatChannel, default=0, const [0,325]}
       foot_roll_l_ctrl.Rotation.X {type=MovieSceneScriptingFloatChannel, default=0, const [0,325]}
       foot_roll_l_ctrl.Rotation.Y {type=MovieSceneScriptingFloatChannel, default=0, const [0,325]}
       foot_roll_l_ctrl.Rotation.Z {type=MovieSceneScriptingFloatChannel, default=0, const [0,325]}
       foot_roll_l_ctrl.Scale.X {type=MovieSceneScriptingFloatChannel, default=1, const [1,325]}
       foot_roll_l_ctrl.Scale.Y {type=MovieSceneScriptingFloatChannel, default=1, const [1,325]}
       foot_roll_l_ctrl.Scale.Z {type=MovieSceneScriptingFloatChannel, default=1, const [1,325]}
       roll_blend_l_ctrl {type=MovieSceneScriptingFloatChannel, default=1, const [1,325]}
       tip_l_ctrl.Location.X {type=MovieSceneScriptingFloatChannel, default=0, const [0,325]}
       tip_l_ctrl.Location.Y {type=MovieSceneScriptingFloatChannel, default=0, const [0,325]}
       tip_l_ctrl.Location.Z {type=MovieSceneScriptingFloatChannel, default=0, const [0,325]}
       tip_l_ctrl.Rotation.X {type=MovieSceneScriptingFloatChannel, default=0, const [0,325]}
       tip_l_ctrl.Rotation.Y {type=MovieSceneScriptingFloatChannel, default=0, const [0,325]}
       tip_l_ctrl.Rotation.Z {type=MovieSceneScriptingFloatChannel, default=0, const [0,325]}
       tip_l_ctrl.Scale.X {type=MovieSceneScriptingFloatChannel, default=1, const [1,325]}
       tip_l_ctrl.Scale.Y {type=MovieSceneScriptingFloatChannel, default=1, const [1,325]}
       tip_l_ctrl.Scale.Z {type=MovieSceneScriptingFloatChannel, default=1, const [1,325]}
       heel_l_ctrl.Location.X {type=MovieSceneScriptingFloatChannel, default=0, const [0,325]}
       heel_l_ctrl.Location.Y {type=MovieSceneScriptingFloatChannel, default=0, const [0,325]}
       heel_l_ctrl.Location.Z {type=MovieSceneScriptingFloatChannel, default=0, const [0,325]}
       heel_l_ctrl.Rotation.X {type=MovieSceneScriptingFloatChannel, default=0, const [0,325]}
       heel_l_ctrl.Rotation.Y {type=MovieSceneScriptingFloatChannel, default=0, const [0,325]}
       heel_l_ctrl.Rotation.Z {type=MovieSceneScriptingFloatChannel, default=0, const [0,325]}
       heel_l_ctrl.Scale.X {type=MovieSceneScriptingFloatChannel, default=1, const [1,325]}
       heel_l_ctrl.Scale.Y {type=MovieSceneScriptingFloatChannel, default=1, const [1,325]}
       heel_l_ctrl.Scale.Z {type=MovieSceneScriptingFloatChannel, default=1, const [1,325]}
       leg_l_softness {type=MovieSceneScriptingFloatChannel, default=0, const [0,325]}
       leg_l_pv_ik_ctrl.Location.X {
        type=MovieSceneScriptingFloatChannel
        default=0
        thinned_from=325
        keys [[0,-0.1492],[12,-0.1492],[13,-0.1485],[14,-0.1491],[28,-0.1492],[31,-0.1483],[32,-0.1472],[35,-0.1424],[38,-0.1391],[39,-0.1384],[42,-0.1377],[65,-0.1378],[66,-0.1375],[67,-0.1378],[82,-0.1377],[84,-0.1381],[85,-0.1387],[86,-0.1398],[87,-0.1414],[91,-0.1493],[93,-0.1535],[95,-0.157],[97,-0.1597],[98,-0.1606],[99,-0.1611],[101,-0.1612],[109,-0.1592],[135,-0.1595],[137,-0.1589],[138,-0.1596],[140,-0.1596],[141,-0.16],[142,-0.1597],[145,-0.1597],[146,-0.1601],[147,-0.1597],[178,-0.1599],[180,-0.1593],[187,-0.1574],[188,-0.157],[190,-0.1552],[197,-0.1459],[199,-0.144],[202,-0.1424],[205,-0.1416],[231,-0.1411],[233,-0.1414],[246,-0.1416],[247,-0.1419],[249,-0.1428],[250,-0.143],[251,-0.1437],[252,-0.1447],[254,-0.1484],[256,-0.1504],[258,-0.1514],[261,-0.1517],[263,-0.1511],[264,-0.1518],[294,-0.15],[301,-0.1493],[302,-0.1505],[303,-0.1492],[314,-0.1488],[317,-0.1496],[318,-0.1491],[324,-0.1492]]
       }
       leg_l_pv_ik_ctrl.Location.Y {type=MovieSceneScriptingFloatChannel, default=0, thinned_from=325, keys [[0,-1.9438],[324,-1.9438]]}
       leg_l_pv_ik_ctrl.Location.Z {type=MovieSceneScriptingFloatChannel, default=0, thinned_from=325, keys [[0,-1.8045],[324,-1.8046]]}
       leg_l_pv_ik_ctrl.Rotation.X {type=MovieSceneScriptingFloatChannel, default=0, thinned_from=325, keys [[0,2.6168],[324,2.6168]]}
       leg_l_pv_ik_ctrl.Rotation.Y {
        type=MovieSceneScriptingFloatChannel
        default=0
        thinned_from=325
        keys [[0,-0.0193],[86,-0.0187],[101,-0.0201],[229,-0.0189],[231,-0.0187],[262,-0.0195],[324,-0.0193]]
       }
       leg_l_pv_ik_ctrl.Rotation.Z {
        type=MovieSceneScriptingFloatChannel
        default=0
        thinned_from=325
        keys [[0,0.2001],[12,0.2],[13,0.1991],[14,0.1999],[28,0.2001],[30,0.1994],[31,0.1988],[32,0.1973],[35,0.1905],[38,0.1857],[39,0.1847],[42,0.1837],[65,0.1839],[66,0.1834],[67,0.1839],[82,0.1837],[84,0.1844],[85,0.1852],[86,0.1868],[87,0.189],[91,0.2002],[92,0.2034],[94,0.2088],[96,0.2132],[97,0.2149],[98,0.2162],[99,0.2169],[101,0.217],[109,0.2142],[135,0.2147],[137,0.2138],[138,0.2148],[140,0.2148],[141,0.2153],[142,0.2149],[145,0.2149],[146,0.2155],[147,0.215],[178,0.2152],[180,0.2143],[187,0.2117],[188,0.2111],[190,0.2086],[197,0.1953],[199,0.1927],[202,0.1904],[205,0.1893],[232,0.1888],[243,0.189],[247,0.1897],[249,0.191],[250,0.1913],[251,0.1923],[252,0.1937],[254,0.199],[256,0.2018],[258,0.2032],[261,0.2037],[263,0.2028],[264,0.2038],[294,0.2012],[301,0.2002],[302,0.2019],[303,0.2001],[314,0.1995],[317,0.2006],[318,0.2],[324,0.2]]
       }
       leg_l_pv_ik_ctrl.Scale.X {type=MovieSceneScriptingFloatChannel, default=1, const [1,325]}
       leg_l_pv_ik_ctrl.Scale.Y {type=MovieSceneScriptingFloatChannel, default=1, const [1,325]}
       leg_l_pv_ik_ctrl.Scale.Z {type=MovieSceneScriptingFloatChannel, default=1, const [1,325]}
       thigh_r_fk_ctrl.Location.X {type=MovieSceneScriptingFloatChannel, default=0, thinned_from=325, keys [[0,-0.0002],[324,-0.0002]]}
       thigh_r_fk_ctrl.Location.Y {type=MovieSceneScriptingFloatChannel, default=0, thinned_from=325, keys [[0,0],[324,0]]}
       thigh_r_fk_ctrl.Location.Z {type=MovieSceneScriptingFloatChannel, default=0, thinned_from=325, keys [[0,0],[324,0]]}
       thigh_r_fk_ctrl.Rotation.X {
        type=MovieSceneScriptingFloatChannel
        default=1.2139
        thinned_from=325
        keys [[0,-4.2125],[28,-4.1166],[29,-4.1393],[31,-4.2409],[32,-4.265],[33,-4.254],[34,-4.167],[35,-3.9278],[36,-3.2691],[37,-2.5162],[38,-2.0642],[39,-1.8111],[40,-1.688],[41,-1.582],[42,-1.4939],[43,-1.4245],[44,-1.3749],[45,-1.3462],[46,-1.339],[47,-1.3457],[48,-1.3574],[49,-1.3732],[51,-1.4136],[55,-1.5043],[57,-1.5406],[60,-1.5794],[63,-1.6054],[67,-1.6249],[83,-1.6391],[84,-1.6165],[86,-1.5107],[87,-1.4845],[88,-1.5283],[89,-1.6817],[90,-1.9675],[91,-2.4754],[92,-3.116],[93,-3.6439],[94,-3.9863],[95,-4.2375],[96,-4.402],[97,-4.4848],[98,-4.4753],[99,-4.3968],[100,-4.3051],[101,-4.2329],[102,-4.1775],[103,-4.1408],[104,-4.1256],[170,-4.2053],[182,-4.1476],[183,-4.1261],[185,-4.0433],[186,-4.0228],[187,-4.0318],[188,-4.1139],[189,-4.3421],[190,-4.8229],[191,-5.4522],[192,-5.9864],[193,-6.3454],[194,-6.6096],[195,-6.7764],[196,-6.8835],[197,-6.971],[198,-7.0401],[199,-7.0919],[200,-7.1278],[201,-7.149],[242,-7.2468],[243,-7.2408],[244,-7.1945],[245,-7.0723],[246,-6.8287],[247,-6.386],[248,-5.4992],[249,-4.7907],[250,-4.6026],[251,-4.4606],[252,-4.3582],[253,-4.2879],[254,-4.2434],[255,-4.2188],[278,-4.1157],[324,-4.2118]]
       }
       thigh_r_fk_ctrl.Rotation.Y {
        type=MovieSceneScriptingFloatChannel
        default=-1.213
        thinned_from=325
        keys [[0,-3.2485],[28,-3.282],[30,-3.2701],[31,-3.259],[32,-3.2309],[35,-3.0964],[38,-2.9895],[39,-2.9682],[41,-2.9502],[83,-2.9158],[85,-2.9501],[86,-2.9825],[87,-3.0273],[93,-3.3702],[95,-3.4647],[96,-3.5038],[97,-3.5358],[98,-3.559],[99,-3.5717],[101,-3.5749],[110,-3.5372],[181,-3.5083],[182,-3.5034],[183,-3.5062],[184,-3.5249],[185,-3.5581],[186,-3.6047],[187,-3.67],[189,-3.8372],[192,-4.0563],[193,-4.1195],[194,-4.157],[195,-4.1697],[202,-4.1459],[245,-4.1051],[246,-4.0908],[247,-4.064],[248,-4.0229],[249,-3.9558],[252,-3.6998],[255,-3.4809],[256,-3.4373],[259,-3.3694],[262,-3.3267],[283,-3.3189],[324,-3.2493]]
       }
       thigh_r_fk_ctrl.Rotation.Z {
        type=MovieSceneScriptingFloatChannel
        default=-0.1592
        thinned_from=325
        keys [[0,2.7101],[1,2.7101],[3,2.6732],[4,2.6467],[6,2.5776],[8,2.4926],[14,2.2081],[16,2.1255],[22,1.9167],[23,1.8866],[24,1.8612],[25,1.8413],[26,1.8288],[28,1.8288],[30,1.8233],[31,1.824],[32,1.8362],[33,1.86],[34,1.898],[35,1.9619],[36,2.0918],[37,2.2376],[38,2.3371],[39,2.4018],[41,2.4806],[47,2.6879],[52,2.8773],[54,2.9391],[56,2.9821],[68,3.1494],[69,3.1494],[70,3.1712],[71,3.1712],[72,3.1905],[73,3.1905],[74,3.2071],[75,3.2071],[76,3.2208],[78,3.2208],[79,3.2356],[83,3.2356],[84,3.2429],[85,3.2396],[86,3.2227],[87,3.189],[88,3.1334],[89,3.0531],[90,2.9465],[91,2.8026],[92,2.6368],[93,2.4888],[94,2.3705],[95,2.268],[96,2.181],[97,2.1098],[98,2.0572],[99,2.0211],[103,1.9169],[106,1.8225],[107,1.7944],[108,1.7707],[109,1.7522],[110,1.7402],[115,1.7402],[116,1.7525],[117,1.7525],[118,1.7683],[119,1.7683],[120,1.7892],[123,1.8304],[126,1.8831],[128,1.9274],[129,1.9566],[131,2.0268],[136,2.2259],[138,2.2933],[139,2.3203],[148,2.4702],[152,2.5214],[153,2.5214],[154,2.5435],[155,2.5435],[156,2.5627],[157,2.5627],[158,2.5794],[159,2.5794],[160,2.5934],[162,2.5934],[163,2.6082],[170,2.6082],[172,2.5456],[173,2.5093],[175,2.4222],[180,2.1714],[182,2.0815],[183,2.0457],[186,1.9673],[187,1.9364],[188,1.8967],[189,1.8383],[190,1.7469],[191,1.6374],[192,1.547],[193,1.4872],[194,1.4469],[195,1.4279],[196,1.4226],[198,1.4212],[200,1.4291],[201,1.4372],[202,1.4372],[203,1.4599],[207,1.517],[210,1.5711],[211,1.592],[212,1.6175],[213,1.6495],[214,1.6869],[219,1.9036],[220,1.9432],[221,1.9784],[222,2.0079],[223,2.0298],[231,2.1568],[235,2.2067],[236,2.2067],[237,2.2279],[238,2.2279],[239,2.2468],[240,2.2468],[241,2.2631],[242,2.2631],[244,2.2909],[245,2.3154],[246,2.3588],[247,2.4327],[248,2.5755],[249,2.689],[250,2.7174],[251,2.7336],[252,2.7396],[253,2.7272],[255,2.6735],[257,2.6018],[260,2.4728],[266,2.1877],[268,2.1056],[274,1.8997],[275,1.8705],[276,1.846],[277,1.827],[278,1.8149],[283,1.8149],[284,1.8294],[285,1.8294],[286,1.8462],[287,1.8462],[288,1.8682],[291,1.9108],[295,1.9859],[296,2.0099],[298,2.0706],[305,2.3366],[307,2.4025],[315,2.5991],[318,2.6566],[320,2.685],[321,2.685],[322,2.7035],[324,2.7035]]
       }
       thigh_r_fk_ctrl.Scale.X {type=MovieSceneScriptingFloatChannel, default=1, const [1,325]}
       thigh_r_fk_ctrl.Scale.Y {type=MovieSceneScriptingFloatChannel, default=1, const [1,325]}
       thigh_r_fk_ctrl.Scale.Z {type=MovieSceneScriptingFloatChannel, default=1, const [1,325]}
       calf_r_fk_ctrl.Location.X {
        type=MovieSceneScriptingFloatChannel
        default=0
        thinned_from=325
        keys [[0,-0.6018],[2,-0.6031],[3,-0.6055],[4,-0.6096],[5,-0.616],[6,-0.6252],[7,-0.6375],[8,-0.6536],[9,-0.6748],[10,-0.7013],[11,-0.7315],[14,-0.8298],[15,-0.8602],[16,-0.8869],[17,-0.9083],[19,-0.9416],[21,-0.9679],[23,-0.9869],[25,-0.9985],[27,-1.0024],[35,-1.0007],[38,-0.9947],[40,-0.9855],[42,-0.972],[43,-0.9635],[44,-0.9521],[45,-0.9363],[46,-0.9171],[47,-0.8951],[52,-0.7693],[53,-0.7461],[54,-0.7253],[55,-0.7076],[56,-0.6938],[58,-0.6725],[60,-0.6551],[62,-0.6413],[64,-0.6307],[66,-0.6229],[69,-0.6158],[75,-0.6129],[78,-0.6155],[80,-0.6203],[82,-0.6289],[84,-0.6419],[85,-0.6504],[87,-0.672],[89,-0.6983],[91,-0.7288],[99,-0.8687],[105,-0.9603],[107,-0.9866],[108,-0.9972],[109,-1.0053],[110,-1.0105],[111,-1.0123],[113,-1.01],[116,-0.9988],[119,-0.98],[125,-0.9275],[128,-0.8967],[130,-0.8703],[137,-0.7601],[139,-0.7331],[143,-0.692],[147,-0.6566],[151,-0.6272],[154,-0.6095],[157,-0.5959],[160,-0.5873],[164,-0.5826],[168,-0.582],[169,-0.583],[170,-0.5863],[171,-0.5914],[172,-0.5984],[173,-0.6068],[175,-0.6276],[178,-0.6655],[184,-0.7489],[185,-0.765],[186,-0.7839],[187,-0.8055],[192,-0.9252],[193,-0.9459],[194,-0.9627],[195,-0.9753],[197,-0.994],[199,-1.0073],[202,-1.0195],[208,-1.0274],[212,-1.0263],[214,-1.0189],[216,-1.0057],[219,-0.9776],[223,-0.9309],[225,-0.9038],[228,-0.8556],[234,-0.7517],[236,-0.7203],[238,-0.6932],[239,-0.6817],[240,-0.6719],[241,-0.6636],[243,-0.6511],[248,-0.6323],[249,-0.6273],[251,-0.6139],[252,-0.6062],[254,-0.5889],[255,-0.5828],[256,-0.5799],[257,-0.5795],[258,-0.5803],[259,-0.5824],[260,-0.5856],[262,-0.5957],[264,-0.6104],[266,-0.6298],[268,-0.654],[269,-0.6679],[270,-0.6871],[271,-0.714],[272,-0.7467],[275,-0.8577],[276,-0.8919],[277,-0.9214],[278,-0.9438],[279,-0.9571],[282,-0.977],[285,-0.9895],[290,-0.9984],[296,-0.9988],[299,-0.9919],[301,-0.9824],[303,-0.9688],[305,-0.9509],[307,-0.9283],[308,-0.9152],[309,-0.8997],[310,-0.8808],[311,-0.8592],[313,-0.8102],[317,-0.7057],[318,-0.6817],[319,-0.6598],[320,-0.6405],[321,-0.6245],[322,-0.6123],[323,-0.6045],[324,-0.6018]]
       }
       calf_r_fk_ctrl.Location.Y {
        type=MovieSceneScriptingFloatChannel
        default=0
        thinned_from=325
        keys [[0,0.0495],[3,0.0498],[5,0.0506],[6,0.0514],[7,0.0524],[8,0.0537],[9,0.0555],[10,0.0577],[12,0.0628],[14,0.0682],[15,0.0707],[16,0.0729],[17,0.0747],[19,0.0774],[21,0.0796],[23,0.0812],[25,0.0821],[27,0.0824],[36,0.0822],[39,0.0815],[42,0.0799],[43,0.0792],[44,0.0783],[45,0.077],[46,0.0754],[48,0.0716],[52,0.0633],[54,0.0596],[55,0.0582],[56,0.057],[59,0.0545],[62,0.0527],[66,0.0512],[72,0.0504],[78,0.0506],[81,0.0513],[84,0.0528],[86,0.0543],[89,0.0574],[93,0.0627],[99,0.0714],[106,0.0801],[108,0.082],[109,0.0827],[110,0.0831],[112,0.0832],[115,0.0825],[119,0.0806],[127,0.0746],[129,0.0727],[137,0.0625],[139,0.0603],[146,0.0547],[152,0.051],[157,0.049],[163,0.0479],[169,0.0479],[171,0.0486],[173,0.0499],[176,0.0526],[184,0.0616],[185,0.0629],[186,0.0645],[188,0.0682],[192,0.0761],[193,0.0778],[194,0.0792],[195,0.0802],[197,0.0817],[200,0.0832],[205,0.0843],[212,0.0844],[214,0.0838],[217,0.082],[223,0.0766],[226,0.0731],[235,0.0605],[238,0.057],[240,0.0552],[242,0.054],[249,0.0516],[252,0.0498],[254,0.0484],[255,0.0479],[256,0.0477],[258,0.0477],[261,0.0485],[264,0.0502],[267,0.0527],[269,0.0549],[270,0.0565],[271,0.0587],[272,0.0614],[275,0.0705],[276,0.0733],[277,0.0758],[278,0.0776],[279,0.0787],[282,0.0803],[286,0.0816],[295,0.0822],[298,0.0818],[301,0.0808],[304,0.079],[307,0.0763],[308,0.0753],[309,0.074],[310,0.0724],[312,0.0687],[317,0.058],[319,0.0542],[320,0.0527],[321,0.0513],[322,0.0503],[323,0.0497],[324,0.0495]]
       }
       calf_r_fk_ctrl.Location.Z {type=MovieSceneScriptingFloatChannel, default=0, thinned_from=325, keys [[0,0.0006],[324,0.0006]]}
       calf_r_fk_ctrl.Rotation.X {type=MovieSceneScriptingFloatChannel, default=0.0561, thinned_from=325, keys [[0,0.0055],[324,0.0055]]}
       calf_r_fk_ctrl.Rotation.Y {type=MovieSceneScriptingFloatChannel, default=1.1982, thinned_from=325, keys [[0,0.0049],[324,0.0049]]}
       calf_r_fk_ctrl.Rotation.Z {type=MovieSceneScriptingFloatChannel, default=0.2616, thinned_from=325, keys [[0,-4.9607],[324,-4.9607]]}
       calf_r_fk_ctrl.Scale.X {type=MovieSceneScriptingFloatChannel, default=1, const [1,325]}
       calf_r_fk_ctrl.Scale.Y {type=MovieSceneScriptingFloatChannel, default=1, const [1,325]}
       calf_r_fk_ctrl.Scale.Z {type=MovieSceneScriptingFloatChannel, default=1, const [1,325]}
       foot_r_fk_ctrl.Location.X {
        type=MovieSceneScriptingFloatChannel
        default=0
        thinned_from=325
        keys [[0,-0.6372],[2,-0.6386],[3,-0.6412],[4,-0.6458],[5,-0.6528],[6,-0.6628],[7,-0.6763],[8,-0.6939],[9,-0.7171],[10,-0.746],[11,-0.7791],[14,-0.8866],[15,-0.9198],[16,-0.949],[17,-0.9725],[19,-1.0089],[21,-1.0376],[23,-1.0584],[25,-1.071],[27,-1.0753],[35,-1.0735],[38,-1.0669],[40,-1.0569],[42,-1.0421],[43,-1.0328],[44,-1.0203],[45,-1.0031],[46,-0.9821],[47,-0.958],[52,-0.8204],[53,-0.7951],[54,-0.7723],[55,-0.753],[56,-0.7378],[58,-0.7145],[60,-0.6955],[62,-0.6804],[64,-0.6688],[66,-0.6603],[69,-0.6525],[73,-0.6493],[77,-0.6507],[79,-0.6544],[81,-0.6616],[83,-0.6732],[84,-0.681],[85,-0.6904],[86,-0.7015],[88,-0.7278],[90,-0.759],[92,-0.7941],[99,-0.9291],[105,-1.0293],[107,-1.0581],[108,-1.0696],[109,-1.0785],[110,-1.0842],[111,-1.0862],[113,-1.0837],[116,-1.0714],[119,-1.0509],[125,-0.9935],[128,-0.9598],[130,-0.9309],[137,-0.8103],[139,-0.7808],[143,-0.7358],[147,-0.6972],[151,-0.665],[154,-0.6456],[157,-0.6307],[159,-0.6238],[162,-0.6179],[168,-0.6155],[169,-0.6167],[170,-0.6202],[171,-0.6259],[172,-0.6335],[173,-0.6427],[175,-0.6654],[177,-0.6923],[184,-0.7981],[185,-0.8157],[186,-0.8364],[187,-0.86],[192,-0.9909],[193,-1.0136],[194,-1.0319],[195,-1.0457],[197,-1.0662],[199,-1.0807],[202,-1.0941],[208,-1.1027],[212,-1.1015],[214,-1.0934],[216,-1.079],[219,-1.0482],[223,-0.9971],[225,-0.9675],[228,-0.9148],[234,-0.8011],[236,-0.7669],[238,-0.7372],[239,-0.7247],[240,-0.7139],[241,-0.7048],[243,-0.6911],[248,-0.6706],[249,-0.6651],[251,-0.6505],[252,-0.642],[254,-0.6231],[255,-0.6165],[256,-0.6132],[257,-0.6128],[258,-0.6137],[259,-0.616],[260,-0.6195],[261,-0.6244],[263,-0.6379],[265,-0.6566],[267,-0.6805],[269,-0.7096],[270,-0.7305],[271,-0.76],[272,-0.7957],[275,-0.917],[276,-0.9545],[277,-0.9867],[278,-1.0112],[279,-1.0258],[282,-1.0476],[285,-1.0613],[290,-1.071],[296,-1.0714],[299,-1.0638],[301,-1.0535],[303,-1.0387],[305,-1.019],[307,-0.9943],[308,-0.98],[309,-0.963],[310,-0.9424],[311,-0.9187],[313,-0.8651],[317,-0.7508],[318,-0.7246],[319,-0.7006],[320,-0.6795],[321,-0.662],[322,-0.6487],[323,-0.6402],[324,-0.6372]]
       }
       foot_r_fk_ctrl.Location.Y {
        type=MovieSceneScriptingFloatChannel
        default=-0.0329
        thinned_from=325
        keys [[0,-0.0029],[44,-0.0046],[90,-0.0034],[186,-0.0037],[270,-0.0033],[308,-0.0044],[324,-0.0029]]
       }
       foot_r_fk_ctrl.Location.Z {type=MovieSceneScriptingFloatChannel, default=-0.9075, thinned_from=325, keys [[0,0.0007],[324,0.0007]]}
       foot_r_fk_ctrl.Rotation.X {
        type=MovieSceneScriptingFloatChannel
        default=-1.1648
        thinned_from=325
        keys [[0,0.3468],[29,0.3482],[31,0.3464],[32,0.3442],[36,0.331],[38,0.3254],[39,0.3236],[42,0.3217],[82,0.3202],[84,0.3212],[85,0.3226],[86,0.3251],[87,0.3286],[93,0.356],[95,0.363],[97,0.3682],[98,0.37],[100,0.3712],[181,0.3662],[188,0.3621],[189,0.3609],[197,0.3414],[200,0.3367],[204,0.3337],[247,0.333],[250,0.3341],[252,0.3373],[254,0.3448],[256,0.3487],[259,0.3508],[324,0.3468]]
       }
       foot_r_fk_ctrl.Rotation.Y {
        type=MovieSceneScriptingFloatChannel
        default=0.1067
        thinned_from=325
        keys [[0,3.2655],[29,3.2851],[31,3.2627],[32,3.2355],[35,3.1064],[38,2.9967],[39,2.9737],[41,2.9541],[82,2.9291],[84,2.9419],[85,2.9597],[86,2.9908],[87,3.0346],[89,3.1464],[92,3.3266],[94,3.4278],[96,3.5057],[97,3.5355],[98,3.5572],[99,3.5693],[101,3.5722],[180,3.514],[188,3.4594],[189,3.4445],[197,3.2044],[199,3.1604],[202,3.1214],[206,3.1015],[247,3.0954],[250,3.1075],[251,3.123],[252,3.1469],[254,3.241],[256,3.2895],[258,3.3115],[324,3.2655]]
       }
       foot_r_fk_ctrl.Rotation.Z {
        type=MovieSceneScriptingFloatChannel
        default=-0.1036
        thinned_from=325
        keys [[0,4.6942],[27,4.8303],[31,4.8302],[33,4.7988],[34,4.7696],[35,4.7245],[37,4.5513],[38,4.4902],[39,4.4532],[43,4.3921],[84,4.3024],[86,4.3162],[87,4.3361],[88,4.3677],[89,4.4132],[90,4.4732],[91,4.5555],[92,4.6495],[93,4.7316],[94,4.7935],[95,4.8439],[96,4.8837],[97,4.9124],[98,4.9284],[170,4.8092],[183,4.8842],[187,4.8799],[188,4.8873],[189,4.906],[190,4.9453],[191,4.9957],[192,5.036],[194,5.0721],[244,4.9027],[245,4.8901],[246,4.8659],[247,4.8226],[248,4.7372],[249,4.6691],[250,4.6538],[252,4.6498],[275,4.8401],[295,4.8156],[324,4.6942]]
       }
       foot_r_fk_ctrl.Scale.X {type=MovieSceneScriptingFloatChannel, default=1, const [1,325]}
       foot_r_fk_ctrl.Scale.Y {type=MovieSceneScriptingFloatChannel, default=1, const [1,325]}
       foot_r_fk_ctrl.Scale.Z {type=MovieSceneScriptingFloatChannel, default=1, const [1,325]}
       ball_r_fk_ctrl.Location.X {type=MovieSceneScriptingFloatChannel, default=0, const [0,325]}
       ball_r_fk_ctrl.Location.Y {type=MovieSceneScriptingFloatChannel, default=0, const [0,325]}
       ball_r_fk_ctrl.Location.Z {type=MovieSceneScriptingFloatChannel, default=0, const [0,325]}
       ball_r_fk_ctrl.Rotation.X {type=MovieSceneScriptingFloatChannel, default=0, const [0,325]}
       ball_r_fk_ctrl.Rotation.Y {type=MovieSceneScriptingFloatChannel, default=0, const [0,325]}
       ball_r_fk_ctrl.Rotation.Z {type=MovieSceneScriptingFloatChannel, default=0, const [0,325]}
       ball_r_fk_ctrl.Scale.X {type=MovieSceneScriptingFloatChannel, default=1, const [1,325]}
       ball_r_fk_ctrl.Scale.Y {type=MovieSceneScriptingFloatChannel, default=1, const [1,325]}
       ball_r_fk_ctrl.Scale.Z {type=MovieSceneScriptingFloatChannel, default=1, const [1,325]}
       foot_r_ik_ctrl.Location.X {
        type=MovieSceneScriptingFloatChannel
        default=0
        thinned_from=325
        keys [[0,-7.0484],[16,-7.0474],[17,-7.0057],[22,-7.0886],[23,-7.0495],[38,-7.0499],[39,-7.094],[46,-7.0097],[47,-7.0477],[100,-7.0476],[101,-7.0059],[106,-7.0886],[107,-7.0475],[122,-7.0483],[123,-7.0898],[130,-7.0139],[131,-7.048],[201,-7.0486],[202,-7.0596],[268,-7.0474],[269,-7.0059],[274,-7.0884],[275,-7.0495],[283,-7.057],[286,-7.0482],[290,-7.0495],[291,-7.0899],[298,-7.0157],[299,-7.0483],[324,-7.0526]]
       }
       foot_r_ik_ctrl.Location.Y {
        type=MovieSceneScriptingFloatChannel
        default=0
        thinned_from=325
        keys [[0,5.7087],[1,5.7024],[16,5.7057],[17,5.5871],[22,5.822],[23,5.7118],[28,5.7025],[38,5.7125],[39,5.8243],[43,5.728],[45,5.657],[46,5.6119],[47,5.7069],[68,5.7079],[69,5.7239],[70,5.7085],[71,5.722],[72,5.7083],[73,5.7203],[74,5.7081],[75,5.718],[76,5.7089],[78,5.7228],[79,5.7083],[83,5.7207],[84,5.7086],[100,5.7062],[101,5.5877],[106,5.8214],[107,5.7062],[110,5.7103],[115,5.7144],[116,5.7073],[117,5.7171],[118,5.7083],[119,5.7217],[122,5.7084],[123,5.825],[127,5.7281],[129,5.6565],[130,5.6111],[131,5.7076],[152,5.7081],[153,5.724],[154,5.7082],[155,5.7218],[156,5.7084],[157,5.7201],[158,5.7082],[159,5.719],[160,5.7088],[162,5.7229],[163,5.7081],[168,5.7204],[169,5.7149],[170,5.696],[171,5.7084],[201,5.7085],[202,5.7212],[203,5.7086],[235,5.7084],[236,5.7231],[237,5.7087],[238,5.7218],[239,5.7086],[240,5.72],[241,5.7087],[242,5.7183],[268,5.7059],[269,5.5878],[274,5.8215],[275,5.7118],[280,5.7013],[283,5.7155],[284,5.7095],[285,5.7174],[286,5.7082],[287,5.7218],[288,5.7103],[290,5.7119],[291,5.8255],[295,5.7284],[298,5.6157],[299,5.7083],[320,5.7087],[321,5.7226],[322,5.7083],[324,5.7174]]
       }
       foot_r_ik_ctrl.Location.Z {
        type=MovieSceneScriptingFloatChannel
        default=0
        thinned_from=325
        keys [[0,0.0002],[22,0.0003],[32,0.0001],[38,0.0001],[39,-0.0006],[43,0],[46,0.001],[47,0.0002],[68,0.0002],[69,0],[70,0.0002],[71,0],[72,0.0002],[73,0],[78,0],[79,0.0002],[82,0.0002],[83,0.0005],[84,0.0002],[100,0.0002],[101,0],[106,0.0004],[107,0.0001],[122,0.0002],[123,0.0004],[130,0.0001],[166,0.0002],[168,0],[185,0.0002],[186,0],[201,0.0001],[202,0],[209,0.0001],[210,0],[274,0.0003],[283,0],[284,0.0001],[290,0.0002],[294,0.0002],[295,0],[324,0.0001]]
       }
       foot_r_ik_ctrl.Rotation.X {
        type=MovieSceneScriptingFloatChannel
        default=0
        thinned_from=325
        keys [[0,-0.0001],[1,0.0045],[2,0.0001],[3,0.0006],[4,-0.0003],[5,-0.0002],[6,-0.0004],[8,0.0003],[9,0.0005],[10,0],[12,0.0005],[13,0.0003],[14,-0.0004],[15,0.0017],[16,0.002],[17,0.0766],[18,0.0478],[19,0.0175],[20,-0.0132],[21,-0.0432],[22,-0.0715],[23,-0.0024],[24,-0.0019],[25,-0.0002],[26,0],[27,0.0044],[28,0.0044],[29,0.0012],[31,0.0004],[32,-0.0005],[33,0.0024],[34,0.0002],[35,0.0001],[36,0.0005],[37,-0.0008],[38,-0.0024],[39,-0.0739],[40,-0.0603],[41,-0.0455],[42,-0.0295],[43,-0.0124],[44,0.008],[45,0.033],[46,0.0619],[47,0.001],[48,0.0009],[49,-0.0006],[50,-0.0001],[52,-0.0001],[53,0.0004],[54,0.0004],[55,-0.0001],[56,0],[57,0.0005],[58,-0.0003],[59,-0.0004],[60,0.0001],[61,-0.0003],[63,-0.0003],[64,0.0001],[65,-0.0002],[66,-0.0002],[67,0.0003],[68,0.0005],[69,-0.0115],[70,0],[71,-0.01],[72,0.0002],[73,-0.0087],[74,0.0014],[75,-0.0071],[76,-0.0002],[77,-0.0051],[78,-0.0106],[79,0.0002],[80,-0.0039],[81,-0.0051],[82,-0.0073],[83,-0.0061],[84,0],[85,0.0002],[87,-0.0003],[88,-0.0001],[89,-0.0003],[90,0],[91,-0.0001],[92,0.0005],[93,0.0002],[94,-0.0002],[95,0.0001],[96,-0.0004],[98,0.0007],[99,0.0016],[100,0.0019],[101,0.0766],[102,0.0478],[103,0.0175],[104,-0.0131],[105,-0.0432],[106,-0.0715],[107,0.0017],[108,-0.0003],[109,0.0012],[110,-0.0013],[111,0.0042],[112,0.0042],[114,-0.0025],[115,-0.0048],[116,0.0009],[117,-0.0064],[118,0.0002],[119,-0.0097],[120,-0.0014],[122,-0.0001],[123,-0.0737],[124,-0.0602],[125,-0.0454],[126,-0.0295],[127,-0.0124],[128,0.008],[129,0.033],[130,0.0618],[131,0.0009],[132,0.0009],[133,0.0006],[134,-0.0001],[137,-0.0003],[138,0.0003],[139,-0.0007],[140,0],[141,0],[142,0.0002],[143,-0.0003],[144,0.0002],[146,0.0003],[147,-0.0002],[148,-0.0002],[149,0],[150,0.0007],[152,0.0003],[153,-0.0114],[154,0.0003],[155,-0.0098],[156,0.0001],[157,-0.0086],[158,0.0003],[159,-0.0077],[160,-0.0002],[161,-0.0057],[162,-0.0106],[163,0.0003],[164,-0.0035],[166,-0.008],[167,-0.0087],[168,-0.0087],[169,-0.0047],[170,0.0093],[171,0.0001],[173,-0.0005],[175,0.0002],[176,-0.0004],[177,0.0008],[178,0.0005],[179,0],[180,0],[181,0.0011],[182,0.0006],[183,0.0028],[184,-0.0002],[185,-0.0001],[186,0.0001],[191,-0.0002],[192,0],[201,0],[202,-0.0113],[203,0],[209,0],[210,0.003],[211,0],[234,0],[235,0.0001],[236,-0.0108],[237,0],[238,-0.0097],[239,0],[240,-0.0084],[241,0],[242,-0.0071],[243,0],[246,-0.0001],[248,0.0001],[250,-0.0001],[251,0.0002],[252,-0.0001],[253,0.0001],[255,0.0003],[256,-0.0001],[257,0.0001],[258,-0.0002],[261,0.0003],[263,0.0002],[264,-0.0004],[266,0.0006],[267,-0.0012],[268,0.0019],[269,0.0766],[270,0.0478],[271,0.0175],[272,-0.0132],[273,-0.0432],[274,-0.0715],[275,-0.0024],[276,-0.0019],[277,-0.0002],[278,0],[279,0.0054],[280,0.0048],[281,-0.0013],[282,-0.0039],[283,-0.0067],[284,-0.0005],[285,-0.0069],[286,0.0002],[287,-0.0103],[288,-0.0014],[289,-0.0028],[290,-0.0025],[291,-0.0738],[292,-0.0602],[293,-0.0454],[294,-0.0259],[295,-0.0052],[296,0.0075],[297,0.0314],[298,0.0585],[299,0],[300,0.0032],[301,0.0001],[302,0.0003],[303,-0.0006],[305,0.0008],[306,0.0001],[307,0.0003],[308,0],[309,0.0003],[310,-0.0002],[312,-0.0002],[313,0.0004],[315,0.0002],[317,0.0003],[318,-0.0003],[319,0.0001],[320,-0.0001],[321,-0.0107],[322,0.0002],[323,-0.0048],[324,-0.0067]]
       }
       foot_r_ik_ctrl.Rotation.Y {
        type=MovieSceneScriptingFloatChannel
        default=0
        thinned_from=325
        keys [[0,0],[1,-0.0003],[2,0],[14,0],[16,-0.0001],[17,-0.0051],[22,0.0048],[23,0.0002],[26,0],[27,-0.0004],[28,-0.0004],[29,0.0047],[30,0],[32,0],[33,-0.0002],[38,0.0001],[39,0.0019],[46,-0.001],[47,0],[68,0],[69,0.0001],[70,0],[71,0.0002],[72,0],[73,0.0002],[74,0],[75,0.0002],[76,0],[78,0.0004],[79,0],[81,0.0007],[82,0.0026],[83,0.007],[84,0],[100,-0.0001],[101,-0.0052],[105,0.0029],[106,0.0048],[107,-0.0001],[108,0],[109,-0.0001],[110,0.0001],[111,-0.001],[112,-0.0012],[115,-0.001],[116,0],[117,0.0002],[118,0],[119,0.0003],[122,0],[123,0.0049],[126,0.002],[127,0.0008],[128,-0.0005],[129,-0.0022],[130,-0.0041],[131,-0.0001],[152,0],[153,0.0005],[154,0],[155,0.0004],[156,0],[157,0.0004],[158,0],[159,0.0004],[160,0],[162,0.0006],[163,0],[167,0.0005],[168,0.0005],[169,0.0003],[170,-0.0005],[171,0],[182,0],[183,-0.0004],[184,0.0001],[185,0.0003],[186,-0.001],[188,-0.0003],[191,-0.0005],[192,0],[193,-0.0001],[194,0.0001],[195,-0.0003],[196,0],[197,-0.0001],[198,0.0002],[201,-0.0002],[202,-0.0037],[203,0],[204,-0.0001],[205,0.0002],[206,0.0002],[208,-0.0002],[209,-0.0001],[210,-0.0005],[211,-0.0001],[212,0.0001],[218,0],[219,-0.0001],[221,0.0002],[222,-0.0001],[223,0.0002],[228,-0.0001],[230,0],[233,-0.0001],[234,0.0002],[235,-0.0001],[236,0.0003],[237,0.0001],[238,0.0005],[239,0.0001],[240,0.0004],[241,0.0001],[242,0.0005],[243,-0.0001],[244,-0.0001],[245,0.0002],[246,0.0002],[248,-0.0002],[250,0.0001],[251,-0.0002],[252,0.0001],[253,-0.0001],[255,-0.0001],[256,0],[266,0],[267,0.0001],[268,-0.0001],[269,-0.0051],[274,0.0048],[275,0.0002],[278,0],[280,-0.0015],[281,-0.002],[283,-0.0035],[284,0],[285,-0.0007],[286,0],[287,-0.0005],[288,0.0001],[290,0.0002],[291,0.0049],[293,0.003],[295,0.0032],[296,-0.0005],[297,-0.0021],[298,-0.0039],[299,0],[300,-0.0002],[301,0],[320,0],[321,-0.0004],[322,0],[324,-0.0003]]
       }
       foot_r_ik_ctrl.Rotation.Z {
        type=MovieSceneScriptingFloatChannel
        default=0
        thinned_from=325
        keys [[0,23.2445],[16,23.2443],[17,23.1722],[22,23.3122],[23,23.2441],[38,23.2472],[39,23.3133],[46,23.1873],[47,23.2429],[100,23.2471],[101,23.1722],[106,23.3122],[107,23.2448],[122,23.2428],[123,23.3143],[130,23.1862],[131,23.2467],[268,23.2449],[269,23.1722],[274,23.3122],[275,23.2442],[290,23.2437],[291,23.3143],[298,23.1893],[299,23.243],[324,23.2445]]
       }
       foot_r_ik_ctrl.Scale.X {type=MovieSceneScriptingFloatChannel, default=1, const [1,325]}
       foot_r_ik_ctrl.Scale.Y {type=MovieSceneScriptingFloatChannel, default=1, const [1,325]}
       foot_r_ik_ctrl.Scale.Z {type=MovieSceneScriptingFloatChannel, default=1, const [1,325]}
       ball_r_ik_ctrl.Location.X {
        type=MovieSceneScriptingFloatChannel
        default=0
        thinned_from=325
        keys [[0,0],[144,0],[145,-0.0002],[146,0],[227,0],[228,0.0001],[230,-0.0001],[253,0],[254,0.0001],[310,0],[311,-0.0002],[324,0]]
       }
       ball_r_ik_ctrl.Location.Y {type=MovieSceneScriptingFloatChannel, default=0, thinned_from=325, keys [[0,0],[140,0],[141,-0.0001],[324,0]]}
       ball_r_ik_ctrl.Location.Z {type=MovieSceneScriptingFloatChannel, default=0, thinned_from=325, keys [[0,0],[324,0]]}
       ball_r_ik_ctrl.Rotation.X {type=MovieSceneScriptingFloatChannel, default=0, thinned_from=325, keys [[0,0],[324,0]]}
       ball_r_ik_ctrl.Rotation.Y {type=MovieSceneScriptingFloatChannel, default=0, thinned_from=325, keys [[0,0],[324,0]]}
       ball_r_ik_ctrl.Rotation.Z {
        type=MovieSceneScriptingFloatChannel
        default=0
        thinned_from=325
        keys [[0,0],[140,0],[141,-0.0004],[142,0],[144,0],[145,0.0009],[146,0],[253,0],[254,-0.0002],[310,0],[311,0.0007],[312,0],[324,0]]
       }
       ball_r_ik_ctrl.Scale.X {type=MovieSceneScriptingFloatChannel, default=1, const [1,325]}
       ball_r_ik_ctrl.Scale.Y {type=MovieSceneScriptingFloatChannel, default=1, const [1,325]}
       ball_r_ik_ctrl.Scale.Z {type=MovieSceneScriptingFloatChannel, default=1, const [1,325]}
       foot_bk1_r_ctrl.Location.X {type=MovieSceneScriptingFloatChannel, default=0, const [0,325]}
       foot_bk1_r_ctrl.Location.Y {type=MovieSceneScriptingFloatChannel, default=0, const [0,325]}
       foot_bk1_r_ctrl.Location.Z {type=MovieSceneScriptingFloatChannel, default=0, const [0,325]}
       foot_bk1_r_ctrl.Rotation.X {type=MovieSceneScriptingFloatChannel, default=0, const [0,325]}
       foot_bk1_r_ctrl.Rotation.Y {type=MovieSceneScriptingFloatChannel, default=0, const [0,325]}
       foot_bk1_r_ctrl.Rotation.Z {type=MovieSceneScriptingFloatChannel, default=0, const [0,325]}
       foot_bk1_r_ctrl.Scale.X {type=MovieSceneScriptingFloatChannel, default=1, const [1,325]}
       foot_bk1_r_ctrl.Scale.Y {type=MovieSceneScriptingFloatChannel, default=1, const [1,325]}
       foot_bk1_r_ctrl.Scale.Z {type=MovieSceneScriptingFloatChannel, default=1, const [1,325]}
       foot_roll_r_ctrl.Location.X {type=MovieSceneScriptingFloatChannel, default=0, const [0,325]}
       foot_roll_r_ctrl.Location.Y {type=MovieSceneScriptingFloatChannel, default=0, const [0,325]}
       foot_roll_r_ctrl.Location.Z {type=MovieSceneScriptingFloatChannel, default=0, const [0,325]}
       foot_roll_r_ctrl.Rotation.X {type=MovieSceneScriptingFloatChannel, default=0, const [0,325]}
       foot_roll_r_ctrl.Rotation.Y {type=MovieSceneScriptingFloatChannel, default=0, const [0,325]}
       foot_roll_r_ctrl.Rotation.Z {type=MovieSceneScriptingFloatChannel, default=0, const [0,325]}
       foot_roll_r_ctrl.Scale.X {type=MovieSceneScriptingFloatChannel, default=1, const [1,325]}
       foot_roll_r_ctrl.Scale.Y {type=MovieSceneScriptingFloatChannel, default=1, const [1,325]}
       foot_roll_r_ctrl.Scale.Z {type=MovieSceneScriptingFloatChannel, default=1, const [1,325]}
       roll_blend_r_ctrl {type=MovieSceneScriptingFloatChannel, default=1, const [1,325]}
       tip_r_ctrl.Location.X {type=MovieSceneScriptingFloatChannel, default=0, const [0,325]}
       tip_r_ctrl.Location.Y {type=MovieSceneScriptingFloatChannel, default=0, const [0,325]}
       tip_r_ctrl.Location.Z {type=MovieSceneScriptingFloatChannel, default=0, const [0,325]}
       tip_r_ctrl.Rotation.X {type=MovieSceneScriptingFloatChannel, default=0, const [0,325]}
       tip_r_ctrl.Rotation.Y {type=MovieSceneScriptingFloatChannel, default=0, const [0,325]}
       tip_r_ctrl.Rotation.Z {type=MovieSceneScriptingFloatChannel, default=0, const [0,325]}
       tip_r_ctrl.Scale.X {type=MovieSceneScriptingFloatChannel, default=1, const [1,325]}
       tip_r_ctrl.Scale.Y {type=MovieSceneScriptingFloatChannel, default=1, const [1,325]}
       tip_r_ctrl.Scale.Z {type=MovieSceneScriptingFloatChannel, default=1, const [1,325]}
       heel_r_ctrl.Location.X {type=MovieSceneScriptingFloatChannel, default=0, const [0,325]}
       heel_r_ctrl.Location.Y {type=MovieSceneScriptingFloatChannel, default=0, const [0,325]}
       heel_r_ctrl.Location.Z {type=MovieSceneScriptingFloatChannel, default=0, const [0,325]}
       heel_r_ctrl.Rotation.X {type=MovieSceneScriptingFloatChannel, default=0, const [0,325]}
       heel_r_ctrl.Rotation.Y {type=MovieSceneScriptingFloatChannel, default=0, const [0,325]}
       heel_r_ctrl.Rotation.Z {type=MovieSceneScriptingFloatChannel, default=0, const [0,325]}
       heel_r_ctrl.Scale.X {type=MovieSceneScriptingFloatChannel, default=1, const [1,325]}
       heel_r_ctrl.Scale.Y {type=MovieSceneScriptingFloatChannel, default=1, const [1,325]}
       heel_r_ctrl.Scale.Z {type=MovieSceneScriptingFloatChannel, default=1, const [1,325]}
       leg_r_softness {type=MovieSceneScriptingFloatChannel, default=0, const [0,325]}
       leg_r_pv_ik_ctrl.Location.X {
        type=MovieSceneScriptingFloatChannel
        default=0
        thinned_from=325
        keys [[0,-0.0589],[23,-0.062],[31,-0.0623],[32,-0.062],[33,-0.0615],[34,-0.0607],[35,-0.0595],[36,-0.0572],[37,-0.0546],[38,-0.0528],[39,-0.0516],[42,-0.0501],[54,-0.048],[56,-0.048],[61,-0.0476],[62,-0.0478],[64,-0.0474],[83,-0.047],[85,-0.0474],[86,-0.0478],[87,-0.0487],[88,-0.0498],[89,-0.0514],[90,-0.0532],[92,-0.058],[93,-0.06],[94,-0.0614],[95,-0.0626],[96,-0.0634],[97,-0.064],[98,-0.0644],[106,-0.0644],[117,-0.0646],[138,-0.0629],[140,-0.0626],[141,-0.0623],[142,-0.0625],[144,-0.0624],[145,-0.063],[146,-0.0623],[171,-0.0618],[179,-0.0631],[180,-0.0634],[181,-0.0633],[187,-0.0634],[188,-0.0636],[190,-0.0649],[191,-0.0661],[192,-0.0671],[193,-0.0677],[194,-0.0681],[197,-0.0683],[205,-0.0676],[206,-0.0673],[208,-0.0673],[213,-0.0669],[223,-0.0653],[231,-0.0649],[233,-0.0645],[236,-0.0646],[243,-0.0644],[245,-0.0639],[246,-0.0632],[247,-0.0621],[248,-0.0598],[249,-0.058],[250,-0.0576],[251,-0.0575],[252,-0.0576],[255,-0.0591],[262,-0.0608],[263,-0.0613],[264,-0.0611],[272,-0.0622],[285,-0.0626],[295,-0.062],[296,-0.0616],[310,-0.0599],[311,-0.0603],[312,-0.0596],[314,-0.0596],[324,-0.0589]]
       }
       leg_r_pv_ik_ctrl.Location.Y {type=MovieSceneScriptingFloatChannel, default=0, thinned_from=325, keys [[0,-1.9484],[324,-1.9484]]}
       leg_r_pv_ik_ctrl.Location.Z {type=MovieSceneScriptingFloatChannel, default=0, thinned_from=325, keys [[0,-1.7958],[324,-1.7958]]}
       leg_r_pv_ik_ctrl.Rotation.X {type=MovieSceneScriptingFloatChannel, default=0, thinned_from=325, keys [[0,2.605],[324,2.605]]}
       leg_r_pv_ik_ctrl.Rotation.Y {
        type=MovieSceneScriptingFloatChannel
        default=0
        thinned_from=325
        keys [[0,-0.0136],[35,-0.0136],[88,-0.013],[246,-0.0139],[254,-0.0135],[324,-0.0136]]
       }
       leg_r_pv_ik_ctrl.Rotation.Z {
        type=MovieSceneScriptingFloatChannel
        default=0
        thinned_from=325
        keys [[0,0.0722],[2,0.0722],[23,0.0766],[24,0.0766],[25,0.0769],[31,0.0769],[32,0.0765],[33,0.0758],[34,0.0747],[35,0.073],[36,0.0697],[37,0.0661],[38,0.0635],[39,0.0618],[41,0.0603],[44,0.0588],[54,0.0567],[56,0.0567],[61,0.0562],[62,0.0564],[64,0.0559],[70,0.0556],[74,0.0555],[83,0.0552],[85,0.0558],[86,0.0565],[87,0.0576],[88,0.0593],[89,0.0615],[90,0.0641],[91,0.0673],[92,0.0708],[93,0.0737],[94,0.0758],[95,0.0774],[96,0.0786],[97,0.0794],[98,0.0799],[106,0.0799],[117,0.0802],[126,0.0795],[127,0.0793],[128,0.0793],[137,0.0778],[138,0.0778],[140,0.0775],[141,0.0769],[142,0.0773],[144,0.0772],[145,0.0779],[146,0.0769],[148,0.0769],[171,0.0763],[179,0.0781],[180,0.0786],[181,0.0784],[184,0.0786],[187,0.0785],[188,0.0787],[190,0.0806],[191,0.0824],[192,0.0838],[193,0.0847],[194,0.0852],[195,0.0854],[197,0.0854],[204,0.0844],[205,0.0844],[206,0.0841],[208,0.0841],[209,0.0838],[210,0.0838],[212,0.0834],[213,0.0834],[223,0.0813],[224,0.0813],[226,0.0809],[227,0.0809],[228,0.0807],[231,0.0807],[233,0.0802],[236,0.0803],[243,0.08],[245,0.0792],[246,0.0783],[247,0.0766],[248,0.0734],[249,0.0709],[250,0.0703],[251,0.0701],[252,0.0703],[255,0.0724],[257,0.0733],[262,0.0748],[263,0.0755],[264,0.0753],[272,0.0769],[273,0.0769],[274,0.0771],[275,0.0771],[276,0.0774],[285,0.0774],[286,0.0771],[291,0.0768],[292,0.0766],[295,0.0766],[296,0.076],[310,0.0735],[311,0.0741],[312,0.0732],[314,0.0731],[315,0.0728],[316,0.0728],[318,0.0725],[324,0.0722]]
       }
       leg_r_pv_ik_ctrl.Scale.X {type=MovieSceneScriptingFloatChannel, default=1, const [1,325]}
       leg_r_pv_ik_ctrl.Scale.Y {type=MovieSceneScriptingFloatChannel, default=1, const [1,325]}
       leg_r_pv_ik_ctrl.Scale.Z {type=MovieSceneScriptingFloatChannel, default=1, const [1,325]}
       upperarm_l_fk_ctrl.Location.X {type=MovieSceneScriptingFloatChannel, default=0, thinned_from=325, keys [[0,0],[324,0]]}
       upperarm_l_fk_ctrl.Location.Y {type=MovieSceneScriptingFloatChannel, default=0, thinned_from=325, keys [[0,0],[324,0]]}
       upperarm_l_fk_ctrl.Location.Z {type=MovieSceneScriptingFloatChannel, default=0, thinned_from=325, keys [[0,0],[324,0]]}
       upperarm_l_fk_ctrl.Rotation.X {
        type=MovieSceneScriptingFloatChannel
        default=0
        thinned_from=325
        keys [[0,-10.5216],[3,-10.5585],[7,-10.7046],[16,-11.2097],[23,-11.4509],[25,-11.4711],[27,-11.4269],[29,-11.3577],[32,-11.3196],[33,-11.2737],[34,-11.1937],[35,-11.0253],[36,-10.5126],[37,-9.9284],[38,-9.5794],[39,-9.3908],[41,-9.2662],[48,-9.0672],[55,-8.7203],[71,-8.2277],[72,-8.2118],[73,-8.1771],[79,-8.1147],[84,-8.092],[85,-8.06],[86,-8.0517],[87,-8.0703],[88,-8.1368],[89,-8.2725],[90,-8.5025],[91,-8.8652],[92,-9.2975],[93,-9.6166],[94,-9.7701],[95,-9.8704],[96,-9.9111],[97,-9.9048],[98,-9.8452],[100,-9.6673],[101,-9.6074],[102,-9.5793],[103,-9.5899],[104,-9.6378],[109,-10.0204],[117,-10.3866],[119,-10.4246],[123,-10.4094],[127,-10.3349],[137,-9.9299],[139,-9.8903],[150,-10.1],[169,-10.0233],[172,-10.0961],[180,-10.4577],[182,-10.4988],[183,-10.4729],[186,-10.2305],[187,-10.175],[188,-10.1778],[189,-10.2925],[190,-10.5941],[191,-11.0055],[192,-11.2417],[193,-11.3349],[194,-11.4527],[196,-11.7554],[198,-12.0862],[199,-12.2215],[200,-12.3268],[205,-12.628],[210,-12.793],[212,-12.8136],[221,-12.5836],[234,-12.6505],[243,-12.5408],[244,-12.5021],[245,-12.3985],[246,-12.1989],[247,-11.8193],[248,-11.0206],[249,-10.3935],[250,-10.2821],[251,-10.2451],[252,-10.2322],[254,-10.2663],[255,-10.2465],[261,-9.9156],[262,-9.9058],[263,-9.9158],[265,-9.9946],[267,-10.1395],[270,-10.4338],[276,-11.2],[278,-11.4124],[280,-11.5749],[286,-11.8251],[290,-11.8972],[294,-11.8643],[296,-11.8075],[300,-11.5697],[307,-11.0537],[314,-10.704],[320,-10.5498],[324,-10.523]]
       }
       upperarm_l_fk_ctrl.Rotation.Y {
        type=MovieSceneScriptingFloatChannel
        default=0
        thinned_from=325
        keys [[0,-12.386],[9,-12.5214],[20,-13.0142],[25,-13.3384],[28,-13.6026],[31,-13.9567],[32,-14.0287],[33,-14.0571],[34,-13.9828],[35,-13.8102],[36,-13.5505],[37,-13.2447],[38,-13.0037],[42,-12.2037],[45,-11.5688],[47,-11.2104],[48,-11.0494],[49,-10.9235],[50,-10.8368],[52,-10.7304],[54,-10.6887],[61,-10.7948],[71,-11.1738],[74,-11.2511],[83,-11.2569],[85,-11.1978],[86,-11.1332],[87,-11.0951],[88,-11.1212],[89,-11.2329],[90,-11.4145],[91,-11.7526],[93,-12.7349],[94,-13.203],[96,-14.0491],[97,-14.4267],[98,-14.7627],[100,-15.327],[103,-16.0574],[104,-16.2587],[105,-16.4114],[106,-16.5147],[107,-16.5604],[111,-16.5084],[114,-16.3403],[118,-15.9643],[129,-15.6764],[131,-15.5766],[141,-14.5849],[149,-13.5926],[150,-13.5113],[155,-13.3175],[160,-13.2271],[162,-13.2294],[163,-13.2083],[171,-13.245],[175,-13.3721],[178,-13.5711],[180,-13.7628],[181,-13.8832],[183,-14.2041],[187,-14.9975],[188,-15.2151],[189,-15.4673],[190,-15.7866],[194,-17.2775],[195,-17.5743],[196,-17.8156],[197,-17.9773],[198,-18.0473],[199,-18.0334],[201,-17.8847],[223,-15.1453],[228,-14.6677],[231,-14.4808],[235,-14.3602],[242,-14.3435],[243,-14.3577],[245,-14.2676],[246,-14.1557],[247,-13.9953],[248,-13.7523],[249,-13.5727],[250,-13.4776],[252,-13.1843],[253,-13.0653],[254,-13.0318],[255,-13.1147],[256,-13.2571],[260,-14.0131],[261,-14.1845],[262,-14.2954],[263,-14.3772],[264,-14.4259],[266,-14.4368],[267,-14.3985],[271,-14.0641],[282,-12.8105],[287,-12.379],[291,-12.1539],[302,-11.9679],[307,-12.0123],[320,-12.3561],[324,-12.3843]]
       }
       upperarm_l_fk_ctrl.Rotation.Z {
        type=MovieSceneScriptingFloatChannel
        default=0
        thinned_from=325
        keys [[0,-9.8449],[1,-9.8449],[3,-9.7404],[5,-9.5677],[7,-9.3276],[9,-9.0301],[14,-8.1595],[16,-7.8499],[22,-7.0552],[23,-6.9431],[24,-6.8532],[25,-6.7919],[26,-6.7663],[27,-6.783],[28,-6.8221],[32,-7.0737],[33,-7.1188],[34,-7.0891],[35,-6.8614],[36,-6.2989],[37,-5.6782],[38,-5.3826],[39,-5.2749],[40,-5.2335],[41,-5.2204],[42,-5.2509],[43,-5.3377],[44,-5.4948],[45,-5.7359],[46,-6.038],[47,-6.3769],[49,-7.1055],[51,-7.8549],[52,-8.204],[53,-8.5091],[54,-8.7752],[55,-9.0027],[56,-9.1827],[58,-9.4527],[60,-9.6521],[70,-10.3261],[72,-10.433],[74,-10.5131],[79,-10.6212],[83,-10.6503],[85,-10.5929],[87,-10.4489],[88,-10.4558],[89,-10.608],[90,-10.9192],[91,-11.5157],[92,-12.1913],[93,-12.7193],[94,-13.0917],[95,-13.2513],[96,-13.2622],[97,-13.1127],[98,-12.8068],[99,-12.3812],[102,-10.9137],[104,-9.8516],[105,-9.3759],[106,-8.9562],[107,-8.6241],[108,-8.3461],[109,-8.0997],[110,-7.9037],[111,-7.7555],[112,-7.6412],[114,-7.4625],[115,-7.3913],[116,-7.3368],[117,-7.3029],[118,-7.291],[120,-7.3219],[123,-7.4335],[127,-7.6806],[128,-7.7913],[129,-7.9201],[130,-8.0751],[134,-8.8142],[137,-9.4159],[138,-9.5869],[139,-9.7301],[140,-9.8388],[144,-10.1443],[158,-10.8296],[164,-10.9911],[169,-11.0199],[170,-10.9763],[171,-10.8966],[172,-10.787],[174,-10.5005],[177,-9.9684],[180,-9.3966],[181,-9.2272],[182,-9.0886],[183,-8.984],[184,-8.9327],[185,-8.9313],[186,-8.9484],[187,-8.9895],[188,-9.0779],[189,-9.2982],[190,-9.7825],[191,-10.3982],[192,-11.2997],[193,-11.9493],[194,-11.9485],[195,-11.5048],[197,-10.1311],[198,-9.6423],[199,-9.2998],[200,-9.0968],[201,-8.9941],[203,-8.8465],[205,-8.7593],[207,-8.7286],[209,-8.7574],[211,-8.8619],[212,-8.9507],[213,-9.0794],[214,-9.2406],[219,-10.2129],[220,-10.3836],[221,-10.5289],[222,-10.6413],[223,-10.7131],[241,-11.4507],[243,-11.4686],[244,-11.4534],[245,-11.4039],[246,-11.2963],[247,-11.1333],[249,-10.6375],[251,-10.2448],[253,-9.9835],[254,-9.9052],[255,-9.9177],[256,-10.0075],[260,-10.5702],[261,-10.6892],[262,-10.7348],[263,-10.7363],[264,-10.6991],[265,-10.6241],[266,-10.5127],[267,-10.371],[269,-10.0272],[270,-9.8292],[271,-9.5889],[272,-9.2839],[275,-8.3054],[276,-8.0017],[277,-7.7226],[278,-7.4717],[279,-7.2732],[280,-7.1351],[281,-7.0178],[282,-6.9167],[284,-6.752],[286,-6.6281],[288,-6.5497],[289,-6.5322],[290,-6.5322],[291,-6.5499],[292,-6.5873],[293,-6.6405],[294,-6.7081],[295,-6.7918],[296,-6.9035],[297,-7.0494],[298,-7.2228],[299,-7.4176],[304,-8.489],[306,-8.8711],[307,-9.035],[309,-9.2978],[311,-9.5206],[313,-9.681],[315,-9.7826],[318,-9.8504],[324,-9.8479]]
       }
       upperarm_l_fk_ctrl.Scale.X {type=MovieSceneScriptingFloatChannel, default=1, const [1,325]}
       upperarm_l_fk_ctrl.Scale.Y {type=MovieSceneScriptingFloatChannel, default=1, const [1,325]}
       upperarm_l_fk_ctrl.Scale.Z {type=MovieSceneScriptingFloatChannel, default=1, const [1,325]}
       lowerarm_l_fk_ctrl.Location.X {
        type=MovieSceneScriptingFloatChannel
        default=0
        thinned_from=325
        keys [[0,0],[1,0.0025],[2,0],[79,0],[80,-0.0002],[81,0],[82,-0.002],[83,0],[165,0],[166,-0.002],[167,0],[168,-0.0006],[170,0],[321,0],[322,-0.0003],[323,0],[324,-0.0014]]
       }
       lowerarm_l_fk_ctrl.Location.Y {
        type=MovieSceneScriptingFloatChannel
        default=0
        thinned_from=325
        keys [[0,0],[1,0.0046],[2,0],[79,0],[80,-0.0004],[81,0],[82,-0.0037],[83,0],[165,0],[166,-0.0037],[167,0],[168,-0.0011],[170,0],[321,0],[322,-0.0007],[323,0],[324,-0.0026]]
       }
       lowerarm_l_fk_ctrl.Location.Z {
        type=MovieSceneScriptingFloatChannel
        default=0
        thinned_from=325
        keys [[0,0],[1,-0.0018],[2,0],[79,0],[80,-0.0014],[81,0],[82,0.001],[83,0],[165,0],[166,0.0011],[167,0],[168,0.0003],[169,-0.0001],[321,0],[322,-0.0046],[323,0],[324,0.0007]]
       }
       lowerarm_l_fk_ctrl.Rotation.X {
        type=MovieSceneScriptingFloatChannel
        default=0
        thinned_from=325
        keys [[0,0.0229],[1,0.0177],[2,0.0187],[3,0.0168],[4,0.0153],[5,0.0144],[6,0.014],[8,0.014],[9,0.0148],[10,0.0172],[11,0.0201],[12,0.0225],[14,0.0264],[15,0.0281],[16,0.0294],[17,0.0303],[21,0.0318],[22,0.0318],[23,0.0321],[24,0.0321],[25,0.0316],[26,0.0303],[27,0.0282],[28,0.0252],[29,0.0211],[30,0.0162],[31,0.0106],[32,0.0038],[33,-0.0048],[34,-0.0147],[35,-0.0252],[36,-0.0363],[37,-0.0483],[38,-0.0597],[39,-0.0688],[40,-0.0763],[41,-0.0823],[42,-0.0873],[43,-0.0913],[45,-0.0983],[46,-0.1013],[47,-0.1034],[48,-0.1043],[49,-0.1038],[50,-0.1017],[51,-0.0975],[52,-0.0916],[53,-0.0848],[54,-0.0761],[55,-0.0646],[56,-0.0514],[57,-0.0371],[58,-0.0216],[59,-0.0044],[60,0.0124],[61,0.0282],[62,0.0416],[63,0.0511],[64,0.0583],[65,0.0637],[66,0.0679],[67,0.0713],[68,0.0741],[69,0.0767],[71,0.0812],[73,0.0853],[74,0.087],[78,0.0917],[79,0.0927],[80,0.0935],[81,0.0945],[82,0.0971],[83,0.0961],[85,0.0976],[86,0.098],[87,0.0976],[88,0.0953],[89,0.0908],[90,0.0826],[91,0.0678],[92,0.0474],[93,0.0221],[94,-0.0071],[96,-0.0692],[97,-0.0972],[98,-0.1207],[99,-0.1363],[100,-0.1468],[101,-0.1542],[102,-0.1593],[103,-0.1636],[105,-0.1704],[110,-0.1861],[113,-0.1938],[115,-0.1981],[117,-0.2014],[119,-0.2037],[122,-0.2054],[123,-0.2054],[124,-0.206],[130,-0.206],[131,-0.2055],[132,-0.2011],[133,-0.1927],[134,-0.1828],[135,-0.1718],[136,-0.1601],[139,-0.1237],[140,-0.1113],[141,-0.0981],[143,-0.0712],[145,-0.0456],[146,-0.033],[147,-0.0211],[148,-0.0098],[149,0.0007],[150,0.0103],[151,0.0191],[152,0.0271],[153,0.0339],[154,0.0399],[155,0.0453],[156,0.0502],[157,0.0548],[158,0.0589],[159,0.0626],[160,0.0659],[161,0.0686],[162,0.0709],[165,0.0767],[166,0.0802],[167,0.0798],[168,0.0817],[169,0.0786],[170,0.0836],[171,0.0844],[172,0.0848],[173,0.0846],[174,0.0839],[175,0.0828],[176,0.0813],[177,0.0795],[178,0.0772],[179,0.0745],[180,0.071],[181,0.0662],[182,0.0598],[183,0.0514],[184,0.0411],[185,0.0289],[187,0.002],[188,-0.0127],[189,-0.0298],[190,-0.0481],[191,-0.0675],[192,-0.0848],[193,-0.0957],[194,-0.1026],[195,-0.1067],[197,-0.1114],[199,-0.115],[201,-0.118],[203,-0.1203],[204,-0.1211],[207,-0.1224],[209,-0.1224],[210,-0.1221],[211,-0.1207],[212,-0.118],[214,-0.111],[215,-0.1068],[216,-0.1016],[217,-0.0955],[218,-0.0886],[219,-0.0809],[220,-0.0724],[221,-0.0631],[222,-0.0531],[223,-0.0426],[224,-0.032],[225,-0.0208],[226,-0.0091],[227,0.0021],[228,0.0121],[229,0.0209],[230,0.0291],[231,0.0363],[232,0.0426],[233,0.0477],[234,0.0518],[236,0.059],[237,0.0619],[238,0.0642],[239,0.0661],[240,0.0677],[242,0.0702],[243,0.0711],[244,0.0718],[245,0.0722],[246,0.0722],[247,0.0716],[248,0.0657],[249,0.0585],[250,0.0525],[251,0.0485],[252,0.0475],[253,0.0482],[254,0.05],[255,0.0531],[256,0.0569],[258,0.0656],[261,0.0778],[263,0.0855],[264,0.0888],[265,0.0914],[266,0.0933],[267,0.0945],[268,0.0949],[269,0.0948],[270,0.0937],[271,0.0912],[272,0.0877],[273,0.0837],[274,0.0792],[275,0.0745],[277,0.0644],[281,0.0432],[282,0.0381],[283,0.0333],[284,0.0288],[285,0.0247],[286,0.0209],[287,0.0174],[288,0.0142],[289,0.0114],[291,0.0063],[293,0.0017],[294,-0.0002],[295,-0.0015],[296,-0.0025],[297,-0.0033],[298,-0.0037],[299,-0.0033],[300,-0.0019],[301,0.0005],[302,0.004],[303,0.0089],[306,0.0256],[307,0.031],[308,0.0359],[309,0.0402],[310,0.0438],[311,0.0463],[312,0.0477],[313,0.0482],[314,0.0482],[315,0.0475],[316,0.0459],[317,0.0436],[319,0.0378],[321,0.0316],[322,0.0326],[323,0.0257],[324,0.0243]]
       }
       lowerarm_l_fk_ctrl.Rotation.Y {
        type=MovieSceneScriptingFloatChannel
        default=0
        thinned_from=325
        keys [[0,0.0923],[1,0.081],[2,0.075],[3,0.0669],[4,0.0608],[5,0.0571],[6,0.0553],[8,0.0553],[9,0.0586],[10,0.0686],[11,0.0806],[12,0.0909],[14,0.1075],[15,0.1148],[16,0.1206],[17,0.1245],[18,0.1268],[21,0.1309],[22,0.1309],[23,0.1322],[24,0.1322],[25,0.1302],[26,0.1243],[27,0.1151],[28,0.1022],[29,0.085],[30,0.0643],[31,0.0414],[32,0.0144],[33,-0.0189],[34,-0.0556],[35,-0.0936],[36,-0.1326],[37,-0.1729],[38,-0.2101],[39,-0.239],[40,-0.2623],[41,-0.2809],[42,-0.2957],[43,-0.3077],[45,-0.3283],[46,-0.3371],[47,-0.3432],[48,-0.3458],[49,-0.3444],[50,-0.3383],[51,-0.3261],[52,-0.3088],[53,-0.2884],[54,-0.2617],[55,-0.2259],[56,-0.1832],[57,-0.1351],[58,-0.0806],[59,-0.0174],[61,0.1151],[62,0.1752],[63,0.2205],[64,0.2563],[65,0.2845],[66,0.3069],[67,0.3255],[68,0.3414],[69,0.3561],[71,0.3827],[73,0.4074],[74,0.4183],[75,0.4264],[78,0.4487],[79,0.4554],[80,0.4583],[81,0.4672],[82,0.4742],[83,0.4786],[85,0.4888],[86,0.4913],[87,0.4885],[88,0.4731],[89,0.4429],[90,0.3912],[91,0.3061],[92,0.2024],[93,0.089],[94,-0.0273],[95,-0.1388],[96,-0.2405],[97,-0.3253],[98,-0.3923],[99,-0.4352],[100,-0.4633],[101,-0.4828],[102,-0.4959],[103,-0.5069],[109,-0.556],[111,-0.5702],[114,-0.588],[117,-0.6007],[120,-0.6082],[122,-0.6104],[123,-0.6104],[124,-0.6118],[130,-0.6118],[131,-0.6107],[132,-0.6001],[133,-0.5797],[134,-0.5554],[135,-0.528],[136,-0.498],[137,-0.4664],[139,-0.4007],[140,-0.3658],[141,-0.3277],[142,-0.2875],[144,-0.2058],[145,-0.164],[146,-0.1212],[147,-0.0789],[148,-0.0375],[149,0.0023],[150,0.0403],[151,0.0765],[152,0.1102],[153,0.1402],[154,0.1675],[155,0.1928],[156,0.2163],[157,0.2388],[158,0.2597],[159,0.2789],[160,0.296],[161,0.3106],[162,0.3234],[163,0.3349],[165,0.3558],[166,0.3672],[167,0.3742],[169,0.3918],[171,0.4022],[172,0.4045],[173,0.403],[174,0.3988],[175,0.392],[176,0.383],[177,0.3723],[178,0.3591],[179,0.3437],[180,0.3238],[181,0.2975],[182,0.2642],[183,0.2219],[184,0.173],[185,0.1183],[186,0.0613],[187,0.0073],[188,-0.0483],[189,-0.1099],[191,-0.2349],[192,-0.2885],[193,-0.3207],[194,-0.3409],[195,-0.3529],[197,-0.3663],[199,-0.3764],[201,-0.3848],[203,-0.3913],[205,-0.3952],[207,-0.3972],[209,-0.3972],[210,-0.3963],[211,-0.3924],[212,-0.3848],[214,-0.3651],[215,-0.3531],[216,-0.3382],[217,-0.3202],[218,-0.2996],[219,-0.2764],[220,-0.2503],[221,-0.221],[222,-0.1888],[223,-0.154],[224,-0.1175],[225,-0.0778],[227,0.0078],[228,0.0475],[229,0.0842],[230,0.119],[231,0.1512],[232,0.1798],[233,0.2039],[234,0.2241],[235,0.2422],[236,0.2598],[237,0.2749],[238,0.2869],[239,0.2973],[240,0.3059],[241,0.3132],[242,0.3195],[243,0.3246],[244,0.3283],[245,0.3307],[246,0.3307],[247,0.3274],[248,0.2951],[249,0.2575],[250,0.2273],[251,0.2079],[252,0.2032],[253,0.2063],[254,0.2153],[255,0.2303],[256,0.2494],[257,0.2714],[258,0.2946],[260,0.3392],[263,0.4088],[264,0.4296],[265,0.4464],[266,0.4591],[267,0.4672],[268,0.4703],[269,0.4692],[270,0.4622],[271,0.4453],[272,0.4229],[273,0.3979],[278,0.2611],[279,0.2341],[280,0.2079],[281,0.1829],[282,0.1594],[283,0.1378],[284,0.118],[285,0.1002],[286,0.0841],[287,0.0694],[288,0.0563],[289,0.0447],[290,0.0342],[291,0.0242],[292,0.0149],[293,0.0064],[294,-0.0009],[295,-0.0062],[296,-0.0101],[297,-0.0131],[298,-0.0144],[299,-0.0129],[300,-0.0077],[301,0.0017],[302,0.0154],[303,0.0348],[304,0.0571],[305,0.0803],[306,0.1039],[307,0.1273],[308,0.1493],[309,0.1691],[310,0.1855],[311,0.1974],[312,0.2042],[313,0.2064],[314,0.2064],[315,0.2034],[316,0.1956],[317,0.1845],[319,0.158],[321,0.1303],[322,0.1038],[323,0.1044],[324,0.0933]]
       }
       lowerarm_l_fk_ctrl.Rotation.Z {
        type=MovieSceneScriptingFloatChannel
        default=0
        thinned_from=325
        keys [[0,-1.2143],[1,-1.0914],[3,-0.8796],[4,-0.7992],[5,-0.7512],[6,-0.7266],[8,-0.7266],[9,-0.7708],[10,-0.9018],[11,-1.0601],[12,-1.1963],[13,-1.3067],[14,-1.4144],[15,-1.5107],[16,-1.5872],[17,-1.638],[18,-1.6687],[21,-1.7222],[22,-1.7222],[23,-1.7395],[24,-1.7395],[25,-1.7129],[26,-1.6347],[27,-1.5141],[28,-1.3445],[29,-1.1182],[30,-0.8458],[31,-0.5434],[32,-0.1873],[33,0.2536],[34,0.7403],[35,1.2456],[36,1.7646],[37,2.3035],[38,2.8021],[39,3.1898],[40,3.5042],[41,3.7548],[42,3.955],[43,4.1164],[45,4.3959],[46,4.515],[47,4.5973],[48,4.6324],[49,4.6129],[50,4.5309],[51,4.3648],[52,4.132],[53,3.8553],[54,3.4966],[55,3.0146],[56,2.441],[57,1.7984],[58,1.0732],[59,0.2333],[60,-0.6431],[61,-1.5138],[62,-2.3027],[63,-2.8959],[64,-3.3641],[65,-3.7319],[66,-4.0231],[67,-4.2658],[68,-4.4721],[69,-4.6633],[71,-5.0099],[73,-5.3301],[74,-5.4717],[75,-5.576],[78,-5.8656],[80,-6.0332],[84,-6.3234],[85,-6.3854],[86,-6.4172],[87,-6.3806],[88,-6.1819],[89,-5.7901],[90,-5.1203],[91,-4.0128],[92,-2.6599],[93,-1.1714],[94,0.3652],[95,1.8483],[96,3.21],[97,4.3541],[98,5.2639],[99,5.8494],[100,6.234],[101,6.5009],[102,6.6809],[103,6.8323],[105,7.0735],[110,7.6132],[113,7.8749],[115,8.0186],[118,8.1709],[120,8.2323],[122,8.2626],[123,8.2626],[124,8.2823],[130,8.2823],[131,8.2668],[132,8.1194],[133,7.8367],[134,7.501],[135,7.1222],[136,6.7105],[137,6.276],[139,5.3788],[140,4.9043],[141,4.3874],[142,3.8444],[144,2.745],[145,2.1851],[146,1.6136],[147,1.0503],[148,0.4999],[149,-0.0264],[150,-0.5286],[151,-1.0068],[152,-1.4504],[153,-1.8443],[154,-2.2027],[155,-2.5336],[156,-2.8414],[157,-3.1352],[158,-3.4085],[159,-3.6584],[160,-3.8821],[161,-4.0721],[162,-4.2387],[163,-4.3879],[166,-4.7956],[169,-5.1088],[170,-5.1963],[171,-5.2627],[172,-5.2929],[173,-5.2732],[174,-5.2186],[175,-5.1306],[176,-5.0132],[177,-4.8741],[178,-4.7033],[179,-4.503],[180,-4.2442],[181,-3.9013],[182,-3.4668],[183,-2.914],[184,-2.2747],[185,-1.5562],[186,-0.8056],[187,-0.0928],[188,0.643],[189,1.4631],[190,2.297],[191,3.1359],[192,3.8567],[193,4.2926],[194,4.5657],[195,4.7287],[197,4.9107],[199,5.0484],[201,5.1621],[203,5.2502],[205,5.3036],[207,5.3307],[209,5.3307],[210,5.3183],[211,5.2661],[212,5.1618],[214,4.8936],[215,4.7315],[216,4.5294],[217,4.2855],[218,4.0075],[219,3.6937],[220,3.3424],[221,2.9491],[222,2.5161],[223,2.0517],[224,1.5645],[225,1.0355],[226,0.4674],[227,-0.0994],[228,-0.6247],[229,-1.1073],[230,-1.5654],[231,-1.9879],[232,-2.3633],[233,-2.6788],[234,-2.9437],[235,-3.1799],[236,-3.4094],[237,-3.6072],[238,-3.7636],[239,-3.8984],[240,-4.0111],[241,-4.106],[242,-4.1883],[243,-4.2543],[244,-4.3029],[245,-4.3337],[246,-4.3337],[247,-4.2908],[248,-3.8701],[249,-3.3791],[250,-2.9846],[251,-2.7309],[252,-2.6693],[253,-2.7106],[254,-2.8275],[255,-3.024],[256,-3.2744],[257,-3.5604],[258,-3.8633],[260,-4.4441],[263,-5.3484],[264,-5.6179],[265,-5.8355],[266,-6.0003],[267,-6.1055],[268,-6.1448],[269,-6.1309],[270,-6.0405],[271,-5.8219],[272,-5.5318],[273,-5.2066],[275,-4.4993],[278,-3.4268],[279,-3.0742],[280,-2.7311],[281,-2.4041],[282,-2.096],[283,-1.8121],[284,-1.5518],[285,-1.3178],[286,-1.106],[287,-0.9131],[288,-0.7395],[289,-0.5873],[290,-0.4489],[291,-0.3163],[292,-0.1939],[293,-0.0809],[294,0.0153],[295,0.0857],[296,0.137],[297,0.1766],[298,0.1938],[299,0.1739],[300,0.1048],[301,-0.0186],[302,-0.1998],[303,-0.456],[304,-0.7513],[305,-1.0567],[306,-1.3677],[307,-1.6741],[308,-1.9629],[309,-2.2227],[310,-2.4386],[311,-2.5937],[312,-2.6832],[313,-2.7118],[314,-2.7118],[315,-2.6718],[316,-2.57],[317,-2.4247],[319,-2.0771],[321,-1.7144],[322,-1.5389],[323,-1.3735],[324,-1.2206]]
       }
       lowerarm_l_fk_ctrl.Scale.X {type=MovieSceneScriptingFloatChannel, default=1, const [1,325]}
       lowerarm_l_fk_ctrl.Scale.Y {type=MovieSceneScriptingFloatChannel, default=1, const [1,325]}
       lowerarm_l_fk_ctrl.Scale.Z {type=MovieSceneScriptingFloatChannel, default=1, const [1,325]}
       hand_l_fk_ctrl.Location.X {type=MovieSceneScriptingFloatChannel, default=0, const [0,325]}
       hand_l_fk_ctrl.Location.Y {type=MovieSceneScriptingFloatChannel, default=0, const [0,325]}
       hand_l_fk_ctrl.Location.Z {type=MovieSceneScriptingFloatChannel, default=0, const [0,325]}
       hand_l_fk_ctrl.Rotation.X {type=MovieSceneScriptingFloatChannel, default=0, const [-0.8734,325]}
       hand_l_fk_ctrl.Rotation.Y {type=MovieSceneScriptingFloatChannel, default=0, const [10.019,325]}
       hand_l_fk_ctrl.Rotation.Z {type=MovieSceneScriptingFloatChannel, default=0, const [-11.4504,325]}
       hand_l_fk_ctrl.Scale.X {type=MovieSceneScriptingFloatChannel, default=1, const [1,325]}
       hand_l_fk_ctrl.Scale.Y {type=MovieSceneScriptingFloatChannel, default=1, const [1,325]}
       hand_l_fk_ctrl.Scale.Z {type=MovieSceneScriptingFloatChannel, default=1, const [1,325]}
       thumb_01_l_ctrl.Location.X {type=MovieSceneScriptingFloatChannel, default=0, const [0.0686,325]}
       thumb_01_l_ctrl.Location.Y {type=MovieSceneScriptingFloatChannel, default=0, const [-0.3985,325]}
       thumb_01_l_ctrl.Location.Z {type=MovieSceneScriptingFloatChannel, default=0, const [-0.4179,325]}
       thumb_01_l_ctrl.Rotation.X {type=MovieSceneScriptingFloatChannel, default=0, const [9.0456,325]}
       thumb_01_l_ctrl.Rotation.Y {type=MovieSceneScriptingFloatChannel, default=0, const [-15.3775,325]}
       thumb_01_l_ctrl.Rotation.Z {type=MovieSceneScriptingFloatChannel, default=0, const [-6.5547,325]}
       thumb_01_l_ctrl.Scale.X {type=MovieSceneScriptingFloatChannel, default=1, const [1,325]}
       thumb_01_l_ctrl.Scale.Y {type=MovieSceneScriptingFloatChannel, default=1, const [1,325]}
       thumb_01_l_ctrl.Scale.Z {type=MovieSceneScriptingFloatChannel, default=1, const [1,325]}
       thumb_02_l_ctrl.Location.X {type=MovieSceneScriptingFloatChannel, default=0, thinned_from=325, keys [[0,0],[324,0]]}
       thumb_02_l_ctrl.Location.Y {type=MovieSceneScriptingFloatChannel, default=0, thinned_from=325, keys [[0,0],[324,0]]}
       thumb_02_l_ctrl.Location.Z {type=MovieSceneScriptingFloatChannel, default=0, thinned_from=325, keys [[0,0],[324,0]]}
       thumb_02_l_ctrl.Rotation.X {type=MovieSceneScriptingFloatChannel, default=0, thinned_from=325, keys [[0,-28.0109],[324,-28.0109]]}
       thumb_02_l_ctrl.Rotation.Y {type=MovieSceneScriptingFloatChannel, default=0, thinned_from=325, keys [[0,-14.6405],[324,-14.6405]]}
       thumb_02_l_ctrl.Rotation.Z {
        type=MovieSceneScriptingFloatChannel
        default=0
        thinned_from=325
        keys [[0,-20.6578],[1,-21.2697],[3,-23.8884],[5,-25.9211],[6,-26.7499],[7,-27.3326],[8,-27.7482],[10,-28.4442],[12,-28.959],[17,-29.6208],[19,-29.7399],[20,-29.7399],[21,-29.8174],[22,-30.0215],[23,-30.4043],[24,-30.9601],[25,-31.7678],[26,-32.7485],[27,-33.6592],[28,-34.4304],[29,-35.1318],[30,-35.7285],[31,-36.201],[32,-36.569],[34,-37.0749],[35,-37.2127],[42,-37.2608],[43,-37.1277],[44,-36.7875],[45,-36.3287],[46,-35.7632],[48,-34.3286],[50,-32.6511],[52,-30.7827],[53,-29.7862],[54,-28.7209],[55,-27.5612],[60,-21.3266],[62,-18.9942],[65,-15.5754],[66,-14.5144],[67,-13.5241],[68,-12.6073],[69,-11.8321],[70,-11.2208],[71,-10.7511],[72,-10.4237],[73,-10.2435],[74,-10.1771],[81,-10.1771],[82,-10.2556],[83,-10.5286],[84,-10.9998],[85,-11.7134],[86,-12.6251],[87,-13.616],[88,-14.6586],[89,-15.7804],[90,-16.9449],[91,-18.1849],[92,-19.4676],[93,-20.6578],[94,-21.7461],[95,-22.7419],[96,-23.5385],[97,-24.0646],[98,-24.3915],[99,-24.6268],[100,-24.7836],[103,-24.9277],[105,-24.9434],[106,-24.8952],[109,-24.4692],[110,-24.2803],[111,-24.0289],[112,-23.6873],[120,-20.3153],[125,-18.1285],[126,-17.7465],[127,-17.4108],[128,-17.114],[129,-16.8616],[130,-16.6584],[131,-16.4996],[132,-16.3764],[133,-16.2957],[137,-16.1691],[138,-16.102],[139,-15.9475],[140,-15.7259],[142,-15.2069],[143,-14.9096],[144,-14.5376],[145,-14.0511],[149,-11.8433],[150,-11.3582],[151,-10.9485],[152,-10.5965],[153,-10.3031],[154,-10.0678],[155,-9.8913],[156,-9.7766],[158,-9.6467],[160,-9.5787],[171,-9.5637],[172,-9.6844],[173,-9.9801],[175,-10.6785],[176,-11.0812],[177,-11.544],[178,-12.0837],[181,-13.9316],[183,-15.2433],[188,-18.7454],[189,-19.3936],[190,-19.9534],[191,-20.4479],[192,-20.8823],[193,-21.2516],[195,-21.8297],[197,-22.2546],[199,-22.5474],[202,-22.7825],[205,-22.8166],[206,-22.7684],[207,-22.6742],[208,-22.528],[209,-22.3346],[210,-22.0951],[211,-21.8036],[213,-21.1006],[215,-20.2938],[218,-18.9369],[219,-18.4375],[221,-17.3345],[224,-15.5468],[225,-14.9101],[228,-12.8352],[230,-11.394],[231,-10.7451],[232,-10.1798],[233,-9.6871],[234,-9.3074],[235,-9.0607],[236,-8.927],[237,-8.8867],[240,-8.8867],[241,-8.9763],[242,-9.2295],[243,-9.623],[244,-10.1586],[245,-10.8346],[246,-11.6215],[247,-12.5466],[251,-16.6642],[252,-17.6183],[253,-18.4738],[254,-19.2608],[255,-20.0011],[256,-20.6996],[258,-21.9631],[260,-23.084],[261,-23.5652],[262,-23.9835],[263,-24.3424],[264,-24.6332],[265,-24.8637],[266,-25.0259],[267,-25.0873],[288,-25.0873],[289,-25.0604],[290,-24.973],[291,-24.8148],[292,-24.5716],[294,-23.9047],[303,-20.0787],[306,-18.9204],[308,-18.2325],[309,-17.969],[310,-17.7882],[311,-17.6764],[312,-17.6381],[313,-17.7091],[314,-17.8972],[315,-18.1652],[316,-18.5382],[318,-19.4115],[320,-20.1078],[321,-20.3661],[322,-20.5341],[323,-20.6283],[324,-20.6578]]
       }
       thumb_02_l_ctrl.Scale.X {type=MovieSceneScriptingFloatChannel, default=1, const [1,325]}
       thumb_02_l_ctrl.Scale.Y {type=MovieSceneScriptingFloatChannel, default=1, const [1,325]}
       thumb_02_l_ctrl.Scale.Z {type=MovieSceneScriptingFloatChannel, default=1, const [1,325]}
       thumb_03_l_ctrl.Location.X {type=MovieSceneScriptingFloatChannel, default=0, thinned_from=325, keys [[0,0],[324,0]]}
       thumb_03_l_ctrl.Location.Y {type=MovieSceneScriptingFloatChannel, default=0, thinned_from=325, keys [[0,0],[324,0]]}
       thumb_03_l_ctrl.Location.Z {type=MovieSceneScriptingFloatChannel, default=0, thinned_from=325, keys [[0,0],[324,0]]}
       thumb_03_l_ctrl.Rotation.X {type=MovieSceneScriptingFloatChannel, default=0, thinned_from=325, keys [[0,-4.6289],[324,-4.6289]]}
       thumb_03_l_ctrl.Rotation.Y {type=MovieSceneScriptingFloatChannel, default=0, thinned_from=325, keys [[0,-5.8424],[324,-5.8424]]}
       thumb_03_l_ctrl.Rotation.Z {
        type=MovieSceneScriptingFloatChannel
        default=0
        thinned_from=325
        keys [[0,-69.3636],[1,-69.9755],[3,-72.5941],[5,-74.6269],[6,-75.4556],[7,-76.0384],[11,-77.4303],[18,-78.4102],[21,-78.5231],[22,-78.7272],[23,-79.11],[24,-79.6658],[25,-80.4735],[27,-82.3649],[29,-83.8375],[31,-84.9068],[34,-85.7806],[42,-85.9665],[43,-85.8335],[45,-85.0344],[48,-83.0344],[52,-79.4884],[54,-77.4266],[62,-67.6999],[66,-63.2202],[68,-61.3131],[69,-60.5378],[70,-59.9265],[71,-59.4568],[72,-59.1294],[74,-58.8828],[81,-58.8828],[82,-58.9614],[83,-59.2344],[84,-59.7056],[85,-60.4191],[88,-63.3644],[91,-66.8906],[93,-69.3636],[95,-71.4476],[96,-72.2443],[97,-72.7703],[99,-73.3326],[105,-73.6491],[110,-72.986],[126,-66.4522],[129,-65.5674],[133,-65.0015],[138,-64.8077],[143,-63.6154],[145,-62.7569],[150,-60.0639],[153,-59.0089],[155,-58.597],[171,-58.2694],[172,-58.3901],[176,-59.787],[178,-60.7895],[189,-68.0994],[192,-69.5881],[196,-70.7679],[202,-71.4883],[206,-71.4741],[209,-71.0404],[214,-69.4168],[219,-67.1432],[225,-63.6159],[231,-59.4508],[233,-58.3929],[234,-58.0132],[236,-57.6327],[240,-57.5924],[241,-57.6821],[242,-57.9353],[243,-58.3287],[244,-58.8644],[246,-60.3272],[252,-66.324],[255,-68.7068],[259,-71.2494],[262,-72.6892],[265,-73.5694],[267,-73.793],[289,-73.7662],[291,-73.5206],[295,-72.2194],[307,-67.2707],[309,-66.6748],[311,-66.3821],[313,-66.4148],[315,-66.871],[320,-68.8136],[322,-69.2398],[324,-69.3636]]
       }
       thumb_03_l_ctrl.Scale.X {type=MovieSceneScriptingFloatChannel, default=1, const [1,325]}
       thumb_03_l_ctrl.Scale.Y {type=MovieSceneScriptingFloatChannel, default=1, const [1,325]}
       thumb_03_l_ctrl.Scale.Z {type=MovieSceneScriptingFloatChannel, default=1, const [1,325]}
       index_metacarpal_l_ctrl.Location.X {type=MovieSceneScriptingFloatChannel, default=0, const [0,325]}
       index_metacarpal_l_ctrl.Location.Y {type=MovieSceneScriptingFloatChannel, default=0, const [0,325]}
       index_metacarpal_l_ctrl.Location.Z {type=MovieSceneScriptingFloatChannel, default=0, const [0,325]}
       index_metacarpal_l_ctrl.Rotation.X {type=MovieSceneScriptingFloatChannel, default=0, const [-2.9693,325]}
       index_metacarpal_l_ctrl.Rotation.Y {type=MovieSceneScriptingFloatChannel, default=0, const [3.5534,325]}
       index_metacarpal_l_ctrl.Rotation.Z {type=MovieSceneScriptingFloatChannel, default=0, const [-13.3204,325]}
       index_metacarpal_l_ctrl.Scale.X {type=MovieSceneScriptingFloatChannel, default=1, const [1,325]}
       index_metacarpal_l_ctrl.Scale.Y {type=MovieSceneScriptingFloatChannel, default=1, const [1,325]}
       index_metacarpal_l_ctrl.Scale.Z {type=MovieSceneScriptingFloatChannel, default=1, const [1,325]}
       index_01_l_ctrl.Location.X {type=MovieSceneScriptingFloatChannel, default=0, const [-0.4363,325]}
       index_01_l_ctrl.Location.Y {type=MovieSceneScriptingFloatChannel, default=0, const [-1.0252,325]}
       index_01_l_ctrl.Location.Z {type=MovieSceneScriptingFloatChannel, default=0, const [-0.034,325]}
       index_01_l_ctrl.Rotation.X {type=MovieSceneScriptingFloatChannel, default=0, thinned_from=325, keys [[0,8.4656],[324,8.4656]]}
       index_01_l_ctrl.Rotation.Y {type=MovieSceneScriptingFloatChannel, default=0, thinned_from=325, keys [[0,-1.3715],[324,-1.3715]]}
       index_01_l_ctrl.Rotation.Z {
        type=MovieSceneScriptingFloatChannel
        default=0
        thinned_from=325
        keys [[0,-66.4322],[1,-67.0441],[3,-69.6628],[5,-71.6955],[6,-72.5243],[7,-73.107],[11,-74.499],[18,-75.4789],[21,-75.5918],[22,-75.7959],[23,-76.1787],[24,-76.7345],[25,-77.5422],[27,-79.4336],[29,-80.9062],[31,-81.9755],[34,-82.8493],[42,-83.0352],[43,-82.9021],[45,-82.1031],[47,-80.8534],[50,-78.4255],[53,-75.5606],[55,-73.3356],[61,-65.9184],[65,-61.3498],[67,-59.2986],[68,-58.3817],[69,-57.6065],[70,-56.9952],[71,-56.5255],[72,-56.1981],[73,-56.0179],[81,-55.9515],[82,-56.0301],[83,-56.3031],[84,-56.7743],[85,-57.4878],[87,-59.3904],[90,-62.7192],[93,-66.4322],[95,-68.5163],[96,-69.313],[97,-69.839],[99,-70.4012],[105,-70.7178],[110,-70.0547],[112,-69.4617],[126,-63.5209],[129,-62.636],[133,-62.0701],[138,-61.8764],[143,-60.684],[145,-59.8256],[150,-57.1326],[152,-56.3709],[154,-55.8423],[160,-55.3531],[171,-55.3381],[172,-55.4588],[176,-56.8556],[178,-57.8581],[189,-65.168],[192,-66.6567],[196,-67.8366],[202,-68.557],[206,-68.5428],[209,-68.1091],[214,-66.4854],[219,-64.2119],[225,-60.6846],[231,-56.5195],[233,-55.4616],[234,-55.0819],[235,-54.8351],[237,-54.6611],[240,-54.6611],[241,-54.7507],[242,-55.004],[243,-55.3974],[244,-55.9331],[246,-57.3959],[252,-63.3926],[255,-65.7755],[259,-68.318],[262,-69.7579],[265,-70.6381],[267,-70.8617],[289,-70.8348],[291,-70.5892],[295,-69.288],[307,-64.3394],[309,-63.7434],[311,-63.4508],[313,-63.4835],[315,-63.9396],[320,-65.8822],[322,-66.3084],[324,-66.4322]]
       }
       index_01_l_ctrl.Scale.X {type=MovieSceneScriptingFloatChannel, default=1, const [1,325]}
       index_01_l_ctrl.Scale.Y {type=MovieSceneScriptingFloatChannel, default=1, const [1,325]}
       index_01_l_ctrl.Scale.Z {type=MovieSceneScriptingFloatChannel, default=1, const [1,325]}
       index_02_l_ctrl.Location.X {type=MovieSceneScriptingFloatChannel, default=0, thinned_from=325, keys [[0,0],[324,0]]}
       index_02_l_ctrl.Location.Y {type=MovieSceneScriptingFloatChannel, default=0, thinned_from=325, keys [[0,0],[324,0]]}
       index_02_l_ctrl.Location.Z {type=MovieSceneScriptingFloatChannel, default=0, thinned_from=325, keys [[0,0],[324,0]]}
       index_02_l_ctrl.Rotation.X {type=MovieSceneScriptingFloatChannel, default=0, thinned_from=325, keys [[0,2.7848],[324,2.7848]]}
       index_02_l_ctrl.Rotation.Y {type=MovieSceneScriptingFloatChannel, default=0, thinned_from=325, keys [[0,-2.6518],[324,-2.6518]]}
       index_02_l_ctrl.Rotation.Z {
        type=MovieSceneScriptingFloatChannel
        default=0
        thinned_from=325
        keys [[0,-67.2511],[1,-67.863],[3,-70.4817],[5,-72.5144],[6,-73.3432],[7,-73.9259],[11,-75.3179],[18,-76.2977],[21,-76.4106],[22,-76.6147],[23,-76.9976],[24,-77.5533],[25,-78.3611],[27,-80.2524],[29,-81.725],[31,-82.7943],[34,-83.6681],[42,-83.854],[43,-83.721],[45,-82.9219],[47,-81.6722],[50,-79.2444],[53,-76.3794],[55,-74.1544],[61,-66.7373],[65,-62.1687],[67,-60.1174],[68,-59.2006],[69,-58.4253],[70,-57.8141],[71,-57.3444],[72,-57.0169],[73,-56.8368],[81,-56.7704],[82,-56.8489],[83,-57.1219],[84,-57.5931],[85,-58.3066],[87,-60.2093],[90,-63.5381],[93,-67.2511],[95,-69.3352],[96,-70.1318],[97,-70.6579],[99,-71.2201],[105,-71.5367],[110,-70.8736],[112,-70.2806],[126,-64.3397],[129,-63.4549],[133,-62.889],[138,-62.6952],[143,-61.5029],[145,-60.6444],[150,-57.9515],[152,-57.1897],[154,-56.6611],[160,-56.172],[171,-56.1523],[172,-56.2777],[176,-57.6745],[178,-58.677],[189,-65.9869],[192,-67.4756],[196,-68.6555],[202,-69.3758],[206,-69.3617],[209,-68.9279],[214,-67.3043],[219,-65.0307],[225,-61.5034],[231,-57.3383],[233,-56.2804],[234,-55.9007],[235,-55.654],[237,-55.48],[240,-55.48],[241,-55.5696],[242,-55.8228],[243,-56.2163],[244,-56.7519],[246,-58.2147],[252,-64.2115],[255,-66.5944],[259,-69.1369],[262,-70.5768],[265,-71.4569],[267,-71.6805],[289,-71.6537],[291,-71.4081],[295,-70.1069],[307,-65.1583],[309,-64.5623],[311,-64.2697],[313,-64.3024],[315,-64.7585],[320,-66.7011],[322,-67.1273],[324,-67.2511]]
       }
       index_02_l_ctrl.Scale.X {type=MovieSceneScriptingFloatChannel, default=1, const [1,325]}
       index_02_l_ctrl.Scale.Y {type=MovieSceneScriptingFloatChannel, default=1, const [1,325]}
       index_02_l_ctrl.Scale.Z {type=MovieSceneScriptingFloatChannel, default=1, const [1,325]}
       index_03_l_ctrl.Location.X {type=MovieSceneScriptingFloatChannel, default=0, thinned_from=325, keys [[0,-0.0001],[324,-0.0001]]}
       index_03_l_ctrl.Location.Y {type=MovieSceneScriptingFloatChannel, default=0, thinned_from=325, keys [[0,0.0003],[324,0.0003]]}
       index_03_l_ctrl.Location.Z {type=MovieSceneScriptingFloatChannel, default=0, thinned_from=325, keys [[0,0],[324,0]]}
       index_03_l_ctrl.Rotation.X {
        type=MovieSceneScriptingFloatChannel
        default=0
        thinned_from=325
        keys [[0,0.046],[1,0.0471],[3,0.0517],[5,0.0552],[6,0.0566],[7,0.0575],[10,0.0594],[13,0.0605],[18,0.0615],[21,0.0616],[22,0.062],[23,0.0626],[24,0.0635],[25,0.0648],[27,0.0677],[29,0.07],[31,0.0716],[33,0.0726],[35,0.0731],[42,0.0732],[43,0.073],[45,0.0718],[47,0.0699],[49,0.0675],[52,0.0632],[54,0.0598],[58,0.0516],[65,0.0369],[67,0.0331],[68,0.0314],[69,0.0299],[70,0.0288],[71,0.0279],[72,0.0273],[73,0.0269],[81,0.0268],[82,0.0269],[83,0.0275],[84,0.0283],[85,0.0297],[87,0.0332],[90,0.0394],[92,0.0439],[94,0.048],[95,0.0497],[96,0.0511],[97,0.052],[99,0.053],[103,0.0535],[106,0.0534],[110,0.0524],[112,0.0514],[126,0.0408],[129,0.0392],[132,0.0383],[138,0.0378],[142,0.0362],[144,0.035],[150,0.029],[152,0.0276],[154,0.0266],[159,0.0257],[171,0.0256],[172,0.0259],[176,0.0285],[178,0.0304],[189,0.0438],[191,0.0457],[193,0.0471],[196,0.0485],[201,0.0497],[206,0.0498],[209,0.049],[212,0.0475],[218,0.043],[224,0.0368],[231,0.0279],[233,0.0259],[234,0.0251],[235,0.0247],[237,0.0243],[240,0.0243],[241,0.0245],[242,0.025],[243,0.0257],[244,0.0268],[245,0.028],[246,0.0295],[251,0.0389],[253,0.0421],[256,0.0461],[259,0.0494],[261,0.0511],[263,0.0525],[265,0.0534],[267,0.0538],[289,0.0537],[291,0.0533],[294,0.0517],[307,0.0423],[309,0.0412],[311,0.0407],[313,0.0408],[315,0.0416],[319,0.0445],[321,0.0455],[323,0.046],[324,0.046]]
       }
       index_03_l_ctrl.Rotation.Y {
        type=MovieSceneScriptingFloatChannel
        default=0
        thinned_from=325
        keys [[0,-0.1016],[1,-0.1011],[3,-0.0989],[6,-0.0962],[10,-0.0944],[18,-0.0931],[21,-0.093],[22,-0.0928],[24,-0.0918],[25,-0.0909],[27,-0.0887],[30,-0.0862],[32,-0.0851],[35,-0.0843],[42,-0.0842],[43,-0.0844],[45,-0.0854],[48,-0.0879],[54,-0.0942],[59,-0.1],[64,-0.1046],[67,-0.1066],[70,-0.1078],[82,-0.1083],[84,-0.1079],[87,-0.1065],[90,-0.1044],[95,-0.0999],[96,-0.0992],[98,-0.0984],[106,-0.0979],[111,-0.0987],[127,-0.1041],[143,-0.1057],[152,-0.1081],[172,-0.1085],[178,-0.1074],[192,-0.1014],[200,-0.1],[207,-0.0999],[233,-0.1085],[241,-0.1088],[244,-0.1083],[247,-0.1071],[259,-0.1001],[263,-0.0985],[266,-0.0978],[290,-0.0979],[308,-0.1035],[312,-0.1039],[315,-0.1035],[321,-0.1019],[324,-0.1016]]
       }
       index_03_l_ctrl.Rotation.Z {
        type=MovieSceneScriptingFloatChannel
        default=0
        thinned_from=325
        keys [[0,-90.9605],[1,-91.5724],[3,-94.1911],[5,-96.2238],[6,-97.0526],[10,-98.7469],[18,-100.0071],[21,-100.12],[23,-100.7071],[24,-101.2627],[28,-104.7331],[30,-106.0312],[33,-107.1592],[42,-107.5634],[44,-107.0903],[46,-106.0659],[50,-102.9538],[54,-99.0235],[65,-85.8781],[68,-82.91],[70,-81.5235],[72,-80.7263],[74,-80.4798],[82,-80.5583],[83,-80.8313],[84,-81.3025],[85,-82.016],[88,-84.9613],[93,-90.9605],[95,-93.0446],[96,-93.8412],[97,-94.3673],[100,-95.0864],[106,-95.1979],[111,-94.3316],[127,-87.7135],[132,-86.6791],[138,-86.4046],[143,-85.2123],[145,-84.3538],[150,-81.6608],[153,-80.6058],[161,-79.8664],[171,-79.8664],[172,-79.9871],[176,-81.3839],[190,-90.256],[193,-91.5543],[199,-92.8502],[206,-93.0711],[210,-92.3978],[218,-89.2396],[225,-85.2128],[231,-81.0477],[233,-79.9898],[235,-79.3634],[240,-79.1894],[241,-79.279],[243,-79.9257],[245,-81.1372],[247,-82.8493],[252,-87.9209],[256,-91.0023],[261,-93.8679],[265,-95.1664],[290,-95.2758],[294,-94.2075],[308,-88.5352],[311,-87.9791],[313,-88.0117],[315,-88.4679],[320,-90.4105],[323,-90.931],[324,-90.9605]]
       }
       index_03_l_ctrl.Scale.X {type=MovieSceneScriptingFloatChannel, default=1, const [1,325]}
       index_03_l_ctrl.Scale.Y {type=MovieSceneScriptingFloatChannel, default=1, const [1,325]}
       index_03_l_ctrl.Scale.Z {type=MovieSceneScriptingFloatChannel, default=1, const [1,325]}
       middle_metacarpal_l_ctrl.Location.X {type=MovieSceneScriptingFloatChannel, default=0, const [0,325]}
       middle_metacarpal_l_ctrl.Location.Y {type=MovieSceneScriptingFloatChannel, default=0, const [0,325]}
       middle_metacarpal_l_ctrl.Location.Z {type=MovieSceneScriptingFloatChannel, default=0, const [0,325]}
       middle_metacarpal_l_ctrl.Rotation.X {type=MovieSceneScriptingFloatChannel, default=0, const [0.2131,325]}
       middle_metacarpal_l_ctrl.Rotation.Y {type=MovieSceneScriptingFloatChannel, default=0, const [-1.1482,325]}
       middle_metacarpal_l_ctrl.Rotation.Z {type=MovieSceneScriptingFloatChannel, default=0, const [-10.5148,325]}
       middle_metacarpal_l_ctrl.Scale.X {type=MovieSceneScriptingFloatChannel, default=1, const [1,325]}
       middle_metacarpal_l_ctrl.Scale.Y {type=MovieSceneScriptingFloatChannel, default=1, const [1,325]}
       middle_metacarpal_l_ctrl.Scale.Z {type=MovieSceneScriptingFloatChannel, default=1, const [1,325]}
       middle_01_l_ctrl.Location.X {type=MovieSceneScriptingFloatChannel, default=0, const [-0.5532,325]}
       middle_01_l_ctrl.Location.Y {type=MovieSceneScriptingFloatChannel, default=0, const [-0.9634,325]}
       middle_01_l_ctrl.Location.Z {type=MovieSceneScriptingFloatChannel, default=0, const [-0.0915,325]}
       middle_01_l_ctrl.Rotation.X {type=MovieSceneScriptingFloatChannel, default=0, thinned_from=325, keys [[0,0],[324,0]]}
       middle_01_l_ctrl.Rotation.Y {type=MovieSceneScriptingFloatChannel, default=0, thinned_from=325, keys [[0,0],[324,0]]}
       middle_01_l_ctrl.Rotation.Z {
        type=MovieSceneScriptingFloatChannel
        default=0
        thinned_from=325
        keys [[0,-80.8858],[1,-81.4977],[3,-84.1164],[5,-86.1491],[6,-86.9779],[8,-87.9762],[12,-89.187],[19,-89.9679],[21,-90.0453],[23,-90.6323],[24,-91.1881],[27,-93.8872],[30,-95.9565],[33,-97.0846],[36,-97.4887],[42,-97.4887],[43,-97.3557],[45,-96.5566],[48,-94.5566],[53,-90.0142],[65,-75.8034],[68,-72.8353],[69,-72.06],[71,-70.9791],[72,-70.6516],[74,-70.4051],[82,-70.4836],[83,-70.7566],[84,-71.2278],[85,-71.9413],[88,-74.8866],[91,-78.4129],[93,-80.8858],[95,-82.9698],[96,-83.7665],[97,-84.2926],[100,-85.0116],[106,-85.1232],[111,-84.2568],[126,-77.9744],[130,-76.8864],[139,-76.1754],[143,-75.1376],[145,-74.2791],[150,-71.5861],[153,-70.5311],[156,-70.0045],[171,-69.7916],[172,-69.9123],[176,-71.3092],[183,-75.4712],[189,-79.6216],[192,-81.1103],[197,-82.4826],[205,-83.0446],[209,-82.5626],[215,-80.5218],[223,-76.3854],[231,-70.973],[233,-69.9151],[235,-69.2887],[240,-69.1147],[241,-69.2043],[242,-69.4575],[243,-69.851],[245,-71.0625],[247,-72.7746],[252,-77.8462],[256,-80.9276],[260,-83.312],[263,-84.5704],[266,-85.2539],[290,-85.201],[293,-84.4855],[308,-78.4605],[311,-77.9044],[313,-77.937],[315,-78.3932],[320,-80.3358],[323,-80.8563],[324,-80.8858]]
       }
       middle_01_l_ctrl.Scale.X {type=MovieSceneScriptingFloatChannel, default=1, const [1,325]}
       middle_01_l_ctrl.Scale.Y {type=MovieSceneScriptingFloatChannel, default=1, const [1,325]}
       middle_01_l_ctrl.Scale.Z {type=MovieSceneScriptingFloatChannel, default=1, const [1,325]}
       middle_02_l_ctrl.Location.X {type=MovieSceneScriptingFloatChannel, default=0, thinned_from=325, keys [[0,-0.0452],[324,-0.0452]]}
       middle_02_l_ctrl.Location.Y {type=MovieSceneScriptingFloatChannel, default=0, thinned_from=325, keys [[0,0.0507],[324,0.0507]]}
       middle_02_l_ctrl.Location.Z {type=MovieSceneScriptingFloatChannel, default=0, const [0.1146,325]}
       middle_02_l_ctrl.Rotation.X {
        type=MovieSceneScriptingFloatChannel
        default=0
        thinned_from=325
        keys [[0,-6.9244],[1,-6.9016],[5,-6.7159],[7,-6.6555],[18,-6.5501],[22,-6.5356],[24,-6.4923],[29,-6.2907],[32,-6.2179],[35,-6.1849],[42,-6.1824],[44,-6.2067],[47,-6.2933],[54,-6.5944],[61,-6.9433],[67,-7.1633],[71,-7.2418],[82,-7.2549],[85,-7.2155],[90,-7.0552],[95,-6.845],[97,-6.7925],[106,-6.7586],[130,-7.0647],[144,-7.1325],[154,-7.2599],[172,-7.2698],[186,-7.0427],[193,-6.9022],[206,-6.844],[234,-7.2794],[242,-7.2813],[246,-7.218],[260,-6.8316],[266,-6.7532],[291,-6.7619],[309,-7.0205],[314,-7.023],[322,-6.929],[324,-6.9244]]
       }
       middle_02_l_ctrl.Rotation.Y {
        type=MovieSceneScriptingFloatChannel
        default=0
        thinned_from=325
        keys [[0,-1.6646],[1,-1.699],[3,-1.8434],[5,-1.9522],[6,-1.9957],[7,-2.026],[10,-2.083],[12,-2.1091],[18,-2.1465],[21,-2.1521],[22,-2.1623],[23,-2.1812],[24,-2.2085],[25,-2.2478],[27,-2.3375],[29,-2.4053],[30,-2.4322],[32,-2.4696],[34,-2.4918],[36,-2.4999],[42,-2.4999],[43,-2.4941],[44,-2.4792],[45,-2.4589],[46,-2.4337],[48,-2.3685],[50,-2.29],[52,-2.1999],[54,-2.0971],[55,-2.0378],[58,-1.8404],[61,-1.6355],[65,-1.3699],[66,-1.3064],[67,-1.2467],[68,-1.1909],[69,-1.1435],[70,-1.1058],[71,-1.0768],[72,-1.0565],[73,-1.0453],[74,-1.0411],[81,-1.0411],[82,-1.046],[83,-1.063],[84,-1.0922],[85,-1.1362],[86,-1.192],[87,-1.2523],[88,-1.3151],[90,-1.4508],[92,-1.597],[93,-1.6646],[94,-1.7256],[95,-1.7807],[96,-1.8244],[97,-1.8529],[98,-1.8706],[99,-1.8833],[100,-1.8917],[103,-1.8994],[105,-1.9002],[107,-1.8913],[110,-1.8646],[111,-1.851],[119,-1.6698],[125,-1.5199],[127,-1.4781],[129,-1.4459],[131,-1.4246],[133,-1.4126],[137,-1.4051],[138,-1.4011],[139,-1.3919],[142,-1.3479],[143,-1.3301],[144,-1.3078],[145,-1.2786],[149,-1.1442],[150,-1.1143],[151,-1.089],[152,-1.0672],[153,-1.049],[154,-1.0343],[155,-1.0233],[156,-1.0162],[159,-1.0055],[171,-1.0026],[172,-1.0104],[173,-1.0289],[175,-1.0723],[176,-1.0972],[177,-1.1257],[178,-1.1589],[182,-1.3104],[188,-1.5555],[189,-1.5927],[190,-1.6247],[191,-1.6527],[192,-1.6772],[193,-1.698],[195,-1.7303],[197,-1.7538],[200,-1.7756],[203,-1.7849],[206,-1.7822],[208,-1.769],[210,-1.745],[212,-1.71],[215,-1.644],[218,-1.5665],[220,-1.5066],[223,-1.4044],[225,-1.3302],[228,-1.2048],[230,-1.1165],[231,-1.0764],[232,-1.0413],[233,-1.0106],[234,-0.9869],[235,-0.9714],[236,-0.963],[237,-0.9605],[240,-0.9605],[241,-0.9661],[242,-0.982],[243,-1.0066],[244,-1.04],[245,-1.0819],[246,-1.1305],[247,-1.1872],[251,-1.4343],[252,-1.4902],[253,-1.5398],[255,-1.6274],[257,-1.7036],[259,-1.7698],[261,-1.8258],[263,-1.8679],[265,-1.896],[266,-1.9046],[289,-1.9065],[291,-1.8933],[292,-1.8803],[294,-1.8443],[305,-1.5867],[308,-1.5259],[309,-1.5106],[310,-1.5001],[311,-1.4936],[312,-1.4913],[313,-1.4955],[314,-1.5064],[315,-1.522],[316,-1.5436],[318,-1.5938],[320,-1.6335],[321,-1.6481],[322,-1.6576],[323,-1.6629],[324,-1.6646]]
       }
       middle_02_l_ctrl.Rotation.Z {
        type=MovieSceneScriptingFloatChannel
        default=0
        thinned_from=325
        keys [[0,-62.576],[1,-63.1872],[3,-65.8031],[5,-67.8339],[6,-68.662],[7,-69.2443],[11,-70.6353],[18,-71.6146],[21,-71.7274],[22,-71.9314],[23,-72.314],[24,-72.8695],[25,-73.6768],[27,-75.5675],[29,-77.0398],[31,-78.1089],[34,-78.9827],[42,-79.1686],[43,-79.0355],[45,-78.2365],[47,-76.987],[50,-74.5598],[53,-71.6962],[55,-69.4727],[61,-62.0629],[65,-57.5006],[67,-55.4525],[68,-54.5372],[69,-53.7633],[70,-53.1531],[71,-52.6842],[72,-52.3574],[73,-52.1775],[81,-52.1112],[82,-52.1896],[83,-52.4621],[84,-52.9325],[85,-53.6448],[87,-55.5442],[90,-58.868],[93,-62.576],[95,-64.6578],[96,-65.4536],[97,-65.9791],[99,-66.5408],[105,-66.8571],[110,-66.1946],[112,-65.6022],[126,-59.6685],[129,-58.7849],[132,-58.3003],[138,-58.0263],[142,-57.1326],[144,-56.4643],[150,-53.2902],[152,-52.5298],[154,-52.0022],[159,-51.541],[171,-51.4943],[172,-51.6194],[175,-52.6118],[177,-53.4757],[184,-57.8496],[189,-61.3134],[192,-62.8003],[196,-63.9788],[202,-64.6984],[206,-64.6842],[209,-64.2509],[213,-63.0183],[219,-60.3585],[225,-56.8363],[231,-52.6782],[233,-51.6222],[234,-51.2432],[235,-50.9969],[237,-50.8232],[240,-50.8232],[241,-50.9127],[242,-51.1654],[243,-51.5581],[244,-52.0928],[245,-52.7675],[246,-53.5531],[247,-54.4766],[251,-58.5877],[253,-60.3948],[256,-62.6178],[260,-64.9995],[263,-66.2567],[266,-66.9395],[289,-66.974],[291,-66.7286],[295,-65.4287],[307,-60.4859],[309,-59.8907],[311,-59.5985],[313,-59.6311],[315,-60.0867],[320,-62.0267],[322,-62.4524],[324,-62.576]]
       }
       middle_02_l_ctrl.Scale.X {type=MovieSceneScriptingFloatChannel, default=1, const [1,325]}
       middle_02_l_ctrl.Scale.Y {type=MovieSceneScriptingFloatChannel, default=1, const [1,325]}
       middle_02_l_ctrl.Scale.Z {type=MovieSceneScriptingFloatChannel, default=1, const [1,325]}
       middle_03_l_ctrl.Location.X {type=MovieSceneScriptingFloatChannel, default=0, thinned_from=325, keys [[0,0.0206],[324,0.0206]]}
       middle_03_l_ctrl.Location.Y {type=MovieSceneScriptingFloatChannel, default=0, thinned_from=325, keys [[0,-0.0583],[324,-0.0583]]}
       middle_03_l_ctrl.Location.Z {type=MovieSceneScriptingFloatChannel, default=0, thinned_from=325, keys [[0,0.1324],[324,0.1324]]}
       middle_03_l_ctrl.Rotation.X {
        type=MovieSceneScriptingFloatChannel
        default=0
        thinned_from=325
        keys [[0,-2.0926],[1,-2.1039],[3,-2.1497],[5,-2.1821],[7,-2.203],[18,-2.2351],[22,-2.2391],[24,-2.2506],[29,-2.2942],[43,-2.3108],[48,-2.2867],[53,-2.2362],[56,-2.1883],[63,-2.0377],[68,-1.9218],[70,-1.8884],[72,-1.8687],[74,-1.8626],[82,-1.8645],[83,-1.8713],[84,-1.883],[85,-1.9004],[93,-2.0926],[95,-2.1302],[96,-2.1438],[98,-2.1579],[106,-2.1661],[111,-2.152],[128,-2.0223],[133,-2.0049],[138,-2.0008],[143,-1.9746],[145,-1.9553],[150,-1.8918],[153,-1.8657],[156,-1.8525],[171,-1.847],[172,-1.8502],[176,-1.885],[189,-2.0684],[193,-2.1036],[202,-2.1309],[208,-2.1264],[215,-2.0857],[221,-2.0269],[226,-1.9595],[231,-1.8767],[233,-1.8502],[235,-1.8343],[240,-1.8298],[241,-1.8321],[243,-1.8486],[245,-1.8789],[247,-1.9203],[251,-2.0128],[254,-2.0658],[259,-2.1267],[263,-2.1571],[267,-2.1691],[290,-2.1673],[309,-2.0399],[312,-2.0332],[315,-2.0439],[320,-2.0822],[324,-2.0926]]
       }
       middle_03_l_ctrl.Rotation.Y {
        type=MovieSceneScriptingFloatChannel
        default=0
        thinned_from=325
        keys [[0,1.0729],[1,1.0505],[3,0.9533],[5,0.8765],[6,0.8448],[7,0.8225],[8,0.8065],[10,0.7796],[11,0.7687],[12,0.7596],[13,0.7526],[17,0.7339],[18,0.7306],[20,0.7293],[21,0.7262],[22,0.7183],[23,0.7033],[24,0.6815],[25,0.6497],[26,0.6109],[27,0.5748],[28,0.544],[29,0.516],[30,0.4921],[31,0.4731],[32,0.4583],[33,0.4467],[34,0.4379],[35,0.4324],[36,0.4304],[42,0.4304],[43,0.4358],[44,0.4495],[45,0.468],[46,0.4907],[47,0.5181],[48,0.5481],[49,0.5802],[50,0.6148],[52,0.6884],[53,0.7274],[54,0.7689],[55,0.8137],[59,1.0027],[60,1.0484],[62,1.1332],[65,1.254],[66,1.2907],[67,1.3244],[68,1.3553],[69,1.3812],[70,1.4014],[71,1.4169],[72,1.4276],[73,1.4334],[81,1.4356],[82,1.433],[83,1.4241],[84,1.4087],[85,1.3852],[86,1.3547],[88,1.2857],[90,1.2061],[92,1.1161],[93,1.0729],[94,1.033],[95,0.9961],[96,0.9664],[97,0.9467],[98,0.9344],[99,0.9255],[100,0.9196],[103,0.9142],[105,0.9136],[106,0.9154],[109,0.9315],[110,0.9386],[111,0.948],[112,0.9609],[125,1.1642],[127,1.1897],[129,1.2091],[131,1.2218],[133,1.2289],[138,1.2357],[140,1.2488],[143,1.2771],[144,1.2899],[149,1.3808],[151,1.4104],[153,1.4315],[155,1.4449],[160,1.455],[171,1.4556],[172,1.4516],[175,1.4192],[177,1.3908],[181,1.3106],[186,1.193],[188,1.1421],[189,1.1188],[190,1.0986],[192,1.0647],[193,1.0512],[195,1.0299],[197,1.0142],[199,1.0033],[202,0.9946],[205,0.9934],[207,0.9986],[209,1.0112],[211,1.0309],[214,1.071],[218,1.1353],[223,1.2338],[230,1.3957],[231,1.4171],[233,1.4515],[234,1.4637],[235,1.4716],[236,1.4759],[240,1.4772],[241,1.4743],[242,1.4662],[243,1.4535],[244,1.4362],[245,1.4141],[246,1.3882],[247,1.3574],[251,1.216],[252,1.1823],[254,1.1236],[256,1.0714],[258,1.025],[260,0.9834],[261,0.9654],[262,0.9498],[263,0.9363],[264,0.9253],[265,0.9166],[266,0.9105],[267,0.9081],[288,0.9081],[289,0.9092],[290,0.9125],[291,0.9184],[292,0.9276],[294,0.9527],[304,1.1085],[308,1.1605],[309,1.1699],[310,1.1763],[311,1.1803],[312,1.1816],[313,1.1791],[314,1.1724],[315,1.1629],[316,1.1496],[318,1.1182],[320,1.093],[321,1.0836],[322,1.0774],[323,1.074],[324,1.0729]]
       }
       middle_03_l_ctrl.Rotation.Z {
        type=MovieSceneScriptingFloatChannel
        default=0
        thinned_from=325
        keys [[0,-76.641],[1,-77.2526],[3,-79.8698],[5,-81.9014],[6,-82.7297],[7,-83.312],[11,-84.7031],[18,-85.6823],[21,-85.7951],[22,-85.9991],[23,-86.3817],[24,-86.937],[27,-89.6341],[29,-91.1056],[31,-92.1741],[34,-93.0472],[42,-93.233],[43,-93.1],[45,-92.3016],[48,-90.3031],[52,-86.7598],[54,-84.6994],[65,-71.5607],[67,-69.5101],[68,-68.5936],[69,-67.8185],[70,-67.2074],[71,-66.7379],[72,-66.4105],[74,-66.164],[81,-66.164],[82,-66.2425],[83,-66.5154],[84,-66.9865],[85,-67.6999],[88,-70.6442],[91,-74.1692],[93,-76.641],[95,-78.724],[96,-79.5202],[97,-80.046],[99,-80.6078],[105,-80.9242],[110,-80.2615],[127,-73.3954],[131,-72.4846],[138,-72.0871],[143,-70.8951],[145,-70.037],[150,-67.3448],[153,-66.29],[156,-65.7636],[171,-65.5461],[172,-65.6714],[176,-67.0679],[183,-71.2287],[189,-75.3774],[192,-76.8654],[196,-78.0446],[203,-78.7987],[207,-78.6563],[211,-77.7862],[218,-74.9209],[225,-70.8957],[231,-66.7318],[233,-65.6741],[234,-65.2945],[236,-64.9141],[240,-64.8738],[241,-64.9634],[242,-65.2166],[243,-65.61],[244,-66.1455],[246,-67.608],[252,-73.6028],[256,-76.6827],[260,-79.0659],[263,-80.3236],[266,-81.0067],[289,-81.0412],[292,-80.5527],[308,-74.2168],[311,-73.6609],[313,-73.6936],[315,-74.1495],[320,-76.0912],[322,-76.5173],[324,-76.641]]
       }
       middle_03_l_ctrl.Scale.X {type=MovieSceneScriptingFloatChannel, default=1, const [1,325]}
       middle_03_l_ctrl.Scale.Y {type=MovieSceneScriptingFloatChannel, default=1, const [1,325]}
       middle_03_l_ctrl.Scale.Z {type=MovieSceneScriptingFloatChannel, default=1, const [1,325]}
       ring_metacarpal_l_ctrl.Location.X {type=MovieSceneScriptingFloatChannel, default=0, const [0,325]}
       ring_metacarpal_l_ctrl.Location.Y {type=MovieSceneScriptingFloatChannel, default=0, const [0,325]}
       ring_metacarpal_l_ctrl.Location.Z {type=MovieSceneScriptingFloatChannel, default=0, const [0,325]}
       ring_metacarpal_l_ctrl.Rotation.X {type=MovieSceneScriptingFloatChannel, default=0, const [5.3943,325]}
       ring_metacarpal_l_ctrl.Rotation.Y {type=MovieSceneScriptingFloatChannel, default=0, const [-2.9431,325]}
       ring_metacarpal_l_ctrl.Rotation.Z {type=MovieSceneScriptingFloatChannel, default=0, const [-10.4667,325]}
       ring_metacarpal_l_ctrl.Scale.X {type=MovieSceneScriptingFloatChannel, default=1, const [1,325]}
       ring_metacarpal_l_ctrl.Scale.Y {type=MovieSceneScriptingFloatChannel, default=1, const [1,325]}
       ring_metacarpal_l_ctrl.Scale.Z {type=MovieSceneScriptingFloatChannel, default=1, const [1,325]}
       ring_01_l_ctrl.Location.X {type=MovieSceneScriptingFloatChannel, default=0, const [-0.3709,325]}
       ring_01_l_ctrl.Location.Y {type=MovieSceneScriptingFloatChannel, default=0, const [-1.0509,325]}
       ring_01_l_ctrl.Location.Z {type=MovieSceneScriptingFloatChannel, default=0, const [-0.0244,325]}
       ring_01_l_ctrl.Rotation.X {type=MovieSceneScriptingFloatChannel, default=0, thinned_from=325, keys [[0,-0.8239],[324,-0.8239]]}
       ring_01_l_ctrl.Rotation.Y {type=MovieSceneScriptingFloatChannel, default=0, thinned_from=325, keys [[0,-2.6856],[324,-2.6856]]}
       ring_01_l_ctrl.Rotation.Z {
        type=MovieSceneScriptingFloatChannel
        default=0
        thinned_from=325
        keys [[0,-84.1083],[1,-84.7202],[3,-87.3388],[5,-89.3716],[6,-90.2004],[8,-91.1986],[12,-92.4094],[19,-93.1903],[21,-93.2678],[23,-93.8547],[24,-94.4105],[28,-97.8809],[30,-99.1789],[33,-100.307],[36,-100.7112],[42,-100.7112],[43,-100.5781],[45,-99.7791],[48,-97.7791],[53,-93.2366],[65,-79.0259],[68,-76.0578],[69,-75.2825],[71,-74.2016],[73,-73.6939],[81,-73.6275],[82,-73.7061],[83,-73.9791],[84,-74.4503],[85,-75.1638],[88,-78.1091],[93,-84.1083],[95,-86.1923],[96,-86.989],[97,-87.5151],[100,-88.2341],[106,-88.3456],[111,-87.4793],[127,-80.8612],[131,-79.9501],[138,-79.5524],[143,-78.3601],[145,-77.5016],[150,-74.8086],[153,-73.7536],[156,-73.227],[171,-73.0095],[172,-73.1348],[176,-74.5317],[184,-79.3754],[189,-82.8441],[192,-84.3328],[197,-85.705],[205,-86.2671],[209,-85.7851],[216,-83.3075],[224,-78.9973],[231,-74.1955],[233,-73.1376],[235,-72.5111],[240,-72.3372],[241,-72.4268],[243,-73.0734],[245,-74.285],[247,-75.9971],[252,-81.0687],[256,-84.1501],[260,-86.5345],[263,-87.7929],[266,-88.4764],[290,-88.4235],[294,-87.3552],[308,-81.683],[311,-81.1268],[313,-81.1595],[315,-81.6157],[320,-83.5583],[323,-84.0788],[324,-84.1083]]
       }
       ring_01_l_ctrl.Scale.X {type=MovieSceneScriptingFloatChannel, default=1, const [1,325]}
       ring_01_l_ctrl.Scale.Y {type=MovieSceneScriptingFloatChannel, default=1, const [1,325]}
       ring_01_l_ctrl.Scale.Z {type=MovieSceneScriptingFloatChannel, default=1, const [1,325]}
       ring_02_l_ctrl.Location.X {
        type=MovieSceneScriptingFloatChannel
        default=0
        thinned_from=325
        keys [[0,0],[19,0],[20,0.0006],[21,0],[104,0],[105,0.0006],[106,0],[324,0]]
       }
       ring_02_l_ctrl.Location.Y {
        type=MovieSceneScriptingFloatChannel
        default=0
        thinned_from=325
        keys [[0,0],[19,0],[20,-0.0007],[21,0],[104,0],[105,-0.0008],[106,0],[324,0]]
       }
       ring_02_l_ctrl.Location.Z {type=MovieSceneScriptingFloatChannel, default=0, thinned_from=325, keys [[0,0],[324,0]]}
       ring_02_l_ctrl.Rotation.X {type=MovieSceneScriptingFloatChannel, default=0, thinned_from=325, keys [[0,-12.8636],[324,-12.8636]]}
       ring_02_l_ctrl.Rotation.Y {type=MovieSceneScriptingFloatChannel, default=0, thinned_from=325, keys [[0,-0.6911],[104,-0.6911],[105,-0.6918],[324,-0.6911]]}
       ring_02_l_ctrl.Rotation.Z {
        type=MovieSceneScriptingFloatChannel
        default=0
        thinned_from=325
        keys [[0,-59.6606],[1,-60.2726],[3,-62.8912],[5,-64.924],[6,-65.7528],[7,-66.3355],[10,-67.4471],[13,-68.1428],[19,-68.7427],[21,-68.8202],[22,-69.0243],[23,-69.4072],[24,-69.9629],[25,-70.7707],[27,-72.662],[29,-74.1346],[31,-75.2039],[34,-76.0777],[42,-76.2636],[43,-76.1306],[45,-75.3315],[47,-74.0818],[50,-71.654],[53,-68.7891],[55,-66.564],[61,-59.1469],[65,-54.5782],[67,-52.527],[68,-51.6101],[69,-50.8349],[70,-50.2236],[71,-49.754],[72,-49.4265],[73,-49.2463],[81,-49.1799],[82,-49.2585],[83,-49.5315],[84,-50.0026],[85,-50.7162],[87,-52.6188],[89,-54.7832],[93,-59.6606],[95,-61.7447],[96,-62.5414],[97,-63.0675],[99,-63.6297],[105,-63.9619],[110,-63.2831],[112,-62.6902],[126,-56.7493],[129,-55.8645],[132,-55.3792],[138,-55.1048],[142,-54.2098],[144,-53.5404],[150,-50.361],[152,-49.5993],[154,-49.0707],[157,-48.7059],[171,-48.5619],[172,-48.6872],[175,-49.6814],[177,-50.5469],[183,-54.2461],[189,-58.3965],[191,-59.4508],[193,-60.2544],[197,-61.2574],[203,-61.8195],[207,-61.6771],[210,-61.0979],[215,-59.2967],[219,-57.4403],[225,-53.913],[231,-49.7479],[233,-48.69],[234,-48.3103],[235,-48.0635],[237,-47.8895],[240,-47.8895],[241,-47.9791],[242,-48.2324],[243,-48.6258],[244,-49.1615],[245,-49.8374],[246,-50.6243],[247,-51.5495],[251,-55.667],[253,-57.4766],[256,-59.7024],[260,-62.0869],[263,-63.3453],[265,-63.8665],[267,-64.0901],[289,-64.0633],[291,-63.8177],[294,-62.9076],[307,-57.5678],[309,-56.9719],[311,-56.6792],[313,-56.7119],[315,-57.1681],[320,-59.1107],[322,-59.5369],[324,-59.6606]]
       }
       ring_02_l_ctrl.Scale.X {type=MovieSceneScriptingFloatChannel, default=1, const [1,325]}
       ring_02_l_ctrl.Scale.Y {type=MovieSceneScriptingFloatChannel, default=1, const [1,325]}
       ring_02_l_ctrl.Scale.Z {type=MovieSceneScriptingFloatChannel, default=1, const [1,325]}
       ring_03_l_ctrl.Location.X {type=MovieSceneScriptingFloatChannel, default=0, thinned_from=325, keys [[0,0],[324,0]]}
       ring_03_l_ctrl.Location.Y {type=MovieSceneScriptingFloatChannel, default=0, thinned_from=325, keys [[0,0],[324,0]]}
       ring_03_l_ctrl.Location.Z {type=MovieSceneScriptingFloatChannel, default=0, thinned_from=325, keys [[0,0],[324,0]]}
       ring_03_l_ctrl.Rotation.X {type=MovieSceneScriptingFloatChannel, default=0, thinned_from=325, keys [[0,0],[324,0]]}
       ring_03_l_ctrl.Rotation.Y {type=MovieSceneScriptingFloatChannel, default=0, thinned_from=325, keys [[0,0],[324,0]]}
       ring_03_l_ctrl.Rotation.Z {
        type=MovieSceneScriptingFloatChannel
        default=0
        thinned_from=325
        keys [[0,-88.4683],[1,-89.0802],[3,-91.6989],[5,-93.7317],[6,-94.5604],[10,-96.2547],[18,-97.515],[21,-97.6278],[23,-98.2148],[24,-98.7706],[28,-102.2409],[30,-103.539],[33,-104.6671],[42,-105.0713],[44,-104.598],[46,-103.5737],[50,-100.4616],[54,-96.5313],[65,-83.3859],[68,-80.4178],[69,-79.6425],[71,-78.5616],[73,-78.054],[82,-78.0661],[83,-78.3391],[84,-78.8103],[85,-79.5238],[88,-82.4691],[93,-88.4683],[95,-90.5524],[96,-91.3491],[97,-91.8751],[100,-92.5942],[106,-92.7057],[111,-91.8394],[127,-85.2213],[132,-84.1869],[138,-83.9125],[143,-82.7201],[145,-81.8616],[150,-79.1687],[153,-78.1136],[160,-77.3892],[171,-77.3696],[172,-77.4949],[176,-78.8917],[190,-87.7639],[193,-89.0621],[199,-90.358],[206,-90.5789],[210,-89.9056],[218,-86.7474],[225,-82.7206],[231,-78.5556],[233,-77.4977],[235,-76.8712],[240,-76.6972],[241,-76.7868],[243,-77.4335],[245,-78.6451],[247,-80.3571],[252,-85.4288],[256,-88.5101],[260,-90.8946],[263,-92.153],[266,-92.8364],[290,-92.7836],[294,-91.7153],[308,-86.043],[311,-85.4869],[313,-85.5196],[315,-85.9757],[320,-87.9183],[323,-88.4389],[324,-88.4683]]
       }
       ring_03_l_ctrl.Scale.X {type=MovieSceneScriptingFloatChannel, default=1, const [1,325]}
       ring_03_l_ctrl.Scale.Y {type=MovieSceneScriptingFloatChannel, default=1, const [1,325]}
       ring_03_l_ctrl.Scale.Z {type=MovieSceneScriptingFloatChannel, default=1, const [1,325]}
       pinky_metacarpal_l_ctrl.Location.X {type=MovieSceneScriptingFloatChannel, default=0, const [0,325]}
       pinky_metacarpal_l_ctrl.Location.Y {type=MovieSceneScriptingFloatChannel, default=0, const [0,325]}
       pinky_metacarpal_l_ctrl.Location.Z {type=MovieSceneScriptingFloatChannel, default=0, const [0,325]}
       pinky_metacarpal_l_ctrl.Rotation.X {type=MovieSceneScriptingFloatChannel, default=0, const [9.8795,325]}
       pinky_metacarpal_l_ctrl.Rotation.Y {type=MovieSceneScriptingFloatChannel, default=0, const [-6.9881,325]}
       pinky_metacarpal_l_ctrl.Rotation.Z {type=MovieSceneScriptingFloatChannel, default=0, const [-8.607,325]}
       pinky_metacarpal_l_ctrl.Scale.X {type=MovieSceneScriptingFloatChannel, default=1, const [1,325]}
       pinky_metacarpal_l_ctrl.Scale.Y {type=MovieSceneScriptingFloatChannel, default=1, const [1,325]}
       pinky_metacarpal_l_ctrl.Scale.Z {type=MovieSceneScriptingFloatChannel, default=1, const [1,325]}
       pinky_01_l_ctrl.Location.X {type=MovieSceneScriptingFloatChannel, default=0, const [-0.3836,325]}
       pinky_01_l_ctrl.Location.Y {type=MovieSceneScriptingFloatChannel, default=0, const [-1.0453,325]}
       pinky_01_l_ctrl.Location.Z {type=MovieSceneScriptingFloatChannel, default=0, const [-0.0533,325]}
       pinky_01_l_ctrl.Rotation.X {type=MovieSceneScriptingFloatChannel, default=0, thinned_from=325, keys [[0,-7.5972],[324,-7.5972]]}
       pinky_01_l_ctrl.Rotation.Y {type=MovieSceneScriptingFloatChannel, default=0, thinned_from=325, keys [[0,-7.928],[324,-7.928]]}
       pinky_01_l_ctrl.Rotation.Z {
        type=MovieSceneScriptingFloatChannel
        default=0
        thinned_from=325
        keys [[0,-83.5712],[1,-84.1831],[3,-86.8017],[5,-88.8345],[6,-89.6633],[8,-90.6615],[12,-91.8723],[19,-92.6533],[21,-92.7307],[23,-93.3177],[24,-93.8734],[28,-97.3438],[30,-98.6418],[33,-99.7699],[36,-100.1741],[42,-100.1741],[43,-100.0411],[45,-99.242],[48,-97.242],[53,-92.6995],[65,-78.4888],[68,-75.5207],[69,-74.7454],[71,-73.6645],[72,-73.3371],[74,-73.0905],[82,-73.169],[83,-73.442],[84,-73.9132],[85,-74.6267],[88,-77.572],[93,-83.5712],[95,-85.6552],[96,-86.4519],[97,-86.9779],[100,-87.697],[106,-87.8085],[111,-86.9422],[127,-80.3242],[131,-79.413],[138,-79.0153],[143,-77.823],[145,-76.9645],[150,-74.2715],[153,-73.2165],[156,-72.69],[171,-72.4724],[172,-72.5978],[176,-73.9946],[184,-78.8383],[189,-82.307],[192,-83.7957],[197,-85.1679],[205,-85.73],[209,-85.248],[216,-82.7704],[224,-78.4602],[231,-73.6584],[233,-72.6005],[235,-71.9741],[240,-71.8001],[241,-71.8897],[243,-72.5364],[245,-73.7479],[247,-75.46],[252,-80.5316],[256,-83.613],[260,-85.9974],[263,-87.2558],[266,-87.9393],[290,-87.8864],[294,-86.8181],[308,-81.1459],[311,-80.5897],[313,-80.6224],[315,-81.0786],[320,-83.0212],[323,-83.5417],[324,-83.5712]]
       }
       pinky_01_l_ctrl.Scale.X {type=MovieSceneScriptingFloatChannel, default=1, const [1,325]}
       pinky_01_l_ctrl.Scale.Y {type=MovieSceneScriptingFloatChannel, default=1, const [1,325]}
       pinky_01_l_ctrl.Scale.Z {type=MovieSceneScriptingFloatChannel, default=1, const [1,325]}
       pinky_02_l_ctrl.Location.X {
        type=MovieSceneScriptingFloatChannel
        default=0
        thinned_from=325
        keys [[0,0.3529],[19,0.3529],[20,0.3532],[104,0.3529],[105,0.3533],[324,0.3529]]
       }
       pinky_02_l_ctrl.Location.Y {
        type=MovieSceneScriptingFloatChannel
        default=0
        thinned_from=325
        keys [[0,0.0017],[19,0.0017],[20,0.0011],[21,0.0017],[104,0.0017],[105,0.001],[106,0.0017],[324,0.0017]]
       }
       pinky_02_l_ctrl.Location.Z {type=MovieSceneScriptingFloatChannel, default=0, thinned_from=325, keys [[0,-0.0756],[105,-0.0755],[324,-0.0756]]}
       pinky_02_l_ctrl.Rotation.X {type=MovieSceneScriptingFloatChannel, default=0, thinned_from=325, keys [[0,1.9939],[324,1.9939]]}
       pinky_02_l_ctrl.Rotation.Y {type=MovieSceneScriptingFloatChannel, default=0, thinned_from=325, keys [[0,10.0966],[324,10.0966]]}
       pinky_02_l_ctrl.Rotation.Z {
        type=MovieSceneScriptingFloatChannel
        default=0
        thinned_from=325
        keys [[0,-57.9295],[1,-58.5414],[3,-61.1601],[5,-63.1928],[6,-64.0216],[7,-64.6043],[10,-65.7159],[13,-66.4117],[19,-67.0116],[21,-67.089],[22,-67.2932],[23,-67.676],[24,-68.2318],[25,-69.0395],[27,-70.9309],[29,-72.4034],[31,-73.4727],[34,-74.3466],[42,-74.5324],[43,-74.3994],[45,-73.6003],[47,-72.3507],[50,-69.9228],[53,-67.0579],[55,-64.8328],[61,-57.4157],[65,-52.8471],[67,-50.7958],[68,-49.879],[69,-49.1037],[70,-48.4924],[71,-48.0228],[72,-47.6953],[73,-47.5151],[81,-47.4488],[82,-47.5273],[83,-47.8003],[84,-48.2715],[85,-48.985],[87,-50.8877],[89,-53.0521],[93,-57.9295],[95,-60.0136],[96,-60.8102],[97,-61.3363],[99,-61.8985],[105,-62.2307],[110,-61.552],[112,-60.959],[126,-55.0181],[129,-54.1333],[132,-53.6481],[138,-53.3736],[142,-52.4786],[144,-51.8093],[150,-48.6298],[152,-47.8681],[154,-47.3395],[156,-47.0482],[171,-46.8307],[172,-46.956],[175,-47.9502],[177,-48.8157],[183,-52.5149],[189,-56.6653],[191,-57.7196],[193,-58.5233],[197,-59.5263],[203,-60.0883],[207,-59.9459],[210,-59.3668],[215,-57.5655],[219,-55.7091],[225,-52.1818],[231,-48.0167],[233,-46.9588],[234,-46.5791],[235,-46.3324],[236,-46.1986],[240,-46.1584],[241,-46.248],[242,-46.5012],[243,-46.8946],[244,-47.4303],[245,-48.1062],[246,-48.8931],[247,-49.8183],[251,-53.9358],[253,-55.7454],[256,-57.9713],[260,-60.3557],[263,-61.6141],[265,-62.1353],[267,-62.3589],[289,-62.3321],[291,-62.0865],[294,-61.1764],[307,-55.8367],[309,-55.2407],[311,-54.9481],[313,-54.9807],[315,-55.4369],[320,-57.3795],[322,-57.8057],[324,-57.9295]]
       }
       pinky_02_l_ctrl.Scale.X {type=MovieSceneScriptingFloatChannel, default=1, const [1,325]}
       pinky_02_l_ctrl.Scale.Y {type=MovieSceneScriptingFloatChannel, default=1, const [1,325]}
       pinky_02_l_ctrl.Scale.Z {type=MovieSceneScriptingFloatChannel, default=1, const [1,325]}
       pinky_03_l_ctrl.Location.X {type=MovieSceneScriptingFloatChannel, default=0, thinned_from=325, keys [[0,0],[324,0]]}
       pinky_03_l_ctrl.Location.Y {type=MovieSceneScriptingFloatChannel, default=0, thinned_from=325, keys [[0,0],[324,0]]}
       pinky_03_l_ctrl.Location.Z {type=MovieSceneScriptingFloatChannel, default=0, thinned_from=325, keys [[0,0],[324,0]]}
       pinky_03_l_ctrl.Rotation.X {type=MovieSceneScriptingFloatChannel, default=0, thinned_from=325, keys [[0,3.1496],[324,3.1496]]}
       pinky_03_l_ctrl.Rotation.Y {type=MovieSceneScriptingFloatChannel, default=0, thinned_from=325, keys [[0,-4.3172],[324,-4.3172]]}
       pinky_03_l_ctrl.Rotation.Z {
        type=MovieSceneScriptingFloatChannel
        default=0
        thinned_from=325
        keys [[0,-111.6788],[1,-112.2907],[3,-114.9093],[5,-116.9421],[6,-117.7709],[11,-119.7456],[22,-121.0424],[24,-121.981],[28,-125.4514],[31,-127.222],[35,-128.2337],[43,-128.1487],[46,-126.7841],[52,-121.8037],[66,-105.5354],[68,-103.6283],[70,-102.2418],[72,-101.4446],[82,-101.2766],[84,-102.0208],[86,-103.646],[90,-107.9658],[93,-111.6788],[95,-113.7629],[96,-114.5595],[98,-115.4125],[106,-115.9162],[111,-115.0498],[127,-108.4317],[133,-107.3167],[139,-106.9684],[144,-105.5586],[150,-102.3791],[154,-101.0888],[171,-100.5847],[176,-102.1022],[190,-110.9743],[195,-112.8507],[205,-113.8376],[210,-113.1161],[219,-109.4584],[232,-101.2008],[234,-100.3284],[236,-99.9479],[241,-99.9973],[243,-100.644],[245,-101.8555],[247,-103.5676],[252,-108.6392],[258,-112.9841],[262,-115.0044],[266,-116.0469],[290,-115.994],[295,-114.5346],[308,-109.2535],[311,-108.6973],[313,-108.73],[316,-109.5592],[321,-111.3871],[324,-111.6788]]
       }
       pinky_03_l_ctrl.Scale.X {type=MovieSceneScriptingFloatChannel, default=1, const [1,325]}
       pinky_03_l_ctrl.Scale.Y {type=MovieSceneScriptingFloatChannel, default=1, const [1,325]}
       pinky_03_l_ctrl.Scale.Z {type=MovieSceneScriptingFloatChannel, default=1, const [1,325]}
       hand_l_ik_ctrl.Location.X {
        type=MovieSceneScriptingFloatChannel
        default=0
        thinned_from=325
        keys [[0,-6.3316],[1,-6.3274],[2,-6.34],[3,-6.3664],[4,-6.4089],[5,-6.4671],[6,-6.5395],[8,-6.7223],[9,-6.8347],[10,-6.9657],[12,-7.2738],[15,-7.7598],[16,-7.8974],[19,-8.2096],[21,-8.3789],[22,-8.4487],[24,-8.5447],[26,-8.592],[29,-8.5936],[31,-8.561],[33,-8.5798],[34,-8.6113],[35,-8.576],[37,-8.3029],[38,-8.2229],[41,-8.1279],[42,-8.1263],[47,-8.3141],[49,-8.3431],[50,-8.3378],[52,-8.2811],[55,-8.1346],[57,-8.0908],[58,-8.0905],[60,-8.1437],[62,-8.2362],[77,-8.528],[82,-8.6834],[85,-8.828],[87,-9.0004],[91,-9.2238],[92,-9.24],[93,-9.2214],[94,-9.2227],[95,-9.1801],[98,-8.9681],[100,-8.7777],[104,-8.3296],[105,-8.2325],[107,-8.0777],[111,-7.8849],[114,-7.7961],[120,-7.7085],[122,-7.6536],[123,-7.638],[128,-7.4344],[130,-7.3161],[131,-7.2691],[135,-7.0064],[138,-6.8513],[141,-6.7454],[144,-6.6831],[148,-6.6973],[157,-6.9167],[168,-7.2836],[174,-7.638],[186,-8.5119],[188,-8.6126],[189,-8.6819],[190,-8.7804],[192,-8.9048],[193,-8.805],[194,-8.5058],[195,-8.1235],[197,-7.169],[198,-6.7894],[199,-6.4757],[200,-6.2355],[201,-6.0604],[202,-5.9192],[203,-5.8129],[204,-5.7364],[205,-5.6876],[206,-5.6584],[210,-5.6089],[212,-5.5972],[213,-5.5803],[215,-5.505],[217,-5.4102],[223,-4.9979],[225,-4.8864],[227,-4.8088],[230,-4.7345],[238,-4.61],[240,-4.5911],[241,-4.5971],[245,-4.5911],[246,-4.5812],[247,-4.5807],[248,-4.6061],[249,-4.653],[250,-4.7507],[251,-4.9813],[252,-5.2635],[253,-5.5822],[254,-5.9657],[256,-6.7995],[257,-7.1854],[258,-7.5441],[259,-7.8685],[260,-8.173],[261,-8.4494],[262,-8.6778],[263,-8.8696],[264,-9.0348],[265,-9.1792],[266,-9.2987],[267,-9.3847],[268,-9.4334],[270,-9.4567],[271,-9.4523],[274,-9.3307],[277,-9.1192],[281,-8.7792],[287,-8.4145],[290,-8.2738],[296,-8.056],[298,-7.9604],[300,-7.8826],[310,-7.3577],[314,-7.0763],[319,-6.6137],[321,-6.4576],[322,-6.4],[323,-6.3582],[324,-6.3379]]
       }
       hand_l_ik_ctrl.Location.Y {
        type=MovieSceneScriptingFloatChannel
        default=0
        thinned_from=325
        keys [[0,7.9333],[9,7.7292],[17,7.686],[23,7.7783],[25,7.8471],[29,8.0378],[31,8.0963],[32,8.0691],[33,7.9897],[34,7.8366],[35,7.5821],[36,7.2633],[37,6.9093],[38,6.6199],[39,6.4043],[40,6.2135],[41,6.0374],[42,5.8813],[43,5.752],[44,5.646],[45,5.5682],[46,5.5176],[47,5.4936],[48,5.4835],[49,5.5086],[50,5.5778],[51,5.6823],[53,5.9397],[54,6.0826],[55,6.2498],[57,6.624],[58,6.8282],[61,7.4929],[62,7.6989],[63,7.8687],[64,8.0141],[66,8.2513],[71,8.7152],[73,8.8491],[74,8.8982],[80,9.0049],[83,9.0254],[87,9.0075],[88,9.0291],[90,9.1293],[91,9.2196],[93,9.4282],[94,9.5149],[96,9.567],[102,9.6013],[103,9.577],[104,9.5327],[106,9.3856],[109,9.0759],[114,8.6571],[116,8.4565],[117,8.3722],[118,8.3088],[120,8.2523],[130,8.2168],[132,8.2428],[134,8.3314],[142,8.905],[145,9.0615],[149,9.2052],[150,9.2626],[156,9.6583],[159,9.8097],[162,9.9154],[163,9.9275],[169,10.0345],[172,10.0145],[178,9.8652],[181,9.822],[183,9.8326],[191,10.0838],[193,10.2467],[194,10.2588],[195,10.1863],[196,10.0454],[200,9.3112],[201,9.1627],[203,8.8428],[205,8.5816],[207,8.3647],[209,8.1796],[211,8.0349],[213,7.9546],[215,7.9107],[217,7.9045],[219,7.9432],[222,8.0646],[224,8.1733],[233,8.8329],[239,9.1665],[241,9.2523],[242,9.2609],[243,9.2923],[245,9.2588],[246,9.1981],[247,9.1168],[251,8.5989],[252,8.5517],[253,8.5636],[254,8.683],[255,8.9098],[258,9.8354],[260,10.4134],[261,10.6616],[262,10.8318],[263,10.9525],[264,11.0205],[265,11.0407],[266,11.0179],[267,10.9463],[268,10.8234],[270,10.4854],[271,10.2726],[274,9.4727],[276,8.9122],[278,8.4008],[279,8.1674],[281,7.766],[283,7.4173],[285,7.1165],[287,6.8618],[288,6.7532],[290,6.5775],[292,6.4562],[294,6.3694],[295,6.3433],[297,6.3312],[298,6.3449],[300,6.4307],[301,6.4967],[302,6.5808],[303,6.6871],[309,7.4656],[311,7.6933],[312,7.7811],[314,7.9059],[316,7.9835],[319,8.0124],[323,7.9614],[324,7.9331]]
       }
       hand_l_ik_ctrl.Location.Z {
        type=MovieSceneScriptingFloatChannel
        default=0
        thinned_from=325
        keys [[0,-12.992],[9,-12.6749],[12,-12.6687],[16,-12.5725],[22,-12.3421],[27,-12.1007],[28,-12.0332],[29,-11.9415],[31,-11.6712],[32,-11.4938],[33,-11.2756],[34,-11.0155],[35,-10.7216],[37,-10.0976],[38,-9.9394],[39,-9.9052],[40,-9.8924],[41,-9.9182],[42,-9.9984],[43,-10.1372],[44,-10.3023],[46,-10.7275],[50,-11.7445],[52,-12.2894],[55,-13.0424],[57,-13.596],[61,-14.8742],[62,-15.1623],[63,-15.3826],[64,-15.5592],[66,-15.8142],[72,-16.256],[79,-16.5079],[87,-16.5545],[89,-16.4835],[90,-16.3808],[91,-16.1906],[92,-15.8674],[93,-15.3963],[94,-14.8453],[95,-14.2094],[97,-12.8875],[98,-12.2715],[99,-11.723],[101,-10.7271],[103,-9.8336],[104,-9.4112],[105,-9.0427],[106,-8.7333],[107,-8.4816],[108,-8.2946],[109,-8.132],[110,-7.9994],[111,-7.8966],[112,-7.8217],[113,-7.7689],[114,-7.7398],[115,-7.7305],[121,-7.824],[122,-7.8301],[123,-7.8591],[127,-7.8973],[129,-7.9621],[130,-8.0088],[131,-8.093],[132,-8.2195],[133,-8.407],[134,-8.6191],[135,-8.8568],[141,-10.4436],[146,-11.8214],[149,-12.573],[151,-13.0036],[153,-13.3647],[156,-13.7862],[159,-14.1118],[163,-14.3829],[169,-14.5662],[171,-14.5589],[173,-14.4709],[177,-14.1224],[180,-13.7717],[182,-13.4384],[185,-12.8337],[189,-11.9663],[190,-11.8099],[191,-11.6242],[192,-11.4719],[193,-11.1227],[194,-10.4758],[195,-9.7368],[196,-8.9014],[197,-8.1119],[198,-7.5237],[199,-7.0798],[200,-6.771],[201,-6.5677],[202,-6.433],[203,-6.3444],[204,-6.2976],[205,-6.2922],[206,-6.3163],[207,-6.3589],[209,-6.4742],[210,-6.547],[211,-6.6431],[212,-6.7663],[215,-7.1774],[217,-7.5021],[218,-7.6848],[223,-8.7416],[225,-9.1985],[227,-9.689],[229,-10.1281],[231,-10.5129],[232,-10.6793],[234,-10.9468],[237,-11.2542],[241,-11.5002],[246,-11.6261],[249,-11.7838],[250,-11.8791],[251,-12.1414],[253,-12.7442],[257,-14.1095],[258,-14.401],[260,-14.8823],[261,-15.0909],[262,-15.2604],[263,-15.3885],[264,-15.4794],[265,-15.5369],[266,-15.5613],[267,-15.548],[268,-15.4953],[270,-15.2643],[271,-15.0865],[274,-14.3361],[276,-13.7918],[278,-13.3152],[279,-13.1053],[281,-12.7662],[283,-12.4835],[286,-12.1471],[289,-11.9014],[290,-11.8406],[294,-11.6921],[296,-11.6672],[298,-11.6982],[300,-11.8416],[301,-11.94],[302,-12.0627],[307,-12.866],[309,-13.1418],[311,-13.3486],[312,-13.4098],[314,-13.4603],[316,-13.4362],[324,-12.9983]]
       }
       hand_l_ik_ctrl.Rotation.X {
        type=MovieSceneScriptingFloatChannel
        default=0
        thinned_from=325
        keys [[0,-30.0019],[16,-30.039],[17,-29.9683],[22,-30.2156],[23,-30.161],[30,-30.5233],[37,-31.2336],[39,-31.3646],[42,-31.1961],[46,-30.7303],[47,-30.6914],[62,-29.334],[89,-29.0821],[90,-29.1292],[91,-29.2425],[92,-29.4235],[95,-30.2545],[99,-31.4679],[100,-31.7185],[101,-31.8658],[104,-32.5603],[106,-32.8856],[114,-33.0539],[122,-32.9249],[123,-33.0016],[130,-32.7833],[131,-32.8264],[150,-30.48],[160,-29.9489],[174,-29.8558],[180,-30.1116],[183,-30.4208],[191,-31.8455],[193,-32.1052],[197,-33.0765],[199,-33.2957],[215,-32.5908],[231,-31.0193],[250,-30.6556],[254,-30.2512],[268,-30.0199],[269,-29.8995],[274,-29.961],[275,-29.8781],[290,-29.8892],[291,-29.9705],[298,-29.8324],[299,-29.8866],[312,-29.6389],[324,-29.9991]]
       }
       hand_l_ik_ctrl.Rotation.Y {
        type=MovieSceneScriptingFloatChannel
        default=0
        thinned_from=325
        keys [[0,-12.7526],[3,-12.4695],[4,-12.405],[6,-12.3592],[8,-12.3847],[9,-12.4386],[10,-12.5695],[12,-12.8652],[15,-13.1971],[17,-13.3606],[21,-13.4338],[22,-13.4328],[23,-13.4784],[25,-13.4752],[26,-13.4142],[27,-13.3166],[28,-13.1761],[29,-12.9913],[30,-12.7668],[31,-12.5027],[32,-12.1986],[33,-11.8121],[34,-11.3376],[35,-10.7149],[36,-9.8487],[37,-8.9391],[38,-8.3138],[39,-7.8988],[40,-7.6123],[41,-7.402],[42,-7.2698],[43,-7.2151],[44,-7.2204],[45,-7.2653],[46,-7.3546],[47,-7.4601],[48,-7.6265],[49,-7.8276],[50,-8.0799],[51,-8.3935],[53,-9.1098],[54,-9.5222],[55,-10.0269],[56,-10.596],[57,-11.2112],[58,-11.8839],[61,-14.1953],[62,-14.8887],[63,-15.41],[64,-15.8222],[65,-16.1462],[66,-16.4034],[68,-16.8014],[72,-17.4328],[74,-17.696],[79,-18.1311],[85,-18.4836],[86,-18.5168],[87,-18.4951],[88,-18.4077],[89,-18.2657],[90,-18.0059],[91,-17.5618],[92,-16.9351],[93,-16.072],[94,-15.0607],[96,-12.8044],[97,-11.7401],[98,-10.7575],[99,-9.9589],[100,-9.2766],[101,-8.7101],[103,-7.6414],[104,-7.1561],[105,-6.7363],[106,-6.3807],[107,-6.1411],[109,-5.7038],[110,-5.5054],[111,-5.3298],[112,-5.173],[113,-5.0334],[114,-4.9177],[115,-4.8172],[117,-4.6773],[118,-4.6333],[120,-4.5761],[122,-4.5479],[123,-4.5225],[124,-4.512],[127,-4.5125],[130,-4.5521],[131,-4.5535],[132,-4.6854],[133,-4.9359],[134,-5.234],[135,-5.5808],[136,-5.9673],[137,-6.3761],[139,-7.2231],[140,-7.6717],[141,-8.1564],[145,-10.2089],[148,-11.7597],[150,-12.7079],[151,-13.1495],[152,-13.5593],[153,-13.9241],[155,-14.5646],[157,-15.1203],[159,-15.6066],[160,-15.815],[163,-16.2943],[169,-16.9312],[171,-17.0576],[172,-17.0686],[174,-16.9672],[177,-16.6277],[179,-16.3047],[180,-16.0871],[181,-15.8024],[182,-15.4427],[183,-14.987],[184,-14.4675],[186,-13.3292],[188,-12.2789],[189,-11.779],[190,-11.4665],[191,-11.2417],[193,-11.4647],[194,-11.1334],[195,-10.5165],[196,-9.747],[197,-9.0256],[198,-8.5168],[199,-8.1539],[200,-7.9429],[201,-7.8063],[203,-7.6332],[205,-7.5197],[207,-7.4523],[209,-7.4319],[210,-7.4464],[211,-7.5005],[212,-7.5998],[214,-7.8502],[215,-7.9996],[216,-8.1821],[217,-8.3982],[218,-8.6418],[219,-8.9156],[220,-9.221],[221,-9.562],[222,-9.9371],[224,-10.7606],[225,-11.2182],[227,-12.2009],[228,-12.6566],[230,-13.4747],[231,-13.843],[232,-14.1709],[233,-14.4473],[236,-15.0866],[237,-15.2592],[239,-15.5152],[241,-15.7096],[243,-15.8125],[244,-15.8355],[245,-15.8244],[246,-15.7697],[247,-15.6419],[248,-15.1713],[249,-14.6361],[250,-14.1916],[251,-13.8964],[252,-13.8143],[253,-13.8378],[254,-13.9622],[255,-14.2137],[256,-14.5688],[259,-15.8977],[261,-16.8531],[263,-17.7168],[264,-18.0905],[265,-18.396],[266,-18.6278],[267,-18.7807],[268,-18.8579],[269,-18.8889],[270,-18.7922],[271,-18.557],[272,-18.2087],[275,-16.9743],[279,-15.2335],[281,-14.4822],[283,-13.8381],[285,-13.2974],[287,-12.8468],[289,-12.4846],[291,-12.1827],[293,-11.9745],[294,-11.8976],[296,-11.8258],[298,-11.8325],[299,-11.855],[300,-11.9432],[301,-12.072],[302,-12.2484],[303,-12.4882],[307,-13.5926],[309,-14.0816],[310,-14.2703],[311,-14.4001],[312,-14.4602],[313,-14.4586],[314,-14.424],[315,-14.3475],[316,-14.2128],[322,-13.062],[324,-12.7617]]
       }
       hand_l_ik_ctrl.Rotation.Z {
        type=MovieSceneScriptingFloatChannel
        default=0
        thinned_from=325
        keys [[0,-13.6249],[3,-13.764],[5,-13.8003],[8,-13.7577],[9,-13.7144],[16,-13.0759],[17,-12.9553],[22,-12.7098],[23,-12.5929],[26,-12.3971],[29,-12.3006],[31,-12.2845],[32,-12.3149],[33,-12.3999],[34,-12.5848],[35,-12.9289],[36,-13.4427],[37,-14.0356],[38,-14.52],[39,-14.9516],[41,-15.5951],[43,-16.1414],[45,-16.561],[47,-16.9055],[48,-17.0149],[49,-17.0639],[50,-17.0453],[51,-16.9661],[53,-16.6938],[54,-16.4997],[55,-16.2405],[57,-15.606],[58,-15.242],[61,-13.9845],[62,-13.5956],[63,-13.2902],[64,-13.0386],[65,-12.8299],[67,-12.4938],[71,-11.9377],[73,-11.7155],[74,-11.6282],[79,-11.3852],[83,-11.2487],[85,-11.2518],[87,-11.3815],[89,-11.5532],[91,-11.7655],[92,-11.8464],[93,-11.964],[94,-12.1073],[96,-12.4905],[97,-12.6492],[98,-12.7777],[99,-12.827],[100,-12.8342],[101,-12.7602],[103,-12.7562],[104,-12.7834],[106,-12.9229],[107,-12.959],[114,-13.8258],[117,-14.2798],[118,-14.3884],[120,-14.5086],[122,-14.5664],[123,-14.6378],[130,-14.7045],[131,-14.788],[132,-14.7803],[134,-14.6539],[140,-13.902],[149,-12.6416],[153,-11.9965],[158,-11.3577],[161,-11.0688],[162,-10.989],[170,-10.5366],[172,-10.4754],[179,-10.5036],[182,-10.5874],[186,-10.7731],[189,-10.7377],[190,-10.6395],[191,-10.5021],[192,-10.1634],[193,-9.7422],[194,-9.4626],[195,-9.3703],[196,-9.385],[197,-9.4828],[198,-9.6127],[200,-9.9255],[201,-10.0665],[204,-10.5787],[207,-11.0238],[210,-11.3999],[211,-11.4902],[214,-11.6464],[216,-11.6887],[218,-11.6657],[220,-11.5788],[222,-11.4122],[224,-11.1766],[225,-11.0278],[228,-10.5122],[231,-10.0549],[233,-9.8016],[237,-9.4128],[241,-9.1421],[242,-9.1177],[243,-9.0724],[244,-9.0829],[245,-9.1186],[246,-9.2192],[247,-9.3935],[248,-9.766],[249,-10.1961],[251,-11.1194],[252,-11.4642],[253,-11.7507],[254,-11.8919],[255,-11.8213],[256,-11.5929],[257,-11.2647],[259,-10.4922],[261,-9.6851],[262,-9.3673],[263,-9.088],[264,-8.8638],[265,-8.7052],[266,-8.6119],[267,-8.5921],[269,-8.7146],[270,-8.8829],[271,-9.1307],[272,-9.4456],[274,-10.1536],[275,-10.4683],[279,-11.8708],[280,-12.1924],[282,-12.7609],[284,-13.2577],[286,-13.6838],[288,-14.0359],[291,-14.4672],[293,-14.6301],[295,-14.725],[297,-14.7483],[298,-14.7356],[299,-14.755],[301,-14.6463],[302,-14.5532],[304,-14.2707],[309,-13.4323],[310,-13.2908],[311,-13.1786],[312,-13.1014],[314,-13.029],[315,-13.0249],[317,-13.104],[324,-13.6238]]
       }
       hand_l_ik_ctrl.Scale.X {type=MovieSceneScriptingFloatChannel, default=1, const [1,325]}
       hand_l_ik_ctrl.Scale.Y {type=MovieSceneScriptingFloatChannel, default=1, const [1,325]}
       hand_l_ik_ctrl.Scale.Z {type=MovieSceneScriptingFloatChannel, default=1, const [1,325]}
       Softness {type=MovieSceneScriptingFloatChannel, default=0, const [0,325]}
       hand_l_ikorient_ctrl.Rotation.Roll {type=MovieSceneScriptingFloatChannel, default=0, const [0,325]}
       hand_l_ikorient_ctrl.Rotation.Pitch {type=MovieSceneScriptingFloatChannel, default=0, const [0,325]}
       hand_l_ikorient_ctrl.Rotation.Yaw {type=MovieSceneScriptingFloatChannel, default=0, const [0,325]}
       arm_l_pv_ik_ctrl.Location.X {
        type=MovieSceneScriptingFloatChannel
        default=0
        thinned_from=325
        keys [[0,-21.9385],[16,-22.1189],[17,-22.099],[22,-22.3161],[32,-22.71],[34,-22.6774],[35,-22.5628],[37,-22.1758],[38,-22.0228],[42,-21.6722],[45,-21.5377],[46,-21.5235],[74,-22.3766],[86,-22.4461],[88,-22.5852],[90,-22.8858],[94,-24.0285],[96,-24.4613],[98,-24.7293],[100,-24.8198],[106,-24.661],[118,-23.9717],[122,-23.9159],[123,-23.951],[130,-23.7878],[145,-23.3658],[150,-23.1196],[177,-23.1464],[181,-23.3186],[192,-24.5246],[193,-24.547],[194,-24.4809],[195,-24.3468],[199,-23.5595],[203,-23.0314],[228,-21.7082],[245,-21.6081],[248,-21.5092],[251,-21.5617],[252,-21.6434],[253,-21.7757],[254,-21.9854],[258,-23.0797],[260,-23.5363],[261,-23.72],[263,-23.9172],[265,-23.9682],[267,-23.9116],[268,-23.8377],[269,-23.7011],[274,-23.1832],[279,-22.49],[288,-21.7731],[290,-21.6767],[291,-21.6824],[298,-21.5059],[324,-21.9399]]
       }
       arm_l_pv_ik_ctrl.Location.Y {
        type=MovieSceneScriptingFloatChannel
        default=0
        thinned_from=325
        keys [[0,-7.4562],[16,-7.5896],[17,-7.5736],[22,-7.5698],[23,-7.5377],[29,-7.4925],[34,-7.3309],[36,-7.2149],[37,-7.1773],[38,-7.21],[41,-7.4502],[42,-7.5671],[44,-7.8907],[47,-8.4866],[48,-8.6691],[50,-8.978],[51,-9.1063],[52,-9.2138],[53,-9.2897],[55,-9.3661],[76,-9.5976],[85,-9.6142],[87,-9.6791],[90,-9.8423],[91,-9.9191],[92,-9.935],[94,-9.8294],[95,-9.708],[96,-9.5428],[98,-9.1514],[99,-8.9251],[100,-8.6793],[104,-7.605],[105,-7.3663],[106,-7.1651],[107,-6.9971],[108,-6.8784],[109,-6.7774],[110,-6.7019],[111,-6.6471],[112,-6.6129],[113,-6.5961],[115,-6.6062],[119,-6.7026],[121,-6.7195],[122,-6.7142],[124,-6.7272],[127,-6.703],[130,-6.7301],[134,-6.8895],[140,-7.2202],[145,-7.4351],[154,-7.68],[163,-7.7922],[179,-7.7451],[181,-7.7613],[182,-7.7854],[183,-7.8316],[186,-8.0565],[188,-8.1582],[190,-8.3005],[191,-8.3018],[192,-8.2168],[193,-7.862],[194,-7.2878],[195,-6.676],[196,-5.9765],[197,-5.2907],[198,-4.7569],[199,-4.3325],[200,-4.013],[201,-3.7938],[202,-3.6356],[203,-3.5299],[204,-3.4694],[205,-3.4534],[206,-3.4688],[207,-3.5029],[209,-3.5907],[210,-3.6421],[213,-3.8308],[214,-3.8841],[217,-4.0167],[220,-4.1247],[221,-4.1477],[223,-4.1704],[241,-4.3453],[243,-4.341],[244,-4.3512],[245,-4.3736],[246,-4.4104],[247,-4.4925],[248,-4.7219],[249,-4.9879],[250,-5.2736],[251,-5.708],[252,-6.1277],[253,-6.5126],[255,-7.3335],[256,-7.7176],[257,-8.0409],[258,-8.3096],[259,-8.5259],[260,-8.7082],[261,-8.8541],[262,-8.951],[263,-9.0014],[265,-9.0245],[266,-9.0132],[267,-8.9808],[268,-8.9234],[271,-8.6687],[274,-8.3456],[275,-8.2139],[277,-8.0066],[279,-7.8356],[282,-7.6815],[288,-7.4734],[290,-7.4257],[291,-7.4265],[297,-7.359],[298,-7.3597],[307,-7.5466],[313,-7.5797],[323,-7.4597],[324,-7.4618]]
       }
       arm_l_pv_ik_ctrl.Location.Z {
        type=MovieSceneScriptingFloatChannel
        default=0
        thinned_from=325
        keys [[0,0.9098],[1,0.9138],[2,0.9437],[3,0.9853],[4,1.0418],[5,1.1139],[6,1.2004],[7,1.3011],[8,1.4153],[9,1.5441],[10,1.686],[11,1.845],[12,2.0209],[14,2.3886],[15,2.5646],[16,2.7223],[17,2.8972],[18,3.008],[19,3.1089],[20,3.2007],[21,3.2839],[22,3.3592],[23,3.4738],[24,3.5457],[25,3.5983],[27,3.6643],[29,3.7184],[31,3.7964],[32,3.8786],[34,4.1],[35,4.101],[36,3.8962],[37,3.6524],[38,3.5155],[39,3.3965],[41,3.2691],[42,3.2234],[43,3.2018],[44,3.1921],[46,3.1984],[47,3.1541],[48,3.1309],[49,3.0977],[50,3.0546],[51,2.9926],[52,2.9192],[55,2.6659],[58,2.4501],[60,2.3193],[61,2.2611],[62,2.2077],[63,2.1591],[64,2.1161],[65,2.0782],[66,2.0459],[67,2.0192],[68,1.9974],[69,1.9803],[71,1.9619],[73,1.9547],[74,1.9596],[75,1.9578],[76,1.971],[77,1.9734],[78,1.9915],[79,2.018],[80,2.0254],[82,2.0863],[85,2.1962],[87,2.3197],[88,2.3942],[89,2.5154],[90,2.7073],[91,3.0246],[92,3.3813],[94,4.0562],[95,4.3187],[96,4.526],[97,4.6668],[98,4.7331],[99,4.7326],[100,4.6896],[101,4.6716],[102,4.5836],[104,4.3818],[105,4.2984],[106,4.2298],[107,4.2338],[108,4.2038],[110,4.1217],[115,3.8828],[117,3.7997],[122,3.6419],[123,3.569],[124,3.5435],[128,3.3905],[129,3.3353],[130,3.2699],[131,3.1521],[132,3.0495],[135,2.6825],[137,2.4501],[138,2.3405],[139,2.2364],[140,2.1412],[143,1.8712],[144,1.7847],[145,1.7112],[147,1.5842],[148,1.5248],[149,1.47],[150,1.4254],[151,1.389],[152,1.3569],[153,1.3297],[154,1.3075],[155,1.2896],[156,1.2759],[157,1.2659],[158,1.2624],[159,1.262],[161,1.2739],[162,1.287],[163,1.3084],[164,1.3169],[165,1.3404],[166,1.3589],[168,1.4121],[169,1.4473],[170,1.5001],[171,1.562],[172,1.6332],[173,1.7136],[174,1.8026],[175,1.8989],[176,2.0012],[177,2.1083],[178,2.2218],[179,2.3441],[180,2.4734],[181,2.6145],[184,3.0685],[186,3.4046],[187,3.5783],[188,3.7645],[189,4.0057],[190,4.3546],[191,4.7447],[192,5.2535],[193,5.6266],[194,5.6727],[195,5.491],[196,5.1666],[197,4.8244],[198,4.5649],[199,4.361],[200,4.2262],[201,4.1306],[202,4.0494],[203,3.9784],[205,3.8623],[209,3.6797],[211,3.5986],[212,3.5462],[213,3.4793],[214,3.3999],[215,3.3103],[216,3.2126],[217,3.1081],[218,2.9854],[222,2.4315],[224,2.1726],[225,2.0473],[226,1.926],[227,1.8113],[228,1.7054],[229,1.6079],[230,1.517],[231,1.4327],[232,1.3548],[233,1.2832],[234,1.2172],[235,1.1574],[236,1.1032],[237,1.0546],[238,1.012],[239,0.9747],[240,0.9422],[241,0.932],[242,0.913],[244,0.8806],[245,0.85],[246,0.8018],[247,0.7276],[249,0.5581],[250,0.4956],[251,0.4748],[252,0.5001],[253,0.5617],[254,0.6805],[255,0.8591],[256,1.0781],[257,1.3152],[259,1.81],[260,2.0644],[261,2.3065],[262,2.5149],[263,2.7058],[264,2.8851],[265,3.0525],[266,3.2016],[267,3.3277],[268,3.4327],[269,3.5574],[271,3.6932],[272,3.739],[274,3.8018],[275,3.8669],[276,3.8783],[277,3.8717],[278,3.8435],[280,3.7558],[283,3.6493],[284,3.6192],[289,3.4416],[290,3.4114],[291,3.3485],[296,3.2637],[297,3.2366],[298,3.2025],[299,3.1189],[300,3.058],[301,2.9793],[302,2.8857],[304,2.6804],[307,2.3588],[312,1.8512],[313,1.7572],[314,1.6573],[317,1.3453],[318,1.2466],[319,1.1559],[320,1.0736],[321,1.0061],[322,0.956],[323,0.9275],[324,0.917]]
       }
       arm_l_pv_ik_ctrl.Rotation.X {
        type=MovieSceneScriptingFloatChannel
        default=0
        thinned_from=325
        keys [[0,6.0233],[1,6.0156],[6,6.0768],[10,6.173],[16,6.3996],[17,6.5125],[22,6.5678],[23,6.6796],[26,6.8349],[28,6.9603],[32,7.2791],[33,7.3409],[34,7.3177],[35,7.0979],[36,6.5902],[37,6.0206],[38,5.7139],[39,5.483],[40,5.3722],[41,5.2779],[42,5.2136],[43,5.1929],[44,5.2215],[45,5.3034],[46,5.4234],[47,5.4725],[49,5.716],[52,6.1441],[53,6.26],[55,6.4571],[56,6.5404],[57,6.6098],[58,6.6642],[60,6.7356],[69,6.9658],[72,7.0553],[74,7.0942],[75,7.0909],[76,7.1025],[78,7.1063],[79,7.1228],[80,7.1027],[81,7.114],[83,7.1146],[87,7.0095],[88,7.0811],[89,7.3031],[90,7.6819],[91,8.3377],[92,9.0936],[93,9.7663],[94,10.3506],[95,10.7612],[96,11.0515],[97,11.2014],[98,11.2057],[99,11.0888],[100,10.8935],[101,10.7491],[102,10.4717],[104,9.8479],[105,9.5662],[106,9.3175],[107,9.2226],[110,8.7924],[112,8.5465],[115,8.2235],[117,8.0497],[118,7.9926],[120,7.9363],[122,7.9084],[123,7.8228],[127,7.8316],[130,7.8739],[131,7.7929],[134,7.6851],[141,7.5808],[145,7.4598],[149,7.298],[150,7.2802],[165,7.3665],[171,7.3722],[175,7.357],[177,7.3938],[179,7.4835],[180,7.5458],[181,7.6271],[182,7.7335],[183,7.8666],[184,8.0331],[185,8.2457],[186,8.4825],[187,8.7437],[188,9.0442],[189,9.4518],[190,10.078],[191,10.8077],[192,11.8181],[193,12.6009],[194,12.7635],[195,12.4644],[197,11.3194],[198,10.8808],[199,10.5473],[200,10.3315],[203,9.9199],[206,9.5849],[209,9.3185],[226,8.3287],[229,8.1984],[232,8.1218],[237,8.0883],[240,8.1004],[241,8.1248],[243,8.1024],[244,8.0678],[245,8.0075],[246,7.9055],[247,7.742],[248,7.5543],[249,7.3497],[250,7.1144],[251,6.8556],[252,6.6581],[253,6.4811],[254,6.3949],[255,6.453],[256,6.6317],[257,6.8865],[258,7.1772],[259,7.4917],[260,7.8317],[261,8.1541],[262,8.4055],[263,8.618],[264,8.7898],[265,8.9194],[266,9.0029],[267,9.038],[268,9.0322],[269,9.0635],[270,8.9684],[271,8.8392],[272,8.6485],[274,8.2067],[275,8.0726],[279,7.2439],[280,7.0788],[281,6.9351],[283,6.6828],[285,6.462],[287,6.2696],[289,6.1135],[290,6.0563],[291,5.9419],[293,5.9201],[295,5.94],[297,6.0064],[298,6.0535],[299,6.0203],[311,6.3054],[313,6.3047],[315,6.2662],[321,6.0675],[322,6.0477],[324,6.0318]]
       }
       arm_l_pv_ik_ctrl.Rotation.Y {
        type=MovieSceneScriptingFloatChannel
        default=0
        thinned_from=325
        keys [[0,-10.2592],[12,-10.3542],[17,-10.475],[22,-10.7309],[24,-10.8154],[28,-11.1113],[30,-11.3124],[33,-11.6486],[34,-11.7156],[36,-11.7548],[37,-11.7292],[39,-11.5105],[41,-11.1662],[46,-10.0511],[48,-9.6634],[49,-9.5037],[50,-9.383],[51,-9.2875],[52,-9.2156],[53,-9.1657],[55,-9.123],[60,-9.1675],[69,-9.4722],[71,-9.5617],[72,-9.5816],[73,-9.623],[83,-9.6508],[85,-9.6123],[88,-9.4539],[90,-9.394],[91,-9.4173],[92,-9.5266],[93,-9.7224],[94,-9.9788],[95,-10.273],[96,-10.5993],[100,-12.0507],[103,-13.0582],[104,-13.351],[105,-13.5785],[106,-13.7439],[107,-13.8054],[110,-13.8661],[112,-13.8493],[114,-13.7654],[118,-13.4289],[121,-13.3324],[127,-13.25],[131,-13.1159],[138,-12.519],[149,-11.2291],[150,-11.1486],[154,-10.977],[160,-10.8552],[162,-10.855],[163,-10.8304],[171,-10.8583],[175,-10.9603],[178,-11.1148],[180,-11.2627],[181,-11.3578],[182,-11.4807],[183,-11.6291],[185,-12.0132],[189,-12.8896],[191,-13.1859],[192,-13.2924],[193,-13.4967],[194,-13.8937],[196,-14.8147],[197,-15.1925],[198,-15.4078],[199,-15.5109],[201,-15.4704],[225,-13.0477],[229,-12.753],[233,-12.5919],[242,-12.5583],[243,-12.578],[245,-12.5445],[247,-12.4258],[249,-12.2507],[250,-12.0941],[253,-11.3639],[254,-11.1908],[255,-11.1285],[256,-11.1428],[257,-11.2035],[261,-11.5538],[263,-11.6075],[265,-11.5812],[267,-11.4848],[269,-11.2989],[274,-10.9335],[282,-10.2846],[287,-9.9681],[290,-9.836],[298,-9.6542],[304,-9.665],[311,-9.8626],[320,-10.2031],[323,-10.2523],[324,-10.2549]]
       }
       arm_l_pv_ik_ctrl.Rotation.Z {
        type=MovieSceneScriptingFloatChannel
        default=0
        thinned_from=325
        keys [[0,-35.062],[16,-35.15],[17,-35.0972],[22,-35.234],[34,-35.2053],[35,-35.1252],[37,-34.7456],[38,-34.6597],[44,-34.7122],[46,-34.8243],[54,-35.4438],[88,-35.6219],[90,-35.8474],[93,-36.4974],[95,-36.7137],[97,-36.7147],[100,-36.3813],[101,-36.1772],[107,-35.469],[122,-35.1059],[123,-35.1599],[130,-35.0514],[186,-35.5106],[189,-35.718],[191,-36.1297],[192,-36.4556],[193,-36.6844],[194,-36.6684],[195,-36.4892],[197,-35.9571],[199,-35.6411],[247,-35.2117],[253,-34.9841],[266,-35.8294],[268,-35.8699],[274,-35.7656],[279,-35.4351],[290,-35.1699],[298,-35.1415],[324,-35.0657]]
       }
       arm_l_pv_ik_ctrl.Scale.X {type=MovieSceneScriptingFloatChannel, default=1, const [1,325]}
       arm_l_pv_ik_ctrl.Scale.Y {type=MovieSceneScriptingFloatChannel, default=1, const [1,325]}
       arm_l_pv_ik_ctrl.Scale.Z {type=MovieSceneScriptingFloatChannel, default=1, const [1,325]}
       upperarm_r_fk_ctrl.Location.X {type=MovieSceneScriptingFloatChannel, default=0, thinned_from=325, keys [[0,0],[324,0]]}
       upperarm_r_fk_ctrl.Location.Y {type=MovieSceneScriptingFloatChannel, default=0, thinned_from=325, keys [[0,0],[324,0]]}
       upperarm_r_fk_ctrl.Location.Z {type=MovieSceneScriptingFloatChannel, default=0, thinned_from=325, keys [[0,0],[324,0]]}
       upperarm_r_fk_ctrl.Rotation.X {
        type=MovieSceneScriptingFloatChannel
        default=0
        thinned_from=325
        keys [[0,14.591],[4,14.5162],[9,14.2538],[17,13.6905],[23,13.3592],[26,13.2759],[28,13.2857],[29,13.3153],[32,13.4822],[33,13.5071],[34,13.4754],[35,13.3016],[36,12.7333],[37,12.0796],[38,11.6984],[39,11.4996],[41,11.3755],[43,11.334],[45,11.3395],[72,11.9654],[83,12.0963],[87,12.0142],[88,12.0332],[89,12.1366],[90,12.3482],[91,12.7475],[92,13.2465],[93,13.6412],[94,13.8826],[95,14.0479],[96,14.1502],[97,14.1903],[98,14.1504],[100,13.9658],[103,13.8135],[110,13.7726],[128,13.8863],[138,14.2364],[169,14.3604],[173,14.2379],[177,13.9982],[183,13.479],[186,13.1805],[187,13.1205],[188,13.1337],[189,13.2908],[190,13.6901],[191,14.2381],[192,14.7034],[193,14.9984],[194,15.2049],[195,15.3295],[203,15.901],[210,16.5779],[221,17.4601],[226,17.5874],[243,17.6289],[244,17.6],[245,17.5284],[246,17.3649],[247,17.0431],[248,16.3554],[249,15.8575],[253,15.5253],[258,14.7864],[262,14.3198],[265,14.0762],[276,13.6972],[279,13.6669],[280,13.6864],[287,13.7316],[304,14.3303],[324,14.5892]]
       }
       upperarm_r_fk_ctrl.Rotation.Y {
        type=MovieSceneScriptingFloatChannel
        default=0
        thinned_from=325
        keys [[0,-5.2004],[11,-5.1554],[14,-5.0977],[23,-4.7984],[25,-4.7148],[26,-4.684],[27,-4.6312],[30,-4.5165],[31,-4.4926],[32,-4.5101],[33,-4.5636],[34,-4.6678],[35,-4.788],[36,-4.9568],[37,-5.1474],[38,-5.2669],[40,-5.4335],[41,-5.5459],[45,-6.0316],[46,-6.1345],[47,-6.2126],[48,-6.2552],[50,-6.2958],[52,-6.2897],[55,-6.2239],[57,-6.1676],[75,-5.9265],[76,-5.9044],[77,-5.9004],[84,-5.7149],[85,-5.7553],[86,-5.837],[89,-6.1344],[90,-6.2196],[91,-6.2407],[92,-6.2053],[93,-6.1832],[95,-6.1785],[96,-6.1443],[98,-6.0274],[99,-5.9594],[101,-5.79],[102,-5.7278],[103,-5.6871],[104,-5.6722],[105,-5.6809],[106,-5.7049],[107,-5.7437],[110,-5.9093],[113,-6.0271],[117,-6.084],[125,-6.2846],[126,-6.2947],[127,-6.2827],[128,-6.2469],[129,-6.1951],[131,-6.0564],[135,-5.7289],[139,-5.4575],[142,-5.289],[143,-5.2524],[146,-5.1845],[148,-5.1627],[168,-5.2265],[169,-5.251],[171,-5.2713],[175,-5.3695],[184,-5.7288],[185,-5.7389],[187,-5.7156],[188,-5.7206],[191,-5.8294],[192,-5.9242],[194,-6.2385],[195,-6.3652],[196,-6.4702],[197,-6.5268],[198,-6.5511],[199,-6.5479],[201,-6.4971],[203,-6.4247],[209,-6.291],[210,-6.2787],[211,-6.2467],[221,-6.0459],[226,-6.0078],[227,-6.0078],[231,-5.9494],[240,-5.6642],[241,-5.613],[243,-5.6031],[244,-5.6124],[245,-5.6563],[246,-5.6641],[247,-5.7045],[248,-5.822],[249,-5.858],[250,-5.683],[251,-5.4692],[252,-5.2756],[253,-5.0999],[254,-4.9736],[255,-4.8956],[256,-4.9078],[257,-5.0056],[258,-5.1813],[259,-5.3979],[260,-5.629],[261,-5.8435],[262,-6.0431],[263,-6.2037],[264,-6.3397],[265,-6.4414],[266,-6.495],[267,-6.5202],[268,-6.5174],[269,-6.4868],[270,-6.4183],[271,-6.3182],[273,-6.0787],[274,-5.9918],[275,-5.9193],[277,-5.8174],[278,-5.7838],[280,-5.7514],[286,-5.7117],[287,-5.7214],[288,-5.7551],[289,-5.8091],[293,-6.0943],[294,-6.1511],[295,-6.1886],[296,-6.2036],[297,-6.187],[299,-6.1193],[301,-6.0208],[306,-5.686],[310,-5.4957],[314,-5.3404],[317,-5.2542],[320,-5.2105],[324,-5.2007]]
       }
       upperarm_r_fk_ctrl.Rotation.Z {
        type=MovieSceneScriptingFloatChannel
        default=0
        thinned_from=325
        keys [[0,-11.8934],[2,-11.8456],[4,-11.7113],[6,-11.5037],[8,-11.236],[14,-10.2556],[15,-10.119],[16,-10.0043],[21,-9.6059],[23,-9.4946],[24,-9.496],[25,-9.543],[26,-9.6155],[32,-10.2387],[33,-10.3546],[34,-10.506],[35,-10.715],[36,-11.0691],[37,-11.4476],[38,-11.7113],[39,-11.8947],[40,-12.0062],[42,-12.1544],[43,-12.2081],[44,-12.1724],[45,-12.0424],[46,-11.8668],[50,-11.0333],[51,-10.8447],[52,-10.6847],[53,-10.5493],[54,-10.439],[55,-10.3539],[57,-10.2794],[58,-10.2623],[61,-10.3108],[75,-10.9395],[78,-11.1689],[81,-11.4818],[83,-11.7496],[84,-11.9324],[85,-12.0803],[86,-12.1274],[87,-12.0781],[88,-11.8825],[89,-11.5529],[90,-11.1019],[91,-10.5277],[92,-9.8302],[93,-9.19],[94,-8.7476],[95,-8.4308],[96,-8.3116],[97,-8.3541],[98,-8.4826],[99,-8.7169],[100,-9.0643],[101,-9.442],[102,-9.7557],[103,-10.0133],[104,-10.2187],[105,-10.374],[106,-10.4838],[107,-10.5564],[111,-10.6498],[114,-10.6433],[117,-10.5589],[119,-10.435],[123,-10.1505],[126,-10.0077],[127,-9.9847],[128,-9.9906],[129,-10.032],[130,-10.1021],[132,-10.3018],[136,-10.751],[138,-10.9022],[139,-10.9415],[142,-10.92],[147,-10.745],[151,-10.5445],[157,-10.3945],[158,-10.3945],[159,-10.3708],[164,-10.3324],[169,-10.2406],[170,-10.1864],[171,-10.1049],[172,-9.9972],[173,-9.8645],[174,-9.7075],[175,-9.5271],[176,-9.3243],[177,-9.099],[178,-8.8422],[182,-7.6155],[183,-7.321],[184,-7.0569],[187,-6.337],[188,-6.0666],[189,-5.7268],[190,-5.3064],[192,-4.4261],[193,-3.9251],[194,-3.4522],[195,-3.0794],[196,-2.8099],[197,-2.6849],[198,-2.652],[199,-2.7369],[200,-2.9521],[201,-3.245],[202,-3.6545],[203,-4.1685],[204,-4.7232],[205,-5.2901],[208,-6.9458],[209,-7.4625],[210,-7.9031],[212,-8.5512],[216,-9.7571],[217,-10.0334],[219,-10.5278],[220,-10.7552],[221,-10.9485],[222,-11.0892],[223,-11.1656],[225,-11.21],[227,-11.2102],[232,-11.0511],[237,-11.0304],[240,-11.0821],[243,-11.1734],[244,-11.28],[245,-11.5354],[246,-11.9988],[247,-12.6714],[248,-13.7335],[249,-14.9443],[250,-15.8665],[251,-16.5905],[252,-17.0527],[253,-17.2653],[254,-17.0682],[255,-16.5687],[256,-15.7848],[257,-14.8179],[258,-13.8025],[259,-12.8702],[260,-11.9883],[261,-11.2005],[262,-10.4976],[263,-9.9609],[264,-9.543],[265,-9.2554],[266,-9.1235],[267,-9.0578],[268,-9.059],[269,-9.1362],[270,-9.2745],[273,-9.7513],[275,-9.9791],[277,-10.1566],[278,-10.2287],[294,-10.82],[296,-10.9348],[300,-11.3348],[302,-11.473],[304,-11.5396],[306,-11.4985],[308,-11.3931],[310,-11.345],[312,-11.3567],[315,-11.5206],[319,-11.7742],[322,-11.874],[324,-11.8885]]
       }
       upperarm_r_fk_ctrl.Scale.X {type=MovieSceneScriptingFloatChannel, default=1, const [1,325]}
       upperarm_r_fk_ctrl.Scale.Y {type=MovieSceneScriptingFloatChannel, default=1, const [1,325]}
       upperarm_r_fk_ctrl.Scale.Z {type=MovieSceneScriptingFloatChannel, default=1, const [1,325]}
       lowerarm_r_fk_ctrl.Location.X {
        type=MovieSceneScriptingFloatChannel
        default=0
        thinned_from=325
        keys [[0,-0.0001],[157,-0.0001],[158,0.0014],[159,-0.0001],[160,0.0016],[161,-0.0001],[162,0.0017],[163,-0.0001],[164,0.0022],[165,-0.0001],[226,-0.0001],[227,0.0015],[228,-0.0001],[323,-0.0001],[324,-0.0014]]
       }
       lowerarm_r_fk_ctrl.Location.Y {
        type=MovieSceneScriptingFloatChannel
        default=0
        thinned_from=325
        keys [[0,0],[157,0],[158,0.003],[159,0],[160,0.0032],[161,0],[162,0.0033],[163,0],[164,0.0043],[165,0],[226,0],[227,0.0032],[228,0],[323,0],[324,-0.0025]]
       }
       lowerarm_r_fk_ctrl.Location.Z {
        type=MovieSceneScriptingFloatChannel
        default=0
        thinned_from=325
        keys [[0,0],[157,0],[158,0.0056],[159,0],[160,0.0035],[161,0],[162,0.0018],[163,0],[164,0.0006],[165,0],[226,0],[227,0.0055],[228,0],[323,0],[324,-0.0005]]
       }
       lowerarm_r_fk_ctrl.Rotation.X {
        type=MovieSceneScriptingFloatChannel
        default=0
        thinned_from=325
        keys [[0,-0.0556],[3,-0.05],[5,-0.0471],[6,-0.0459],[7,-0.0451],[8,-0.0447],[9,-0.0464],[10,-0.0508],[11,-0.057],[12,-0.0642],[13,-0.0716],[14,-0.0794],[15,-0.0875],[16,-0.0958],[17,-0.1044],[19,-0.1222],[21,-0.1409],[28,-0.2096],[29,-0.2179],[30,-0.2242],[31,-0.2286],[32,-0.2308],[33,-0.2313],[34,-0.2291],[35,-0.2212],[36,-0.2056],[37,-0.1863],[39,-0.1466],[40,-0.127],[41,-0.1101],[42,-0.0982],[43,-0.09],[44,-0.0853],[45,-0.0838],[46,-0.0849],[47,-0.0886],[48,-0.0952],[49,-0.1047],[50,-0.118],[51,-0.133],[52,-0.1475],[53,-0.1612],[54,-0.173],[55,-0.1831],[56,-0.1908],[58,-0.2042],[59,-0.2103],[61,-0.2216],[62,-0.2268],[63,-0.2313],[64,-0.2351],[65,-0.2381],[67,-0.2429],[69,-0.246],[72,-0.2488],[73,-0.2488],[74,-0.2493],[78,-0.2493],[79,-0.2503],[80,-0.2518],[81,-0.2541],[82,-0.2571],[83,-0.2616],[84,-0.2686],[86,-0.2857],[88,-0.3043],[89,-0.3131],[90,-0.3199],[91,-0.3252],[92,-0.3282],[93,-0.3306],[96,-0.3407],[97,-0.3425],[98,-0.3419],[99,-0.3398],[100,-0.336],[101,-0.3306],[102,-0.3235],[103,-0.3151],[104,-0.3059],[106,-0.2854],[107,-0.2743],[109,-0.2513],[110,-0.2408],[111,-0.2319],[112,-0.2259],[113,-0.2218],[114,-0.219],[115,-0.2174],[116,-0.2165],[117,-0.2165],[118,-0.2172],[119,-0.2193],[120,-0.222],[121,-0.2254],[123,-0.2332],[125,-0.2399],[128,-0.2482],[129,-0.2502],[131,-0.2528],[134,-0.2547],[137,-0.2551],[138,-0.2532],[139,-0.2495],[140,-0.2429],[141,-0.2326],[142,-0.2196],[143,-0.2054],[144,-0.1904],[147,-0.1438],[148,-0.1288],[149,-0.1145],[150,-0.1011],[151,-0.0888],[152,-0.0778],[153,-0.0681],[154,-0.0595],[155,-0.052],[156,-0.0453],[157,-0.0395],[158,-0.028],[159,-0.0301],[160,-0.0215],[161,-0.0224],[162,-0.0156],[163,-0.0158],[164,-0.0105],[165,-0.0099],[168,-0.0021],[174,0.0126],[175,0.0146],[177,0.0174],[178,0.0185],[179,0.0191],[180,0.0193],[181,0.0189],[182,0.0178],[183,0.0163],[184,0.0138],[185,0.0098],[186,0.0049],[187,-0.0006],[188,-0.0068],[189,-0.0137],[190,-0.0212],[191,-0.0295],[192,-0.0381],[193,-0.046],[194,-0.0533],[195,-0.0571],[196,-0.06],[197,-0.0623],[198,-0.0633],[200,-0.0633],[201,-0.0614],[202,-0.0577],[203,-0.0531],[204,-0.0479],[207,-0.0311],[209,-0.0203],[210,-0.0153],[211,-0.0107],[212,-0.0066],[213,-0.0028],[215,0.004],[216,0.0069],[217,0.0093],[218,0.0111],[219,0.0122],[220,0.0127],[221,0.0124],[222,0.0115],[223,0.0096],[226,0.003],[227,0.015],[228,-0.0015],[234,-0.0129],[236,-0.0163],[238,-0.0191],[240,-0.0213],[244,-0.0244],[247,-0.0258],[248,-0.0261],[251,-0.0261],[252,-0.0258],[253,-0.0225],[255,-0.0115],[256,-0.0068],[257,-0.0049],[258,-0.0046],[259,-0.0076],[260,-0.0156],[261,-0.029],[262,-0.0447],[263,-0.0587],[264,-0.0718],[265,-0.0835],[266,-0.094],[267,-0.1039],[268,-0.1118],[269,-0.1161],[270,-0.1173],[271,-0.1153],[272,-0.1099],[273,-0.1015],[274,-0.0909],[276,-0.0658],[277,-0.0551],[278,-0.0455],[279,-0.0368],[280,-0.0294],[281,-0.0235],[282,-0.0187],[283,-0.0147],[284,-0.0113],[285,-0.0085],[286,-0.006],[287,-0.0037],[289,0.0003],[290,0.0021],[291,0.0036],[292,0.0047],[293,0.0052],[294,0.0052],[295,0.0046],[296,0.0026],[297,-0.0009],[298,-0.0059],[299,-0.0125],[300,-0.0202],[302,-0.0366],[303,-0.0442],[304,-0.051],[305,-0.0564],[306,-0.0607],[307,-0.0643],[308,-0.0671],[309,-0.0694],[310,-0.071],[311,-0.0719],[313,-0.0719],[314,-0.0715],[315,-0.0704],[317,-0.0676],[323,-0.0572],[324,-0.0548]]
       }
       lowerarm_r_fk_ctrl.Rotation.Y {
        type=MovieSceneScriptingFloatChannel
        default=0
        thinned_from=325
        keys [[0,-0.1985],[3,-0.1804],[4,-0.1753],[5,-0.1708],[6,-0.1668],[7,-0.1639],[8,-0.1628],[9,-0.1683],[10,-0.1828],[11,-0.2032],[12,-0.2261],[14,-0.2735],[17,-0.3476],[25,-0.5496],[27,-0.5984],[28,-0.6216],[29,-0.6412],[30,-0.6558],[31,-0.6659],[32,-0.6709],[33,-0.6721],[34,-0.6671],[35,-0.6489],[36,-0.6121],[37,-0.5654],[38,-0.5158],[40,-0.4114],[41,-0.364],[42,-0.3297],[43,-0.3056],[44,-0.2914],[45,-0.2868],[46,-0.2901],[47,-0.3012],[48,-0.3209],[49,-0.3486],[50,-0.3864],[51,-0.4276],[52,-0.4664],[53,-0.5022],[54,-0.5323],[55,-0.5575],[56,-0.5765],[58,-0.6087],[59,-0.6234],[61,-0.6498],[62,-0.6618],[63,-0.6722],[64,-0.6807],[66,-0.6935],[68,-0.7025],[72,-0.7117],[78,-0.7129],[80,-0.7183],[81,-0.7235],[82,-0.7301],[83,-0.7399],[84,-0.7553],[89,-0.8488],[90,-0.8627],[91,-0.8733],[93,-0.8842],[96,-0.9042],[97,-0.9077],[98,-0.9065],[99,-0.9024],[100,-0.8949],[101,-0.8842],[102,-0.8699],[103,-0.8529],[104,-0.834],[106,-0.7911],[107,-0.7676],[109,-0.7172],[110,-0.6936],[111,-0.6735],[112,-0.6597],[113,-0.6502],[114,-0.6437],[115,-0.6399],[116,-0.6379],[117,-0.6379],[118,-0.6395],[119,-0.6443],[120,-0.6507],[124,-0.6845],[127,-0.7047],[129,-0.7148],[131,-0.7206],[134,-0.7247],[137,-0.7257],[138,-0.7214],[139,-0.7131],[140,-0.6984],[141,-0.6751],[142,-0.6451],[143,-0.6116],[144,-0.5754],[145,-0.537],[149,-0.3764],[150,-0.338],[151,-0.3018],[152,-0.2685],[153,-0.2384],[154,-0.2112],[155,-0.1867],[156,-0.1648],[157,-0.1453],[158,-0.1177],[159,-0.1126],[160,-0.0927],[161,-0.0856],[162,-0.0708],[163,-0.0616],[164,-0.0504],[165,-0.0399],[166,-0.0297],[167,-0.0199],[172,0.0278],[173,0.0376],[174,0.0478],[175,0.0559],[176,0.0618],[177,0.0674],[178,0.0717],[179,0.0744],[180,0.0754],[181,0.0736],[182,0.069],[183,0.0626],[184,0.0524],[185,0.0364],[186,0.0169],[187,-0.0047],[188,-0.0284],[189,-0.0539],[190,-0.0812],[191,-0.1107],[192,-0.1405],[193,-0.1672],[194,-0.191],[195,-0.2036],[196,-0.2127],[197,-0.22],[198,-0.2233],[200,-0.2233],[201,-0.2174],[202,-0.2052],[203,-0.1905],[204,-0.1733],[205,-0.1547],[209,-0.0781],[210,-0.0597],[211,-0.043],[212,-0.0275],[213,-0.0131],[214,0.0004],[215,0.0133],[216,0.0249],[217,0.0343],[218,0.0415],[219,0.0462],[220,0.048],[221,0.0468],[222,0.043],[223,0.0357],[226,0.0092],[227,0.0062],[228,-0.0079],[229,-0.0156],[230,-0.023],[234,-0.0511],[235,-0.0577],[236,-0.0636],[237,-0.0688],[238,-0.0735],[239,-0.0777],[240,-0.0815],[241,-0.0848],[242,-0.0876],[244,-0.0926],[245,-0.0947],[246,-0.0963],[247,-0.0977],[248,-0.0987],[251,-0.0987],[252,-0.0977],[253,-0.0857],[254,-0.0664],[255,-0.0459],[256,-0.0283],[257,-0.0212],[258,-0.0198],[259,-0.0313],[260,-0.0608],[261,-0.1088],[262,-0.1625],[263,-0.2085],[264,-0.2501],[265,-0.2859],[266,-0.3173],[267,-0.3464],[268,-0.3689],[269,-0.3809],[270,-0.3845],[271,-0.3787],[272,-0.3634],[273,-0.3393],[274,-0.3081],[275,-0.2704],[276,-0.2314],[277,-0.1969],[278,-0.1652],[279,-0.136],[280,-0.1102],[281,-0.0895],[282,-0.0723],[283,-0.0576],[284,-0.0452],[285,-0.0345],[286,-0.0251],[287,-0.0165],[288,-0.0086],[289,-0.0011],[290,0.0057],[291,0.0117],[292,0.0161],[293,0.0179],[294,0.0179],[295,0.0157],[296,0.008],[297,-0.0057],[298,-0.0248],[299,-0.0496],[300,-0.0776],[301,-0.1067],[302,-0.1353],[303,-0.1611],[304,-0.1836],[305,-0.2012],[306,-0.2152],[307,-0.2265],[308,-0.2354],[309,-0.2426],[310,-0.2477],[311,-0.2504],[313,-0.2504],[314,-0.2492],[315,-0.2458],[316,-0.2417],[318,-0.2319],[323,-0.2038],[324,-0.2003]]
       }
       lowerarm_r_fk_ctrl.Rotation.Z {
        type=MovieSceneScriptingFloatChannel
        default=0
        thinned_from=325
        keys [[0,2.6708],[3,2.4274],[4,2.3598],[5,2.2999],[6,2.2461],[7,2.2071],[8,2.1922],[9,2.2663],[10,2.4605],[11,2.7326],[12,3.0407],[13,3.3564],[15,4.0077],[19,5.3663],[26,7.7837],[28,8.4414],[29,8.7136],[30,8.9182],[31,9.0593],[32,9.1292],[33,9.145],[34,9.0758],[35,8.8212],[36,8.3089],[37,7.6627],[38,6.9776],[40,5.547],[41,4.9027],[42,4.4381],[43,4.1121],[44,3.9198],[45,3.8581],[46,3.9028],[47,4.0521],[48,4.3182],[49,4.6943],[50,5.2078],[51,5.7687],[52,6.2995],[53,6.7911],[54,7.2052],[55,7.5521],[56,7.8157],[58,8.2626],[59,8.4659],[61,8.8336],[62,9.0018],[63,9.1474],[64,9.2663],[66,9.4452],[68,9.5719],[71,9.6777],[72,9.6998],[78,9.7171],[80,9.7938],[81,9.8658],[82,9.9598],[83,10.0978],[84,10.3137],[87,11.1045],[88,11.3844],[89,11.6409],[90,11.8395],[91,11.9907],[93,12.148],[96,12.4345],[97,12.4849],[98,12.4676],[99,12.408],[100,12.3007],[101,12.147],[102,11.9422],[103,11.6995],[104,11.4296],[106,10.8207],[107,10.4883],[109,9.7777],[110,9.4473],[111,9.1651],[112,8.9729],[113,8.8395],[114,8.7498],[115,8.6965],[116,8.6681],[117,8.6681],[118,8.6911],[119,8.7573],[120,8.8474],[124,9.3191],[127,9.6018],[129,9.7435],[131,9.8262],[134,9.8829],[137,9.8973],[138,9.8363],[139,9.7207],[140,9.5133],[141,9.1881],[142,8.7687],[143,8.3026],[144,7.7997],[145,7.2697],[148,5.6131],[149,5.0711],[150,4.5503],[151,4.0604],[152,3.611],[153,3.206],[154,2.8407],[155,2.5127],[156,2.2194],[157,1.9583],[158,1.7339],[159,1.5226],[160,1.3447],[161,1.163],[162,1.007],[163,0.844],[164,0.7071],[165,0.5562],[166,0.4212],[167,0.2904],[168,0.1627],[169,0.037],[170,-0.0879],[171,-0.2133],[172,-0.3403],[173,-0.47],[174,-0.6036],[175,-0.7107],[176,-0.7893],[177,-0.8627],[178,-0.9189],[179,-0.9543],[180,-0.9675],[181,-0.9446],[182,-0.8839],[183,-0.7988],[184,-0.6654],[185,-0.4542],[186,-0.1954],[187,0.0898],[188,0.4037],[189,0.7418],[190,1.1039],[191,1.4966],[192,1.8942],[193,2.2508],[194,2.5694],[195,2.7387],[196,2.8607],[197,2.9593],[198,3.0036],[200,3.0036],[201,2.9241],[202,2.7607],[203,2.5631],[204,2.3325],[205,2.0843],[207,1.5706],[209,1.0626],[210,0.819],[211,0.5973],[212,0.3916],[213,0.2007],[214,0.0227],[215,-0.1489],[216,-0.3017],[217,-0.4264],[218,-0.5211],[219,-0.5827],[220,-0.6063],[221,-0.5907],[222,-0.5402],[223,-0.4442],[224,-0.3271],[225,-0.2123],[226,-0.0947],[227,0.0289],[229,0.2345],[230,0.3325],[231,0.4271],[233,0.6128],[234,0.7045],[235,0.7916],[236,0.8704],[237,0.9399],[238,1.002],[239,1.0577],[240,1.1082],[241,1.1524],[242,1.1901],[244,1.2565],[245,1.2841],[246,1.3053],[247,1.3235],[248,1.3369],[251,1.3369],[252,1.3239],[253,1.1647],[254,0.9072],[255,0.636],[256,0.4027],[257,0.3075],[258,0.2894],[259,0.4424],[260,0.8328],[261,1.4714],[262,2.1888],[263,2.804],[264,3.3631],[265,3.8461],[266,4.2698],[267,4.6632],[268,4.9697],[269,5.1324],[270,5.1811],[271,5.1027],[272,4.8948],[273,4.5671],[274,4.1458],[275,3.6371],[276,3.1111],[277,2.6492],[278,2.2248],[279,1.8337],[280,1.4904],[281,1.2147],[282,0.9859],[283,0.791],[284,0.6259],[285,0.4843],[286,0.3603],[287,0.2461],[288,0.1415],[289,0.0415],[290,-0.0483],[291,-0.1278],[292,-0.1859],[293,-0.2088],[294,-0.2088],[295,-0.1808],[296,-0.0785],[297,0.1031],[298,0.3556],[299,0.685],[300,1.0563],[301,1.4439],[302,1.8246],[303,2.1694],[304,2.471],[305,2.7057],[306,2.8937],[307,3.0457],[308,3.1656],[309,3.2623],[310,3.3309],[311,3.368],[313,3.368],[314,3.3515],[315,3.3059],[316,3.2508],[318,3.1191],[324,2.6646]]
       }
       lowerarm_r_fk_ctrl.Scale.X {type=MovieSceneScriptingFloatChannel, default=1, const [1,325]}
       lowerarm_r_fk_ctrl.Scale.Y {type=MovieSceneScriptingFloatChannel, default=1, const [1,325]}
       lowerarm_r_fk_ctrl.Scale.Z {type=MovieSceneScriptingFloatChannel, default=1, const [1,325]}
       hand_r_fk_ctrl.Location.X {type=MovieSceneScriptingFloatChannel, default=0, const [-0.0001,325]}
       hand_r_fk_ctrl.Location.Y {type=MovieSceneScriptingFloatChannel, default=0, const [0,325]}
       hand_r_fk_ctrl.Location.Z {type=MovieSceneScriptingFloatChannel, default=0, const [0,325]}
       hand_r_fk_ctrl.Rotation.X {type=MovieSceneScriptingFloatChannel, default=0, const [0.321,325]}
       hand_r_fk_ctrl.Rotation.Y {type=MovieSceneScriptingFloatChannel, default=0, const [4.7589,325]}
       hand_r_fk_ctrl.Rotation.Z {type=MovieSceneScriptingFloatChannel, default=0, const [-11.7032,325]}
       hand_r_fk_ctrl.Scale.X {type=MovieSceneScriptingFloatChannel, default=1, const [1,325]}
       hand_r_fk_ctrl.Scale.Y {type=MovieSceneScriptingFloatChannel, default=1, const [1,325]}
       hand_r_fk_ctrl.Scale.Z {type=MovieSceneScriptingFloatChannel, default=1, const [1,325]}
       thumb_01_r_ctrl.Location.X {type=MovieSceneScriptingFloatChannel, default=0, const [0.0685,325]}
       thumb_01_r_ctrl.Location.Y {type=MovieSceneScriptingFloatChannel, default=0, const [-0.3984,325]}
       thumb_01_r_ctrl.Location.Z {type=MovieSceneScriptingFloatChannel, default=0, const [-0.4179,325]}
       thumb_01_r_ctrl.Rotation.X {type=MovieSceneScriptingFloatChannel, default=0, thinned_from=325, keys [[0,9.0456],[324,9.0456]]}
       thumb_01_r_ctrl.Rotation.Y {type=MovieSceneScriptingFloatChannel, default=0, thinned_from=325, keys [[0,-15.3775],[324,-15.3775]]}
       thumb_01_r_ctrl.Rotation.Z {
        type=MovieSceneScriptingFloatChannel
        default=0
        thinned_from=325
        keys [[0,-6.5546],[1,-6.4773],[3,-6.1466],[5,-5.8995],[7,-5.6825],[10,-5.4086],[11,-5.327],[12,-5.2582],[13,-5.2001],[14,-5.1549],[15,-5.1367],[16,-5.1596],[17,-5.2264],[18,-5.3342],[19,-5.4956],[23,-6.2987],[24,-6.4738],[25,-6.6232],[26,-6.7487],[27,-6.8309],[29,-6.9134],[30,-6.9375],[31,-6.9375],[32,-6.9604],[34,-6.9604],[37,-6.8996],[39,-6.8241],[41,-6.6875],[43,-6.5112],[45,-6.2971],[50,-5.6182],[51,-5.5178],[52,-5.4618],[53,-5.4373],[54,-5.4373],[55,-5.4955],[56,-5.6412],[58,-5.9624],[59,-6.1381],[60,-6.2968],[61,-6.4322],[62,-6.5507],[63,-6.6406],[64,-6.6947],[65,-6.7202],[74,-6.7202],[75,-6.6942],[76,-6.6475],[78,-6.5366],[80,-6.4805],[82,-6.4805],[86,-6.6685],[91,-7.0016],[97,-7.4669],[99,-7.5864],[101,-7.6709],[110,-7.6709],[111,-7.6353],[112,-7.5567],[113,-7.4356],[114,-7.2914],[115,-7.1291],[116,-6.9437],[117,-6.7391],[118,-6.5105],[122,-5.5137],[123,-5.2898],[124,-5.1008],[125,-4.9384],[126,-4.8046],[127,-4.6924],[128,-4.6087],[129,-4.5759],[130,-4.5952],[131,-4.6694],[132,-4.8228],[133,-5.0848],[134,-5.4261],[135,-5.783],[136,-6.1534],[137,-6.5394],[138,-6.8839],[139,-7.1722],[140,-7.419],[141,-7.6109],[142,-7.7425],[143,-7.8192],[144,-7.8441],[159,-7.8441],[160,-7.8265],[162,-7.7288],[170,-7.142],[172,-7.0273],[174,-6.9547],[175,-6.9324],[176,-6.9324],[177,-6.9152],[178,-6.9152],[180,-6.9942],[189,-7.5767],[192,-7.7199],[194,-7.7685],[195,-7.7685],[196,-7.7971],[217,-7.7971],[221,-7.6016],[225,-7.3448],[227,-7.2561],[229,-7.2008],[230,-7.2008],[231,-7.1801],[235,-7.1801],[238,-7.3302],[244,-7.736],[246,-7.8381],[250,-7.9562],[251,-7.9734],[252,-7.9734],[253,-7.9996],[280,-7.9996],[281,-7.9817],[283,-7.8942],[284,-7.8295],[285,-7.7489],[286,-7.6491],[289,-7.2855],[291,-7.0136],[293,-6.6977],[294,-6.5546],[295,-6.4381],[296,-6.34],[297,-6.2609],[298,-6.2042],[300,-6.1385],[301,-6.1174],[305,-6.1174],[307,-6.1888],[313,-6.4667],[315,-6.5265],[316,-6.5265],[317,-6.5463],[324,-6.5463]]
       }
       thumb_01_r_ctrl.Scale.X {type=MovieSceneScriptingFloatChannel, default=1, const [1,325]}
       thumb_01_r_ctrl.Scale.Y {type=MovieSceneScriptingFloatChannel, default=1, const [1,325]}
       thumb_01_r_ctrl.Scale.Z {type=MovieSceneScriptingFloatChannel, default=1, const [1,325]}
       thumb_02_r_ctrl.Location.X {type=MovieSceneScriptingFloatChannel, default=0, thinned_from=325, keys [[0,0.0001],[324,0.0001]]}
       thumb_02_r_ctrl.Location.Y {type=MovieSceneScriptingFloatChannel, default=0, thinned_from=325, keys [[0,0],[324,0]]}
       thumb_02_r_ctrl.Location.Z {type=MovieSceneScriptingFloatChannel, default=0, thinned_from=325, keys [[0,0],[324,0]]}
       thumb_02_r_ctrl.Rotation.X {type=MovieSceneScriptingFloatChannel, default=0, thinned_from=325, keys [[0,-28.0109],[324,-28.0109]]}
       thumb_02_r_ctrl.Rotation.Y {type=MovieSceneScriptingFloatChannel, default=0, thinned_from=325, keys [[0,-14.6405],[324,-14.6405]]}
       thumb_02_r_ctrl.Rotation.Z {
        type=MovieSceneScriptingFloatChannel
        default=0
        thinned_from=325
        keys [[0,-20.6578],[1,-20.6319],[2,-20.5154],[3,-20.2497],[4,-19.7554],[6,-18.4992],[8,-17.4476],[9,-17.0258],[10,-16.7186],[11,-16.5],[12,-16.3504],[13,-16.2649],[15,-16.2333],[16,-16.2915],[17,-16.4513],[18,-16.6662],[19,-16.9281],[20,-17.245],[21,-17.5984],[22,-17.9901],[24,-18.8584],[26,-19.8591],[27,-20.2252],[29,-20.6611],[31,-20.9569],[32,-21.0376],[34,-21.0658],[40,-20.8768],[41,-20.8129],[42,-20.7061],[43,-20.5396],[45,-20.0995],[50,-18.7735],[51,-18.6113],[52,-18.5508],[54,-18.5285],[55,-18.5938],[56,-18.7681],[57,-19.0018],[58,-19.2986],[62,-20.7494],[63,-21.0445],[64,-21.2426],[65,-21.3668],[66,-21.4097],[74,-21.4097],[76,-21.3193],[77,-21.2095],[78,-21.048],[79,-20.8302],[81,-20.2529],[84,-19.1144],[86,-18.2528],[87,-17.8866],[88,-17.5952],[89,-17.3733],[90,-17.2851],[91,-17.3173],[92,-17.4398],[93,-17.692],[94,-18.1303],[98,-20.3539],[99,-20.8303],[100,-21.2208],[101,-21.5439],[102,-21.7892],[103,-21.9461],[104,-22.0252],[110,-22.0319],[112,-21.9216],[113,-21.7998],[114,-21.6308],[115,-21.416],[116,-21.1542],[117,-20.8423],[118,-20.4625],[122,-18.7512],[124,-18.0073],[125,-17.689],[126,-17.447],[127,-17.303],[128,-17.2352],[130,-17.2336],[131,-17.3181],[132,-17.5296],[133,-17.9094],[134,-18.4162],[135,-18.9839],[136,-19.6344],[137,-20.3458],[138,-20.987],[139,-21.5211],[140,-21.985],[141,-22.3585],[142,-22.621],[143,-22.7931],[145,-22.9988],[159,-23.0324],[160,-22.995],[162,-22.7267],[168,-21.3936],[170,-21.0196],[171,-20.8831],[172,-20.8101],[177,-20.7201],[179,-20.7913],[181,-20.9686],[182,-21.0933],[183,-21.285],[184,-21.5718],[188,-23.057],[189,-23.4887],[192,-25.0632],[194,-26.0045],[195,-26.4],[196,-26.7234],[197,-26.9848],[198,-27.179],[199,-27.2948],[207,-27.3435],[208,-27.3009],[209,-27.154],[210,-26.9283],[211,-26.6332],[212,-26.2594],[213,-25.795],[214,-25.2094],[218,-22.4565],[219,-21.8088],[220,-21.2117],[221,-20.6671],[222,-20.2438],[223,-19.9912],[224,-19.8598],[226,-19.6988],[228,-19.6691],[229,-19.6886],[231,-19.8666],[233,-20.1753],[234,-20.3979],[235,-20.6883],[242,-23.1443],[244,-23.7185],[247,-24.3448],[251,-24.9181],[256,-25.3038],[277,-25.4297],[278,-25.3602],[279,-25.2219],[280,-24.9852],[282,-24.3259],[285,-23.1176],[287,-22.2194],[288,-21.8137],[289,-21.4766],[290,-21.1884],[291,-20.9484],[292,-20.7639],[294,-20.5175],[297,-20.322],[304,-20.2054],[315,-20.6296],[324,-20.6494]]
       }
       thumb_02_r_ctrl.Scale.X {type=MovieSceneScriptingFloatChannel, default=1, const [1,325]}
       thumb_02_r_ctrl.Scale.Y {type=MovieSceneScriptingFloatChannel, default=1, const [1,325]}
       thumb_02_r_ctrl.Scale.Z {type=MovieSceneScriptingFloatChannel, default=1, const [1,325]}
       thumb_03_r_ctrl.Location.X {type=MovieSceneScriptingFloatChannel, default=0, thinned_from=325, keys [[0,0],[324,0]]}
       thumb_03_r_ctrl.Location.Y {type=MovieSceneScriptingFloatChannel, default=0, thinned_from=325, keys [[0,0],[324,0]]}
       thumb_03_r_ctrl.Location.Z {type=MovieSceneScriptingFloatChannel, default=0, thinned_from=325, keys [[0,0],[324,0]]}
       thumb_03_r_ctrl.Rotation.X {type=MovieSceneScriptingFloatChannel, default=0, thinned_from=325, keys [[0,-4.6289],[324,-4.6289]]}
       thumb_03_r_ctrl.Rotation.Y {type=MovieSceneScriptingFloatChannel, default=0, thinned_from=325, keys [[0,-5.8424],[324,-5.8424]]}
       thumb_03_r_ctrl.Rotation.Z {
        type=MovieSceneScriptingFloatChannel
        default=0
        thinned_from=325
        keys [[0,-69.3636],[2,-69.2212],[3,-68.9555],[4,-68.4612],[8,-66.1533],[10,-65.4243],[12,-65.0562],[15,-64.9391],[17,-65.1571],[20,-65.9508],[24,-67.5642],[26,-68.5649],[29,-69.3669],[32,-69.7434],[41,-69.5186],[46,-68.5498],[50,-67.4793],[54,-67.2343],[56,-67.4738],[63,-69.7503],[65,-70.0725],[76,-70.025],[79,-69.5359],[84,-67.8201],[87,-66.5924],[89,-66.0791],[91,-66.023],[93,-66.3977],[99,-69.536],[101,-70.2497],[103,-70.6519],[111,-70.702],[115,-70.1218],[118,-69.1682],[124,-66.7131],[126,-66.1527],[129,-65.924],[131,-66.0238],[132,-66.2353],[134,-67.1219],[138,-69.6928],[140,-70.6907],[142,-71.3268],[146,-71.7382],[160,-71.7007],[171,-69.5888],[179,-69.497],[183,-69.9907],[189,-72.1945],[194,-74.7102],[197,-75.6905],[200,-76.0492],[208,-76.0066],[211,-75.3389],[213,-74.5008],[220,-69.9174],[222,-68.9495],[226,-68.4045],[230,-68.4652],[234,-69.1036],[244,-72.4242],[254,-73.8941],[277,-74.1354],[279,-73.9277],[283,-72.6556],[288,-70.5194],[291,-69.6541],[304,-68.9112],[324,-69.3552]]
       }
       thumb_03_r_ctrl.Scale.X {type=MovieSceneScriptingFloatChannel, default=1, const [1,325]}
       thumb_03_r_ctrl.Scale.Y {type=MovieSceneScriptingFloatChannel, default=1, const [1,325]}
       thumb_03_r_ctrl.Scale.Z {type=MovieSceneScriptingFloatChannel, default=1, const [1,325]}
       index_metacarpal_r_ctrl.Location.X {type=MovieSceneScriptingFloatChannel, default=0, const [0.0001,325]}
       index_metacarpal_r_ctrl.Location.Y {type=MovieSceneScriptingFloatChannel, default=0, const [0.0002,325]}
       index_metacarpal_r_ctrl.Location.Z {type=MovieSceneScriptingFloatChannel, default=0, const [0,325]}
       index_metacarpal_r_ctrl.Rotation.X {type=MovieSceneScriptingFloatChannel, default=0, thinned_from=325, keys [[0,-2.9693],[324,-2.9693]]}
       index_metacarpal_r_ctrl.Rotation.Y {type=MovieSceneScriptingFloatChannel, default=0, thinned_from=325, keys [[0,3.5534],[324,3.5535]]}
       index_metacarpal_r_ctrl.Rotation.Z {
        type=MovieSceneScriptingFloatChannel
        default=0
        thinned_from=325
        keys [[0,-13.3204],[1,-13.243],[3,-12.9123],[7,-12.4482],[11,-12.0927],[13,-11.9658],[14,-11.9206],[15,-11.9024],[16,-11.9253],[17,-11.9921],[18,-12.0999],[19,-12.2613],[23,-13.0644],[25,-13.3889],[26,-13.5144],[27,-13.5967],[30,-13.7032],[35,-13.7184],[39,-13.5898],[42,-13.3687],[45,-13.0628],[50,-12.3839],[51,-12.2835],[52,-12.2275],[53,-12.203],[54,-12.203],[55,-12.2612],[58,-12.7281],[60,-13.0625],[62,-13.3164],[63,-13.4063],[64,-13.4604],[74,-13.4859],[76,-13.4132],[78,-13.3024],[80,-13.2462],[82,-13.2462],[88,-13.5554],[98,-14.2974],[101,-14.4367],[109,-14.4517],[111,-14.401],[112,-14.3224],[114,-14.0571],[116,-13.7094],[118,-13.2762],[122,-12.2794],[123,-12.0555],[124,-11.8665],[125,-11.7042],[127,-11.4581],[128,-11.3744],[129,-11.3416],[130,-11.3609],[131,-11.4351],[132,-11.5885],[133,-11.8505],[135,-12.5487],[137,-13.3051],[138,-13.6496],[139,-13.9379],[140,-14.1847],[141,-14.3766],[142,-14.5082],[143,-14.5849],[159,-14.6098],[161,-14.549],[171,-13.8439],[174,-13.7204],[178,-13.6852],[181,-13.8137],[191,-14.4454],[196,-14.5628],[216,-14.5745],[221,-14.3673],[226,-14.0618],[230,-13.9512],[235,-13.9512],[241,-14.298],[246,-14.6039],[253,-14.7653],[280,-14.7778],[283,-14.66],[285,-14.5146],[290,-13.9187],[294,-13.3204],[296,-13.1057],[298,-12.9699],[301,-12.8831],[304,-12.8715],[307,-12.9545],[314,-13.2688],[324,-13.3168]]
       }
       index_metacarpal_r_ctrl.Scale.X {type=MovieSceneScriptingFloatChannel, default=1, const [1,325]}
       index_metacarpal_r_ctrl.Scale.Y {type=MovieSceneScriptingFloatChannel, default=1, const [1,325]}
       index_metacarpal_r_ctrl.Scale.Z {type=MovieSceneScriptingFloatChannel, default=1, const [1,325]}
       index_01_r_ctrl.Location.X {type=MovieSceneScriptingFloatChannel, default=0, thinned_from=325, keys [[0,-0.4363],[324,-0.4363]]}
       index_01_r_ctrl.Location.Y {type=MovieSceneScriptingFloatChannel, default=0, thinned_from=325, keys [[0,-1.0252],[324,-1.0252]]}
       index_01_r_ctrl.Location.Z {type=MovieSceneScriptingFloatChannel, default=0, thinned_from=325, keys [[0,-0.034],[324,-0.034]]}
       index_01_r_ctrl.Rotation.X {type=MovieSceneScriptingFloatChannel, default=0, thinned_from=325, keys [[0,8.4656],[324,8.4656]]}
       index_01_r_ctrl.Rotation.Y {type=MovieSceneScriptingFloatChannel, default=0, thinned_from=325, keys [[0,-1.3715],[324,-1.3714]]}
       index_01_r_ctrl.Rotation.Z {
        type=MovieSceneScriptingFloatChannel
        default=0
        thinned_from=325
        keys [[0,-66.4322],[2,-66.2898],[3,-66.0241],[4,-65.5298],[7,-63.7324],[9,-62.8002],[11,-62.2744],[14,-62.0077],[16,-62.0659],[19,-62.7025],[24,-64.6328],[26,-65.6335],[27,-65.9996],[31,-66.7313],[41,-66.5873],[46,-65.6184],[50,-64.5479],[54,-64.3029],[56,-64.5425],[63,-66.8189],[65,-67.1412],[76,-67.0937],[79,-66.6046],[84,-64.8888],[87,-63.661],[89,-63.1477],[91,-63.0917],[92,-63.2142],[93,-63.4663],[94,-63.9047],[99,-66.6047],[101,-67.3183],[103,-67.7206],[111,-67.7706],[114,-67.4053],[117,-66.6167],[124,-63.7817],[126,-63.2214],[129,-62.9926],[131,-63.0925],[132,-63.304],[134,-64.1906],[136,-65.4088],[138,-66.7614],[140,-67.7594],[142,-68.3954],[145,-68.7733],[160,-68.7694],[171,-66.6575],[179,-66.5657],[182,-66.8677],[184,-67.3462],[189,-69.2632],[194,-71.7789],[197,-72.7592],[200,-73.1179],[208,-73.0753],[211,-72.4076],[213,-71.5694],[220,-66.9861],[222,-66.0182],[224,-65.6342],[229,-65.4631],[233,-65.9497],[244,-69.4929],[253,-70.8894],[277,-71.2041],[279,-70.9963],[283,-69.7243],[288,-67.5881],[291,-66.7228],[304,-65.9799],[324,-66.4239]]
       }
       index_01_r_ctrl.Scale.X {type=MovieSceneScriptingFloatChannel, default=1, const [1,325]}
       index_01_r_ctrl.Scale.Y {type=MovieSceneScriptingFloatChannel, default=1, const [1,325]}
       index_01_r_ctrl.Scale.Z {type=MovieSceneScriptingFloatChannel, default=1, const [1,325]}
       index_02_r_ctrl.Location.X {type=MovieSceneScriptingFloatChannel, default=0, thinned_from=325, keys [[0,-0.0001],[324,-0.0001]]}
       index_02_r_ctrl.Location.Y {type=MovieSceneScriptingFloatChannel, default=0, thinned_from=325, keys [[0,-0.0001],[324,-0.0001]]}
       index_02_r_ctrl.Location.Z {type=MovieSceneScriptingFloatChannel, default=0, thinned_from=325, keys [[0,-0.0001],[324,-0.0001]]}
       index_02_r_ctrl.Rotation.X {type=MovieSceneScriptingFloatChannel, default=0, thinned_from=325, keys [[0,2.7848],[324,2.7848]]}
       index_02_r_ctrl.Rotation.Y {type=MovieSceneScriptingFloatChannel, default=0, thinned_from=325, keys [[0,-2.6518],[324,-2.6518]]}
       index_02_r_ctrl.Rotation.Z {
        type=MovieSceneScriptingFloatChannel
        default=0
        thinned_from=325
        keys [[0,-67.2511],[2,-67.1087],[3,-66.843],[4,-66.3487],[7,-64.5513],[9,-63.6191],[11,-63.0932],[14,-62.8266],[16,-62.8848],[19,-63.5214],[24,-65.4517],[26,-66.4524],[29,-67.2544],[32,-67.6309],[41,-67.4062],[46,-66.4373],[50,-65.3668],[54,-65.1218],[56,-65.3614],[63,-67.6378],[65,-67.9601],[76,-67.9126],[79,-67.4235],[84,-65.7076],[87,-64.4799],[89,-63.9666],[91,-63.9105],[92,-64.0331],[93,-64.2852],[94,-64.7236],[99,-67.4236],[101,-68.1372],[103,-68.5394],[111,-68.5895],[114,-68.2241],[117,-67.4356],[124,-64.6006],[126,-64.0403],[129,-63.8115],[131,-63.9114],[132,-64.1229],[134,-65.0094],[136,-66.2276],[138,-67.5803],[140,-68.5783],[142,-69.2143],[146,-69.6257],[160,-69.5883],[171,-67.4763],[179,-67.3846],[182,-67.6866],[184,-68.1651],[189,-70.082],[194,-72.5978],[197,-73.5781],[200,-73.9367],[208,-73.8941],[211,-73.2265],[213,-72.3883],[220,-67.805],[222,-66.8371],[224,-66.4531],[229,-66.2819],[233,-66.7686],[244,-70.3118],[253,-71.7082],[277,-72.0229],[279,-71.8152],[283,-70.5431],[288,-68.4069],[291,-67.5417],[304,-66.7987],[324,-67.2428]]
       }
       index_02_r_ctrl.Scale.X {type=MovieSceneScriptingFloatChannel, default=1, const [1,325]}
       index_02_r_ctrl.Scale.Y {type=MovieSceneScriptingFloatChannel, default=1, const [1,325]}
       index_02_r_ctrl.Scale.Z {type=MovieSceneScriptingFloatChannel, default=1, const [1,325]}
       index_03_r_ctrl.Location.X {type=MovieSceneScriptingFloatChannel, default=0, thinned_from=325, keys [[0,-0.0001],[324,-0.0001]]}
       index_03_r_ctrl.Location.Y {type=MovieSceneScriptingFloatChannel, default=0, thinned_from=325, keys [[0,0.0001],[324,0.0001]]}
       index_03_r_ctrl.Location.Z {type=MovieSceneScriptingFloatChannel, default=0, thinned_from=325, keys [[0,-0.0001],[324,-0.0001]]}
       index_03_r_ctrl.Rotation.X {type=MovieSceneScriptingFloatChannel, default=0, thinned_from=325, keys [[0,0],[324,0]]}
       index_03_r_ctrl.Rotation.Y {type=MovieSceneScriptingFloatChannel, default=0, thinned_from=325, keys [[0,0],[324,0]]}
       index_03_r_ctrl.Rotation.Z {
        type=MovieSceneScriptingFloatChannel
        default=0
        thinned_from=325
        keys [[0,-90.9605],[2,-90.818],[3,-90.5524],[8,-87.7502],[10,-87.0212],[13,-86.5675],[16,-86.5941],[20,-87.5476],[27,-90.5278],[32,-91.3403],[42,-91.0088],[51,-88.9139],[55,-88.8964],[58,-89.6012],[63,-91.3472],[76,-91.6219],[80,-90.8633],[88,-87.8978],[90,-87.5877],[92,-87.7424],[93,-87.9946],[99,-91.133],[102,-92.0919],[112,-92.2243],[116,-91.4569],[125,-87.9917],[128,-87.5378],[131,-87.6207],[133,-88.212],[136,-89.937],[138,-91.2897],[141,-92.6612],[145,-93.3015],[160,-93.2976],[171,-91.1857],[181,-91.2712],[184,-91.8745],[189,-93.7914],[195,-96.7026],[198,-97.4817],[208,-97.6035],[211,-96.9358],[213,-96.0977],[220,-91.5144],[222,-90.5464],[227,-89.9718],[233,-90.478],[245,-94.2527],[277,-95.7323],[280,-95.2879],[290,-91.4911],[306,-90.5564],[324,-90.9521]]
       }
       index_03_r_ctrl.Scale.X {type=MovieSceneScriptingFloatChannel, default=1, const [1,325]}
       index_03_r_ctrl.Scale.Y {type=MovieSceneScriptingFloatChannel, default=1, const [1,325]}
       index_03_r_ctrl.Scale.Z {type=MovieSceneScriptingFloatChannel, default=1, const [1,325]}
       middle_metacarpal_r_ctrl.Location.X {type=MovieSceneScriptingFloatChannel, default=0, const [0,325]}
       middle_metacarpal_r_ctrl.Location.Y {type=MovieSceneScriptingFloatChannel, default=0, const [0.0001,325]}
       middle_metacarpal_r_ctrl.Location.Z {type=MovieSceneScriptingFloatChannel, default=0, const [0,325]}
       middle_metacarpal_r_ctrl.Rotation.X {type=MovieSceneScriptingFloatChannel, default=0, thinned_from=325, keys [[0,0.2131],[324,0.2131]]}
       middle_metacarpal_r_ctrl.Rotation.Y {type=MovieSceneScriptingFloatChannel, default=0, thinned_from=325, keys [[0,-1.1482],[324,-1.1482]]}
       middle_metacarpal_r_ctrl.Rotation.Z {
        type=MovieSceneScriptingFloatChannel
        default=0
        thinned_from=325
        keys [[0,-10.5148],[1,-10.4375],[3,-10.1067],[6,-9.7466],[10,-9.3688],[12,-9.2184],[14,-9.1151],[15,-9.0968],[16,-9.1197],[17,-9.1865],[18,-9.2944],[19,-9.4558],[23,-10.2589],[24,-10.434],[25,-10.5834],[26,-10.7088],[27,-10.7911],[29,-10.8736],[31,-10.9128],[35,-10.9128],[38,-10.8287],[40,-10.7227],[43,-10.4714],[45,-10.2573],[50,-9.5784],[51,-9.478],[52,-9.422],[53,-9.3975],[54,-9.3975],[55,-9.4557],[58,-9.9226],[60,-10.257],[62,-10.5109],[63,-10.6008],[64,-10.6549],[65,-10.6804],[74,-10.6804],[76,-10.6077],[78,-10.4968],[80,-10.4407],[82,-10.4407],[87,-10.6867],[98,-11.4919],[101,-11.6311],[109,-11.6462],[111,-11.5954],[112,-11.5169],[113,-11.3958],[115,-11.0893],[117,-10.6993],[122,-9.4739],[123,-9.2499],[124,-9.061],[125,-8.8986],[126,-8.7648],[127,-8.6526],[128,-8.5689],[129,-8.536],[130,-8.5553],[131,-8.6295],[132,-8.783],[133,-9.045],[135,-9.7432],[137,-10.4996],[138,-10.8441],[139,-11.1324],[140,-11.3792],[141,-11.5711],[142,-11.7027],[143,-11.7794],[144,-11.8043],[159,-11.8043],[160,-11.7867],[163,-11.6246],[171,-11.0383],[174,-10.9149],[176,-10.8796],[178,-10.8796],[181,-11.0082],[190,-11.5909],[193,-11.7088],[196,-11.7572],[216,-11.7689],[218,-11.7123],[223,-11.4364],[225,-11.305],[228,-11.1847],[230,-11.1457],[235,-11.1457],[238,-11.2904],[245,-11.7509],[250,-11.9164],[280,-11.9722],[283,-11.8544],[285,-11.7091],[289,-11.2457],[291,-10.9738],[294,-10.5148],[296,-10.3002],[297,-10.2211],[299,-10.1268],[302,-10.066],[304,-10.066],[306,-10.1108],[314,-10.4632],[316,-10.4985],[324,-10.5113]]
       }
       middle_metacarpal_r_ctrl.Scale.X {type=MovieSceneScriptingFloatChannel, default=1, const [1,325]}
       middle_metacarpal_r_ctrl.Scale.Y {type=MovieSceneScriptingFloatChannel, default=1, const [1,325]}
       middle_metacarpal_r_ctrl.Scale.Z {type=MovieSceneScriptingFloatChannel, default=1, const [1,325]}
       middle_01_r_ctrl.Location.X {type=MovieSceneScriptingFloatChannel, default=0, const [-0.553,325]}
       middle_01_r_ctrl.Location.Y {type=MovieSceneScriptingFloatChannel, default=0, thinned_from=325, keys [[0,-0.9634],[324,-0.9634]]}
       middle_01_r_ctrl.Location.Z {type=MovieSceneScriptingFloatChannel, default=0, thinned_from=325, keys [[0,-0.0914],[324,-0.0914]]}
       middle_01_r_ctrl.Rotation.X {type=MovieSceneScriptingFloatChannel, default=0, thinned_from=325, keys [[0,0],[324,0]]}
       middle_01_r_ctrl.Rotation.Y {type=MovieSceneScriptingFloatChannel, default=0, thinned_from=325, keys [[0,0],[324,0]]}
       middle_01_r_ctrl.Rotation.Z {
        type=MovieSceneScriptingFloatChannel
        default=0
        thinned_from=325
        keys [[0,-80.8858],[2,-80.7434],[3,-80.4777],[8,-77.6755],[10,-76.9465],[13,-76.4928],[16,-76.5195],[19,-77.1561],[24,-79.0864],[26,-80.0871],[30,-81.0574],[41,-81.0409],[51,-78.8393],[55,-78.8217],[58,-79.5266],[63,-81.2725],[66,-81.6377],[76,-81.5472],[80,-80.7886],[87,-78.1146],[89,-77.6013],[91,-77.5452],[93,-77.9199],[99,-81.0583],[102,-82.0172],[111,-82.2242],[115,-81.644],[118,-80.6905],[125,-77.917],[127,-77.5309],[130,-77.4462],[132,-77.7576],[134,-78.6441],[138,-81.215],[140,-82.213],[142,-82.849],[160,-83.223],[171,-81.111],[179,-81.0192],[183,-81.5129],[189,-83.7167],[194,-86.2325],[197,-87.2128],[200,-87.5714],[208,-87.5288],[211,-86.8612],[213,-86.023],[220,-81.4397],[222,-80.4718],[226,-79.9267],[231,-80.0946],[234,-80.6259],[244,-83.9465],[259,-85.6368],[278,-85.5882],[281,-84.9009],[289,-81.7046],[293,-80.8556],[324,-80.8774]]
       }
       middle_01_r_ctrl.Scale.X {type=MovieSceneScriptingFloatChannel, default=1, const [1,325]}
       middle_01_r_ctrl.Scale.Y {type=MovieSceneScriptingFloatChannel, default=1, const [1,325]}
       middle_01_r_ctrl.Scale.Z {type=MovieSceneScriptingFloatChannel, default=1, const [1,325]}
       middle_02_r_ctrl.Location.X {type=MovieSceneScriptingFloatChannel, default=0, thinned_from=325, keys [[0,0.0008],[324,0.0008]]}
       middle_02_r_ctrl.Location.Y {type=MovieSceneScriptingFloatChannel, default=0, thinned_from=325, keys [[0,0.0009],[324,0.0009]]}
       middle_02_r_ctrl.Location.Z {type=MovieSceneScriptingFloatChannel, default=0, thinned_from=325, keys [[0,0.0476],[324,0.0476]]}
       middle_02_r_ctrl.Rotation.X {type=MovieSceneScriptingFloatChannel, default=0, thinned_from=325, keys [[0,-3.6824],[324,-3.6824]]}
       middle_02_r_ctrl.Rotation.Y {type=MovieSceneScriptingFloatChannel, default=0, thinned_from=325, keys [[0,0.451],[324,0.451]]}
       middle_02_r_ctrl.Rotation.Z {
        type=MovieSceneScriptingFloatChannel
        default=0
        thinned_from=325
        keys [[0,-62.569],[2,-62.4266],[3,-62.1609],[4,-61.6666],[7,-59.8691],[9,-58.9369],[11,-58.4111],[14,-58.1445],[16,-58.2026],[19,-58.8392],[24,-60.7696],[26,-61.7703],[27,-62.1364],[31,-62.868],[41,-62.724],[46,-61.7552],[50,-60.6847],[53,-60.4397],[55,-60.5049],[57,-60.913],[63,-62.9557],[65,-63.2779],[76,-63.2304],[79,-62.7413],[84,-61.0255],[87,-59.7978],[89,-59.2845],[90,-59.1963],[92,-59.3509],[93,-59.6031],[94,-60.0415],[98,-62.265],[100,-63.132],[102,-63.7003],[104,-63.9364],[112,-63.8328],[115,-63.3272],[118,-62.3736],[124,-59.9185],[126,-59.3581],[128,-59.1463],[131,-59.2292],[132,-59.4407],[133,-59.8205],[135,-60.8951],[138,-62.8982],[140,-63.8962],[142,-64.5322],[145,-64.91],[160,-64.9061],[163,-64.4346],[170,-62.9308],[178,-62.6499],[182,-63.0045],[184,-63.483],[189,-65.3999],[194,-67.9157],[197,-68.8959],[199,-69.206],[208,-69.212],[210,-68.8395],[212,-68.1705],[214,-67.1206],[219,-63.72],[221,-62.5783],[222,-62.1549],[224,-61.771],[229,-61.5998],[233,-62.0865],[244,-65.6296],[253,-67.0261],[277,-67.3408],[279,-67.1331],[283,-65.861],[288,-63.7248],[291,-62.8596],[301,-62.1348],[324,-62.5606]]
       }
       middle_02_r_ctrl.Scale.X {type=MovieSceneScriptingFloatChannel, default=1, const [1,325]}
       middle_02_r_ctrl.Scale.Y {type=MovieSceneScriptingFloatChannel, default=1, const [1,325]}
       middle_02_r_ctrl.Scale.Z {type=MovieSceneScriptingFloatChannel, default=1, const [1,325]}
       middle_03_r_ctrl.Location.X {type=MovieSceneScriptingFloatChannel, default=0, thinned_from=325, keys [[0,0.0005],[324,0.0005]]}
       middle_03_r_ctrl.Location.Y {type=MovieSceneScriptingFloatChannel, default=0, thinned_from=325, keys [[0,0],[324,0]]}
       middle_03_r_ctrl.Location.Z {type=MovieSceneScriptingFloatChannel, default=0, thinned_from=325, keys [[0,0.0061],[324,0.0061]]}
       middle_03_r_ctrl.Rotation.X {type=MovieSceneScriptingFloatChannel, default=0, thinned_from=325, keys [[0,-0.0051],[135,-0.0054],[213,-0.0042],[324,-0.0051]]}
       middle_03_r_ctrl.Rotation.Y {type=MovieSceneScriptingFloatChannel, default=0, thinned_from=325, keys [[0,-0.0098],[324,-0.0098]]}
       middle_03_r_ctrl.Rotation.Z {
        type=MovieSceneScriptingFloatChannel
        default=0
        thinned_from=325
        keys [[0,-76.6019],[2,-76.4595],[3,-76.1938],[8,-73.3916],[10,-72.6626],[13,-72.2089],[16,-72.2356],[19,-72.8722],[24,-74.8025],[26,-75.8032],[30,-76.7736],[41,-76.757],[51,-74.5554],[55,-74.5379],[58,-75.2427],[63,-76.9886],[65,-77.3109],[76,-77.2634],[79,-76.7743],[88,-73.5393],[90,-73.2292],[92,-73.3839],[93,-73.636],[99,-76.7744],[102,-77.7333],[111,-77.9403],[115,-77.3601],[118,-76.4066],[124,-73.9514],[126,-73.3911],[130,-73.1623],[132,-73.4737],[134,-74.3603],[138,-76.9311],[140,-77.9291],[142,-78.5651],[146,-78.9765],[160,-78.9391],[171,-76.8272],[179,-76.7354],[183,-77.2291],[189,-79.4328],[194,-81.9486],[197,-82.9289],[200,-83.2876],[208,-83.245],[211,-82.5773],[213,-81.7392],[220,-77.1558],[222,-76.1879],[226,-75.6429],[231,-75.8107],[234,-76.342],[244,-79.6626],[256,-81.2479],[278,-81.3043],[280,-80.9293],[289,-77.4208],[293,-76.5717],[324,-76.5936]]
       }
       middle_03_r_ctrl.Scale.X {type=MovieSceneScriptingFloatChannel, default=1, const [1,325]}
       middle_03_r_ctrl.Scale.Y {type=MovieSceneScriptingFloatChannel, default=1, const [1,325]}
       middle_03_r_ctrl.Scale.Z {type=MovieSceneScriptingFloatChannel, default=1, const [1,325]}
       ring_metacarpal_r_ctrl.Location.X {type=MovieSceneScriptingFloatChannel, default=0, const [0.0001,325]}
       ring_metacarpal_r_ctrl.Location.Y {type=MovieSceneScriptingFloatChannel, default=0, const [0.0002,325]}
       ring_metacarpal_r_ctrl.Location.Z {type=MovieSceneScriptingFloatChannel, default=0, const [0,325]}
       ring_metacarpal_r_ctrl.Rotation.X {type=MovieSceneScriptingFloatChannel, default=0, thinned_from=325, keys [[0,5.3943],[324,5.3943]]}
       ring_metacarpal_r_ctrl.Rotation.Y {type=MovieSceneScriptingFloatChannel, default=0, thinned_from=325, keys [[0,-2.9431],[324,-2.9431]]}
       ring_metacarpal_r_ctrl.Rotation.Z {
        type=MovieSceneScriptingFloatChannel
        default=0
        thinned_from=325
        keys [[0,-10.4667],[1,-10.3894],[3,-10.0586],[6,-9.6985],[10,-9.3207],[12,-9.1703],[14,-9.0669],[15,-9.0487],[16,-9.0716],[17,-9.1384],[18,-9.2463],[19,-9.4077],[23,-10.2108],[24,-10.3859],[25,-10.5353],[26,-10.6607],[27,-10.743],[29,-10.8255],[31,-10.8647],[35,-10.8647],[38,-10.7806],[40,-10.6746],[43,-10.4233],[45,-10.2092],[50,-9.5303],[51,-9.4299],[52,-9.3739],[53,-9.3494],[54,-9.3494],[55,-9.4076],[58,-9.8745],[60,-10.2089],[62,-10.4628],[63,-10.5527],[64,-10.6068],[65,-10.6323],[74,-10.6323],[76,-10.5596],[78,-10.4487],[80,-10.3926],[82,-10.3926],[87,-10.6386],[98,-11.4438],[101,-11.583],[109,-11.5981],[111,-11.5473],[112,-11.4688],[113,-11.3477],[115,-11.0412],[117,-10.6512],[122,-9.4258],[123,-9.2018],[124,-9.0128],[125,-8.8505],[126,-8.7167],[127,-8.6045],[128,-8.5208],[129,-8.4879],[130,-8.5072],[131,-8.5814],[132,-8.7349],[133,-8.9969],[135,-9.6951],[137,-10.4515],[138,-10.7959],[139,-11.0843],[140,-11.3311],[141,-11.523],[142,-11.6546],[143,-11.7313],[144,-11.7562],[159,-11.7562],[160,-11.7386],[163,-11.5765],[171,-10.9902],[174,-10.8668],[176,-10.8315],[178,-10.8315],[181,-10.9601],[190,-11.5428],[193,-11.6607],[196,-11.7091],[216,-11.7208],[218,-11.6642],[223,-11.3883],[225,-11.2569],[228,-11.1366],[230,-11.0976],[235,-11.0976],[238,-11.2423],[245,-11.7028],[250,-11.8683],[280,-11.9241],[283,-11.8063],[285,-11.661],[289,-11.1976],[291,-10.9257],[294,-10.4667],[296,-10.2521],[297,-10.173],[299,-10.0787],[302,-10.0179],[304,-10.0179],[306,-10.0627],[314,-10.4151],[316,-10.4504],[324,-10.4632]]
       }
       ring_metacarpal_r_ctrl.Scale.X {type=MovieSceneScriptingFloatChannel, default=1, const [1,325]}
       ring_metacarpal_r_ctrl.Scale.Y {type=MovieSceneScriptingFloatChannel, default=1, const [1,325]}
       ring_metacarpal_r_ctrl.Scale.Z {type=MovieSceneScriptingFloatChannel, default=1, const [1,325]}
       ring_01_r_ctrl.Location.X {type=MovieSceneScriptingFloatChannel, default=0, thinned_from=325, keys [[0,-0.371],[324,-0.371]]}
       ring_01_r_ctrl.Location.Y {type=MovieSceneScriptingFloatChannel, default=0, thinned_from=325, keys [[0,-1.0509],[324,-1.0509]]}
       ring_01_r_ctrl.Location.Z {type=MovieSceneScriptingFloatChannel, default=0, thinned_from=325, keys [[0,-0.0243],[324,-0.0243]]}
       ring_01_r_ctrl.Rotation.X {type=MovieSceneScriptingFloatChannel, default=0, thinned_from=325, keys [[0,-0.8239],[324,-0.8239]]}
       ring_01_r_ctrl.Rotation.Y {type=MovieSceneScriptingFloatChannel, default=0, thinned_from=325, keys [[0,-2.6856],[324,-2.6856]]}
       ring_01_r_ctrl.Rotation.Z {
        type=MovieSceneScriptingFloatChannel
        default=0
        thinned_from=325
        keys [[0,-84.1083],[2,-83.9659],[3,-83.7002],[8,-80.898],[10,-80.169],[13,-79.7153],[16,-79.742],[20,-80.6954],[27,-83.6756],[32,-84.4881],[42,-84.1566],[51,-82.0618],[55,-82.0442],[58,-82.7491],[63,-84.495],[66,-84.8602],[76,-84.7697],[80,-84.0111],[88,-81.0456],[90,-80.7356],[92,-80.8902],[93,-81.1424],[99,-84.2808],[102,-85.2396],[111,-85.4467],[115,-84.8665],[118,-83.9129],[125,-81.1395],[127,-80.7534],[130,-80.684],[132,-80.98],[134,-81.8666],[138,-84.4375],[141,-85.809],[145,-86.4493],[160,-86.4454],[171,-84.3335],[180,-84.3231],[183,-84.7354],[189,-86.9392],[194,-89.455],[197,-90.4352],[208,-90.7513],[211,-90.0836],[213,-89.2455],[220,-84.6621],[222,-83.6943],[226,-83.1492],[231,-83.3171],[234,-83.8483],[244,-87.1689],[260,-88.8801],[278,-88.8106],[281,-88.1234],[289,-84.9271],[294,-83.968],[324,-84.0999]]
       }
       ring_01_r_ctrl.Scale.X {type=MovieSceneScriptingFloatChannel, default=1, const [1,325]}
       ring_01_r_ctrl.Scale.Y {type=MovieSceneScriptingFloatChannel, default=1, const [1,325]}
       ring_01_r_ctrl.Scale.Z {type=MovieSceneScriptingFloatChannel, default=1, const [1,325]}
       ring_02_r_ctrl.Location.X {type=MovieSceneScriptingFloatChannel, default=0, thinned_from=325, keys [[0,0.0003],[109,0.0003],[110,-0.0004],[111,0.0003],[324,0.0003]]}
       ring_02_r_ctrl.Location.Y {type=MovieSceneScriptingFloatChannel, default=0, thinned_from=325, keys [[0,0.0001],[109,0.0001],[110,0.0008],[111,0.0001],[324,0.0001]]}
       ring_02_r_ctrl.Location.Z {type=MovieSceneScriptingFloatChannel, default=0, thinned_from=325, keys [[0,0],[324,0]]}
       ring_02_r_ctrl.Rotation.X {type=MovieSceneScriptingFloatChannel, default=0, thinned_from=325, keys [[0,-12.8636],[324,-12.8636]]}
       ring_02_r_ctrl.Rotation.Y {type=MovieSceneScriptingFloatChannel, default=0, thinned_from=325, keys [[0,-0.6911],[109,-0.6911],[110,-0.6903],[324,-0.6911]]}
       ring_02_r_ctrl.Rotation.Z {
        type=MovieSceneScriptingFloatChannel
        default=0
        thinned_from=325
        keys [[0,-59.6606],[2,-59.5183],[3,-59.2526],[4,-58.7582],[6,-57.5021],[8,-56.4504],[9,-56.0286],[11,-55.5028],[13,-55.2677],[16,-55.2943],[19,-55.9309],[24,-57.8612],[26,-58.862],[27,-59.228],[31,-59.9597],[41,-59.8157],[45,-59.1024],[50,-57.7764],[52,-57.5536],[55,-57.5966],[57,-58.0046],[63,-60.0473],[65,-60.3696],[76,-60.3221],[79,-59.833],[82,-58.905],[87,-56.8895],[89,-56.3762],[90,-56.288],[92,-56.4426],[93,-56.6948],[94,-57.1332],[98,-59.3567],[100,-60.2236],[102,-60.792],[104,-61.028],[111,-60.9991],[114,-60.6337],[117,-59.8451],[124,-57.0101],[126,-56.4498],[128,-56.238],[130,-56.2211],[132,-56.5324],[133,-56.9122],[135,-57.9868],[138,-59.9899],[140,-60.9878],[142,-61.6239],[145,-62.0017],[160,-61.9978],[163,-61.5263],[170,-60.0224],[178,-59.7416],[182,-60.0962],[184,-60.5746],[189,-62.4916],[194,-65.0074],[197,-65.9877],[199,-66.2977],[208,-66.3037],[210,-65.9312],[212,-65.2622],[214,-64.2123],[219,-60.8117],[221,-59.67],[222,-59.2466],[223,-58.9941],[227,-58.672],[231,-58.8695],[234,-59.4007],[243,-62.4567],[251,-63.9209],[277,-64.4325],[279,-64.2248],[283,-62.9527],[288,-60.8165],[291,-59.9512],[298,-59.2879],[324,-59.6523]]
       }
       ring_02_r_ctrl.Scale.X {type=MovieSceneScriptingFloatChannel, default=1, const [1,325]}
       ring_02_r_ctrl.Scale.Y {type=MovieSceneScriptingFloatChannel, default=1, const [1,325]}
       ring_02_r_ctrl.Scale.Z {type=MovieSceneScriptingFloatChannel, default=1, const [1,325]}
       ring_03_r_ctrl.Location.X {type=MovieSceneScriptingFloatChannel, default=0, thinned_from=325, keys [[0,0],[324,0]]}
       ring_03_r_ctrl.Location.Y {type=MovieSceneScriptingFloatChannel, default=0, thinned_from=325, keys [[0,-0.0001],[324,-0.0001]]}
       ring_03_r_ctrl.Location.Z {type=MovieSceneScriptingFloatChannel, default=0, thinned_from=325, keys [[0,0],[324,0]]}
       ring_03_r_ctrl.Rotation.X {type=MovieSceneScriptingFloatChannel, default=0, thinned_from=325, keys [[0,0],[324,0]]}
       ring_03_r_ctrl.Rotation.Y {type=MovieSceneScriptingFloatChannel, default=0, thinned_from=325, keys [[0,0],[324,0]]}
       ring_03_r_ctrl.Rotation.Z {
        type=MovieSceneScriptingFloatChannel
        default=0
        thinned_from=325
        keys [[0,-88.4683],[2,-88.3259],[3,-88.0602],[8,-85.2581],[10,-84.529],[13,-84.0754],[16,-84.102],[20,-85.0555],[27,-88.0357],[32,-88.8481],[42,-88.5166],[51,-86.4218],[55,-86.4043],[58,-87.1091],[63,-88.855],[66,-89.2203],[76,-89.1298],[80,-88.3711],[88,-85.4057],[90,-85.0956],[92,-85.2503],[93,-85.5024],[99,-88.6408],[102,-89.5997],[112,-89.7322],[116,-88.9647],[125,-85.4995],[128,-85.0457],[131,-85.1286],[133,-85.7199],[136,-87.4449],[138,-88.7975],[141,-90.1691],[145,-90.8094],[160,-90.8055],[171,-88.6936],[181,-88.7791],[184,-89.3823],[189,-91.2993],[195,-94.2105],[198,-94.9895],[208,-95.1114],[211,-94.4437],[213,-93.6055],[220,-89.0222],[222,-88.0543],[227,-87.4797],[232,-87.8133],[246,-91.9692],[277,-93.2402],[280,-92.7957],[289,-89.2872],[295,-88.2411],[324,-88.46]]
       }
       ring_03_r_ctrl.Scale.X {type=MovieSceneScriptingFloatChannel, default=1, const [1,325]}
       ring_03_r_ctrl.Scale.Y {type=MovieSceneScriptingFloatChannel, default=1, const [1,325]}
       ring_03_r_ctrl.Scale.Z {type=MovieSceneScriptingFloatChannel, default=1, const [1,325]}
       pinky_metacarpal_r_ctrl.Location.X {type=MovieSceneScriptingFloatChannel, default=0, const [-0.0001,325]}
       pinky_metacarpal_r_ctrl.Location.Y {type=MovieSceneScriptingFloatChannel, default=0, const [0,325]}
       pinky_metacarpal_r_ctrl.Location.Z {type=MovieSceneScriptingFloatChannel, default=0, const [0,325]}
       pinky_metacarpal_r_ctrl.Rotation.X {type=MovieSceneScriptingFloatChannel, default=0, thinned_from=325, keys [[0,9.8795],[324,9.8795]]}
       pinky_metacarpal_r_ctrl.Rotation.Y {type=MovieSceneScriptingFloatChannel, default=0, thinned_from=325, keys [[0,-6.9881],[324,-6.9881]]}
       pinky_metacarpal_r_ctrl.Rotation.Z {
        type=MovieSceneScriptingFloatChannel
        default=0
        thinned_from=325
        keys [[0,-8.607],[1,-8.5297],[3,-8.1989],[6,-7.8388],[10,-7.461],[12,-7.3106],[14,-7.2072],[15,-7.189],[16,-7.2119],[17,-7.2787],[18,-7.3866],[19,-7.548],[23,-8.3511],[24,-8.5262],[25,-8.6756],[26,-8.801],[27,-8.8833],[29,-8.9657],[31,-9.005],[35,-9.005],[38,-8.9209],[40,-8.8149],[43,-8.5635],[45,-8.3495],[50,-7.6706],[51,-7.5702],[52,-7.5142],[53,-7.4896],[54,-7.4896],[55,-7.5478],[58,-8.0148],[59,-8.1904],[60,-8.3491],[62,-8.6031],[63,-8.693],[64,-8.7471],[65,-8.7725],[74,-8.7725],[75,-8.7466],[78,-8.589],[80,-8.5329],[82,-8.5329],[86,-8.7208],[93,-9.2072],[97,-9.5193],[100,-9.6861],[101,-9.7233],[109,-9.7384],[110,-9.7233],[111,-9.6876],[112,-9.6091],[113,-9.488],[115,-9.1814],[116,-8.996],[117,-8.7915],[118,-8.5628],[122,-7.5661],[123,-7.3421],[124,-7.1531],[125,-6.9908],[126,-6.857],[127,-6.7448],[128,-6.6611],[129,-6.6282],[130,-6.6475],[131,-6.7217],[132,-6.8751],[133,-7.1372],[134,-7.4785],[136,-8.2058],[137,-8.5918],[138,-8.9362],[139,-9.2246],[140,-9.4713],[141,-9.6633],[142,-9.7948],[143,-9.8716],[144,-9.8965],[159,-9.8965],[160,-9.8788],[162,-9.7811],[171,-9.1305],[173,-9.0387],[175,-8.9848],[178,-8.9718],[180,-9.0466],[189,-9.6291],[192,-9.7723],[196,-9.8494],[216,-9.8611],[217,-9.8442],[222,-9.5948],[225,-9.3971],[228,-9.2768],[230,-9.2379],[235,-9.2379],[238,-9.3825],[244,-9.7884],[247,-9.929],[253,-10.0519],[280,-10.0644],[282,-9.9971],[284,-9.8818],[286,-9.7015],[290,-9.2053],[292,-8.9114],[293,-8.75],[294,-8.607],[295,-8.4905],[296,-8.3924],[297,-8.3133],[298,-8.2566],[301,-8.1697],[304,-8.1546],[306,-8.203],[314,-8.5554],[316,-8.5907],[324,-8.6035]]
       }
       pinky_metacarpal_r_ctrl.Scale.X {type=MovieSceneScriptingFloatChannel, default=1, const [1,325]}
       pinky_metacarpal_r_ctrl.Scale.Y {type=MovieSceneScriptingFloatChannel, default=1, const [1,325]}
       pinky_metacarpal_r_ctrl.Scale.Z {type=MovieSceneScriptingFloatChannel, default=1, const [1,325]}
       pinky_01_r_ctrl.Location.X {type=MovieSceneScriptingFloatChannel, default=0, thinned_from=325, keys [[0,-0.3834],[324,-0.3834]]}
       pinky_01_r_ctrl.Location.Y {type=MovieSceneScriptingFloatChannel, default=0, const [-1.045,325]}
       pinky_01_r_ctrl.Location.Z {type=MovieSceneScriptingFloatChannel, default=0, thinned_from=325, keys [[0,-0.0533],[324,-0.0533]]}
       pinky_01_r_ctrl.Rotation.X {type=MovieSceneScriptingFloatChannel, default=0, thinned_from=325, keys [[0,-7.5971],[324,-7.5972]]}
       pinky_01_r_ctrl.Rotation.Y {type=MovieSceneScriptingFloatChannel, default=0, thinned_from=325, keys [[0,-7.928],[324,-7.9281]]}
       pinky_01_r_ctrl.Rotation.Z {
        type=MovieSceneScriptingFloatChannel
        default=0
        thinned_from=325
        keys [[0,-83.5712],[2,-83.4287],[3,-83.1631],[8,-80.3609],[10,-79.6319],[13,-79.1782],[16,-79.2049],[20,-80.1584],[27,-83.1385],[32,-83.951],[42,-83.6195],[51,-81.5247],[55,-81.5071],[58,-82.212],[63,-83.9579],[66,-84.3231],[76,-84.2326],[80,-83.474],[88,-80.5086],[90,-80.1985],[92,-80.3532],[93,-80.6053],[99,-83.7437],[102,-84.7026],[111,-84.9096],[115,-84.3294],[118,-83.3758],[125,-80.6024],[127,-80.2163],[130,-80.1469],[132,-80.443],[134,-81.3295],[138,-83.9004],[141,-85.2719],[145,-85.9122],[160,-85.9083],[171,-83.7964],[180,-83.786],[183,-84.1983],[189,-86.4021],[194,-88.9178],[197,-89.8981],[208,-90.2142],[211,-89.5465],[213,-88.7084],[220,-84.125],[222,-83.1572],[226,-82.6121],[231,-82.78],[234,-83.3112],[244,-86.6318],[260,-88.343],[278,-88.2735],[281,-87.5863],[289,-84.39],[294,-83.4308],[324,-83.5628]]
       }
       pinky_01_r_ctrl.Scale.X {type=MovieSceneScriptingFloatChannel, default=1, const [1,325]}
       pinky_01_r_ctrl.Scale.Y {type=MovieSceneScriptingFloatChannel, default=1, const [1,325]}
       pinky_01_r_ctrl.Scale.Z {type=MovieSceneScriptingFloatChannel, default=1, const [1,325]}
       pinky_02_r_ctrl.Location.X {type=MovieSceneScriptingFloatChannel, default=0, thinned_from=325, keys [[0,0.3528],[109,0.3528],[110,0.3524],[324,0.3528]]}
       pinky_02_r_ctrl.Location.Y {type=MovieSceneScriptingFloatChannel, default=0, thinned_from=325, keys [[0,0.0018],[109,0.0018],[110,0.0025],[111,0.0018],[324,0.0018]]}
       pinky_02_r_ctrl.Location.Z {type=MovieSceneScriptingFloatChannel, default=0, thinned_from=325, keys [[0,-0.0756],[324,-0.0756]]}
       pinky_02_r_ctrl.Rotation.X {type=MovieSceneScriptingFloatChannel, default=0, thinned_from=325, keys [[0,1.9939],[324,1.9939]]}
       pinky_02_r_ctrl.Rotation.Y {type=MovieSceneScriptingFloatChannel, default=0, thinned_from=325, keys [[0,10.0966],[324,10.0966]]}
       pinky_02_r_ctrl.Rotation.Z {
        type=MovieSceneScriptingFloatChannel
        default=0
        thinned_from=325
        keys [[0,-57.9295],[2,-57.7871],[3,-57.5214],[4,-57.0271],[6,-55.7709],[8,-54.7192],[9,-54.2974],[11,-53.7716],[13,-53.5365],[16,-53.5632],[19,-54.1998],[22,-55.2618],[26,-57.1308],[27,-57.4969],[31,-58.2286],[41,-58.0846],[45,-57.3712],[50,-56.0452],[52,-55.8225],[55,-55.8654],[57,-56.2735],[63,-58.3162],[65,-58.6384],[76,-58.591],[79,-58.1018],[82,-57.1738],[87,-55.1583],[89,-54.645],[90,-54.5568],[92,-54.7115],[93,-54.9636],[94,-55.402],[98,-57.6256],[100,-58.4925],[102,-59.0609],[104,-59.2969],[111,-59.2679],[114,-58.9025],[117,-58.114],[124,-55.279],[126,-54.7186],[128,-54.5069],[130,-54.4899],[131,-54.5897],[132,-54.8013],[133,-55.181],[135,-56.2556],[138,-58.2587],[140,-59.2567],[142,-59.8927],[145,-60.2705],[160,-60.2667],[163,-59.7951],[170,-58.2913],[178,-58.0104],[182,-58.365],[184,-58.8435],[189,-60.7604],[194,-63.2762],[196,-63.9951],[198,-64.4507],[208,-64.5725],[210,-64.2],[212,-63.5311],[214,-62.4811],[219,-59.0805],[221,-57.9388],[222,-57.5155],[223,-57.2629],[227,-56.9408],[230,-57.0311],[233,-57.447],[244,-60.9902],[252,-62.2973],[277,-62.7013],[279,-62.4936],[283,-61.2215],[288,-59.0853],[291,-58.2201],[297,-57.5937],[324,-57.9211]]
       }
       pinky_02_r_ctrl.Scale.X {type=MovieSceneScriptingFloatChannel, default=1, const [1,325]}
       pinky_02_r_ctrl.Scale.Y {type=MovieSceneScriptingFloatChannel, default=1, const [1,325]}
       pinky_02_r_ctrl.Scale.Z {type=MovieSceneScriptingFloatChannel, default=1, const [1,325]}
       pinky_03_r_ctrl.Location.X {type=MovieSceneScriptingFloatChannel, default=0, thinned_from=325, keys [[0,0.0002],[324,0.0002]]}
       pinky_03_r_ctrl.Location.Y {type=MovieSceneScriptingFloatChannel, default=0, thinned_from=325, keys [[0,-0.0002],[324,-0.0002]]}
       pinky_03_r_ctrl.Location.Z {type=MovieSceneScriptingFloatChannel, default=0, thinned_from=325, keys [[0,0],[324,0]]}
       pinky_03_r_ctrl.Rotation.X {type=MovieSceneScriptingFloatChannel, default=0, thinned_from=325, keys [[0,3.1496],[324,3.1496]]}
       pinky_03_r_ctrl.Rotation.Y {type=MovieSceneScriptingFloatChannel, default=0, thinned_from=325, keys [[0,-4.3172],[324,-4.3172]]}
       pinky_03_r_ctrl.Rotation.Z {
        type=MovieSceneScriptingFloatChannel
        default=0
        thinned_from=325
        keys [[0,-111.6788],[2,-111.5364],[3,-111.2707],[8,-108.4685],[11,-107.5209],[16,-107.3125],[21,-108.6193],[27,-111.2462],[33,-112.0868],[42,-111.7271],[51,-109.6323],[55,-109.6147],[64,-112.2636],[77,-112.2305],[84,-110.1353],[88,-108.6161],[90,-108.3061],[92,-108.4607],[94,-109.1513],[99,-111.8513],[102,-112.8102],[112,-112.9426],[117,-111.8632],[125,-108.71],[130,-108.2545],[132,-108.5506],[135,-110.0049],[139,-112.5421],[142,-113.642],[160,-114.016],[172,-111.8311],[182,-112.1143],[189,-114.5097],[195,-117.421],[199,-118.3158],[208,-118.3218],[212,-117.2803],[221,-111.6881],[223,-111.0122],[230,-110.7804],[235,-111.7093],[246,-115.1796],[278,-116.3812],[291,-111.9694],[324,-111.6704]]
       }
       pinky_03_r_ctrl.Scale.X {type=MovieSceneScriptingFloatChannel, default=1, const [1,325]}
       pinky_03_r_ctrl.Scale.Y {type=MovieSceneScriptingFloatChannel, default=1, const [1,325]}
       pinky_03_r_ctrl.Scale.Z {type=MovieSceneScriptingFloatChannel, default=1, const [1,325]}
       hand_r_ik_ctrl.Location.X {
        type=MovieSceneScriptingFloatChannel
        default=0
        thinned_from=325
        keys [[0,0.4144],[1,0.4105],[2,0.3924],[3,0.3624],[4,0.3205],[5,0.2668],[6,0.2021],[7,0.1266],[8,0.0406],[9,-0.0583],[10,-0.1698],[11,-0.2955],[12,-0.4359],[13,-0.5838],[14,-0.7337],[15,-0.8791],[16,-1.0138],[17,-1.1512],[18,-1.2629],[19,-1.3686],[20,-1.4673],[21,-1.5577],[22,-1.6391],[23,-1.7268],[24,-1.7957],[25,-1.8531],[26,-1.9005],[27,-1.9421],[29,-2.014],[30,-2.0437],[31,-2.0646],[32,-2.0493],[33,-2.0022],[34,-1.9072],[35,-1.8242],[37,-1.7005],[38,-1.6648],[39,-1.604],[40,-1.5343],[41,-1.4319],[43,-1.2084],[44,-1.0768],[45,-0.9147],[46,-0.7374],[47,-0.5319],[48,-0.3321],[49,-0.139],[50,0.047],[51,0.2318],[52,0.4005],[53,0.5399],[54,0.6574],[55,0.7595],[56,0.8454],[57,0.9194],[58,0.9851],[59,1.0453],[60,1.0939],[61,1.1306],[62,1.1644],[64,1.2251],[66,1.2785],[68,1.322],[69,1.3399],[70,1.3548],[71,1.366],[73,1.3804],[74,1.3783],[75,1.3722],[76,1.3615],[77,1.345],[78,1.3245],[79,1.2964],[80,1.2618],[81,1.2197],[82,1.1705],[83,1.1128],[84,1.0429],[85,0.9643],[86,0.8889],[87,0.8104],[88,0.6956],[89,0.5433],[90,0.3717],[91,0.1829],[92,-0.0006],[93,-0.1646],[94,-0.3154],[95,-0.4777],[96,-0.6699],[97,-0.8749],[98,-1.0776],[100,-1.4925],[101,-1.7047],[102,-1.8825],[103,-2.0435],[104,-2.1872],[105,-2.311],[106,-2.4109],[107,-2.5039],[108,-2.5626],[109,-2.6],[110,-2.6183],[111,-2.617],[112,-2.6032],[113,-2.5833],[114,-2.5516],[116,-2.4723],[120,-2.2806],[122,-2.1964],[123,-2.1412],[125,-2.0493],[126,-1.9986],[127,-1.9408],[128,-1.8706],[129,-1.7871],[130,-1.6905],[131,-1.573],[132,-1.4597],[134,-1.2082],[135,-1.0805],[140,-0.4567],[141,-0.3281],[142,-0.1959],[143,-0.066],[144,0.0606],[145,0.1749],[146,0.2781],[147,0.3793],[148,0.4746],[149,0.5635],[150,0.6456],[151,0.7172],[152,0.7788],[153,0.8343],[154,0.8826],[155,0.9234],[156,0.9574],[157,0.9846],[158,1.0049],[159,1.021],[160,1.0309],[161,1.0374],[162,1.0391],[164,1.0321],[166,1.0137],[167,1.0001],[168,0.9841],[169,0.9612],[170,0.9228],[171,0.8733],[172,0.8145],[173,0.7483],[174,0.6766],[175,0.6014],[176,0.5243],[177,0.4485],[178,0.3737],[180,0.228],[181,0.1494],[182,0.0757],[183,0.0114],[184,-0.0499],[185,-0.1121],[186,-0.1809],[187,-0.2666],[188,-0.3718],[189,-0.505],[190,-0.667],[191,-0.7993],[192,-0.9293],[193,-1.1635],[195,-1.6967],[196,-1.9566],[197,-2.1649],[198,-2.3226],[199,-2.4678],[200,-2.5888],[202,-2.8114],[204,-3.0449],[205,-3.1495],[206,-3.2343],[208,-3.3857],[209,-3.4556],[210,-3.5044],[211,-3.5119],[212,-3.5012],[215,-3.4019],[216,-3.3579],[217,-3.305],[218,-3.2392],[222,-2.9388],[224,-2.7677],[229,-2.305],[231,-2.131],[232,-2.0519],[233,-1.9799],[234,-1.9128],[235,-1.8502],[236,-1.7922],[237,-1.7414],[238,-1.6984],[239,-1.6621],[240,-1.632],[241,-1.6137],[242,-1.6026],[243,-1.6012],[244,-1.6148],[245,-1.6508],[246,-1.7185],[247,-1.8369],[248,-1.9779],[249,-2.1364],[250,-2.2421],[251,-2.2603],[252,-2.1638],[253,-1.9396],[254,-1.5632],[255,-1.1441],[256,-0.7479],[257,-0.4196],[258,-0.1609],[259,0.0149],[260,0.127],[261,0.1818],[262,0.1907],[263,0.1413],[264,0.0476],[265,-0.0798],[266,-0.2341],[267,-0.3968],[268,-0.563],[269,-0.7432],[271,-1.0899],[272,-1.2562],[273,-1.4181],[275,-1.7206],[276,-1.8366],[277,-1.9317],[278,-2.0033],[279,-2.0433],[281,-2.0897],[282,-2.108],[284,-2.1337],[289,-2.1517],[290,-2.1521],[291,-2.1316],[293,-2.1217],[294,-2.1101],[295,-2.0888],[296,-2.0527],[297,-2.0015],[298,-1.9378],[299,-1.8444],[300,-1.7551],[301,-1.6489],[302,-1.5283],[303,-1.3995],[304,-1.2655],[306,-0.9807],[307,-0.8408],[308,-0.7086],[309,-0.5873],[310,-0.4714],[311,-0.3601],[312,-0.2572],[313,-0.1707],[314,-0.088],[315,-0.0085],[316,0.0657],[317,0.1352],[318,0.2002],[319,0.2587],[320,0.31],[321,0.3525],[322,0.3849],[323,0.405],[324,0.4117]]
       }
       hand_r_ik_ctrl.Location.Y {
        type=MovieSceneScriptingFloatChannel
        default=0
        thinned_from=325
        keys [[0,10.4986],[3,10.4082],[9,10.0849],[19,9.7397],[25,9.6259],[31,9.6145],[32,9.6379],[34,9.7377],[35,9.7553],[36,9.7173],[39,9.5372],[40,9.5095],[41,9.5475],[42,9.6482],[43,9.7783],[44,9.9523],[45,10.1827],[46,10.44],[51,11.8381],[52,12.0915],[53,12.3026],[54,12.4692],[55,12.6025],[59,12.9345],[64,13.1685],[71,13.2787],[75,13.2396],[83,12.934],[84,12.8793],[85,12.8749],[87,12.9539],[90,13.1516],[91,13.1651],[92,13.1407],[93,13.0878],[94,12.9957],[95,12.8683],[96,12.6712],[98,12.1691],[100,11.6017],[101,11.2997],[102,11.0553],[103,10.8433],[104,10.6683],[105,10.5265],[107,10.3061],[109,10.1912],[111,10.1299],[114,10.1565],[116,10.2232],[118,10.3327],[122,10.6535],[125,10.9029],[127,11.0156],[131,11.0917],[139,11.1534],[141,11.0959],[145,10.8811],[152,10.4727],[158,10.243],[159,10.2301],[168,9.9841],[169,9.9742],[176,9.6972],[178,9.6575],[181,9.6697],[182,9.6918],[185,9.8283],[187,9.8891],[188,9.9425],[189,10.0484],[190,10.1818],[191,10.361],[193,10.7831],[194,10.9714],[196,11.2371],[197,11.3175],[198,11.3577],[199,11.3408],[200,11.291],[201,11.2074],[202,11.0785],[208,10.1342],[209,9.991],[211,9.77],[213,9.6327],[216,9.4905],[218,9.4629],[220,9.4887],[222,9.5617],[230,10.0833],[233,10.2045],[237,10.2829],[240,10.276],[241,10.2477],[245,10.2443],[246,10.1521],[247,10.0227],[248,9.8211],[249,9.5584],[251,8.9985],[252,8.8062],[253,8.644],[254,8.6134],[255,8.7097],[256,8.9118],[257,9.1982],[258,9.542],[259,9.9094],[261,10.6924],[262,11.0559],[263,11.3147],[264,11.5025],[265,11.6128],[266,11.6395],[267,11.6256],[268,11.5624],[269,11.427],[270,11.2443],[271,11.0067],[275,9.8879],[276,9.6516],[277,9.4521],[278,9.2891],[279,9.1704],[281,9.0031],[284,8.8451],[286,8.7763],[287,8.7638],[288,8.7769],[290,8.8605],[298,9.377],[305,10.041],[308,10.278],[311,10.4365],[322,10.5139],[324,10.4965]]
       }
       hand_r_ik_ctrl.Location.Z {
        type=MovieSceneScriptingFloatChannel
        default=0
        thinned_from=325
        keys [[0,2.2064],[2,2.168],[3,2.1613],[4,2.1659],[5,2.1791],[6,2.1998],[7,2.2303],[8,2.2728],[9,2.3492],[10,2.4692],[11,2.6169],[13,2.9331],[14,3.0858],[15,3.2316],[16,3.3674],[17,3.4527],[19,3.7016],[22,4.0475],[23,4.1111],[24,4.1897],[25,4.2454],[28,4.3709],[29,4.3885],[30,4.3659],[31,4.3183],[32,4.2373],[33,4.1368],[34,4.0035],[35,3.8394],[36,3.7073],[37,3.5608],[38,3.372],[40,3.0853],[41,3.0299],[42,3.0578],[43,3.1194],[44,3.2505],[45,3.4703],[46,3.7355],[48,4.448],[49,4.8305],[51,5.6464],[52,5.997],[53,6.2802],[54,6.4902],[55,6.647],[56,6.7432],[58,6.8802],[62,7.0707],[64,7.1181],[68,7.1296],[75,7.0251],[76,6.9967],[83,6.7144],[84,6.688],[85,6.6963],[86,6.7848],[87,6.9399],[89,7.3927],[90,7.5992],[92,7.894],[93,7.9946],[94,7.9971],[95,7.9271],[96,7.7307],[97,7.445],[98,7.1301],[99,6.7657],[100,6.3273],[101,5.831],[102,5.434],[103,5.0687],[104,4.7374],[105,4.4411],[106,4.1833],[107,3.9131],[108,3.7011],[109,3.5233],[110,3.3804],[111,3.2813],[112,3.2406],[113,3.2292],[114,3.2621],[115,3.3124],[116,3.3764],[117,3.4608],[118,3.5597],[119,3.6795],[120,3.8075],[122,4.0416],[123,4.1923],[125,4.3955],[127,4.5787],[128,4.6595],[129,4.7213],[130,4.7679],[131,4.8456],[132,4.8767],[134,4.8975],[137,4.8989],[138,4.8827],[139,4.8551],[140,4.8045],[141,4.7192],[142,4.6029],[143,4.4738],[148,3.7526],[149,3.6147],[152,3.2302],[153,3.123],[154,3.028],[155,2.9424],[156,2.867],[157,2.8014],[159,2.6952],[161,2.6123],[163,2.5428],[165,2.4901],[168,2.4264],[169,2.4089],[170,2.3992],[171,2.3963],[172,2.4001],[173,2.4095],[174,2.4244],[175,2.4554],[176,2.5022],[177,2.5545],[178,2.6243],[179,2.7237],[180,2.842],[181,2.9748],[182,3.1337],[183,3.3233],[184,3.5459],[185,3.7915],[187,4.2924],[188,4.5274],[189,4.728],[191,4.9811],[192,5.0754],[193,5.0072],[194,4.8899],[195,4.7129],[196,4.5081],[197,4.3226],[198,4.1513],[199,3.9292],[200,3.668],[201,3.3566],[202,2.9591],[203,2.4912],[204,2.0017],[205,1.5258],[206,1.0879],[207,0.6722],[208,0.258],[209,-0.124],[210,-0.4359],[211,-0.6401],[212,-0.8235],[213,-0.9807],[214,-1.1415],[215,-1.3135],[216,-1.4744],[217,-1.6104],[218,-1.7354],[219,-1.8559],[220,-1.9711],[221,-2.0689],[222,-2.1367],[223,-2.1611],[224,-2.1555],[225,-2.141],[226,-2.1144],[227,-2.0788],[228,-2.0348],[231,-1.8781],[232,-1.8348],[236,-1.7057],[237,-1.6826],[238,-1.669],[241,-1.6459],[243,-1.6471],[244,-1.6807],[245,-1.7814],[246,-1.9638],[247,-2.2321],[248,-2.5517],[249,-2.9362],[250,-3.2168],[251,-3.3166],[252,-3.2302],[253,-3.0058],[254,-2.4941],[255,-1.8152],[256,-1.0458],[257,-0.2444],[258,0.5241],[259,1.227],[260,1.9225],[261,2.607],[262,3.2354],[263,3.6876],[264,4.0217],[265,4.2315],[266,4.3154],[267,4.3439],[268,4.2937],[269,4.1117],[270,3.9119],[271,3.6696],[272,3.4002],[274,2.8079],[275,2.459],[276,2.1702],[277,1.9247],[278,1.7176],[279,1.5564],[280,1.4265],[281,1.3081],[282,1.2163],[283,1.1408],[284,1.0738],[285,1.0185],[286,0.9687],[287,0.9248],[288,0.8879],[289,0.8547],[290,0.8255],[291,0.838],[292,0.8195],[293,0.8141],[294,0.8171],[295,0.8337],[296,0.8771],[297,0.9425],[298,1.0311],[299,1.189],[300,1.3406],[303,1.8458],[304,2.009],[305,2.166],[306,2.3149],[307,2.453],[308,2.5739],[309,2.6622],[310,2.7221],[311,2.7611],[312,2.7713],[313,2.7648],[314,2.736],[315,2.6877],[319,2.4314],[320,2.3726],[321,2.3195],[322,2.273],[323,2.2325],[324,2.2061]]
       }
       hand_r_ik_ctrl.Rotation.X {
        type=MovieSceneScriptingFloatChannel
        default=0
        thinned_from=325
        keys [[0,26.1428],[9,26.2659],[16,25.9567],[17,25.8679],[22,25.7429],[30,25.4281],[34,25.456],[36,25.6206],[40,26.1658],[42,26.2901],[44,26.2724],[48,25.929],[53,25.3185],[84,24.9399],[87,24.7564],[91,24.3617],[95,24.2081],[99,24.2849],[101,24.3466],[106,24.7169],[112,24.9036],[120,24.8312],[122,24.7569],[123,24.7691],[130,24.5919],[140,24.8934],[155,26.1071],[178,26.6709],[188,26.3701],[195,25.7603],[198,25.6702],[202,25.784],[216,26.7702],[222,26.8619],[240,26.546],[245,26.5398],[248,26.6808],[253,27.1627],[255,27.2141],[256,27.1629],[258,26.88],[260,26.4417],[262,25.9174],[264,25.5525],[266,25.3369],[269,25.1834],[271,25.3308],[278,26.2671],[280,26.4674],[288,26.7751],[295,26.6785],[305,26.0503],[324,26.1432]]
       }
       hand_r_ik_ctrl.Rotation.Y {
        type=MovieSceneScriptingFloatChannel
        default=0
        thinned_from=325
        keys [[0,3.3134],[1,3.2528],[2,3.1718],[3,3.1023],[5,2.9878],[6,2.9367],[7,2.8963],[8,2.8744],[9,2.9266],[10,3.0796],[11,3.2955],[16,4.5103],[17,4.6638],[21,5.7309],[22,5.9897],[23,6.1357],[24,6.3309],[28,7.0339],[29,7.1666],[30,7.2393],[31,7.2782],[32,7.2532],[33,7.1776],[34,7.025],[35,6.705],[36,6.1494],[38,4.8206],[39,4.2324],[40,3.5684],[41,2.9871],[42,2.5698],[43,2.281],[44,2.202],[45,2.3414],[46,2.631],[47,3.1608],[48,3.7483],[49,4.4444],[50,5.2545],[51,6.0903],[52,6.8678],[53,7.5764],[54,8.178],[55,8.6752],[56,9.0355],[58,9.5981],[59,9.8317],[61,10.2252],[62,10.3929],[63,10.5328],[64,10.6449],[66,10.8085],[68,10.9142],[71,10.9739],[75,10.9337],[76,10.8972],[79,10.72],[81,10.6453],[83,10.6155],[84,10.6411],[85,10.7213],[86,10.8619],[87,11.0674],[88,11.3681],[91,12.4568],[92,12.8574],[93,13.2119],[94,13.4458],[95,13.5765],[96,13.54],[97,13.3366],[98,13.0229],[99,12.6007],[100,12.0498],[101,11.3607],[104,9.6935],[105,9.1674],[106,8.6717],[107,8.0947],[108,7.6267],[109,7.2013],[110,6.8341],[111,6.5421],[112,6.3613],[113,6.2423],[114,6.1757],[115,6.1595],[116,6.1617],[117,6.2216],[118,6.3109],[119,6.4571],[120,6.6391],[122,7.0448],[123,7.3384],[125,7.6963],[127,8.0176],[128,8.1599],[129,8.2801],[130,8.3809],[131,8.5716],[134,8.8939],[137,9.1824],[138,9.2215],[139,9.2154],[140,9.1285],[141,8.9408],[142,8.6723],[143,8.3667],[144,8.03],[146,7.2968],[148,6.5405],[149,6.1742],[151,5.4868],[152,5.1669],[153,4.8871],[154,4.6353],[155,4.4084],[156,4.2053],[157,4.0261],[158,3.8653],[159,3.7225],[161,3.4706],[163,3.2352],[164,3.1401],[165,3.0339],[166,2.9443],[168,2.7492],[169,2.6606],[174,2.177],[175,2.1079],[176,2.069],[177,2.037],[178,2.0346],[179,2.0775],[180,2.1527],[181,2.2602],[182,2.409],[183,2.5865],[184,2.8164],[185,3.1081],[186,3.44],[187,3.793],[188,4.1785],[189,4.6061],[190,5.0655],[191,5.5384],[192,6.0561],[194,7.2497],[195,7.6758],[196,7.9982],[197,8.1764],[198,8.2274],[199,8.1472],[200,7.9376],[201,7.6113],[202,7.1],[203,6.4633],[204,5.7607],[205,5.0355],[206,4.3248],[208,2.9241],[209,2.2682],[210,1.7008],[211,1.2707],[212,0.8908],[213,0.5791],[214,0.3065],[215,0.0513],[216,-0.1683],[217,-0.3332],[218,-0.4537],[219,-0.5384],[220,-0.5945],[221,-0.6075],[222,-0.5707],[223,-0.4701],[224,-0.3291],[225,-0.1786],[226,-0.0139],[227,0.1577],[228,0.3322],[229,0.5099],[230,0.6837],[231,0.8487],[232,1.0001],[233,1.1358],[234,1.2654],[235,1.3876],[236,1.5027],[237,1.5993],[238,1.6762],[239,1.741],[240,1.798],[241,1.8372],[242,1.8664],[243,1.8851],[244,1.8397],[245,1.6597],[246,1.3028],[247,0.8129],[248,0.0569],[249,-0.906],[250,-1.7274],[251,-2.3672],[252,-2.7832],[253,-3.1101],[254,-3.1626],[255,-2.9668],[256,-2.498],[257,-1.7697],[258,-0.9637],[259,-0.1187],[260,0.8622],[261,1.9541],[262,3.0251],[263,3.8503],[264,4.5158],[265,4.997],[266,5.2924],[267,5.5156],[268,5.6147],[269,5.4594],[270,5.2754],[271,4.9594],[272,4.5348],[273,4.0156],[274,3.4489],[275,2.719],[276,2.0859],[277,1.5313],[278,1.0524],[279,0.6635],[280,0.3392],[281,0.0707],[282,-0.1559],[283,-0.3388],[284,-0.5085],[285,-0.6412],[286,-0.7649],[287,-0.8772],[288,-0.9754],[289,-1.0666],[290,-1.1487],[291,-1.1344],[292,-1.1948],[293,-1.2229],[294,-1.2303],[295,-1.2106],[296,-1.1234],[297,-0.9653],[298,-0.7375],[299,-0.3296],[300,0.066],[301,0.5125],[302,0.972],[303,1.4197],[304,1.846],[305,2.2362],[306,2.5946],[307,2.9183],[308,3.1989],[309,3.4226],[310,3.5995],[311,3.7347],[312,3.8157],[313,3.8648],[314,3.8801],[315,3.8583],[316,3.8144],[318,3.6986],[322,3.4369],[323,3.3631],[324,3.312]]
       }
       hand_r_ik_ctrl.Rotation.Z {
        type=MovieSceneScriptingFloatChannel
        default=0
        thinned_from=325
        keys [[0,9.359],[7,9.0713],[8,9.0456],[9,9.0608],[10,9.132],[16,9.7945],[17,9.8601],[22,10.4732],[23,10.5374],[28,10.9959],[29,11.0707],[30,11.1197],[31,11.1327],[32,11.1217],[33,11.0865],[34,11.0085],[35,10.8416],[36,10.5456],[38,9.8471],[40,9.261],[41,9.0374],[42,8.9223],[43,8.8787],[44,8.9225],[45,9.0598],[46,9.2549],[48,9.8265],[49,10.1462],[51,10.8961],[52,11.2381],[53,11.542],[54,11.7898],[55,11.988],[56,12.1237],[59,12.452],[62,12.7],[65,12.8576],[69,12.9471],[74,12.9329],[81,12.7712],[83,12.7677],[84,12.7956],[85,12.9015],[86,13.071],[87,13.2847],[88,13.5612],[90,14.1935],[91,14.4795],[92,14.719],[93,14.906],[94,15.0303],[95,15.1029],[96,15.0953],[97,15.0095],[98,14.8683],[99,14.6741],[100,14.4235],[101,14.0989],[103,13.6303],[105,13.2428],[106,13.0751],[107,12.8613],[109,12.5722],[111,12.3466],[113,12.2547],[114,12.2328],[116,12.25],[118,12.3352],[122,12.7096],[123,12.8614],[126,13.1012],[127,13.145],[129,13.1535],[130,13.1257],[131,13.1433],[137,12.9313],[139,12.7876],[140,12.6626],[141,12.4847],[144,11.8073],[148,10.8188],[150,10.3701],[152,9.9692],[154,9.6442],[156,9.3814],[158,9.1583],[160,8.9991],[165,8.6646],[168,8.4969],[169,8.4667],[174,8.2224],[175,8.1944],[177,8.1772],[178,8.185],[180,8.254],[181,8.3122],[183,8.4783],[186,8.7486],[187,8.8261],[188,8.9306],[189,9.101],[190,9.3166],[191,9.5598],[192,9.8322],[194,10.4442],[195,10.657],[196,10.8167],[197,10.9075],[198,10.9429],[199,10.9045],[200,10.8378],[201,10.7147],[202,10.5264],[208,9.1573],[209,8.9439],[211,8.587],[213,8.3105],[215,8.072],[216,7.9675],[217,7.8869],[218,7.8267],[220,7.7558],[221,7.7506],[222,7.7691],[223,7.8152],[230,8.2288],[232,8.3099],[236,8.4012],[240,8.4013],[241,8.3773],[245,8.4098],[247,8.2431],[248,8.0895],[249,7.8786],[251,7.4256],[253,7.1124],[254,7.0192],[255,6.998],[256,7.0955],[257,7.3451],[258,7.7014],[259,8.1493],[260,8.7022],[262,9.9467],[263,10.4396],[264,10.8537],[265,11.1718],[266,11.3877],[267,11.5528],[268,11.6421],[269,11.5843],[270,11.4911],[271,11.3048],[272,11.0351],[274,10.3906],[275,9.9836],[276,9.6327],[277,9.3265],[278,9.0609],[279,8.8429],[280,8.6675],[281,8.5132],[282,8.3897],[284,8.193],[286,8.0445],[287,7.9971],[288,7.9806],[289,7.9879],[290,8.0132],[291,8.0906],[295,8.3007],[297,8.4414],[298,8.5367],[302,9.2114],[303,9.3532],[304,9.4745],[306,9.6525],[308,9.7778],[310,9.839],[311,9.8463],[314,9.7809],[323,9.3912],[324,9.3588]]
       }
       hand_r_ik_ctrl.Scale.X {type=MovieSceneScriptingFloatChannel, default=1, const [1,325]}
       hand_r_ik_ctrl.Scale.Y {type=MovieSceneScriptingFloatChannel, default=1, const [1,325]}
       hand_r_ik_ctrl.Scale.Z {type=MovieSceneScriptingFloatChannel, default=1, const [1,325]}
       Softness_2 {type=MovieSceneScriptingFloatChannel, default=0, const [0,325]}
       hand_r_ikorient_ctrl.Rotation.Roll {type=MovieSceneScriptingFloatChannel, default=0, const [0,325]}
       hand_r_ikorient_ctrl.Rotation.Pitch {type=MovieSceneScriptingFloatChannel, default=0, const [0,325]}
       hand_r_ikorient_ctrl.Rotation.Yaw {type=MovieSceneScriptingFloatChannel, default=0, const [0,325]}
       arm_r_pv_ik_ctrl.Location.X {
        type=MovieSceneScriptingFloatChannel
        default=0
        thinned_from=325
        keys [[0,5.7569],[10,5.8532],[16,6.016],[17,6.0209],[22,6.2427],[23,6.246],[25,6.3086],[26,6.3292],[29,6.4096],[31,6.4406],[32,6.427],[34,6.3471],[35,6.2908],[38,6.0355],[39,5.9849],[40,5.8868],[41,5.7567],[45,5.1908],[46,5.0567],[48,4.8788],[49,4.8032],[51,4.6789],[52,4.6349],[53,4.6123],[55,4.6073],[74,4.6712],[77,4.7051],[81,4.7865],[83,4.8478],[84,4.8909],[85,4.9091],[86,4.9156],[89,4.8932],[90,4.897],[91,4.9379],[92,5.0055],[93,5.0898],[94,5.196],[95,5.3155],[96,5.4671],[97,5.63],[100,6.0628],[102,6.2615],[103,6.3382],[104,6.3893],[105,6.4195],[106,6.4349],[107,6.3992],[112,6.2731],[113,6.2354],[118,6.1142],[122,5.997],[123,5.9999],[126,5.917],[128,5.9002],[130,5.9112],[131,5.9575],[138,6.1364],[142,6.1916],[148,6.1781],[157,6.0979],[158,6.1],[168,6.0596],[177,5.9933],[182,5.8804],[184,5.7903],[187,5.6966],[188,5.6528],[190,5.5096],[191,5.4117],[192,5.3002],[193,5.215],[196,4.9993],[197,4.9403],[198,4.9001],[199,4.8868],[201,4.9121],[208,5.1745],[211,5.2535],[218,5.3679],[222,5.3992],[231,5.3815],[240,5.5028],[241,5.5305],[244,5.5394],[245,5.5313],[246,5.5586],[247,5.599],[248,5.6559],[249,5.732],[251,5.9521],[253,6.1378],[254,6.1652],[255,6.1136],[256,6.0135],[257,5.8878],[260,5.4255],[261,5.2894],[262,5.1693],[263,5.0892],[264,5.0375],[265,5.0165],[266,5.0323],[267,5.0682],[269,5.1726],[270,5.2778],[273,5.6367],[274,5.729],[277,5.8572],[278,5.8833],[280,5.9109],[286,5.9226],[287,5.9137],[289,5.8601],[290,5.82],[291,5.8036],[294,5.6607],[296,5.5971],[298,5.5758],[299,5.6006],[306,5.6817],[312,5.6996],[319,5.7582],[324,5.7579]]
       }
       arm_r_pv_ik_ctrl.Location.Y {
        type=MovieSceneScriptingFloatChannel
        default=0
        thinned_from=325
        keys [[0,7.3282],[11,7.3826],[15,7.3293],[16,7.3041],[17,7.2356],[22,7.1153],[23,7.0513],[25,6.9653],[29,6.7499],[32,6.5353],[33,6.479],[34,6.4385],[35,6.4203],[36,6.499],[37,6.6015],[38,6.6632],[40,6.8992],[41,7.0904],[44,7.7204],[46,8.2003],[47,8.4839],[48,8.737],[50,9.1958],[51,9.4028],[52,9.5735],[53,9.6916],[54,9.7661],[55,9.8154],[62,9.8807],[75,9.7444],[79,9.6019],[82,9.4325],[85,9.2075],[86,9.2045],[87,9.246],[89,9.3714],[90,9.4035],[92,9.3576],[93,9.3132],[94,9.2023],[95,9.0382],[96,8.7741],[98,8.141],[99,7.803],[100,7.4279],[101,7.015],[102,6.7152],[103,6.4575],[104,6.2411],[105,6.0657],[107,5.7907],[108,5.7118],[109,5.6651],[110,5.6422],[111,5.6469],[113,5.7099],[116,5.9316],[120,6.2816],[122,6.4178],[123,6.5185],[129,6.8694],[130,6.919],[131,7.0078],[139,7.3055],[150,7.8438],[156,8.0215],[174,8.2566],[177,8.3365],[179,8.4281],[182,8.6235],[186,9.0327],[187,9.1083],[188,9.1552],[189,9.1499],[191,8.9739],[192,8.8542],[193,8.5355],[196,7.4276],[197,7.1262],[200,6.386],[201,6.1225],[202,5.8291],[204,5.1699],[205,4.8687],[206,4.612],[208,4.1521],[209,3.9495],[210,3.8026],[211,3.7361],[212,3.685],[213,3.6482],[214,3.6003],[216,3.4803],[219,3.3377],[221,3.2273],[222,3.1862],[223,3.1663],[224,3.1641],[225,3.1694],[226,3.1828],[227,3.2029],[228,3.2298],[231,3.3336],[232,3.3615],[236,3.4349],[238,3.4534],[241,3.4562],[243,3.4437],[244,3.4131],[245,3.345],[246,3.2325],[247,3.0537],[248,2.8876],[249,2.7156],[250,2.6305],[251,2.7028],[252,2.9364],[253,3.3383],[254,3.9968],[256,5.5156],[257,6.1996],[258,6.8054],[259,7.2981],[260,7.7071],[261,8.0356],[262,8.2956],[263,8.4399],[264,8.5006],[265,8.4826],[266,8.3843],[267,8.2505],[268,8.0798],[269,7.8362],[271,7.4058],[273,7.0218],[274,6.8635],[275,6.6833],[276,6.5702],[277,6.4819],[278,6.4179],[279,6.3858],[280,6.3752],[281,6.3443],[287,6.2966],[290,6.3053],[291,6.347],[295,6.3809],[298,6.4419],[299,6.5041],[302,6.6547],[304,6.7926],[308,7.1555],[309,7.2263],[311,7.3311],[312,7.366],[314,7.3933],[324,7.3271]]
       }
       arm_r_pv_ik_ctrl.Location.Z {
        type=MovieSceneScriptingFloatChannel
        default=0
        thinned_from=325
        keys [[0,3.9239],[1,3.9286],[2,3.9614],[3,4.007],[4,4.0686],[5,4.1466],[6,4.2394],[7,4.3465],[8,4.467],[9,4.6014],[10,4.748],[11,4.9115],[12,5.0936],[15,5.6739],[16,5.8528],[17,6.0571],[19,6.3374],[21,6.5894],[22,6.7023],[23,6.8548],[25,7.0856],[27,7.2575],[29,7.3903],[30,7.4424],[31,7.4772],[32,7.4649],[33,7.4182],[34,7.3285],[35,7.2744],[38,7.3789],[40,7.3215],[41,7.2538],[43,7.0577],[44,6.8883],[45,6.6433],[46,6.356],[47,5.9943],[51,4.631],[52,4.3097],[53,4.0222],[54,3.7689],[55,3.5517],[56,3.3834],[57,3.243],[58,3.1126],[59,3.0007],[60,2.9114],[61,2.8411],[62,2.7785],[63,2.7226],[65,2.6268],[66,2.5866],[67,2.5517],[68,2.5222],[69,2.4985],[70,2.4816],[71,2.4706],[73,2.4655],[74,2.4769],[75,2.496],[77,2.5643],[78,2.6158],[80,2.748],[82,2.9248],[83,3.0345],[84,3.1672],[85,3.2725],[86,3.3578],[87,3.4243],[88,3.4821],[89,3.5287],[90,3.5557],[91,3.5729],[93,3.564],[94,3.6121],[95,3.717],[96,3.9236],[97,4.2095],[98,4.5296],[99,4.8891],[100,5.2868],[101,5.7218],[102,6.0587],[103,6.3609],[104,6.6256],[105,6.8495],[107,7.2271],[108,7.3352],[109,7.4087],[110,7.4534],[111,7.4719],[112,7.4677],[114,7.4367],[116,7.3642],[118,7.2397],[122,6.8883],[123,6.7543],[129,6.2324],[130,6.1374],[131,5.9893],[132,5.8673],[138,4.9963],[140,4.7274],[144,4.2185],[145,4.1043],[149,3.6895],[150,3.5906],[151,3.5053],[153,3.3637],[154,3.3027],[155,3.2506],[156,3.2066],[157,3.1699],[158,3.1502],[159,3.1204],[161,3.0946],[162,3.0926],[163,3.0991],[164,3.0988],[166,3.1167],[169,3.1647],[170,3.2134],[171,3.2713],[172,3.3392],[177,3.7436],[178,3.8181],[182,4.073],[184,4.1802],[185,4.2534],[186,4.3399],[188,4.533],[189,4.5935],[190,4.6217],[191,4.609],[192,4.5408],[194,4.3259],[195,4.2488],[196,4.2025],[197,4.2089],[198,4.2435],[199,4.3276],[200,4.4544],[201,4.6013],[202,4.7935],[205,5.4956],[208,6.1479],[209,6.3392],[210,6.4892],[211,6.5866],[212,6.6439],[213,6.6656],[214,6.6623],[215,6.6455],[216,6.6122],[217,6.5581],[218,6.4795],[221,6.1901],[223,5.9668],[229,5.204],[231,4.9702],[232,4.8694],[234,4.7048],[236,4.5663],[237,4.5111],[238,4.4691],[240,4.4131],[241,4.4153],[242,4.4036],[243,4.4026],[244,4.4344],[245,4.5169],[246,4.7148],[247,4.9923],[248,5.4205],[249,5.952],[250,6.4284],[251,6.8143],[252,7.0365],[253,7.0996],[254,6.9369],[255,6.6381],[256,6.236],[258,5.343],[259,4.9608],[260,4.629],[261,4.367],[262,4.1624],[263,4.0613],[264,4.0369],[265,4.0898],[266,4.2269],[267,4.3949],[268,4.5917],[270,5.1137],[271,5.4058],[273,6.0111],[274,6.2786],[275,6.569],[276,6.7821],[277,6.9601],[278,7.0915],[279,7.1618],[282,7.2711],[286,7.342],[287,7.3435],[289,7.3047],[290,7.2713],[291,7.192],[297,6.9317],[298,6.871],[299,6.7481],[300,6.6428],[301,6.5095],[303,6.1897],[304,6.0151],[307,5.4416],[308,5.2614],[310,4.9506],[311,4.8084],[312,4.6811],[314,4.4891],[316,4.3206],[318,4.1728],[320,4.0463],[321,3.996],[322,3.9574],[323,3.9381],[324,3.9274]]
       }
       arm_r_pv_ik_ctrl.Rotation.X {
        type=MovieSceneScriptingFloatChannel
        default=0
        thinned_from=325
        keys [[0,13.5438],[10,13.645],[14,13.7851],[16,13.8987],[17,14.04],[22,14.3002],[23,14.4584],[30,15.3152],[35,15.8165],[37,16.0608],[38,16.1437],[39,16.1258],[41,16.142],[43,16.0775],[44,15.9416],[45,15.6941],[46,15.3881],[47,14.9486],[50,13.7687],[51,13.3995],[52,13.069],[53,12.7791],[54,12.5343],[55,12.3396],[56,12.2137],[58,12.0449],[61,11.9214],[68,11.8876],[75,12.0197],[78,12.1956],[80,12.3686],[82,12.5944],[83,12.735],[84,12.9096],[85,13.038],[86,13.101],[87,13.0923],[88,12.9794],[89,12.7816],[90,12.5169],[91,12.2081],[92,11.8329],[93,11.5021],[94,11.3345],[95,11.2799],[96,11.4061],[97,11.675],[98,12.0007],[99,12.405],[100,12.9032],[101,13.4908],[102,13.9081],[103,14.2799],[104,14.6053],[105,14.88],[106,15.102],[107,15.3742],[108,15.5161],[109,15.6176],[111,15.7162],[114,15.6945],[116,15.6218],[118,15.4729],[122,15.0301],[123,14.8425],[127,14.4746],[130,14.3119],[131,14.1817],[143,13.3266],[151,12.5896],[156,12.2871],[162,12.0889],[175,11.8364],[178,11.7044],[182,11.3136],[184,11.0781],[188,10.7396],[189,10.611],[191,10.31],[192,10.1146],[193,9.788],[194,9.4368],[195,9.1278],[196,8.8885],[197,8.7856],[198,8.7679],[199,8.856],[200,9.073],[201,9.3401],[202,9.7205],[203,10.1937],[205,11.2079],[208,12.6634],[209,13.1012],[210,13.4587],[211,13.7064],[212,13.9124],[213,14.066],[215,14.3057],[216,14.4005],[221,14.6095],[223,14.5999],[227,14.3778],[231,14.0541],[236,13.8357],[240,13.7971],[243,13.8324],[244,13.9024],[245,14.0903],[246,14.4685],[247,14.9692],[248,15.7365],[249,16.7058],[250,17.5547],[251,18.231],[252,18.6626],[253,18.8715],[254,18.7056],[255,18.2691],[256,17.5719],[257,16.7109],[258,15.8116],[259,15.0061],[260,14.2643],[261,13.6328],[262,13.0996],[263,12.7436],[264,12.5111],[265,12.4122],[266,12.4676],[267,12.5804],[268,12.7537],[269,13.0587],[270,13.3204],[273,14.2115],[274,14.4597],[275,14.7779],[277,15.1693],[278,15.3017],[282,15.475],[287,15.5736],[290,15.5318],[291,15.4308],[298,15.3256],[299,15.2142],[300,15.1461],[302,14.9272],[304,14.6472],[308,13.9601],[310,13.7295],[312,13.5711],[315,13.5102],[324,13.5452]]
       }
       arm_r_pv_ik_ctrl.Rotation.Y {
        type=MovieSceneScriptingFloatChannel
        default=0
        thinned_from=325
        keys [[0,-7.9882],[10,-7.8798],[17,-7.6832],[22,-7.4829],[26,-7.3753],[28,-7.3152],[37,-7.1855],[38,-7.1913],[39,-7.2127],[40,-7.2831],[41,-7.3849],[44,-7.742],[46,-7.9419],[47,-8.0013],[48,-8.0274],[50,-8.0309],[53,-7.9557],[57,-7.8085],[76,-7.6148],[77,-7.6212],[84,-7.5289],[85,-7.5742],[86,-7.655],[87,-7.7619],[88,-7.8982],[90,-8.2313],[91,-8.3705],[92,-8.4574],[93,-8.5077],[95,-8.5459],[97,-8.5205],[102,-8.3767],[103,-8.372],[104,-8.3913],[106,-8.4781],[110,-8.7487],[113,-8.8718],[122,-9.0122],[126,-9.0431],[127,-9.0237],[129,-8.9292],[130,-8.8654],[136,-8.3819],[142,-7.9807],[146,-7.8328],[149,-7.7733],[157,-7.7421],[158,-7.7273],[159,-7.7406],[160,-7.7312],[168,-7.735],[169,-7.7565],[183,-7.9282],[184,-7.9174],[185,-7.8599],[186,-7.7736],[188,-7.5757],[189,-7.528],[190,-7.5107],[192,-7.5249],[196,-7.6128],[198,-7.6291],[199,-7.6227],[201,-7.6573],[202,-7.694],[203,-7.7535],[209,-8.226],[210,-8.2925],[212,-8.3468],[222,-8.3714],[228,-8.2744],[240,-7.8467],[241,-7.799],[243,-7.7947],[244,-7.8105],[245,-7.8772],[246,-7.9273],[247,-8.0155],[249,-8.2672],[250,-8.3239],[253,-8.308],[256,-8.226],[257,-8.233],[258,-8.3012],[262,-8.7731],[264,-8.9533],[265,-9.0233],[266,-9.0684],[267,-9.0932],[269,-9.0969],[270,-9.0511],[271,-8.9779],[273,-8.7904],[274,-8.7238],[277,-8.6143],[286,-8.5576],[287,-8.5714],[288,-8.6086],[291,-8.8049],[293,-8.9585],[294,-9.0204],[295,-9.0631],[296,-9.0841],[298,-9.0507],[300,-8.9693],[302,-8.8431],[307,-8.4428],[311,-8.2125],[314,-8.0937],[318,-8.008],[324,-7.9884]]
       }
       arm_r_pv_ik_ctrl.Rotation.Z {
        type=MovieSceneScriptingFloatChannel
        default=0
        thinned_from=325
        keys [[0,27.3313],[16,27.2105],[17,27.1053],[22,27.1221],[23,27.0136],[38,26.4166],[39,26.4715],[43,26.4361],[46,26.6242],[47,26.8227],[53,27.6101],[58,27.882],[79,27.7998],[86,27.5012],[88,27.5385],[91,27.8065],[93,28.0603],[95,28.1402],[97,27.9965],[99,27.7339],[100,27.5558],[101,27.2853],[105,26.8718],[106,26.8122],[107,26.6618],[113,26.5216],[122,26.776],[123,26.8995],[130,26.9913],[131,27.0872],[189,28.4165],[192,28.599],[196,29.0476],[198,29.0924],[200,28.9807],[202,28.7377],[209,27.4812],[212,27.1899],[222,26.9356],[243,27.2422],[245,27.1366],[247,26.8166],[248,26.5389],[250,25.8894],[251,25.6508],[252,25.4994],[253,25.4264],[254,25.4846],[255,25.6384],[256,25.8847],[259,26.7953],[261,27.2841],[262,27.4748],[264,27.6844],[266,27.6968],[268,27.5945],[269,27.4273],[274,27.0386],[275,26.8717],[278,26.6856],[290,26.6009],[291,26.6919],[298,26.6285],[299,26.7105],[312,27.3171],[324,27.3311]]
       }
       arm_r_pv_ik_ctrl.Scale.X {type=MovieSceneScriptingFloatChannel, default=1, const [1,325]}
       arm_r_pv_ik_ctrl.Scale.Y {type=MovieSceneScriptingFloatChannel, default=1, const [1,325]}
       arm_r_pv_ik_ctrl.Scale.Z {type=MovieSceneScriptingFloatChannel, default=1, const [1,325]}
       body_ctrl.Location.X {
        type=MovieSceneScriptingFloatChannel
        default=0
        thinned_from=325
        keys [[0,1.5753],[16,1.6067],[17,1.6144],[22,1.6102],[23,1.6151],[28,1.6186],[29,1.6244],[30,1.6354],[31,1.6541],[32,1.7007],[33,1.7707],[35,1.9238],[36,1.9862],[38,2.0953],[39,2.1238],[40,2.1409],[41,2.1519],[43,2.1604],[46,2.1581],[47,2.1495],[56,2.1219],[82,2.1035],[83,2.0971],[84,2.0842],[85,2.0562],[86,2.0077],[87,1.9392],[88,1.8552],[91,1.5767],[92,1.4899],[93,1.4096],[94,1.3393],[95,1.2784],[96,1.2253],[97,1.1822],[98,1.1514],[99,1.1356],[100,1.1327],[102,1.1501],[105,1.1849],[106,1.1941],[107,1.2083],[109,1.2205],[111,1.2252],[122,1.2233],[123,1.2176],[130,1.2176],[131,1.2102],[139,1.1863],[170,1.1707],[176,1.1801],[178,1.1887],[180,1.2045],[182,1.2268],[183,1.2414],[184,1.2595],[185,1.2821],[186,1.3091],[187,1.342],[188,1.3809],[189,1.4301],[190,1.4923],[191,1.564],[193,1.7126],[194,1.7763],[196,1.8959],[197,1.9516],[198,1.9962],[199,2.031],[200,2.0593],[201,2.0822],[203,2.1145],[205,2.1322],[213,2.1336],[246,2.0952],[247,2.0859],[248,2.066],[249,2.0417],[250,2.0092],[251,1.958],[252,1.8922],[254,1.6846],[255,1.6174],[256,1.5687],[257,1.538],[258,1.5188],[259,1.5065],[260,1.4997],[262,1.4975],[268,1.5169],[269,1.5258],[274,1.5355],[275,1.5443],[290,1.5772],[291,1.5738],[298,1.589],[299,1.5842],[324,1.5753]]
       }
       body_ctrl.Location.Y {
        type=MovieSceneScriptingFloatChannel
        default=0
        thinned_from=325
        keys [[0,-1.7915],[5,-1.7812],[10,-1.7517],[16,-1.7026],[17,-1.6834],[22,-1.6711],[23,-1.6526],[26,-1.6386],[33,-1.6407],[35,-1.6445],[38,-1.6406],[39,-1.6526],[45,-1.6521],[46,-1.6545],[47,-1.6727],[53,-1.727],[55,-1.7394],[85,-1.7844],[94,-1.7506],[100,-1.7042],[101,-1.6847],[106,-1.6718],[107,-1.6526],[109,-1.6421],[122,-1.6403],[123,-1.6541],[129,-1.6562],[130,-1.6588],[131,-1.6775],[137,-1.7331],[139,-1.7456],[169,-1.7915],[174,-1.7754],[183,-1.7032],[187,-1.6873],[210,-1.7153],[212,-1.7234],[221,-1.8088],[245,-1.8576],[248,-1.8534],[252,-1.8276],[268,-1.7019],[269,-1.6829],[274,-1.671],[275,-1.6526],[278,-1.6386],[290,-1.6406],[291,-1.654],[297,-1.6557],[298,-1.658],[299,-1.6756],[306,-1.7366],[318,-1.7848],[324,-1.7915]]
       }
       body_ctrl.Location.Z {
        type=MovieSceneScriptingFloatChannel
        default=0
        thinned_from=325
        keys [[0,1.6458],[2,1.649],[3,1.6544],[4,1.6637],[5,1.6778],[6,1.6978],[7,1.7247],[8,1.7594],[9,1.8049],[10,1.8615],[11,1.9259],[14,2.135],[15,2.1994],[16,2.256],[17,2.3025],[19,2.3726],[21,2.428],[23,2.4692],[25,2.494],[27,2.5024],[34,2.4931],[38,2.4675],[41,2.431],[43,2.397],[44,2.3728],[45,2.3397],[46,2.2992],[48,2.2006],[52,1.9838],[53,1.9343],[54,1.8899],[55,1.8521],[56,1.8226],[58,1.777],[60,1.7398],[62,1.7101],[65,1.678],[68,1.6585],[72,1.6477],[77,1.6492],[80,1.6619],[82,1.6797],[84,1.7078],[85,1.7269],[86,1.7503],[88,1.808],[90,1.8772],[98,2.1946],[101,2.2989],[106,2.454],[108,2.5032],[109,2.5203],[110,2.5312],[111,2.5351],[114,2.5244],[117,2.4951],[122,2.4155],[128,2.291],[130,2.2354],[137,1.9986],[139,1.9407],[144,1.8329],[149,1.7442],[153,1.6881],[157,1.6472],[160,1.6286],[168,1.6168],[169,1.6192],[170,1.6264],[171,1.6377],[172,1.6529],[174,1.693],[176,1.743],[185,2.0021],[187,2.0697],[191,2.2173],[193,2.2798],[194,2.3045],[196,2.3382],[200,2.3797],[206,2.403],[212,2.4009],[214,2.3847],[216,2.3558],[219,2.2943],[224,2.1663],[227,2.0697],[234,1.8142],[236,1.7479],[238,1.6906],[239,1.6663],[240,1.6454],[242,1.6133],[244,1.5917],[246,1.5787],[249,1.5706],[254,1.5729],[257,1.5864],[259,1.6029],[261,1.6262],[263,1.6568],[265,1.6956],[267,1.743],[269,1.8009],[270,1.8412],[271,1.8978],[272,1.9664],[275,2.2003],[276,2.2724],[277,2.3343],[278,2.3814],[279,2.4094],[282,2.4509],[286,2.4826],[296,2.4941],[298,2.486],[301,2.4568],[304,2.4083],[306,2.3642],[308,2.3104],[309,2.2771],[310,2.2369],[311,2.1909],[313,2.0869],[317,1.8656],[318,1.8148],[319,1.7684],[320,1.7276],[321,1.6937],[322,1.6679],[323,1.6516],[324,1.6458]]
       }
       body_ctrl.Rotation.X {
        type=MovieSceneScriptingFloatChannel
        default=0
        thinned_from=325
        keys [[0,-1.3872],[1,-1.3832],[2,-1.3347],[3,-1.2714],[4,-1.1869],[5,-1.0807],[6,-0.9554],[7,-0.8119],[8,-0.6519],[9,-0.4755],[10,-0.2849],[11,-0.0834],[12,0.1231],[13,0.3292],[14,0.5299],[15,0.7232],[16,0.8998],[17,1.1307],[18,1.2529],[19,1.3833],[21,1.6487],[22,1.7738],[23,1.9821],[24,2.1012],[25,2.1951],[26,2.255],[27,2.2763],[30,2.2724],[33,2.22],[35,2.1661],[36,2.1325],[37,2.0913],[38,2.0429],[39,1.9172],[40,1.8693],[41,1.8158],[42,1.7565],[43,1.6916],[44,1.6095],[45,1.5011],[46,1.3715],[47,1.1326],[48,0.9409],[50,0.5387],[51,0.3405],[52,0.1518],[53,-0.0211],[54,-0.1733],[55,-0.2993],[56,-0.3923],[57,-0.4653],[58,-0.5365],[59,-0.6042],[60,-0.6682],[61,-0.7303],[62,-0.7889],[63,-0.8446],[64,-0.8971],[65,-0.9474],[66,-0.9945],[67,-1.0384],[69,-1.1195],[71,-1.1892],[73,-1.2488],[75,-1.2981],[76,-1.3195],[78,-1.3538],[79,-1.3575],[80,-1.3794],[81,-1.3807],[82,-1.3949],[84,-1.3962],[85,-1.3923],[86,-1.3524],[87,-1.2907],[88,-1.2043],[89,-1.0957],[90,-0.9665],[91,-0.8198],[92,-0.6563],[93,-0.4785],[94,-0.2876],[95,-0.0863],[96,0.1191],[97,0.3256],[98,0.5271],[99,0.7192],[100,0.8958],[101,1.127],[102,1.2497],[103,1.3806],[105,1.6471],[106,1.7727],[107,1.9856],[108,2.1027],[109,2.1967],[110,2.254],[111,2.2767],[114,2.2714],[115,2.2535],[116,2.248],[117,2.2181],[119,2.1646],[121,2.0903],[122,2.0458],[123,1.9206],[124,1.8765],[125,1.8267],[126,1.771],[127,1.7091],[128,1.6289],[129,1.522],[130,1.3934],[131,1.155],[132,0.9636],[133,0.7638],[134,0.5613],[135,0.3628],[136,0.1737],[137,-0.0005],[138,-0.153],[139,-0.2805],[140,-0.3738],[141,-0.448],[142,-0.5188],[143,-0.5873],[144,-0.6516],[145,-0.7133],[146,-0.772],[147,-0.8283],[148,-0.8813],[149,-0.9311],[151,-1.0223],[152,-1.0641],[153,-1.1036],[155,-1.1734],[156,-1.2046],[158,-1.2592],[159,-1.2833],[161,-1.3228],[162,-1.3387],[163,-1.3424],[164,-1.3641],[165,-1.3661],[166,-1.38],[168,-1.3806],[169,-1.3773],[170,-1.3296],[171,-1.2582],[172,-1.1624],[173,-1.0438],[174,-0.9037],[175,-0.745],[176,-0.5706],[177,-0.3757],[178,-0.1651],[179,0.0552],[180,0.2787],[181,0.4986],[182,0.705],[183,0.8945],[184,1.052],[185,1.1999],[187,1.5049],[188,1.654],[189,1.7959],[190,1.9265],[191,2.0421],[192,2.1396],[193,2.2158],[194,2.2662],[195,2.2859],[198,2.2841],[199,2.2716],[200,2.2671],[203,2.1898],[204,2.1565],[205,2.1173],[206,2.0719],[207,2.0198],[208,1.9609],[209,1.8947],[210,1.8208],[211,1.7387],[212,1.6307],[213,1.4847],[214,1.3085],[215,1.1099],[216,0.8965],[218,0.4568],[219,0.2461],[220,0.0518],[221,-0.1182],[222,-0.256],[223,-0.3542],[224,-0.4281],[225,-0.4988],[226,-0.5662],[227,-0.6304],[228,-0.6915],[229,-0.7496],[230,-0.8047],[231,-0.8567],[232,-0.906],[233,-0.9523],[234,-0.996],[235,-1.0368],[236,-1.075],[237,-1.1107],[238,-1.1438],[239,-1.1743],[241,-1.2282],[243,-1.2722],[245,-1.3076],[246,-1.3117],[247,-1.3361],[250,-1.372],[251,-1.3798],[252,-1.3844],[253,-1.3728],[254,-1.3349],[255,-1.272],[256,-1.1868],[257,-1.0798],[258,-0.9538],[259,-0.8091],[260,-0.6476],[261,-0.4696],[262,-0.2784],[263,-0.078],[264,0.1261],[265,0.3311],[266,0.5308],[267,0.7186],[268,0.8966],[269,1.127],[270,1.2493],[271,1.3799],[273,1.6463],[274,1.7719],[275,1.9808],[276,2.1004],[277,2.1947],[278,2.2549],[279,2.2776],[280,2.2776],[282,2.271],[283,2.2534],[284,2.2464],[287,2.1645],[288,2.1292],[290,2.0434],[291,1.9206],[292,1.8764],[293,1.8267],[294,1.771],[295,1.709],[296,1.6307],[297,1.5286],[298,1.4067],[299,1.1808],[300,1.003],[302,0.6162],[303,0.4217],[304,0.234],[305,0.0564],[306,-0.1083],[307,-0.2528],[308,-0.3738],[309,-0.4791],[310,-0.5816],[311,-0.6797],[312,-0.7736],[313,-0.8617],[314,-0.9453],[315,-1.0232],[316,-1.0946],[317,-1.1593],[318,-1.2176],[319,-1.2671],[320,-1.3092],[321,-1.3427],[322,-1.3667],[323,-1.371],[324,-1.3872]]
       }
       body_ctrl.Rotation.Y {
        type=MovieSceneScriptingFloatChannel
        default=0
        thinned_from=325
        keys [[0,0],[30,0],[31,-0.027],[32,-0.079],[33,-0.1586],[34,-0.2729],[35,-0.3649],[36,-0.3975],[37,-0.4169],[38,-0.4226],[39,-0.4181],[40,-0.4094],[41,-0.3993],[43,-0.3774],[45,-0.3611],[46,-0.3541],[48,-0.3423],[50,-0.3329],[51,-0.3291],[52,-0.3265],[53,-0.3255],[54,-0.3264],[57,-0.3317],[59,-0.333],[61,-0.332],[70,-0.325],[79,-0.324],[85,-0.3252],[86,-0.3237],[87,-0.3091],[88,-0.2649],[89,-0.203],[90,-0.1384],[91,-0.0751],[92,-0.0151],[93,0.0185],[94,0.0175],[95,0.0094],[96,-0.0029],[97,-0.0136],[98,-0.0203],[99,-0.0249],[100,-0.0281],[101,-0.0303],[102,-0.0319],[103,-0.0328],[104,-0.033],[105,-0.0328],[107,-0.0315],[109,-0.0297],[110,-0.0283],[111,-0.0268],[114,-0.0268],[115,-0.0168],[116,-0.0168],[117,-0.0111],[118,-0.0087],[119,-0.0067],[120,-0.0054],[121,-0.005],[172,-0.0088],[177,-0.0083],[181,-0.0071],[182,-0.0059],[183,-0.0119],[184,-0.0343],[185,-0.0721],[186,-0.123],[187,-0.1963],[188,-0.2887],[190,-0.4906],[191,-0.5988],[193,-0.8556],[194,-0.9531],[195,-1.0193],[196,-1.0743],[197,-1.1144],[198,-1.1349],[199,-1.1632],[200,-1.174],[201,-1.1884],[203,-1.2029],[205,-1.2114],[213,-1.218],[221,-1.2647],[244,-1.2959],[245,-1.2917],[246,-1.2802],[247,-1.2418],[248,-1.1858],[249,-1.0854],[250,-0.942],[252,-0.6291],[253,-0.4681],[254,-0.314],[255,-0.1844],[256,-0.0978],[257,-0.0452],[258,-0.0048],[259,0.0252],[260,0.0465],[261,0.0609],[262,0.0686],[263,0.0702],[264,0.0674],[265,0.0618],[266,0.0551],[267,0.0474],[268,0.0392],[269,0.0315],[270,0.0247],[271,0.0193],[272,0.015],[273,0.0113],[274,0.0082],[275,0.0058],[276,0.004],[277,0.0029],[279,0.0017],[282,0.0017],[283,0.0004],[284,0.0004],[324,0]]
       }
       body_ctrl.Rotation.Z {
        type=MovieSceneScriptingFloatChannel
        default=0
        thinned_from=325
        keys [[0,-19.4399],[16,-19.4397],[17,-19.3676],[22,-19.5076],[23,-19.4394],[29,-19.4176],[32,-19.2893],[33,-19.2997],[34,-19.3855],[35,-19.6281],[36,-20.295],[37,-21.0586],[38,-21.5205],[39,-21.8444],[42,-22.1315],[44,-22.2228],[46,-22.2166],[47,-22.2692],[83,-22.0356],[87,-22.1881],[88,-22.1423],[89,-21.985],[90,-21.6935],[91,-21.1767],[92,-20.5246],[93,-19.9855],[94,-19.6321],[95,-19.3711],[96,-19.1969],[97,-19.1069],[98,-19.1116],[100,-19.2756],[101,-19.2699],[103,-19.4117],[106,-19.5075],[107,-19.4401],[122,-19.4389],[123,-19.5104],[130,-19.3822],[131,-19.4426],[183,-19.4561],[186,-19.5524],[187,-19.54],[188,-19.4533],[189,-19.2198],[190,-18.7297],[191,-18.0894],[192,-17.5451],[193,-17.1787],[194,-16.9092],[195,-16.7397],[197,-16.543],[199,-16.4218],[243,-16.3612],[244,-16.4087],[245,-16.5333],[246,-16.7812],[247,-17.2318],[248,-18.133],[249,-18.8537],[250,-19.0463],[251,-19.1919],[253,-19.3672],[268,-19.4417],[269,-19.3687],[274,-19.5079],[275,-19.4397],[290,-19.4391],[291,-19.5097],[298,-19.3847],[299,-19.4384],[324,-19.4399]]
       }
       body_ctrl.Scale.X {type=MovieSceneScriptingFloatChannel, default=1, const [1,325]}
       body_ctrl.Scale.Y {type=MovieSceneScriptingFloatChannel, default=1, const [1,325]}
       body_ctrl.Scale.Z {type=MovieSceneScriptingFloatChannel, default=1, const [1,325]}
       hips_ctrl.Location.X {type=MovieSceneScriptingFloatChannel, default=0, thinned_from=325, keys [[0,0],[324,0]]}
       hips_ctrl.Location.Y {type=MovieSceneScriptingFloatChannel, default=0, thinned_from=325, keys [[0,0],[324,0]]}
       hips_ctrl.Location.Z {type=MovieSceneScriptingFloatChannel, default=0, thinned_from=325, keys [[0,0],[324,0]]}
       hips_ctrl.Rotation.X {
        type=MovieSceneScriptingFloatChannel
        default=0
        thinned_from=325
        keys [[0,-1.2664],[1,-1.2664],[2,-1.3044],[3,-1.3505],[4,-1.4135],[5,-1.4926],[6,-1.5869],[7,-1.6956],[8,-1.8178],[9,-1.9548],[10,-2.1055],[11,-2.2656],[14,-2.7597],[15,-2.9146],[16,-3.0574],[18,-3.3058],[21,-3.7016],[22,-3.8267],[23,-3.9402],[24,-4.037],[25,-4.1124],[26,-4.1612],[27,-4.1785],[30,-4.1785],[34,-4.1223],[36,-4.0772],[38,-4.012],[40,-3.9223],[42,-3.8096],[43,-3.7446],[44,-3.6625],[45,-3.5541],[46,-3.4245],[47,-3.2784],[48,-3.1206],[51,-2.6259],[52,-2.4702],[53,-2.3274],[54,-2.2023],[55,-2.0998],[56,-2.0244],[58,-1.9096],[60,-1.8053],[62,-1.7113],[64,-1.6273],[66,-1.553],[68,-1.488],[70,-1.4316],[72,-1.3837],[74,-1.3436],[76,-1.3113],[78,-1.2864],[79,-1.2864],[80,-1.2684],[81,-1.2684],[82,-1.2568],[85,-1.2568],[86,-1.2864],[87,-1.3315],[88,-1.3955],[89,-1.4773],[90,-1.575],[91,-1.6873],[92,-1.813],[93,-1.9519],[94,-2.1028],[95,-2.2627],[98,-2.7559],[99,-2.9106],[100,-3.0533],[102,-3.3026],[105,-3.7],[106,-3.8256],[107,-3.9396],[108,-4.0369],[109,-4.1125],[110,-4.1615],[111,-4.1789],[114,-4.1789],[115,-4.1624],[116,-4.1624],[120,-4.0756],[122,-4.0124],[124,-3.9294],[126,-3.8239],[127,-3.762],[128,-3.6818],[129,-3.5749],[130,-3.4463],[131,-3.3007],[132,-3.1432],[135,-2.6482],[136,-2.492],[137,-2.3485],[138,-2.2226],[139,-2.119],[140,-2.0428],[142,-1.9268],[144,-1.8219],[146,-1.7276],[148,-1.6435],[150,-1.5691],[152,-1.5039],[154,-1.4475],[156,-1.3993],[158,-1.3591],[160,-1.3267],[162,-1.3017],[163,-1.3017],[164,-1.2836],[165,-1.2836],[166,-1.272],[169,-1.272],[170,-1.3078],[171,-1.3595],[172,-1.4301],[173,-1.5184],[174,-1.6232],[175,-1.7431],[176,-1.877],[177,-2.0278],[178,-2.1949],[181,-2.7307],[182,-2.8992],[183,-3.0518],[184,-3.1818],[188,-3.6704],[189,-3.786],[190,-3.8926],[191,-3.9871],[192,-4.0669],[193,-4.1297],[194,-4.1716],[195,-4.1884],[198,-4.1909],[199,-4.1819],[200,-4.1819],[204,-4.1003],[206,-4.0371],[208,-3.9524],[210,-3.8437],[211,-3.7793],[212,-3.693],[213,-3.5746],[214,-3.4305],[215,-3.2674],[218,-2.729],[219,-2.5552],[220,-2.395],[221,-2.2552],[222,-2.1423],[223,-2.0627],[225,-1.9472],[227,-1.8431],[229,-1.7498],[231,-1.6668],[233,-1.5935],[235,-1.5295],[237,-1.4742],[239,-1.427],[241,-1.3875],[243,-1.3555],[245,-1.3301],[246,-1.3301],[247,-1.309],[250,-1.2798],[251,-1.2726],[252,-1.2686],[253,-1.2763],[254,-1.3036],[255,-1.3495],[256,-1.4129],[257,-1.4927],[258,-1.5879],[259,-1.698],[260,-1.8221],[261,-1.9606],[262,-2.1115],[266,-2.7596],[267,-2.9127],[268,-3.0542],[270,-3.3022],[273,-3.6992],[274,-3.8248],[275,-3.9388],[276,-4.0362],[277,-4.112],[278,-4.1611],[279,-4.1785],[282,-4.1785],[283,-4.1622],[284,-4.1622],[288,-4.0755],[290,-4.0124],[292,-3.9293],[294,-3.8239],[295,-3.7619],[296,-3.6836],[297,-3.5815],[298,-3.4596],[299,-3.3221],[300,-3.173],[304,-2.5403],[305,-2.3932],[306,-2.2586],[307,-2.1405],[308,-2.0428],[310,-1.8777],[311,-1.8007],[312,-1.7278],[313,-1.6593],[314,-1.5954],[315,-1.5364],[316,-1.4825],[317,-1.4339],[318,-1.391],[319,-1.354],[320,-1.3232],[321,-1.2987],[322,-1.281],[323,-1.281],[324,-1.2664]]
       }
       hips_ctrl.Rotation.Y {
        type=MovieSceneScriptingFloatChannel
        default=0
        thinned_from=325
        keys [[0,0],[30,0],[31,0.027],[32,0.0789],[33,0.1582],[34,0.2723],[35,0.3642],[36,0.3967],[37,0.4161],[38,0.4218],[39,0.4174],[40,0.4087],[41,0.3986],[43,0.3768],[45,0.3607],[46,0.3537],[48,0.342],[50,0.3327],[51,0.3291],[52,0.3265],[53,0.3255],[54,0.3264],[57,0.3318],[59,0.3332],[61,0.3321],[70,0.3253],[79,0.3243],[85,0.3254],[86,0.3239],[87,0.3094],[88,0.2651],[89,0.2031],[90,0.1384],[91,0.0751],[92,0.0151],[93,-0.0185],[94,-0.0175],[95,-0.0094],[96,0.0029],[97,0.0136],[98,0.0203],[99,0.0249],[100,0.0281],[101,0.0303],[102,0.0318],[103,0.0327],[104,0.033],[105,0.0328],[107,0.0314],[109,0.0296],[110,0.0283],[111,0.0267],[114,0.0267],[115,0.0168],[116,0.0168],[117,0.0111],[118,0.0087],[119,0.0067],[120,0.0054],[121,0.005],[172,0.0088],[177,0.0083],[181,0.0071],[182,0.0059],[183,0.0035],[185,-0.0018],[186,-0.0037],[187,-0.0043],[188,-0.0038],[189,-0.0013],[190,0.0049],[191,0.0165],[192,0.0427],[193,0.0802],[194,0.1128],[195,0.1373],[196,0.1583],[197,0.172],[198,0.172],[199,0.1855],[200,0.1855],[201,0.1925],[203,0.1966],[204,0.1979],[205,0.1984],[208,0.1977],[210,0.1978],[211,0.1972],[212,0.1974],[213,0.1994],[214,0.2035],[215,0.2091],[220,0.2407],[221,0.2464],[222,0.251],[223,0.2542],[227,0.2623],[233,0.2704],[243,0.2779],[244,0.2781],[245,0.2766],[246,0.2766],[247,0.2616],[248,0.2447],[249,0.2106],[250,0.159],[251,0.1028],[252,0.0551],[253,0.0175],[254,-0.0163],[255,-0.0445],[256,-0.0653],[257,-0.0788],[258,-0.0869],[259,-0.0907],[260,-0.0914],[261,-0.09],[262,-0.0861],[263,-0.0797],[264,-0.0717],[265,-0.0633],[267,-0.0474],[268,-0.0392],[269,-0.0315],[270,-0.0247],[271,-0.0193],[272,-0.015],[273,-0.0113],[274,-0.0082],[275,-0.0058],[276,-0.004],[277,-0.0029],[279,-0.0017],[282,-0.0017],[283,-0.0004],[284,-0.0004],[324,0]]
       }
       hips_ctrl.Rotation.Z {
        type=MovieSceneScriptingFloatChannel
        default=0
        thinned_from=325
        keys [[0,0],[30,0],[31,-0.002],[32,-0.0057],[33,-0.0114],[34,-0.0196],[35,-0.0261],[36,-0.0283],[37,-0.0294],[38,-0.0296],[39,-0.029],[41,-0.027],[51,-0.0151],[53,-0.0132],[55,-0.012],[70,-0.0081],[85,-0.0071],[86,-0.0073],[87,-0.0072],[88,-0.0065],[89,-0.0052],[91,-0.0022],[92,-0.0005],[93,0.0006],[94,0.0006],[95,0.0004],[98,-0.001],[104,-0.0021],[114,-0.002],[115,-0.0012],[116,-0.0012],[117,-0.0008],[182,-0.0003],[187,0.0003],[189,0.0001],[190,-0.0003],[191,-0.0012],[192,-0.003],[193,-0.0058],[194,-0.0082],[195,-0.0101],[196,-0.0116],[197,-0.0126],[198,-0.0126],[199,-0.0136],[200,-0.0136],[201,-0.014],[205,-0.0141],[216,-0.0116],[222,-0.0094],[245,-0.0064],[246,-0.0064],[248,-0.0056],[249,-0.0047],[251,-0.0023],[252,-0.0012],[255,0.001],[258,0.0024],[262,0.0032],[266,0.0027],[274,0.0005],[324,0]]
       }
       hips_ctrl.Scale.X {type=MovieSceneScriptingFloatChannel, default=1, const [1,325]}
       hips_ctrl.Scale.Y {type=MovieSceneScriptingFloatChannel, default=1, const [1,325]}
       hips_ctrl.Scale.Z {type=MovieSceneScriptingFloatChannel, default=1, const [1,325]}
       hips_tan_ctrl.Location.X {
        type=MovieSceneScriptingFloatChannel
        default=0
        thinned_from=325
        keys [[0,0],[31,0],[33,-0.0005],[35,-0.0014],[38,-0.0021],[43,-0.0049],[46,-0.0056],[52,-0.0053],[60,-0.004],[85,-0.004],[88,-0.0043],[90,-0.0038],[92,-0.0026],[96,0.0001],[98,0.001],[106,0.0021],[112,0.0018],[120,0.0004],[182,0.0003],[187,-0.0003],[189,-0.0001],[190,0.0003],[191,0.0011],[192,0.003],[193,0.0057],[194,0.0081],[195,0.01],[196,0.0115],[197,0.0125],[199,0.0134],[203,0.0138],[216,0.0112],[245,0.0113],[246,0.0111],[247,0.0106],[248,0.0099],[249,0.0085],[251,0.0041],[252,0.0022],[254,-0.0006],[255,-0.0017],[256,-0.0025],[258,-0.0034],[261,-0.0036],[267,-0.0024],[274,-0.0005],[324,0]]
       }
       hips_tan_ctrl.Location.Y {
        type=MovieSceneScriptingFloatChannel
        default=0.008
        thinned_from=325
        keys [[0,0.1374],[1,0.1372],[2,0.1367],[3,0.1359],[5,0.1334],[7,0.1302],[10,0.1241],[15,0.1131],[17,0.1094],[21,0.1028],[23,0.1],[24,0.0989],[25,0.098],[26,0.0975],[28,0.0973],[32,0.0979],[35,0.0992],[37,0.1005],[39,0.1023],[41,0.1047],[43,0.1077],[51,0.1226],[53,0.1256],[55,0.1279],[64,0.1332],[77,0.137],[84,0.1374],[85,0.1372],[86,0.1367],[87,0.1358],[89,0.1333],[91,0.13],[94,0.1238],[98,0.1148],[100,0.1108],[105,0.1026],[107,0.0999],[108,0.0988],[109,0.098],[110,0.0975],[111,0.0973],[115,0.0977],[118,0.0987],[120,0.0998],[122,0.1014],[124,0.1034],[126,0.106],[128,0.1093],[134,0.1207],[136,0.1241],[138,0.1268],[140,0.1287],[150,0.134],[163,0.1372],[168,0.1374],[169,0.1371],[170,0.1365],[171,0.1355],[173,0.1328],[175,0.1291],[178,0.1224],[181,0.1153],[183,0.111],[187,0.1045],[189,0.1017],[191,0.0994],[192,0.0985],[193,0.0979],[194,0.0975],[198,0.0977],[201,0.0984],[203,0.0993],[205,0.1006],[207,0.1024],[209,0.1048],[211,0.1078],[213,0.1115],[218,0.122],[220,0.1254],[221,0.1268],[222,0.128],[231,0.1333],[244,0.137],[252,0.1373],[253,0.1371],[254,0.1366],[256,0.1346],[258,0.1317],[260,0.128],[267,0.1127],[269,0.1091],[273,0.1026],[275,0.0999],[276,0.0988],[277,0.098],[278,0.0975],[280,0.0973],[284,0.0979],[287,0.0992],[289,0.1005],[291,0.1023],[293,0.1047],[295,0.1076],[303,0.1219],[305,0.125],[307,0.1276],[313,0.133],[318,0.136],[322,0.1372],[324,0.1373]]
       }
       hips_tan_ctrl.Location.Z {
        type=MovieSceneScriptingFloatChannel
        default=-0.0073
        thinned_from=325
        keys [[0,0.042],[1,0.0428],[2,0.0454],[3,0.0494],[4,0.0548],[5,0.0615],[6,0.0692],[7,0.078],[8,0.0876],[9,0.0979],[10,0.1087],[11,0.1201],[12,0.1319],[17,0.1928],[18,0.2044],[19,0.2152],[20,0.2253],[21,0.2344],[22,0.2424],[23,0.2493],[24,0.2548],[25,0.2589],[26,0.2615],[27,0.2624],[30,0.2617],[32,0.2602],[34,0.2576],[36,0.2536],[38,0.2477],[39,0.2439],[40,0.2396],[41,0.2347],[42,0.2292],[43,0.2231],[44,0.2156],[45,0.2064],[46,0.1958],[47,0.1841],[48,0.1715],[51,0.1321],[52,0.1194],[53,0.1073],[54,0.0962],[55,0.0862],[56,0.0778],[58,0.0628],[59,0.0556],[60,0.0487],[61,0.042],[62,0.0356],[63,0.0296],[64,0.024],[65,0.0189],[66,0.0142],[67,0.0101],[68,0.0065],[69,0.0036],[70,0.0012],[71,-0.0004],[72,-0.0013],[73,-0.0012],[74,-0.0007],[75,0.0006],[76,0.0026],[77,0.0053],[78,0.0087],[79,0.0128],[80,0.0175],[81,0.0229],[82,0.029],[83,0.0357],[84,0.0431],[85,0.0512],[86,0.0598],[87,0.0689],[88,0.0783],[89,0.088],[90,0.098],[92,0.1188],[97,0.1714],[99,0.1917],[101,0.211],[102,0.22],[103,0.2283],[104,0.2358],[105,0.2425],[106,0.2483],[107,0.2532],[108,0.2571],[109,0.26],[110,0.2618],[111,0.2624],[114,0.2617],[116,0.2602],[118,0.2575],[120,0.2533],[122,0.2473],[123,0.2435],[124,0.2392],[125,0.2344],[126,0.2289],[127,0.2227],[128,0.2152],[129,0.206],[130,0.1954],[131,0.1837],[132,0.1711],[135,0.1317],[136,0.1189],[137,0.1069],[138,0.0957],[139,0.0858],[140,0.0773],[142,0.0624],[143,0.0552],[144,0.0482],[145,0.0415],[146,0.0352],[147,0.0292],[148,0.0236],[149,0.0184],[150,0.0138],[151,0.0096],[152,0.0061],[153,0.0031],[154,0.0008],[155,-0.0008],[156,-0.0017],[157,-0.0017],[158,-0.0012],[159,0.0002],[160,0.0023],[161,0.005],[162,0.0085],[163,0.0126],[164,0.0174],[165,0.0229],[166,0.029],[167,0.0358],[168,0.0433],[169,0.0513],[170,0.0598],[171,0.0688],[172,0.0782],[173,0.0879],[174,0.0978],[183,0.1906],[185,0.2101],[186,0.2191],[187,0.2274],[188,0.235],[189,0.2418],[190,0.2478],[191,0.2529],[192,0.2569],[193,0.2599],[194,0.2618],[195,0.2624],[198,0.2618],[200,0.2604],[202,0.2579],[204,0.2538],[205,0.2511],[206,0.2478],[207,0.2441],[208,0.2397],[209,0.2347],[210,0.2291],[211,0.2227],[212,0.2148],[213,0.2047],[214,0.1929],[215,0.1798],[216,0.1658],[218,0.1369],[219,0.1227],[220,0.1093],[221,0.0971],[222,0.0863],[223,0.0773],[224,0.0695],[225,0.0619],[226,0.0545],[227,0.0475],[228,0.0408],[229,0.0344],[230,0.0284],[231,0.0229],[232,0.0178],[233,0.0132],[234,0.0091],[235,0.0056],[236,0.0028],[237,0.0005],[238,-0.001],[239,-0.0019],[240,-0.0019],[241,-0.0014],[242,-0.0003],[243,0.0014],[244,0.0037],[245,0.0066],[246,0.0101],[247,0.0142],[248,0.0188],[249,0.0242],[250,0.0301],[251,0.0367],[252,0.0439],[253,0.0518],[254,0.0602],[255,0.0692],[256,0.0785],[257,0.0883],[258,0.0982],[266,0.1816],[268,0.2015],[269,0.211],[270,0.22],[271,0.2282],[272,0.2357],[273,0.2424],[274,0.2483],[275,0.2532],[276,0.2571],[277,0.26],[278,0.2618],[279,0.2624],[282,0.2617],[284,0.2602],[286,0.2575],[288,0.2533],[290,0.2473],[291,0.2435],[292,0.2392],[293,0.2344],[294,0.2289],[295,0.2227],[296,0.2151],[297,0.2057],[298,0.1947],[299,0.1825],[300,0.1694],[303,0.129],[304,0.1162],[305,0.1043],[306,0.0936],[307,0.0845],[308,0.0773],[309,0.0715],[310,0.0664],[311,0.0619],[312,0.058],[313,0.0546],[314,0.0517],[315,0.0493],[316,0.0473],[317,0.0458],[318,0.0445],[319,0.0436],[320,0.0429],[322,0.0421],[324,0.042]]
       }
       hips_tan_ctrl.Rotation.X {type=MovieSceneScriptingFloatChannel, default=0, const [0,325]}
       hips_tan_ctrl.Rotation.Y {type=MovieSceneScriptingFloatChannel, default=0, const [0,325]}
       hips_tan_ctrl.Rotation.Z {type=MovieSceneScriptingFloatChannel, default=0, const [0,325]}
       hips_tan_ctrl.Scale.X {type=MovieSceneScriptingFloatChannel, default=1, const [1,325]}
       hips_tan_ctrl.Scale.Y {type=MovieSceneScriptingFloatChannel, default=1, const [1,325]}
       hips_tan_ctrl.Scale.Z {type=MovieSceneScriptingFloatChannel, default=1, const [1,325]}
       spine_01_ctrl.Location.X {
        type=MovieSceneScriptingFloatChannel
        default=0
        thinned_from=325
        keys [[0,0],[29,0],[31,0.0004],[32,0.0013],[33,0.0026],[34,0.0044],[35,0.006],[36,0.0065],[37,0.0069],[52,0.0067],[84,0.0075],[86,0.0073],[87,0.0069],[88,0.0058],[91,0.0016],[92,0.0003],[93,-0.0004],[190,0.0001],[191,0.0003],[195,0.0022],[197,0.0028],[214,0.0037],[239,0.0068],[245,0.0068],[247,0.0063],[248,0.0059],[249,0.005],[251,0.0024],[252,0.0013],[254,-0.0004],[256,-0.0014],[258,-0.0018],[263,-0.0016],[324,0]]
       }
       spine_01_ctrl.Location.Y {
        type=MovieSceneScriptingFloatChannel
        default=0
        thinned_from=325
        keys [[0,0.1326],[2,0.1321],[9,0.1271],[13,0.1258],[29,0.1248],[34,0.1256],[38,0.1274],[41,0.1295],[44,0.1324],[46,0.1336],[48,0.134],[51,0.1334],[56,0.1314],[69,0.1293],[73,0.1293],[85,0.1324],[87,0.132],[95,0.1274],[107,0.1249],[115,0.1249],[119,0.1259],[122,0.1274],[125,0.1297],[128,0.1327],[130,0.134],[132,0.1345],[135,0.1338],[140,0.1318],[153,0.1297],[157,0.1297],[169,0.1328],[170,0.1326],[172,0.1316],[177,0.1273],[179,0.1264],[192,0.1248],[201,0.1256],[205,0.127],[208,0.129],[212,0.133],[214,0.1343],[216,0.1347],[219,0.1339],[223,0.1322],[236,0.1301],[240,0.1301],[253,0.1328],[255,0.1325],[262,0.128],[270,0.1255],[281,0.1248],[286,0.1256],[290,0.1274],[293,0.1297],[296,0.1327],[298,0.1339],[300,0.1343],[303,0.1335],[307,0.1319],[324,0.1326]]
       }
       spine_01_ctrl.Location.Z {
        type=MovieSceneScriptingFloatChannel
        default=0
        thinned_from=325
        keys [[0,0.0359],[1,0.0366],[2,0.0386],[3,0.0418],[4,0.0462],[5,0.0515],[6,0.0577],[7,0.0647],[8,0.0723],[9,0.0806],[10,0.0892],[11,0.0983],[12,0.1077],[17,0.1564],[18,0.1657],[19,0.1744],[20,0.1824],[21,0.1897],[22,0.1962],[23,0.2017],[24,0.2061],[25,0.2094],[26,0.2115],[27,0.2122],[30,0.2117],[32,0.2104],[34,0.2084],[36,0.2051],[38,0.2003],[39,0.1973],[40,0.1938],[41,0.1898],[42,0.1854],[43,0.1804],[44,0.1744],[45,0.167],[46,0.1584],[47,0.149],[48,0.1389],[51,0.1075],[52,0.0973],[53,0.0877],[54,0.0788],[55,0.0709],[56,0.0641],[58,0.0522],[60,0.0409],[61,0.0356],[62,0.0305],[63,0.0257],[64,0.0212],[65,0.0171],[66,0.0134],[67,0.0101],[68,0.0073],[69,0.0049],[70,0.0031],[71,0.0018],[72,0.0011],[73,0.0011],[74,0.0016],[75,0.0026],[76,0.0043],[77,0.0065],[78,0.0092],[79,0.0125],[80,0.0163],[81,0.0206],[82,0.0255],[83,0.0309],[84,0.0369],[85,0.0434],[86,0.0503],[87,0.0576],[88,0.0651],[89,0.0729],[91,0.0891],[97,0.1394],[99,0.1556],[101,0.1711],[102,0.1783],[103,0.1849],[104,0.1909],[105,0.1963],[106,0.2009],[107,0.2049],[108,0.208],[109,0.2103],[110,0.2117],[111,0.2122],[114,0.2117],[116,0.2104],[118,0.2082],[120,0.2048],[122,0.2],[123,0.1969],[124,0.1934],[125,0.1895],[126,0.185],[127,0.18],[128,0.174],[129,0.1666],[130,0.158],[131,0.1486],[132,0.1386],[135,0.1071],[136,0.0969],[137,0.0873],[138,0.0784],[139,0.0705],[140,0.0637],[142,0.0518],[144,0.0405],[145,0.0352],[146,0.0301],[147,0.0253],[148,0.0208],[149,0.0167],[150,0.013],[151,0.0097],[152,0.0068],[153,0.0045],[154,0.0026],[155,0.0014],[156,0.0007],[157,0.0007],[158,0.0011],[159,0.0022],[160,0.0039],[161,0.0062],[162,0.0089],[163,0.0123],[164,0.0162],[165,0.0206],[166,0.0255],[167,0.031],[168,0.0369],[169,0.0434],[170,0.0502],[171,0.0574],[172,0.065],[173,0.0727],[175,0.0889],[183,0.1546],[185,0.1703],[186,0.1775],[187,0.1841],[188,0.1902],[189,0.1957],[190,0.2005],[191,0.2046],[192,0.2078],[193,0.2102],[194,0.2117],[195,0.2122],[198,0.2117],[200,0.2106],[202,0.2085],[204,0.2052],[205,0.203],[206,0.2004],[207,0.1973],[208,0.1938],[209,0.1898],[210,0.1852],[211,0.18],[212,0.1736],[213,0.1655],[214,0.156],[215,0.1455],[216,0.1343],[218,0.1112],[219,0.0999],[220,0.0892],[221,0.0794],[222,0.0708],[223,0.0637],[225,0.0514],[226,0.0455],[227,0.0399],[228,0.0345],[229,0.0294],[230,0.0246],[231,0.0202],[232,0.0161],[233,0.0125],[234,0.0092],[235,0.0065],[236,0.0042],[237,0.0024],[238,0.0012],[239,0.0005],[240,0.0005],[241,0.0009],[242,0.0018],[243,0.0032],[244,0.0051],[245,0.0074],[246,0.0102],[247,0.0135],[248,0.0173],[249,0.0216],[250,0.0264],[251,0.0317],[252,0.0375],[253,0.0438],[254,0.0506],[255,0.0577],[256,0.0652],[257,0.073],[259,0.0892],[266,0.1475],[268,0.1634],[269,0.171],[270,0.1782],[271,0.1848],[272,0.1909],[273,0.1962],[274,0.2009],[275,0.2048],[276,0.208],[277,0.2103],[278,0.2117],[279,0.2122],[282,0.2117],[284,0.2104],[286,0.2082],[288,0.2048],[290,0.2],[291,0.1969],[292,0.1934],[293,0.1895],[294,0.185],[295,0.18],[296,0.1739],[297,0.1663],[298,0.1574],[299,0.1476],[300,0.1372],[303,0.1049],[304,0.0947],[305,0.0852],[306,0.0767],[307,0.0695],[308,0.0637],[309,0.0591],[310,0.0551],[311,0.0515],[312,0.0484],[313,0.0457],[314,0.0435],[315,0.0416],[316,0.04],[317,0.0388],[318,0.0378],[319,0.0371],[321,0.0362],[323,0.0359],[324,0.0359]]
       }
       spine_01_ctrl.Rotation.X {
        type=MovieSceneScriptingFloatChannel
        default=0
        thinned_from=325
        keys [[0,2.2454],[2,2.2382],[10,2.1502],[15,2.1193],[27,2.0912],[33,2.1007],[38,2.1316],[41,2.1636],[45,2.2187],[47,2.2333],[50,2.2363],[67,2.1735],[73,2.174],[83,2.2345],[85,2.2493],[87,2.2485],[90,2.2262],[98,2.1468],[109,2.0938],[116,2.0974],[121,2.1231],[125,2.1648],[128,2.2094],[130,2.2302],[132,2.2397],[135,2.2365],[151,2.1767],[157,2.1772],[167,2.2382],[169,2.2524],[170,2.2527],[172,2.2403],[178,2.1622],[182,2.1296],[192,2.0944],[201,2.1034],[205,2.1253],[208,2.1545],[213,2.225],[215,2.2397],[217,2.2428],[233,2.1803],[240,2.1797],[251,2.2395],[253,2.2532],[255,2.2522],[258,2.2288],[266,2.1474],[277,2.0937],[284,2.0973],[289,2.1231],[293,2.1648],[296,2.2091],[298,2.2278],[300,2.2344],[308,2.2112],[324,2.245]]
       }
       spine_01_ctrl.Rotation.Y {
        type=MovieSceneScriptingFloatChannel
        default=0
        thinned_from=325
        keys [[0,0],[28,0],[29,-0.0001],[30,-0.0011],[31,-0.0036],[32,-0.011],[33,-0.0226],[34,-0.0398],[35,-0.0539],[36,-0.0594],[37,-0.0636],[39,-0.0699],[40,-0.0735],[42,-0.0801],[43,-0.0829],[44,-0.085],[45,-0.0869],[46,-0.0885],[48,-0.0913],[51,-0.0947],[53,-0.0966],[64,-0.1049],[67,-0.1066],[70,-0.1076],[73,-0.1078],[76,-0.1069],[79,-0.1052],[82,-0.1026],[85,-0.0992],[86,-0.0977],[87,-0.0938],[88,-0.0828],[89,-0.0675],[91,-0.0351],[92,-0.0191],[93,-0.0082],[94,-0.0044],[95,-0.0018],[96,0.0002],[97,0.0017],[100,0.0051],[103,0.0077],[105,0.0089],[107,0.0097],[109,0.01],[110,0.0098],[111,0.0094],[113,0.0078],[117,0.0037],[119,0.0022],[120,0.0017],[173,0.0002],[181,0.001],[182,0.0009],[184,0.0001],[186,-0.0008],[187,-0.0011],[188,-0.001],[189,-0.0004],[190,0.0015],[191,0.0053],[192,0.0142],[193,0.0274],[194,0.0394],[195,0.0484],[196,0.0558],[197,0.0605],[198,0.063],[199,0.0646],[200,0.0654],[201,0.0656],[202,0.0654],[203,0.0648],[204,0.0638],[205,0.0623],[206,0.0604],[208,0.0558],[209,0.0533],[210,0.0506],[211,0.0476],[218,0.0247],[219,0.0216],[220,0.0189],[221,0.0164],[222,0.0144],[223,0.0128],[228,0.0068],[232,0.0029],[235,0.0007],[238,-0.0009],[239,-0.0012],[240,-0.0012],[241,-0.0017],[244,-0.0018],[247,-0.0013],[251,0.0001],[253,0.0001],[254,-0.0001],[256,-0.0012],[262,-0.0066],[263,-0.0073],[264,-0.0077],[266,-0.008],[267,-0.0077],[268,-0.0071],[271,-0.0045],[275,-0.0018],[277,-0.001],[279,-0.0006],[280,-0.0006],[324,0]]
       }
       spine_01_ctrl.Rotation.Z {
        type=MovieSceneScriptingFloatChannel
        default=0
        thinned_from=325
        keys [[0,0],[29,0],[30,0.0044],[31,0.0153],[32,0.0297],[33,0.0445],[34,0.0564],[35,0.0617],[36,0.0582],[37,0.0385],[38,-0.0102],[39,-0.1012],[40,-0.2209],[41,-0.3416],[42,-0.4637],[43,-0.5697],[44,-0.6427],[45,-0.7006],[46,-0.7417],[47,-0.768],[48,-0.7842],[49,-0.7924],[50,-0.7947],[51,-0.7916],[52,-0.7807],[53,-0.76],[54,-0.7275],[55,-0.6916],[56,-0.6618],[57,-0.638],[58,-0.6198],[59,-0.6072],[60,-0.5998],[61,-0.5974],[85,-0.5983],[86,-0.6127],[87,-0.6393],[88,-0.6538],[89,-0.6431],[90,-0.6094],[91,-0.5511],[92,-0.4604],[93,-0.345],[94,-0.2269],[95,-0.104],[96,0.0213],[97,0.1196],[98,0.1864],[99,0.2388],[100,0.2797],[101,0.3123],[102,0.3397],[103,0.3622],[104,0.379],[105,0.3907],[106,0.398],[107,0.4019],[108,0.403],[109,0.3989],[110,0.3869],[111,0.3674],[112,0.3408],[113,0.3077],[114,0.2695],[116,0.1881],[117,0.1497],[118,0.1157],[119,0.0887],[120,0.0707],[121,0.0642],[181,0.0642],[182,0.0559],[183,0.0349],[185,-0.0195],[186,-0.0404],[187,-0.0488],[188,-0.0447],[189,-0.0158],[190,0.0624],[191,0.2146],[192,0.5681],[193,1.0862],[194,1.5478],[195,1.8951],[196,2.186],[197,2.3734],[198,2.4755],[199,2.5459],[200,2.5904],[201,2.6149],[202,2.6252],[203,2.6274],[204,2.6183],[205,2.5938],[206,2.5582],[209,2.4281],[210,2.3808],[212,2.2678],[213,2.2176],[214,2.1818],[216,2.1344],[218,2.099],[219,2.087],[220,2.0795],[244,2.0743],[245,2.0594],[246,2.0191],[247,1.9405],[248,1.8111],[249,1.5554],[250,1.1713],[251,0.7553],[252,0.4044],[253,0.1285],[254,-0.1188],[255,-0.3252],[256,-0.4783],[257,-0.5795],[258,-0.6437],[259,-0.6792],[260,-0.6944],[261,-0.6976],[262,-0.6858],[263,-0.6548],[264,-0.6108],[266,-0.5099],[267,-0.4549],[268,-0.3911],[269,-0.3252],[270,-0.2637],[271,-0.2132],[272,-0.1719],[273,-0.1342],[274,-0.101],[275,-0.0735],[276,-0.0526],[277,-0.0394],[278,-0.031],[279,-0.0238],[280,-0.0238],[281,-0.0131],[282,-0.0092],[283,-0.0061],[284,-0.0039],[285,-0.0022],[286,-0.0011],[287,-0.0005],[288,-0.0001],[324,0]]
       }
       spine_01_ctrl.Scale.X {type=MovieSceneScriptingFloatChannel, default=1, const [1,325]}
       spine_01_ctrl.Scale.Y {type=MovieSceneScriptingFloatChannel, default=1, const [1,325]}
       spine_01_ctrl.Scale.Z {type=MovieSceneScriptingFloatChannel, default=1, const [1,325]}
       spine_02_ctrl.Location.X {
        type=MovieSceneScriptingFloatChannel
        default=0
        thinned_from=325
        keys [[0,0],[29,0],[31,-0.0004],[32,-0.0012],[33,-0.0023],[34,-0.0039],[35,-0.0052],[36,-0.0057],[37,-0.0059],[38,-0.0058],[43,-0.0036],[48,-0.0027],[54,-0.0027],[86,-0.0031],[87,-0.0028],[91,0.0002],[92,0.0009],[93,0.0011],[98,-0.0008],[109,-0.0016],[121,-0.0003],[183,-0.0001],[188,0.0002],[190,-0.0002],[191,-0.0008],[192,-0.0022],[193,-0.0042],[194,-0.006],[195,-0.0074],[196,-0.0085],[197,-0.0092],[199,-0.0099],[204,-0.0102],[218,-0.0084],[245,-0.0083],[247,-0.0078],[248,-0.0073],[249,-0.0063],[251,-0.003],[252,-0.0016],[254,0.0005],[255,0.0013],[256,0.0019],[258,0.0026],[262,0.0028],[275,0.0003],[324,0]]
       }
       spine_02_ctrl.Location.Y {
        type=MovieSceneScriptingFloatChannel
        default=0
        thinned_from=325
        keys [[0,0.1227],[1,0.1225],[2,0.122],[3,0.121],[4,0.1198],[5,0.1183],[6,0.1165],[8,0.1121],[10,0.1069],[14,0.0952],[16,0.0899],[20,0.0811],[22,0.0771],[23,0.0754],[24,0.074],[25,0.0729],[26,0.0722],[27,0.072],[31,0.0724],[34,0.0734],[36,0.0745],[38,0.0762],[40,0.0784],[42,0.0811],[43,0.0827],[44,0.0845],[45,0.0867],[47,0.0915],[51,0.1018],[53,0.1063],[54,0.1082],[55,0.1098],[60,0.1149],[66,0.1195],[71,0.122],[77,0.1233],[83,0.123],[85,0.1224],[86,0.1216],[87,0.1206],[88,0.1192],[89,0.1175],[91,0.1135],[93,0.1086],[98,0.0945],[99,0.0918],[100,0.0893],[105,0.0788],[106,0.077],[107,0.0753],[108,0.0739],[109,0.0729],[110,0.0722],[111,0.072],[115,0.0724],[118,0.0734],[120,0.0746],[122,0.0762],[124,0.0783],[126,0.081],[127,0.0825],[128,0.0844],[129,0.0864],[131,0.0913],[135,0.1016],[137,0.1061],[138,0.1081],[139,0.1097],[143,0.1139],[149,0.1187],[154,0.1215],[160,0.1231],[166,0.1231],[168,0.1227],[169,0.1222],[170,0.1214],[171,0.1202],[172,0.1187],[173,0.1169],[174,0.1149],[176,0.1101],[177,0.1074],[181,0.0952],[182,0.0923],[183,0.0897],[184,0.0874],[187,0.0813],[189,0.0777],[190,0.0761],[191,0.0747],[192,0.0736],[193,0.0727],[194,0.0722],[195,0.072],[199,0.0724],[202,0.0733],[204,0.0745],[206,0.0761],[208,0.0782],[210,0.0809],[211,0.0825],[212,0.0845],[213,0.0867],[215,0.092],[218,0.1006],[219,0.1032],[220,0.1057],[221,0.1078],[222,0.1096],[223,0.1109],[229,0.1164],[235,0.1204],[241,0.1226],[248,0.1232],[252,0.1226],[253,0.1222],[254,0.1215],[255,0.1204],[256,0.119],[257,0.1174],[258,0.1155],[260,0.111],[262,0.1058],[266,0.0944],[268,0.0893],[273,0.0788],[274,0.077],[275,0.0753],[276,0.0739],[277,0.0729],[278,0.0722],[279,0.072],[283,0.0724],[286,0.0734],[288,0.0746],[290,0.0762],[292,0.0783],[294,0.081],[295,0.0825],[296,0.0843],[298,0.0886],[304,0.1033],[306,0.1076],[307,0.1094],[310,0.1135],[314,0.1179],[317,0.1203],[320,0.1219],[323,0.1227],[324,0.1227]]
       }
       spine_02_ctrl.Location.Z {
        type=MovieSceneScriptingFloatChannel
        default=0
        thinned_from=325
        keys [[0,0.022],[1,0.0226],[2,0.0243],[3,0.0269],[4,0.0305],[5,0.0349],[6,0.0401],[7,0.0459],[8,0.0522],[9,0.0589],[10,0.066],[11,0.0734],[13,0.089],[17,0.1209],[18,0.1285],[19,0.1356],[20,0.1422],[21,0.1481],[22,0.1534],[23,0.1578],[24,0.1615],[25,0.1642],[26,0.1658],[27,0.1664],[30,0.166],[32,0.1649],[34,0.1632],[35,0.162],[37,0.1587],[38,0.1565],[39,0.154],[40,0.1511],[41,0.1478],[42,0.1441],[43,0.14],[44,0.135],[45,0.1289],[46,0.1219],[47,0.1142],[48,0.106],[51,0.0803],[52,0.072],[53,0.0642],[54,0.057],[55,0.0505],[56,0.045],[58,0.0353],[60,0.026],[61,0.0217],[62,0.0175],[63,0.0136],[64,0.01],[65,0.0066],[66,0.0036],[67,0.0009],[68,-0.0014],[69,-0.0033],[70,-0.0048],[71,-0.0059],[72,-0.0065],[73,-0.0065],[74,-0.0061],[75,-0.0052],[76,-0.0039],[77,-0.0021],[78,0.0002],[79,0.0028],[80,0.006],[81,0.0095],[82,0.0135],[83,0.018],[84,0.0229],[85,0.0282],[86,0.0339],[87,0.0399],[88,0.0461],[90,0.0591],[97,0.1071],[99,0.1203],[101,0.133],[102,0.1388],[103,0.1442],[104,0.1491],[105,0.1535],[106,0.1573],[107,0.1605],[108,0.163],[109,0.1649],[110,0.166],[111,0.1664],[114,0.166],[116,0.1649],[118,0.1631],[120,0.1603],[121,0.1584],[122,0.1562],[123,0.1537],[124,0.1508],[125,0.1475],[126,0.1438],[127,0.1396],[128,0.1346],[129,0.1285],[130,0.1215],[131,0.1138],[132,0.1056],[135,0.0799],[136,0.0717],[137,0.0638],[138,0.0566],[139,0.0502],[140,0.0447],[142,0.0349],[144,0.0257],[145,0.0213],[146,0.0172],[147,0.0133],[148,0.0096],[149,0.0062],[150,0.0032],[151,0.0005],[152,-0.0018],[153,-0.0037],[154,-0.0052],[155,-0.0063],[156,-0.0068],[157,-0.0069],[158,-0.0065],[159,-0.0056],[160,-0.0042],[161,-0.0023],[162,-0.0001],[163,0.0027],[164,0.0058],[165,0.0094],[166,0.0135],[167,0.018],[168,0.0228],[169,0.0281],[170,0.0338],[171,0.0397],[172,0.046],[174,0.059],[182,0.1129],[184,0.126],[185,0.1323],[186,0.1381],[187,0.1436],[188,0.1486],[189,0.153],[190,0.1569],[191,0.1602],[192,0.1629],[193,0.1648],[194,0.166],[195,0.1664],[198,0.166],[200,0.1651],[202,0.1633],[204,0.1606],[205,0.1587],[206,0.1566],[207,0.154],[208,0.1511],[209,0.1477],[210,0.1439],[211,0.1396],[212,0.1343],[213,0.1276],[214,0.1198],[215,0.1112],[216,0.1021],[218,0.0833],[219,0.0741],[220,0.0654],[221,0.0574],[222,0.0504],[223,0.0446],[225,0.0345],[227,0.0251],[228,0.0208],[229,0.0166],[230,0.0127],[231,0.0091],[232,0.0058],[233,0.0028],[234,0.0001],[235,-0.0021],[236,-0.004],[237,-0.0055],[238,-0.0065],[239,-0.007],[240,-0.0071],[241,-0.0067],[242,-0.0059],[243,-0.0048],[244,-0.0033],[245,-0.0014],[246,0.0009],[247,0.0036],[248,0.0067],[249,0.0102],[250,0.0142],[251,0.0185],[252,0.0233],[253,0.0285],[254,0.0341],[255,0.04],[256,0.0462],[258,0.0592],[265,0.107],[267,0.1203],[269,0.1329],[270,0.1388],[271,0.1442],[272,0.1491],[273,0.1535],[274,0.1572],[275,0.1604],[276,0.163],[277,0.1649],[278,0.166],[279,0.1664],[282,0.166],[284,0.1649],[286,0.1631],[288,0.1603],[289,0.1584],[290,0.1562],[291,0.1537],[292,0.1508],[293,0.1475],[294,0.1438],[295,0.1396],[296,0.1346],[297,0.1283],[298,0.121],[299,0.113],[300,0.1045],[303,0.0781],[304,0.0698],[305,0.0621],[306,0.0552],[307,0.0493],[308,0.0447],[309,0.0409],[310,0.0376],[311,0.0347],[312,0.0322],[313,0.03],[314,0.0282],[315,0.0266],[316,0.0254],[317,0.0244],[319,0.023],[321,0.0223],[324,0.022]]
       }
       spine_02_ctrl.Rotation.X {
        type=MovieSceneScriptingFloatChannel
        default=0
        thinned_from=325
        keys [[0,2.4488],[1,2.4455],[2,2.4357],[3,2.4199],[4,2.3983],[5,2.3713],[6,2.3393],[8,2.2616],[10,2.1659],[14,1.9518],[15,1.902],[16,1.8565],[21,1.6661],[22,1.6315],[23,1.6005],[24,1.5743],[25,1.5541],[26,1.5412],[27,1.5366],[30,1.5405],[33,1.5545],[35,1.5709],[37,1.5945],[39,1.6268],[41,1.6686],[43,1.7202],[44,1.7514],[45,1.788],[47,1.8724],[51,2.0564],[53,2.138],[54,2.1723],[55,2.2006],[60,2.286],[68,2.3814],[77,2.4388],[84,2.4509],[85,2.4474],[86,2.4376],[87,2.4218],[88,2.4],[89,2.3726],[90,2.3401],[92,2.2614],[94,2.1657],[98,1.9532],[99,1.9036],[100,1.8583],[105,1.6671],[106,1.6322],[107,1.601],[108,1.5746],[109,1.5543],[110,1.5412],[111,1.5365],[114,1.5405],[117,1.5546],[119,1.5712],[121,1.5948],[123,1.6262],[125,1.6666],[127,1.7169],[128,1.7478],[129,1.784],[131,1.8682],[135,2.0523],[137,2.1343],[138,2.1687],[139,2.1973],[144,2.2833],[152,2.379],[161,2.4366],[168,2.4487],[169,2.4447],[170,2.4334],[171,2.4154],[172,2.3911],[173,2.361],[174,2.3256],[176,2.2407],[177,2.1908],[181,1.9614],[182,1.9072],[183,1.8585],[184,1.8177],[188,1.6744],[190,1.613],[191,1.5875],[192,1.5662],[193,1.5499],[194,1.5393],[195,1.5354],[198,1.5381],[201,1.5509],[203,1.5668],[205,1.59],[207,1.6215],[209,1.6626],[211,1.7145],[212,1.7473],[213,1.7867],[214,1.8314],[219,2.0798],[220,2.1237],[221,2.162],[222,2.1932],[223,2.2155],[230,2.3198],[238,2.3979],[250,2.4465],[252,2.4481],[253,2.4447],[254,2.4348],[255,2.4187],[256,2.3969],[257,2.3697],[258,2.3375],[260,2.259],[262,2.1633],[266,1.9521],[267,1.903],[268,1.8581],[273,1.6673],[274,1.6324],[275,1.6012],[276,1.5747],[277,1.5543],[278,1.5412],[279,1.5366],[282,1.5405],[285,1.5546],[287,1.5712],[289,1.5948],[291,1.6262],[293,1.6666],[295,1.7169],[296,1.7472],[297,1.782],[299,1.8614],[304,2.0798],[306,2.1574],[307,2.1905],[311,2.2906],[315,2.369],[318,2.412],[321,2.4393],[323,2.4477],[324,2.4477]]
       }
       spine_02_ctrl.Rotation.Y {
        type=MovieSceneScriptingFloatChannel
        default=0
        thinned_from=325
        keys [[0,0],[28,0],[29,0.0003],[30,0.0027],[31,0.0092],[32,0.0265],[33,0.0528],[34,0.0904],[35,0.1206],[36,0.1313],[37,0.1373],[38,0.1383],[39,0.1352],[40,0.1305],[42,0.1201],[43,0.1155],[44,0.1125],[45,0.1101],[46,0.1085],[47,0.1075],[48,0.1069],[49,0.1067],[50,0.1067],[51,0.1069],[52,0.1075],[53,0.1084],[55,0.1115],[56,0.1128],[57,0.1139],[59,0.1154],[61,0.116],[83,0.1167],[85,0.1164],[86,0.1151],[87,0.1097],[88,0.094],[89,0.0722],[90,0.0494],[91,0.027],[92,0.0059],[93,-0.0061],[94,-0.0061],[95,-0.0034],[96,0.0011],[97,0.0053],[98,0.0082],[99,0.0105],[100,0.0124],[101,0.0138],[102,0.0151],[103,0.0161],[104,0.0169],[106,0.0178],[108,0.0181],[109,0.018],[110,0.0174],[111,0.0166],[112,0.0154],[113,0.0139],[117,0.0067],[118,0.0052],[119,0.004],[120,0.0032],[121,0.0029],[181,0.0028],[182,0.0025],[183,0.0015],[185,-0.0009],[186,-0.0018],[187,-0.0022],[188,-0.002],[189,-0.0007],[190,0.0028],[191,0.0097],[192,0.0256],[193,0.049],[194,0.0698],[195,0.0855],[196,0.0986],[197,0.1071],[198,0.1117],[199,0.1148],[200,0.1168],[201,0.1178],[202,0.1181],[203,0.1181],[204,0.1175],[205,0.1162],[206,0.1143],[209,0.1073],[210,0.1046],[212,0.0985],[213,0.0957],[214,0.0937],[216,0.0907],[218,0.0884],[219,0.0875],[220,0.0868],[244,0.0818],[245,0.081],[246,0.0793],[247,0.0761],[248,0.0709],[249,0.0608],[250,0.0457],[251,0.0295],[252,0.0158],[253,0.005],[254,-0.0047],[255,-0.0129],[256,-0.0192],[257,-0.0236],[258,-0.0266],[259,-0.0285],[260,-0.0295],[261,-0.0299],[262,-0.0297],[263,-0.0285],[264,-0.0268],[266,-0.0225],[267,-0.0201],[270,-0.0117],[271,-0.0095],[273,-0.006],[274,-0.0045],[275,-0.0033],[276,-0.0024],[277,-0.0018],[279,-0.0011],[280,-0.0011],[281,-0.0006],[324,0]]
       }
       spine_02_ctrl.Rotation.Z {
        type=MovieSceneScriptingFloatChannel
        default=0
        thinned_from=325
        keys [[0,0],[29,0],[30,0.0033],[31,0.0114],[32,0.0211],[33,0.0298],[34,0.0346],[35,0.0349],[36,0.0305],[37,0.0133],[38,-0.0272],[39,-0.1024],[40,-0.2013],[41,-0.301],[42,-0.4017],[43,-0.4892],[44,-0.5492],[45,-0.5967],[46,-0.6303],[47,-0.6517],[48,-0.6646],[49,-0.6709],[50,-0.6723],[51,-0.6693],[52,-0.6598],[53,-0.6422],[54,-0.615],[55,-0.585],[56,-0.5602],[57,-0.5403],[58,-0.5252],[59,-0.5146],[60,-0.5083],[61,-0.5062],[85,-0.5054],[86,-0.5173],[87,-0.5391],[88,-0.5501],[89,-0.5396],[90,-0.51],[91,-0.4599],[92,-0.3828],[93,-0.2859],[94,-0.1879],[95,-0.0861],[96,0.0176],[97,0.0988],[98,0.154],[99,0.1973],[100,0.2311],[101,0.258],[102,0.2806],[103,0.2993],[104,0.3131],[105,0.3228],[106,0.3289],[107,0.3321],[108,0.333],[109,0.3296],[110,0.3197],[111,0.3036],[112,0.2816],[113,0.2543],[114,0.2227],[116,0.1555],[117,0.1237],[118,0.0956],[119,0.0733],[120,0.0584],[121,0.0531],[181,0.0531],[182,0.0462],[183,0.0289],[185,-0.0161],[186,-0.0334],[187,-0.0403],[188,-0.0369],[189,-0.0131],[190,0.0515],[191,0.1774],[192,0.4694],[193,0.8975],[194,1.279],[195,1.566],[196,1.8063],[197,1.9612],[198,2.0456],[199,2.1037],[200,2.1404],[201,2.1606],[202,2.1692],[203,2.171],[204,2.1634],[205,2.1431],[206,2.1136],[209,2.006],[210,1.9669],[212,1.8734],[213,1.8319],[214,1.8023],[216,1.7632],[218,1.7342],[219,1.7245],[220,1.7184],[244,1.7159],[245,1.7036],[246,1.6703],[247,1.6053],[248,1.4983],[249,1.2868],[250,0.969],[251,0.6249],[252,0.3346],[253,0.1063],[254,-0.0983],[255,-0.269],[256,-0.3957],[257,-0.4793],[258,-0.5323],[259,-0.5616],[260,-0.5741],[261,-0.5766],[262,-0.5669],[263,-0.5411],[264,-0.5048],[266,-0.4213],[267,-0.3759],[268,-0.3232],[269,-0.2687],[270,-0.2179],[271,-0.1762],[272,-0.142],[273,-0.1109],[274,-0.0835],[275,-0.0607],[276,-0.0435],[277,-0.0325],[278,-0.0256],[279,-0.0197],[280,-0.0197],[281,-0.0108],[282,-0.0076],[283,-0.0051],[284,-0.0032],[285,-0.0019],[286,-0.0009],[287,-0.0004],[289,0],[324,0]]
       }
       spine_02_ctrl.Scale.X {type=MovieSceneScriptingFloatChannel, default=1, const [1,325]}
       spine_02_ctrl.Scale.Y {type=MovieSceneScriptingFloatChannel, default=1, const [1,325]}
       spine_02_ctrl.Scale.Z {type=MovieSceneScriptingFloatChannel, default=1, const [1,325]}
       spine_03_ctrl.Location.X {
        type=MovieSceneScriptingFloatChannel
        default=0
        thinned_from=325
        keys [[0,0],[29,-0.0001],[30,-0.0006],[31,-0.002],[32,-0.0059],[33,-0.0118],[34,-0.0203],[35,-0.0272],[36,-0.0296],[37,-0.0311],[38,-0.0316],[39,-0.0314],[43,-0.0293],[46,-0.0285],[52,-0.0284],[61,-0.0296],[85,-0.0295],[86,-0.0292],[87,-0.0281],[88,-0.0245],[89,-0.0194],[91,-0.0085],[92,-0.0033],[93,-0.0001],[94,0.0004],[95,0.0003],[98,-0.001],[106,-0.002],[113,-0.0015],[120,-0.0003],[182,-0.0003],[187,0.0002],[189,0.0001],[190,-0.0003],[191,-0.001],[192,-0.0028],[193,-0.0053],[194,-0.0075],[195,-0.0091],[196,-0.0105],[197,-0.0114],[199,-0.0123],[203,-0.0128],[214,-0.0112],[244,-0.0116],[246,-0.0113],[247,-0.0108],[248,-0.0101],[249,-0.0087],[251,-0.0042],[252,-0.0022],[254,0.0007],[255,0.0018],[256,0.0027],[257,0.0033],[259,0.0039],[262,0.0039],[273,0.0007],[324,0]]
       }
       spine_03_ctrl.Location.Y {
        type=MovieSceneScriptingFloatChannel
        default=0
        thinned_from=325
        keys [[0,0.2854],[1,0.2849],[2,0.2834],[3,0.281],[4,0.2777],[5,0.2735],[6,0.2684],[7,0.2623],[8,0.2554],[9,0.2475],[10,0.2386],[14,0.199],[15,0.1895],[16,0.1807],[17,0.1729],[20,0.1505],[21,0.1433],[22,0.1365],[23,0.1305],[24,0.1254],[25,0.1215],[26,0.1189],[27,0.118],[28,0.118],[30,0.1187],[32,0.12],[34,0.1221],[35,0.1236],[36,0.1253],[37,0.1273],[38,0.1297],[39,0.1325],[40,0.1357],[41,0.1392],[42,0.1431],[43,0.1474],[44,0.1526],[45,0.1591],[46,0.1666],[47,0.1749],[51,0.2113],[52,0.22],[53,0.228],[54,0.235],[55,0.2409],[56,0.2454],[60,0.2589],[64,0.27],[68,0.2786],[72,0.2843],[76,0.2874],[80,0.2878],[84,0.2859],[85,0.2848],[86,0.2828],[87,0.2799],[88,0.2761],[89,0.2715],[90,0.2659],[91,0.2596],[92,0.2524],[93,0.2444],[94,0.2355],[98,0.1967],[99,0.1875],[100,0.179],[101,0.1715],[105,0.1427],[106,0.1361],[107,0.1302],[108,0.1252],[109,0.1214],[110,0.1189],[111,0.118],[113,0.1183],[115,0.1192],[117,0.121],[118,0.1222],[119,0.1236],[120,0.1254],[121,0.1274],[122,0.1297],[123,0.1324],[124,0.1354],[125,0.1387],[126,0.1425],[127,0.1467],[128,0.1518],[129,0.1581],[130,0.1656],[131,0.1739],[135,0.2103],[136,0.219],[137,0.227],[138,0.2341],[139,0.2401],[140,0.2446],[144,0.2582],[148,0.2693],[152,0.2779],[156,0.2836],[160,0.2867],[164,0.2871],[168,0.2852],[169,0.284],[170,0.2818],[171,0.2787],[172,0.2747],[173,0.2697],[174,0.2639],[175,0.2573],[176,0.2498],[177,0.2412],[178,0.2314],[181,0.1995],[182,0.1893],[183,0.1801],[184,0.1723],[187,0.1513],[188,0.1447],[189,0.1385],[190,0.1328],[191,0.1279],[192,0.1238],[193,0.1206],[194,0.1186],[195,0.1178],[197,0.1179],[199,0.1187],[201,0.1203],[202,0.1214],[203,0.1228],[204,0.1245],[205,0.1265],[206,0.1289],[207,0.1315],[208,0.1346],[209,0.138],[210,0.1419],[211,0.1462],[212,0.1516],[213,0.1586],[214,0.1669],[215,0.1762],[218,0.2062],[219,0.2159],[220,0.2249],[221,0.2328],[222,0.2392],[223,0.2439],[227,0.2576],[231,0.2686],[235,0.2769],[239,0.2826],[243,0.2857],[249,0.2863],[252,0.2851],[253,0.2839],[254,0.2819],[255,0.279],[256,0.2753],[257,0.2707],[258,0.2652],[259,0.259],[260,0.2518],[261,0.2438],[262,0.235],[266,0.1965],[267,0.1874],[268,0.1789],[269,0.1715],[273,0.1427],[274,0.1361],[275,0.1302],[276,0.1252],[277,0.1214],[278,0.1189],[279,0.118],[280,0.118],[282,0.1187],[284,0.12],[286,0.1222],[287,0.1236],[288,0.1254],[289,0.1274],[290,0.1297],[291,0.1324],[292,0.1354],[293,0.1387],[294,0.1425],[295,0.1467],[296,0.1517],[297,0.1579],[298,0.165],[299,0.1729],[301,0.1902],[304,0.2168],[305,0.225],[306,0.2325],[307,0.2391],[308,0.2446],[310,0.2537],[312,0.2618],[314,0.2688],[316,0.2746],[318,0.2793],[320,0.2826],[322,0.2847],[324,0.2854]]
       }
       spine_03_ctrl.Location.Z {
        type=MovieSceneScriptingFloatChannel
        default=0
        thinned_from=325
        keys [[0,0.0392],[1,0.04],[2,0.0422],[3,0.0458],[4,0.0507],[5,0.0566],[6,0.0635],[7,0.0713],[8,0.0798],[9,0.0887],[10,0.0983],[11,0.1082],[12,0.1184],[17,0.1715],[18,0.1817],[19,0.1912],[20,0.1999],[21,0.2078],[22,0.2147],[23,0.2207],[24,0.2254],[25,0.229],[26,0.2312],[27,0.2319],[30,0.2313],[32,0.23],[34,0.2277],[36,0.2241],[38,0.2188],[39,0.2155],[40,0.2117],[41,0.2074],[42,0.2025],[43,0.197],[44,0.1904],[45,0.1823],[46,0.173],[47,0.1627],[48,0.1517],[51,0.1174],[52,0.1063],[53,0.0958],[54,0.0862],[55,0.0775],[56,0.0701],[58,0.0569],[59,0.0506],[60,0.0444],[61,0.0385],[62,0.0329],[63,0.0276],[64,0.0227],[65,0.0182],[66,0.0141],[67,0.0104],[68,0.0073],[69,0.0047],[70,0.0026],[71,0.0012],[72,0.0005],[73,0.0004],[74,0.001],[75,0.0022],[76,0.004],[77,0.0064],[78,0.0095],[79,0.0131],[80,0.0173],[81,0.0222],[82,0.0276],[83,0.0336],[84,0.0402],[85,0.0474],[86,0.0551],[87,0.0632],[88,0.0717],[90,0.0892],[96,0.1444],[99,0.1711],[101,0.1879],[102,0.1957],[103,0.2028],[104,0.2093],[105,0.215],[106,0.22],[107,0.2242],[108,0.2275],[109,0.2299],[110,0.2314],[111,0.2319],[114,0.2313],[116,0.23],[118,0.2276],[120,0.2239],[122,0.2185],[123,0.2152],[124,0.2114],[125,0.207],[126,0.2021],[127,0.1966],[128,0.19],[129,0.1819],[130,0.1726],[131,0.1623],[132,0.1513],[135,0.117],[136,0.1059],[137,0.0955],[138,0.0858],[139,0.0771],[140,0.0697],[142,0.0565],[143,0.0502],[144,0.044],[145,0.0381],[146,0.0325],[147,0.0272],[148,0.0223],[149,0.0178],[150,0.0136],[151,0.01],[152,0.0068],[153,0.0042],[154,0.0022],[155,0.0008],[156,0],[157,0],[158,0.0006],[159,0.0018],[160,0.0037],[161,0.0062],[162,0.0093],[163,0.013],[164,0.0173],[165,0.0222],[166,0.0276],[167,0.0337],[168,0.0403],[169,0.0475],[170,0.0551],[171,0.0632],[172,0.0715],[174,0.089],[181,0.1522],[184,0.1784],[185,0.1868],[186,0.1946],[187,0.2018],[188,0.2084],[189,0.2143],[190,0.2195],[191,0.2238],[192,0.2273],[193,0.2298],[194,0.2314],[195,0.2319],[198,0.2314],[200,0.2302],[202,0.2279],[204,0.2243],[205,0.2219],[206,0.219],[207,0.2156],[208,0.2118],[209,0.2073],[210,0.2023],[211,0.1966],[212,0.1895],[213,0.1807],[214,0.1703],[215,0.1589],[216,0.1467],[218,0.1215],[219,0.1092],[220,0.0975],[221,0.0869],[222,0.0774],[223,0.0696],[225,0.056],[226,0.0495],[227,0.0433],[228,0.0374],[229,0.0318],[230,0.0265],[231,0.0216],[232,0.0171],[233,0.0131],[234,0.0095],[235,0.0064],[236,0.0039],[237,0.0019],[238,0.0006],[239,-0.0002],[240,-0.0002],[241,0.0003],[242,0.0013],[243,0.0029],[244,0.0049],[245,0.0075],[246,0.0106],[247,0.0143],[248,0.0185],[249,0.0232],[250,0.0286],[251,0.0345],[252,0.0409],[253,0.0479],[254,0.0555],[255,0.0635],[256,0.0718],[258,0.0893],[265,0.1533],[267,0.171],[269,0.1878],[270,0.1956],[271,0.2028],[272,0.2093],[273,0.215],[274,0.22],[275,0.2242],[276,0.2275],[277,0.2299],[278,0.2314],[279,0.2319],[282,0.2313],[284,0.23],[286,0.2276],[288,0.2239],[290,0.2185],[291,0.2152],[292,0.2114],[293,0.207],[294,0.2021],[295,0.1966],[296,0.1899],[297,0.1816],[298,0.1719],[299,0.1612],[300,0.1498],[303,0.1145],[304,0.1034],[305,0.0931],[306,0.0839],[307,0.0759],[308,0.0697],[309,0.0646],[310,0.0601],[311,0.0562],[312,0.0528],[313,0.0499],[314,0.0474],[315,0.0454],[316,0.0437],[317,0.0423],[318,0.0413],[319,0.0405],[321,0.0395],[324,0.0392]]
       }
       spine_03_ctrl.Rotation.X {
        type=MovieSceneScriptingFloatChannel
        default=0
        thinned_from=325
        keys [[0,4.2433],[1,4.2367],[2,4.2168],[3,4.1836],[4,4.1372],[5,4.0776],[6,4.0048],[7,3.9189],[8,3.8199],[9,3.7055],[10,3.5762],[14,2.9987],[15,2.8612],[16,2.7352],[17,2.6247],[21,2.2011],[22,2.1037],[23,2.0165],[24,1.9428],[25,1.886],[26,1.8494],[27,1.8365],[28,1.8365],[30,1.8456],[32,1.8639],[34,1.8936],[35,1.9135],[36,1.9373],[37,1.9653],[38,1.9979],[39,2.0355],[40,2.0782],[41,2.1258],[42,2.1786],[43,2.2368],[44,2.3081],[45,2.3984],[46,2.5043],[47,2.6224],[51,3.1468],[52,3.2727],[53,3.389],[54,3.4918],[55,3.5772],[56,3.6413],[60,3.8328],[64,3.9897],[68,4.1108],[72,4.1958],[77,4.2518],[82,4.258],[84,4.248],[85,4.2352],[86,4.2103],[87,4.173],[88,4.1226],[89,4.0592],[90,3.983],[91,3.8943],[92,3.7932],[93,3.678],[94,3.5496],[98,2.9824],[99,2.8478],[100,2.7244],[101,2.6161],[105,2.1984],[106,2.1019],[107,2.0154],[108,1.9422],[109,1.8857],[110,1.8493],[111,1.8364],[113,1.8404],[115,1.8534],[117,1.8775],[119,1.9144],[120,1.9383],[121,1.9661],[122,1.9981],[123,2.0346],[124,2.0758],[125,2.1219],[126,2.1733],[127,2.2302],[128,2.3006],[129,2.3903],[130,2.4958],[131,2.6137],[133,2.8724],[135,3.1381],[136,3.2642],[137,3.3808],[138,3.484],[139,3.5698],[140,3.6342],[144,3.8266],[148,3.9836],[152,4.1049],[156,4.19],[161,4.2461],[166,4.2521],[168,4.2421],[169,4.2284],[170,4.2009],[171,4.1597],[172,4.105],[173,4.037],[174,3.9557],[175,3.8615],[176,3.7546],[177,3.63],[178,3.4872],[181,3.0166],[182,2.8677],[183,2.7337],[184,2.6211],[187,2.3195],[188,2.2236],[189,2.1336],[190,2.0517],[191,1.9799],[192,1.9203],[193,1.8751],[194,1.8464],[195,1.8364],[197,1.8398],[199,1.8517],[201,1.8742],[203,1.9097],[204,1.933],[205,1.9604],[206,1.9923],[207,2.0288],[208,2.0704],[209,2.1174],[210,2.1699],[211,2.2285],[212,2.3034],[213,2.4019],[214,2.5191],[215,2.6506],[219,3.2231],[220,3.353],[221,3.4674],[222,3.561],[223,3.6283],[227,3.8199],[231,3.9752],[235,4.0943],[239,4.1776],[244,4.2341],[250,4.2478],[252,4.24],[253,4.2277],[254,4.2026],[255,4.1649],[256,4.1145],[257,4.0514],[258,3.9755],[259,3.8868],[260,3.7852],[261,3.6693],[262,3.5408],[266,2.9788],[267,2.8459],[268,2.7237],[269,2.6161],[273,2.199],[274,2.1024],[275,2.0158],[276,1.9425],[277,1.8859],[278,1.8494],[279,1.8365],[280,1.8365],[282,1.8456],[284,1.8639],[286,1.8942],[288,1.9383],[289,1.9661],[290,1.9981],[291,2.0346],[292,2.0758],[293,2.1219],[294,2.1733],[295,2.2302],[296,2.2993],[297,2.3856],[298,2.4861],[299,2.5981],[300,2.7187],[304,3.2291],[305,3.3484],[306,3.4578],[307,3.5542],[308,3.6342],[311,3.8308],[313,3.9432],[315,4.0392],[317,4.118],[319,4.1784],[321,4.2196],[323,4.2406],[324,4.2433]]
       }
       spine_03_ctrl.Rotation.Y {
        type=MovieSceneScriptingFloatChannel
        default=0
        thinned_from=325
        keys [[0,0],[28,0],[29,0.0011],[30,0.0094],[31,0.0318],[32,0.0929],[33,0.1862],[34,0.3205],[35,0.4288],[36,0.4674],[37,0.4908],[38,0.4981],[39,0.4939],[40,0.4854],[42,0.4656],[43,0.4571],[44,0.452],[45,0.4487],[46,0.4473],[47,0.4475],[49,0.4511],[51,0.457],[53,0.465],[55,0.4745],[57,0.4816],[60,0.4884],[71,0.4986],[80,0.4974],[85,0.4925],[86,0.4877],[87,0.469],[88,0.4105],[89,0.3267],[91,0.1484],[92,0.0615],[93,0.0067],[94,-0.0037],[95,-0.0041],[96,0.0022],[97,0.0097],[98,0.0153],[99,0.0198],[100,0.0234],[101,0.0264],[102,0.0291],[103,0.0314],[104,0.0333],[105,0.0348],[106,0.0359],[107,0.0367],[108,0.0372],[109,0.0371],[110,0.0362],[111,0.0344],[112,0.0319],[113,0.0288],[114,0.0252],[116,0.0176],[117,0.014],[118,0.0108],[119,0.0083],[120,0.0066],[121,0.006],[170,0.0041],[181,0.0053],[182,0.0046],[183,0.0029],[185,-0.0017],[186,-0.0035],[187,-0.0043],[188,-0.004],[189,-0.0014],[190,0.0057],[191,0.0197],[192,0.0526],[193,0.1012],[194,0.1448],[195,0.1776],[196,0.2048],[197,0.2224],[198,0.2319],[199,0.2384],[200,0.2424],[201,0.2445],[202,0.2451],[203,0.2449],[204,0.2436],[205,0.2407],[206,0.2366],[209,0.2215],[210,0.2159],[211,0.2093],[213,0.1953],[214,0.1891],[217,0.1727],[218,0.1677],[219,0.1632],[220,0.1593],[221,0.1562],[222,0.1538],[223,0.1521],[231,0.142],[237,0.1361],[244,0.1312],[245,0.1298],[246,0.1268],[247,0.1215],[248,0.1131],[249,0.0969],[250,0.0729],[251,0.0469],[252,0.0251],[253,0.008],[254,-0.0075],[255,-0.0209],[256,-0.0314],[257,-0.0391],[258,-0.0446],[259,-0.0484],[260,-0.0508],[261,-0.0523],[262,-0.0525],[263,-0.0511],[264,-0.0485],[266,-0.0417],[267,-0.0377],[268,-0.0328],[269,-0.0275],[270,-0.0226],[271,-0.0185],[272,-0.0151],[273,-0.012],[274,-0.0091],[275,-0.0067],[276,-0.0049],[277,-0.0037],[279,-0.0022],[280,-0.0022],[281,-0.0012],[284,-0.0004],[324,0]]
       }
       spine_03_ctrl.Rotation.Z {
        type=MovieSceneScriptingFloatChannel
        default=0
        thinned_from=325
        keys [[0,0],[28,0],[29,-0.0002],[30,0.0035],[31,0.0124],[32,0.0196],[33,0.0224],[34,0.016],[35,0.006],[36,-0.0035],[37,-0.029],[38,-0.0844],[39,-0.1856],[40,-0.3182],[41,-0.4518],[42,-0.5866],[43,-0.7037],[44,-0.784],[45,-0.8474],[46,-0.892],[47,-0.9202],[48,-0.937],[49,-0.9449],[50,-0.9463],[51,-0.9417],[52,-0.9285],[53,-0.9046],[54,-0.8678],[55,-0.8274],[56,-0.794],[57,-0.7673],[58,-0.7469],[59,-0.7326],[60,-0.7241],[61,-0.7212],[85,-0.7184],[86,-0.734],[87,-0.7618],[88,-0.772],[89,-0.7513],[90,-0.7043],[91,-0.6298],[92,-0.5193],[93,-0.3847],[94,-0.2522],[95,-0.1154],[96,0.0234],[97,0.1322],[98,0.206],[99,0.2639],[100,0.3092],[101,0.3454],[102,0.3757],[103,0.4007],[104,0.4193],[105,0.4323],[106,0.4405],[107,0.4449],[108,0.4461],[109,0.4416],[110,0.4283],[111,0.4067],[112,0.3774],[113,0.3407],[114,0.2983],[116,0.2083],[117,0.1657],[118,0.1281],[119,0.0981],[120,0.0783],[121,0.0711],[181,0.071],[182,0.0618],[183,0.0386],[185,-0.0215],[186,-0.0447],[187,-0.054],[188,-0.0494],[189,-0.0175],[190,0.069],[191,0.2376],[192,0.6289],[193,1.2025],[194,1.7137],[195,2.0982],[196,2.4203],[197,2.6277],[198,2.7408],[199,2.8186],[200,2.8677],[201,2.8947],[202,2.9061],[203,2.9083],[204,2.898],[205,2.8707],[206,2.831],[209,2.6861],[210,2.6333],[212,2.5075],[213,2.4515],[214,2.4114],[216,2.3581],[218,2.3183],[219,2.3048],[220,2.2962],[244,2.2899],[245,2.2735],[246,2.229],[247,2.1423],[248,1.9995],[249,1.7172],[250,1.2931],[251,0.8339],[252,0.4465],[253,0.1418],[254,-0.1312],[255,-0.359],[256,-0.5281],[257,-0.6397],[258,-0.7106],[259,-0.7497],[260,-0.7665],[261,-0.7702],[262,-0.7573],[263,-0.7231],[264,-0.6748],[266,-0.5635],[267,-0.5029],[268,-0.4325],[269,-0.3596],[270,-0.2917],[271,-0.2359],[272,-0.1902],[273,-0.1485],[274,-0.1118],[275,-0.0814],[276,-0.0582],[277,-0.0436],[278,-0.0343],[279,-0.0264],[280,-0.0264],[281,-0.0145],[282,-0.0102],[283,-0.0068],[284,-0.0043],[285,-0.0025],[286,-0.0013],[287,-0.0005],[288,-0.0002],[324,0]]
       }
       spine_03_ctrl.Scale.X {type=MovieSceneScriptingFloatChannel, default=1, const [1,325]}
       spine_03_ctrl.Scale.Y {type=MovieSceneScriptingFloatChannel, default=1, const [1,325]}
       spine_03_ctrl.Scale.Z {type=MovieSceneScriptingFloatChannel, default=1, const [1,325]}
       spine_04_ctrl.Location.X {
        type=MovieSceneScriptingFloatChannel
        default=0
        thinned_from=325
        keys [[0,0],[28,0],[29,-0.0002],[30,-0.0019],[31,-0.0063],[32,-0.0185],[33,-0.0372],[34,-0.0642],[35,-0.086],[36,-0.0938],[37,-0.0987],[38,-0.1005],[39,-0.1005],[40,-0.0998],[43,-0.097],[45,-0.0959],[47,-0.0954],[52,-0.0957],[58,-0.0969],[85,-0.0975],[86,-0.097],[87,-0.094],[88,-0.0834],[89,-0.0676],[90,-0.0505],[92,-0.0157],[93,-0.0042],[94,-0.0012],[95,-0.0001],[96,-0.0002],[99,-0.0017],[106,-0.0032],[110,-0.0033],[119,-0.0008],[182,-0.0004],[187,0.0004],[189,0.0001],[190,-0.0005],[191,-0.0018],[192,-0.0047],[193,-0.0091],[194,-0.0131],[195,-0.0161],[196,-0.0185],[197,-0.0201],[198,-0.021],[199,-0.0216],[201,-0.0222],[204,-0.0222],[210,-0.0198],[220,-0.0131],[223,-0.012],[246,-0.0084],[247,-0.0081],[248,-0.0076],[249,-0.0065],[251,-0.0032],[252,-0.0017],[255,0.0015],[257,0.003],[259,0.0038],[261,0.0043],[263,0.0043],[267,0.0032],[274,0.0008],[324,0]]
       }
       spine_04_ctrl.Location.Y {
        type=MovieSceneScriptingFloatChannel
        default=0
        thinned_from=325
        keys [[0,0.5737],[1,0.5729],[2,0.5702],[3,0.5656],[4,0.559],[5,0.5503],[6,0.5395],[7,0.5263],[8,0.5108],[9,0.4924],[10,0.4709],[11,0.4473],[14,0.3713],[15,0.3469],[16,0.3243],[17,0.3043],[18,0.2852],[20,0.2454],[21,0.2259],[22,0.2076],[23,0.1911],[24,0.1772],[25,0.1664],[26,0.1594],[27,0.1569],[28,0.1571],[29,0.1576],[30,0.1586],[31,0.1599],[32,0.1618],[33,0.1641],[34,0.167],[35,0.1704],[36,0.1746],[37,0.1794],[38,0.185],[39,0.1915],[40,0.1988],[41,0.2069],[42,0.2159],[43,0.2258],[44,0.2381],[45,0.2541],[46,0.2731],[47,0.2944],[48,0.3174],[51,0.3887],[52,0.411],[53,0.4315],[54,0.4494],[55,0.4642],[56,0.4751],[59,0.4997],[62,0.5206],[65,0.538],[68,0.5518],[72,0.5652],[76,0.5731],[81,0.5763],[84,0.575],[85,0.5734],[86,0.5701],[87,0.565],[88,0.5578],[89,0.5484],[90,0.5368],[91,0.523],[92,0.5068],[93,0.4879],[94,0.4664],[95,0.4428],[98,0.3676],[99,0.3436],[100,0.3214],[101,0.3018],[102,0.2831],[104,0.2439],[105,0.2248],[106,0.2068],[107,0.1906],[108,0.1768],[109,0.1662],[110,0.1594],[111,0.1569],[112,0.1571],[113,0.1576],[114,0.1586],[115,0.1599],[116,0.1618],[117,0.1641],[118,0.1671],[119,0.1706],[120,0.1747],[121,0.1795],[122,0.185],[123,0.1913],[124,0.1983],[125,0.2061],[126,0.2148],[127,0.2244],[128,0.2365],[129,0.2524],[130,0.2713],[131,0.2926],[132,0.3155],[135,0.3868],[136,0.4091],[137,0.4297],[138,0.4477],[139,0.4625],[140,0.4736],[143,0.4983],[146,0.5192],[149,0.5365],[152,0.5504],[156,0.5638],[160,0.5717],[165,0.5748],[168,0.5736],[169,0.5719],[170,0.5682],[171,0.5626],[172,0.5548],[173,0.5449],[174,0.5327],[175,0.5181],[176,0.5012],[177,0.4807],[178,0.4566],[179,0.43],[181,0.3743],[182,0.3477],[183,0.3236],[184,0.3032],[187,0.2478],[188,0.2299],[189,0.2131],[190,0.1977],[191,0.1842],[192,0.1729],[193,0.1643],[194,0.1589],[195,0.157],[196,0.1572],[197,0.1577],[198,0.1585],[199,0.1598],[200,0.1615],[201,0.1637],[202,0.1665],[203,0.1699],[204,0.174],[205,0.1787],[206,0.1842],[207,0.1904],[208,0.1975],[209,0.2055],[210,0.2144],[211,0.2242],[212,0.2372],[213,0.2547],[214,0.2758],[215,0.2995],[216,0.3249],[218,0.3771],[219,0.402],[220,0.4249],[221,0.4449],[222,0.4611],[223,0.4726],[226,0.4973],[229,0.5181],[232,0.5352],[235,0.5488],[239,0.5619],[243,0.5698],[250,0.574],[252,0.5733],[253,0.5718],[254,0.5684],[255,0.5632],[256,0.556],[257,0.5468],[258,0.5353],[259,0.5215],[260,0.5052],[261,0.4862],[262,0.4647],[263,0.4412],[266,0.3669],[267,0.3432],[268,0.3213],[269,0.3018],[270,0.2831],[272,0.244],[273,0.2249],[274,0.2069],[275,0.1906],[276,0.1769],[277,0.1662],[278,0.1594],[279,0.1569],[280,0.1571],[281,0.1576],[282,0.1586],[283,0.1599],[284,0.1618],[285,0.1641],[286,0.1671],[287,0.1706],[288,0.1747],[289,0.1795],[290,0.185],[291,0.1913],[292,0.1983],[293,0.2061],[294,0.2148],[295,0.2244],[296,0.2363],[297,0.2516],[298,0.2697],[299,0.29],[300,0.3119],[303,0.381],[304,0.4033],[305,0.4243],[306,0.4433],[307,0.46],[308,0.4736],[310,0.4961],[312,0.516],[314,0.5331],[316,0.5474],[318,0.5587],[320,0.567],[322,0.572],[324,0.5737]]
       }
       spine_04_ctrl.Location.Z {
        type=MovieSceneScriptingFloatChannel
        default=0
        thinned_from=325
        keys [[0,0.0672],[1,0.0681],[2,0.0709],[3,0.0753],[4,0.0813],[5,0.0885],[6,0.097],[7,0.1064],[8,0.1168],[9,0.1279],[10,0.1394],[11,0.1514],[17,0.2265],[18,0.2385],[19,0.2496],[20,0.2596],[21,0.2687],[22,0.2766],[23,0.2832],[24,0.2885],[25,0.2924],[26,0.2949],[27,0.2957],[30,0.295],[32,0.2935],[34,0.2906],[36,0.2863],[38,0.2801],[39,0.2763],[40,0.2719],[41,0.2669],[42,0.2612],[43,0.2548],[44,0.2471],[45,0.2376],[46,0.2268],[47,0.2148],[48,0.2019],[51,0.1613],[52,0.148],[53,0.1353],[54,0.1235],[55,0.113],[56,0.104],[58,0.0879],[59,0.0801],[60,0.0726],[61,0.0654],[62,0.0586],[63,0.0521],[64,0.046],[65,0.0405],[66,0.0354],[67,0.0309],[68,0.0271],[69,0.0239],[70,0.0214],[71,0.0197],[72,0.0187],[73,0.0187],[74,0.0194],[75,0.0209],[76,0.0232],[77,0.0262],[78,0.0299],[79,0.0344],[80,0.0396],[81,0.0456],[82,0.0523],[83,0.0598],[84,0.0679],[85,0.0768],[86,0.0863],[87,0.0963],[88,0.1068],[89,0.1176],[94,0.1738],[97,0.2064],[99,0.227],[101,0.2467],[102,0.2558],[103,0.264],[104,0.2713],[105,0.2776],[106,0.2831],[107,0.2876],[108,0.2911],[109,0.2936],[110,0.2952],[113,0.2954],[116,0.2935],[118,0.2907],[120,0.2865],[122,0.2803],[123,0.2765],[124,0.272],[125,0.267],[126,0.2612],[127,0.2548],[128,0.2471],[129,0.2376],[130,0.2268],[131,0.2148],[132,0.2019],[135,0.1612],[136,0.148],[137,0.1352],[138,0.1235],[139,0.113],[140,0.104],[142,0.0879],[143,0.0802],[144,0.0727],[145,0.0655],[146,0.0586],[147,0.0521],[148,0.0461],[149,0.0405],[150,0.0354],[151,0.031],[152,0.0271],[153,0.0239],[154,0.0214],[155,0.0197],[156,0.0188],[157,0.0187],[158,0.0194],[159,0.021],[160,0.0233],[161,0.0264],[162,0.0302],[163,0.0348],[164,0.0401],[165,0.0461],[166,0.0529],[167,0.0604],[168,0.0686],[169,0.0774],[170,0.0869],[171,0.0967],[172,0.107],[173,0.1177],[178,0.1727],[182,0.2147],[184,0.235],[185,0.2448],[186,0.2539],[187,0.2621],[188,0.2696],[189,0.2763],[190,0.282],[191,0.2868],[192,0.2906],[193,0.2934],[194,0.2951],[195,0.2956],[199,0.2945],[201,0.2926],[203,0.2892],[205,0.2841],[206,0.2808],[207,0.2769],[208,0.2724],[209,0.2673],[210,0.2614],[211,0.2548],[212,0.2465],[213,0.2362],[214,0.2241],[215,0.2107],[216,0.1964],[219,0.1519],[220,0.1378],[221,0.1248],[222,0.1134],[223,0.1038],[224,0.0955],[225,0.0873],[226,0.0794],[227,0.0718],[228,0.0645],[229,0.0576],[230,0.0512],[231,0.0452],[232,0.0397],[233,0.0347],[234,0.0303],[235,0.0265],[236,0.0234],[237,0.021],[238,0.0194],[239,0.0185],[240,0.0184],[241,0.0191],[242,0.0204],[243,0.0223],[244,0.0248],[245,0.028],[246,0.0319],[247,0.0364],[248,0.0416],[249,0.0475],[250,0.0541],[251,0.0614],[252,0.0693],[253,0.078],[254,0.0873],[255,0.0971],[256,0.1074],[257,0.1181],[263,0.1847],[266,0.2167],[268,0.2369],[269,0.2466],[270,0.2557],[271,0.2639],[272,0.2712],[273,0.2776],[274,0.283],[275,0.2876],[276,0.2911],[277,0.2936],[278,0.2952],[281,0.2954],[284,0.2935],[286,0.2907],[288,0.2865],[290,0.2803],[291,0.2765],[292,0.272],[293,0.267],[294,0.2612],[295,0.2548],[296,0.247],[297,0.2372],[298,0.2259],[299,0.2133],[300,0.1999],[303,0.158],[304,0.1447],[305,0.1322],[306,0.1211],[307,0.1115],[308,0.104],[309,0.0979],[310,0.0925],[311,0.0878],[312,0.0837],[313,0.0802],[314,0.0772],[315,0.0747],[316,0.0726],[317,0.071],[318,0.0697],[319,0.0688],[320,0.0681],[322,0.0673],[324,0.0672]]
       }
       spine_04_ctrl.Rotation.X {
        type=MovieSceneScriptingFloatChannel
        default=0
        thinned_from=325
        keys [[0,8.6154],[2,8.6046],[4,8.5535],[5,8.5042],[6,8.4343],[7,8.3403],[8,8.2188],[9,8.0586],[10,7.8575],[11,7.6258],[15,6.5953],[16,6.3626],[21,5.3762],[22,5.1925],[23,5.0272],[24,4.8869],[25,4.7785],[26,4.7085],[27,4.6837],[30,4.7009],[33,4.7556],[35,4.8145],[37,4.8941],[39,4.9954],[41,5.1193],[43,5.267],[44,5.3698],[45,5.5132],[46,5.6898],[47,5.8923],[51,6.814],[52,7.0358],[53,7.2396],[54,7.4181],[55,7.564],[56,7.6701],[60,7.9689],[65,8.2484],[71,8.4665],[79,8.5982],[85,8.6106],[87,8.5816],[88,8.5482],[89,8.4975],[90,8.4257],[91,8.3291],[92,8.2041],[93,8.041],[94,7.839],[95,7.6078],[99,6.5869],[100,6.3567],[105,5.3744],[106,5.1912],[107,5.0262],[108,4.8863],[109,4.778],[110,4.7081],[111,4.6834],[114,4.7008],[117,4.7558],[119,4.8156],[121,4.8955],[123,4.9969],[125,5.121],[127,5.269],[128,5.3719],[129,5.5154],[130,5.6921],[131,5.8947],[135,6.8165],[136,7.0383],[137,7.242],[138,7.4204],[139,7.5663],[140,7.6723],[144,7.9708],[149,8.2502],[155,8.4683],[163,8.6003],[169,8.6124],[171,8.5812],[172,8.5455],[173,8.4916],[174,8.4158],[175,8.3142],[176,8.1834],[177,8.0039],[178,7.7702],[179,7.4988],[182,6.6248],[183,6.369],[184,6.1586],[188,5.4203],[189,5.2505],[190,5.095],[191,4.9582],[192,4.8439],[193,4.7557],[194,4.6983],[195,4.6768],[198,4.688],[201,4.7389],[203,4.797],[205,4.8764],[207,4.979],[209,5.1061],[211,5.2595],[212,5.3707],[213,5.5305],[214,5.7291],[215,5.9567],[219,6.9637],[220,7.191],[221,7.3891],[222,7.5486],[223,7.6597],[227,7.9576],[232,8.2348],[238,8.4493],[251,8.6116],[254,8.6011],[256,8.5447],[257,8.491],[258,8.4154],[259,8.3141],[260,8.1833],[261,8.0158],[262,7.8127],[263,7.5831],[267,6.5807],[268,6.3544],[273,5.3763],[274,5.1928],[275,5.0276],[276,4.8873],[277,4.7787],[278,4.7085],[279,4.6837],[282,4.7009],[285,4.7558],[287,4.8156],[289,4.8955],[291,4.9969],[293,5.121],[295,5.269],[296,5.3697],[297,5.5073],[298,5.6755],[299,5.8681],[301,6.3014],[304,6.9784],[305,7.1865],[306,7.3757],[307,7.5396],[308,7.6723],[311,7.9869],[314,8.24],[317,8.4297],[320,8.5542],[323,8.6115],[324,8.6154]]
       }
       spine_04_ctrl.Rotation.Y {
        type=MovieSceneScriptingFloatChannel
        default=0
        thinned_from=325
        keys [[0,0],[28,0],[29,0.0035],[30,0.0284],[31,0.0957],[32,0.2834],[33,0.5722],[34,0.9898],[35,1.3275],[36,1.4486],[37,1.5261],[38,1.5603],[39,1.569],[40,1.571],[45,1.5561],[85,1.5795],[86,1.5736],[87,1.5294],[88,1.3642],[89,1.1172],[90,0.8477],[91,0.5709],[92,0.2912],[93,0.1009],[94,0.0426],[95,0.0126],[96,0.0016],[97,0.0003],[99,0.0014],[102,0.0038],[107,0.0093],[108,0.0101],[109,0.0107],[110,0.0108],[111,0.0104],[113,0.0087],[117,0.0042],[119,0.0025],[120,0.002],[130,0.002],[169,-0.0009],[182,0.0003],[184,0.0001],[187,-0.0007],[188,-0.0008],[189,-0.0003],[190,0.0014],[191,0.0051],[192,0.0146],[193,0.0295],[194,0.0435],[195,0.054],[196,0.0624],[197,0.0679],[198,0.0709],[199,0.073],[200,0.0744],[201,0.0753],[203,0.0763],[204,0.0765],[208,0.0758],[211,0.0761],[212,0.0748],[213,0.0712],[214,0.0661],[215,0.0598],[216,0.0525],[217,0.0446],[219,0.0286],[220,0.0212],[221,0.0146],[222,0.0092],[223,0.0054],[225,-0.0002],[227,-0.0054],[229,-0.01],[231,-0.0143],[233,-0.018],[235,-0.0212],[237,-0.024],[239,-0.0263],[241,-0.028],[243,-0.0292],[244,-0.0296],[245,-0.0297],[246,-0.0294],[247,-0.0284],[248,-0.0266],[249,-0.0229],[250,-0.0173],[251,-0.0111],[252,-0.0059],[253,-0.0018],[254,0.0016],[255,0.0042],[256,0.0057],[257,0.0062],[258,0.0061],[259,0.0055],[261,0.0034],[264,-0.0003],[266,-0.0021],[267,-0.0027],[268,-0.003],[272,-0.0027],[277,-0.0011],[287,0],[288,0.0002],[324,0]]
       }
       spine_04_ctrl.Rotation.Z {
        type=MovieSceneScriptingFloatChannel
        default=0
        thinned_from=325
        keys [[0,0],[28,0],[29,-0.0002],[30,0.0042],[31,0.015],[32,0.0231],[33,0.025],[34,0.0148],[35,0.0004],[36,-0.0124],[37,-0.045],[38,-0.1149],[39,-0.2419],[40,-0.4084],[41,-0.5761],[42,-0.7455],[43,-0.8929],[44,-0.9943],[45,-1.0745],[46,-1.1312],[47,-1.1674],[48,-1.1893],[49,-1.2001],[50,-1.2026],[51,-1.1976],[52,-1.1817],[53,-1.1522],[54,-1.1064],[55,-1.056],[56,-1.0142],[57,-0.9808],[58,-0.9552],[59,-0.9373],[60,-0.9267],[61,-0.9231],[85,-0.9211],[86,-0.9408],[87,-0.9756],[88,-0.9867],[89,-0.9581],[90,-0.896],[91,-0.7991],[92,-0.6568],[93,-0.4852],[94,-0.3178],[95,-0.1453],[96,0.0294],[97,0.1661],[98,0.2588],[99,0.3313],[100,0.388],[101,0.4332],[102,0.4711],[103,0.5023],[104,0.5254],[105,0.5415],[106,0.5516],[107,0.5569],[108,0.5583],[109,0.5526],[110,0.5359],[111,0.5088],[112,0.4721],[113,0.4262],[114,0.3732],[116,0.2606],[117,0.2073],[118,0.1603],[119,0.1228],[120,0.098],[121,0.089],[181,0.0892],[182,0.0775],[183,0.0485],[185,-0.027],[186,-0.056],[187,-0.0676],[188,-0.0619],[189,-0.022],[190,0.0864],[191,0.2974],[192,0.787],[193,1.5045],[194,2.1438],[195,2.6248],[196,3.0277],[197,3.2872],[198,3.4286],[199,3.5261],[200,3.5877],[201,3.6217],[202,3.6361],[203,3.6391],[204,3.6265],[205,3.5926],[206,3.5433],[209,3.3632],[210,3.2976],[212,3.1413],[213,3.0721],[214,3.023],[216,2.9588],[218,2.9117],[219,2.896],[220,2.8865],[244,2.8859],[245,2.8653],[246,2.8092],[247,2.7],[248,2.5199],[249,2.1642],[250,1.6297],[251,1.051],[252,0.5627],[253,0.1787],[254,-0.1653],[255,-0.4524],[256,-0.6654],[257,-0.806],[258,-0.8951],[259,-0.9443],[260,-0.9652],[261,-0.9694],[262,-0.9529],[263,-0.9095],[264,-0.8483],[266,-0.7077],[267,-0.6312],[268,-0.5426],[269,-0.4511],[270,-0.3657],[271,-0.2956],[272,-0.2383],[273,-0.1859],[274,-0.14],[275,-0.1018],[276,-0.0729],[277,-0.0545],[278,-0.0429],[279,-0.033],[280,-0.0248],[281,-0.0181],[282,-0.0127],[283,-0.0085],[284,-0.0054],[285,-0.0031],[286,-0.0016],[287,-0.0007],[288,0],[324,0]]
       }
       spine_04_ctrl.Scale.X {type=MovieSceneScriptingFloatChannel, default=1, const [1,325]}
       spine_04_ctrl.Scale.Y {type=MovieSceneScriptingFloatChannel, default=1, const [1,325]}
       spine_04_ctrl.Scale.Z {type=MovieSceneScriptingFloatChannel, default=1, const [1,325]}
       head_ctrl.Location.X {type=MovieSceneScriptingFloatChannel, default=0, thinned_from=325, keys [[0,0],[324,0]]}
       head_ctrl.Location.Y {type=MovieSceneScriptingFloatChannel, default=0, thinned_from=325, keys [[0,0],[324,0]]}
       head_ctrl.Location.Z {type=MovieSceneScriptingFloatChannel, default=0, thinned_from=325, keys [[0,0],[324,0]]}
       head_ctrl.Rotation.X {
        type=MovieSceneScriptingFloatChannel
        default=0
        thinned_from=325
        keys [[0,9.8055],[1,9.7918],[3,9.6995],[4,9.6203],[6,9.4038],[8,9.1179],[10,8.7667],[16,7.5686],[18,7.1978],[22,6.357],[23,6.1559],[24,5.9715],[25,5.8103],[26,5.6784],[27,5.5822],[28,5.5746],[29,5.7201],[30,6.0299],[31,6.5745],[32,7.2905],[33,7.9844],[34,8.4741],[35,8.6753],[36,8.7517],[40,8.5977],[41,8.5802],[42,8.5965],[43,8.642],[44,8.7154],[45,8.8254],[46,8.9663],[47,9.1322],[49,9.5163],[53,10.3317],[54,10.5105],[55,10.6677],[63,11.5481],[70,12.1264],[79,12.5678],[81,12.727],[83,12.9571],[85,13.2749],[87,13.6565],[88,13.8779],[89,13.9343],[90,13.7912],[91,13.5716],[92,13.2618],[93,12.8381],[95,11.8245],[97,10.7182],[98,10.3559],[103,9.3785],[106,8.7353],[108,8.347],[109,8.185],[110,8.0531],[111,7.958],[115,7.6853],[118,7.5385],[120,7.4831],[122,7.471],[124,7.5096],[125,7.5504],[126,7.6068],[127,7.6799],[128,7.7827],[129,7.9234],[130,8.0956],[131,8.2933],[133,8.7399],[136,9.4446],[138,9.8646],[139,10.0408],[144,10.6724],[149,11.1831],[158,11.8405],[164,12.1257],[168,12.2216],[169,12.2216],[170,12.1841],[171,12.121],[172,12.0314],[174,11.7819],[176,11.4537],[178,11.0518],[180,10.6084],[181,10.4163],[182,10.2824],[183,10.2081],[184,10.2005],[185,10.2442],[186,10.501],[187,10.8943],[188,11.0503],[189,10.9156],[190,10.7213],[191,10.4888],[193,9.9488],[194,9.6465],[195,9.3194],[196,9.0248],[197,8.8041],[198,8.6608],[199,8.5566],[200,8.483],[203,8.3588],[206,8.2952],[208,8.2939],[210,8.3358],[211,8.3772],[212,8.4444],[213,8.5444],[214,8.6729],[216,8.9933],[220,9.7241],[221,9.8892],[222,10.0346],[229,10.7436],[234,11.1367],[244,11.7052],[246,11.7845],[247,11.9311],[249,12.4814],[250,12.8184],[251,13.015],[252,12.7923],[254,12.0685],[255,11.6428],[258,10.223],[259,9.7663],[260,9.3555],[261,9.0125],[262,8.7621],[268,7.5358],[270,7.1575],[274,6.3098],[275,6.1087],[276,5.9251],[277,5.7653],[278,5.636],[279,5.5435],[282,5.3415],[284,5.2303],[286,5.1466],[288,5.0976],[290,5.0907],[291,5.1053],[292,5.1331],[293,5.1751],[294,5.2322],[295,5.3054],[296,5.4049],[297,5.5372],[298,5.6974],[299,5.8805],[300,6.0816],[305,7.1838],[306,7.3889],[307,7.5775],[312,8.3442],[315,8.7504],[319,9.2181],[322,9.6563],[323,9.7587],[324,9.8041]]
       }
       head_ctrl.Rotation.Y {
        type=MovieSceneScriptingFloatChannel
        default=0
        thinned_from=325
        keys [[0,-5.1409],[3,-5.1184],[6,-5.0565],[10,-4.925],[16,-4.6911],[23,-4.464],[25,-4.4213],[26,-4.4129],[28,-4.4266],[29,-4.3945],[30,-4.266],[31,-3.9987],[32,-2.95],[33,-1.3704],[34,0.2484],[35,1.4421],[36,2.267],[37,2.7359],[38,3.1534],[39,3.5059],[40,3.7804],[41,3.962],[42,4.0297],[43,4.0371],[44,4.0537],[46,4.1128],[53,4.4253],[55,4.4943],[61,4.5932],[75,4.7009],[84,4.7332],[85,4.7083],[86,4.5259],[87,3.6234],[88,2.115],[89,0.4026],[90,-1.3191],[91,-2.9119],[92,-4.0167],[93,-4.4545],[94,-4.7069],[95,-4.8321],[96,-4.8646],[97,-4.8416],[106,-4.5663],[109,-4.505],[111,-4.4916],[122,-4.5552],[127,-4.6228],[130,-4.6939],[136,-4.8802],[139,-4.9437],[169,-5.0567],[175,-4.9997],[182,-4.8663],[183,-4.8335],[184,-4.7768],[185,-4.8773],[186,-5.2448],[187,-5.7322],[188,-6.2011],[189,-6.8146],[190,-7.6254],[191,-8.3515],[192,-8.8066],[193,-9.0523],[194,-9.1734],[195,-9.2227],[196,-9.2057],[198,-9.1462],[201,-9.1329],[205,-9.2077],[209,-9.3784],[211,-9.501],[212,-9.5845],[213,-9.6954],[215,-9.9731],[219,-10.597],[221,-10.8554],[222,-10.9525],[232,-11.411],[245,-11.7143],[247,-11.8054],[249,-11.9827],[250,-11.7985],[251,-10.8662],[252,-8.491],[253,-6.4744],[254,-5.9173],[255,-5.5581],[256,-5.3158],[257,-5.1616],[258,-5.0816],[259,-5.0479],[261,-5.0217],[263,-4.9701],[275,-4.4985],[277,-4.4434],[278,-4.4266],[287,-4.4387],[289,-4.4455],[295,-4.5273],[297,-4.572],[306,-4.8902],[315,-5.0699],[317,-5.1065],[324,-5.1406]]
       }
       head_ctrl.Rotation.Z {
        type=MovieSceneScriptingFloatChannel
        default=0
        thinned_from=325
        keys [[0,11.2508],[16,11.4527],[22,11.7514],[24,11.8047],[25,11.7953],[26,11.717],[27,11.5191],[28,11.1521],[29,10.5417],[30,9.6304],[31,8.095],[32,5.6777],[33,2.4898],[34,-2.1866],[35,-7.4686],[36,-10.6498],[37,-11.8022],[38,-12.7594],[39,-13.5],[40,-14.0362],[41,-14.3356],[42,-14.4807],[43,-14.5436],[46,-14.5224],[47,-14.4826],[48,-14.3937],[49,-14.2649],[51,-13.9246],[55,-13.1739],[58,-12.729],[61,-12.3796],[64,-12.1149],[68,-11.874],[73,-11.7158],[81,-11.6606],[82,-11.8095],[84,-12.4428],[85,-12.5413],[86,-12.3632],[87,-11.4375],[88,-7.7729],[89,-2.9192],[90,0.5221],[91,2.5154],[92,4.1163],[93,5.2298],[94,5.9961],[95,6.6272],[96,7.1324],[97,7.5214],[98,7.82],[100,8.3619],[101,8.5945],[102,8.795],[103,8.963],[104,9.0961],[106,9.3007],[108,9.4562],[110,9.5651],[113,9.6536],[118,9.6889],[120,9.6578],[122,9.5763],[124,9.451],[127,9.1954],[130,8.8692],[132,8.6226],[133,8.4762],[134,8.2972],[136,7.877],[138,7.4336],[139,7.2256],[140,7.0389],[143,6.546],[145,6.2526],[147,5.9937],[149,5.7748],[150,5.682],[151,5.6012],[152,5.5331],[153,5.4783],[154,5.4376],[155,5.4117],[169,5.3848],[176,5.4562],[181,5.5481],[182,5.6784],[183,6.0831],[184,6.8153],[185,7.9108],[186,9.4051],[187,12.3057],[188,17.423],[189,24.6682],[190,32.6404],[191,39.3372],[192,43.5573],[193,45.58],[194,46.5177],[195,47.3357],[196,48.0211],[197,48.5621],[199,49.3082],[201,49.7519],[204,50.0099],[207,49.9767],[212,49.1625],[218,47.7643],[223,47.1363],[243,46.8475],[244,46.9827],[245,47.3287],[246,47.8663],[247,48.6083],[248,49.7381],[249,50.5604],[250,48.3161],[251,42.2438],[252,34.0503],[253,24.7639],[254,17.4513],[255,13.6724],[256,11.4787],[257,10.3649],[258,9.7416],[259,9.3286],[260,9.0842],[261,8.9669],[262,8.9344],[263,8.9444],[264,9.0028],[265,9.1287],[266,9.2991],[269,9.8789],[272,10.5626],[273,10.7742],[274,10.9537],[276,11.2521],[277,11.3757],[278,11.4724],[279,11.5344],[324,11.2509]]
       }
       head_ctrl.Scale.X {type=MovieSceneScriptingFloatChannel, default=1, const [1,325]}
       head_ctrl.Scale.Y {type=MovieSceneScriptingFloatChannel, default=1, const [1,325]}
       head_ctrl.Scale.Z {type=MovieSceneScriptingFloatChannel, default=1, const [1,325]}
       clavicle_l_ctrl.Location.X {type=MovieSceneScriptingFloatChannel, default=0, thinned_from=325, keys [[0,0],[324,0]]}
       clavicle_l_ctrl.Location.Y {type=MovieSceneScriptingFloatChannel, default=0, thinned_from=325, keys [[0,0],[324,0]]}
       clavicle_l_ctrl.Location.Z {type=MovieSceneScriptingFloatChannel, default=0, thinned_from=325, keys [[0,0],[324,0]]}
       clavicle_l_ctrl.Rotation.X {
        type=MovieSceneScriptingFloatChannel
        default=0
        thinned_from=325
        keys [[0,-0.0085],[1,-0.01],[2,-0.014],[3,-0.0203],[4,-0.0283],[5,-0.0378],[6,-0.0483],[7,-0.0594],[9,-0.0819],[10,-0.0924],[11,-0.1041],[12,-0.118],[13,-0.1329],[14,-0.1474],[15,-0.1604],[16,-0.1706],[17,-0.1788],[18,-0.1862],[19,-0.193],[20,-0.1987],[21,-0.2034],[22,-0.2069],[23,-0.2092],[24,-0.21],[25,-0.2097],[26,-0.209],[28,-0.2063],[30,-0.2021],[32,-0.1964],[34,-0.1894],[36,-0.1814],[38,-0.1724],[41,-0.158],[42,-0.1524],[43,-0.1453],[44,-0.1366],[45,-0.1256],[46,-0.1137],[47,-0.1014],[48,-0.0894],[49,-0.0785],[50,-0.0691],[51,-0.0625],[53,-0.052],[55,-0.0423],[56,-0.0377],[57,-0.0334],[58,-0.0296],[59,-0.0264],[60,-0.0237],[61,-0.0215],[62,-0.0197],[63,-0.0182],[65,-0.016],[68,-0.0138],[75,-0.0103],[76,-0.0103],[77,-0.0095],[79,-0.0095],[80,-0.0089],[85,-0.0089],[86,-0.0142],[87,-0.0204],[88,-0.0285],[89,-0.0379],[90,-0.0484],[91,-0.0595],[92,-0.0708],[93,-0.0814],[94,-0.0925],[95,-0.1048],[96,-0.118],[97,-0.1328],[98,-0.1474],[99,-0.1599],[100,-0.1707],[101,-0.1791],[102,-0.1862],[103,-0.1929],[104,-0.1986],[105,-0.2034],[106,-0.2069],[107,-0.2091],[108,-0.2099],[110,-0.2091],[112,-0.2067],[114,-0.2029],[116,-0.1977],[118,-0.1915],[120,-0.1842],[121,-0.1802],[123,-0.1802],[124,-0.167],[127,-0.167],[128,-0.1445],[129,-0.1318],[131,-0.1025],[132,-0.0893],[133,-0.0786],[134,-0.0692],[135,-0.0623],[136,-0.057],[137,-0.0521],[138,-0.0475],[140,-0.0391],[144,-0.0237],[145,-0.0211],[147,-0.0182],[148,-0.0169],[150,-0.0151],[152,-0.0138],[159,-0.0102],[160,-0.0102],[161,-0.0094],[163,-0.0094],[164,-0.0088],[168,-0.0088],[169,-0.0101],[170,-0.0141],[171,-0.0203],[172,-0.0284],[173,-0.0379],[174,-0.0483],[176,-0.0708],[177,-0.0814],[178,-0.0924],[179,-0.1047],[180,-0.118],[182,-0.1474],[183,-0.1599],[184,-0.1706],[185,-0.1791],[186,-0.1862],[187,-0.1929],[188,-0.1987],[189,-0.2034],[190,-0.207],[191,-0.2091],[192,-0.2099],[193,-0.2097],[194,-0.209],[195,-0.2078],[197,-0.2043],[199,-0.1994],[201,-0.193],[203,-0.1855],[205,-0.177],[208,-0.1628],[209,-0.1587],[210,-0.1525],[211,-0.1427],[212,-0.1324],[216,-0.0889],[217,-0.0786],[218,-0.0691],[219,-0.0622],[222,-0.0471],[223,-0.0422],[224,-0.0376],[225,-0.0333],[226,-0.0295],[227,-0.0262],[228,-0.0234],[229,-0.0213],[230,-0.0195],[231,-0.0181],[232,-0.017],[234,-0.0155],[238,-0.0134],[240,-0.0119],[241,-0.0114],[242,-0.0114],[243,-0.0104],[244,-0.0104],[245,-0.0097],[249,-0.0088],[252,-0.0087],[253,-0.0101],[254,-0.0142],[255,-0.0204],[256,-0.0285],[257,-0.038],[258,-0.0484],[259,-0.0595],[261,-0.0819],[262,-0.0925],[263,-0.1042],[264,-0.118],[265,-0.1329],[266,-0.1474],[267,-0.1604],[268,-0.1706],[269,-0.1787],[270,-0.1862],[271,-0.1929],[272,-0.1987],[273,-0.2034],[274,-0.207],[275,-0.2092],[276,-0.21],[277,-0.2097],[279,-0.2078],[281,-0.2042],[283,-0.1991],[285,-0.1927],[287,-0.1851],[289,-0.1766],[292,-0.1625],[294,-0.1524],[295,-0.1455],[296,-0.1356],[297,-0.1237],[298,-0.1111],[299,-0.0987],[300,-0.0876],[301,-0.0778],[302,-0.0691],[303,-0.0622],[305,-0.051],[306,-0.0455],[307,-0.0403],[308,-0.0353],[309,-0.0307],[310,-0.0266],[311,-0.0231],[312,-0.0202],[319,-0.0115],[321,-0.0097],[322,-0.0091],[323,-0.0087],[324,-0.0087]]
       }
       clavicle_l_ctrl.Rotation.Y {
        type=MovieSceneScriptingFloatChannel
        default=0
        thinned_from=325
        keys [[0,-6.7741],[1,-6.7628],[2,-6.7302],[3,-6.679],[4,-6.6114],[6,-6.437],[10,-5.9989],[11,-5.8601],[12,-5.6841],[14,-5.2867],[15,-5.0981],[16,-4.9383],[18,-4.6768],[20,-4.4459],[24,-4.039],[25,-3.986],[26,-3.9653],[29,-3.8846],[32,-3.8338],[37,-3.7966],[42,-3.7914],[43,-3.7998],[44,-3.8222],[45,-3.8575],[47,-3.9534],[48,-4.0161],[49,-4.0912],[50,-4.1963],[51,-4.3593],[53,-4.7448],[56,-5.3489],[57,-5.5375],[58,-5.7099],[59,-5.8564],[60,-5.9819],[61,-6.0853],[63,-6.2561],[65,-6.3953],[67,-6.5046],[70,-6.6225],[74,-6.7167],[75,-6.7316],[76,-6.7316],[77,-6.7529],[79,-6.7529],[80,-6.7682],[85,-6.7682],[87,-6.651],[88,-6.5672],[90,-6.3578],[94,-5.8783],[95,-5.7412],[96,-5.5912],[98,-5.2481],[100,-4.9383],[104,-4.3931],[105,-4.2729],[106,-4.1697],[107,-4.0926],[108,-4.035],[109,-3.9957],[112,-3.9124],[115,-3.8544],[120,-3.8056],[123,-3.8012],[127,-3.7945],[128,-3.8003],[129,-3.8199],[131,-3.8988],[132,-3.9734],[133,-4.0924],[134,-4.233],[135,-4.3922],[143,-5.816],[144,-5.9819],[145,-6.096],[148,-6.3271],[150,-6.4506],[153,-6.5866],[157,-6.6985],[159,-6.7325],[160,-6.7325],[161,-6.7543],[163,-6.7543],[164,-6.7701],[168,-6.7701],[169,-6.7584],[170,-6.7174],[171,-6.6528],[172,-6.569],[174,-6.3592],[178,-5.8791],[179,-5.7419],[180,-5.5919],[182,-5.2485],[184,-4.9383],[188,-4.3932],[189,-4.2728],[190,-4.1691],[191,-4.0909],[192,-4.0318],[193,-3.9909],[196,-3.9045],[199,-3.8462],[203,-3.8055],[210,-3.7913],[211,-3.8003],[212,-3.8256],[213,-3.8643],[214,-3.9139],[217,-4.1004],[218,-4.2177],[219,-4.3875],[220,-4.5664],[224,-5.3501],[225,-5.5358],[226,-5.7067],[227,-5.8572],[228,-5.9819],[229,-6.0846],[231,-6.2512],[233,-6.3774],[240,-6.6984],[241,-6.7152],[242,-6.7152],[243,-6.7407],[244,-6.7407],[252,-6.7716],[253,-6.7569],[254,-6.7154],[255,-6.6508],[256,-6.5671],[258,-6.3576],[262,-5.8782],[263,-5.7473],[264,-5.5914],[267,-5.0838],[271,-4.5233],[272,-4.3906],[273,-4.2695],[274,-4.1653],[275,-4.0831],[276,-4.0282],[278,-3.9577],[281,-3.8805],[284,-3.8319],[289,-3.7964],[294,-3.7913],[295,-3.7969],[296,-3.8138],[297,-3.8419],[298,-3.8812],[299,-3.9318],[300,-3.9937],[301,-4.0918],[302,-4.2348],[303,-4.3984],[304,-4.5725],[308,-5.3456],[309,-5.531],[310,-5.7025],[311,-5.8546],[312,-5.9819],[319,-6.5621],[321,-6.6902],[322,-6.7351],[323,-6.764],[324,-6.764]]
       }
       clavicle_l_ctrl.Rotation.Z {
        type=MovieSceneScriptingFloatChannel
        default=0
        thinned_from=325
        keys [[0,-0.1435],[1,-0.1595],[2,-0.2044],[3,-0.2734],[4,-0.3615],[5,-0.464],[6,-0.576],[8,-0.8092],[9,-0.9207],[10,-1.0224],[11,-1.1262],[12,-1.2422],[13,-1.3622],[14,-1.4782],[15,-1.5821],[16,-1.6659],[17,-1.7354],[18,-1.8008],[19,-1.8604],[20,-1.9128],[21,-1.9562],[22,-1.9892],[23,-2.0102],[24,-2.0175],[26,-2.0129],[28,-1.9992],[30,-1.9767],[33,-1.9286],[35,-1.8857],[36,-1.8346],[37,-1.7349],[38,-1.5523],[39,-1.21],[40,-0.8164],[41,-0.434],[42,-0.1229],[43,0.0137],[44,0.0429],[45,-0.0209],[46,-0.1353],[47,-0.2507],[48,-0.3617],[49,-0.4494],[50,-0.5046],[51,-0.5295],[52,-0.5492],[53,-0.564],[54,-0.5742],[55,-0.5802],[56,-0.5821],[85,-0.5821],[87,-0.6311],[88,-0.672],[89,-0.7364],[90,-0.8245],[91,-0.9416],[92,-1.0802],[93,-1.2376],[94,-1.4049],[95,-1.5852],[96,-1.74],[97,-1.8282],[98,-1.8595],[99,-1.8378],[100,-1.7855],[101,-1.7104],[102,-1.6153],[103,-1.5037],[104,-1.3805],[105,-1.2476],[106,-1.1177],[107,-1.0059],[108,-0.9059],[109,-0.8172],[110,-0.7362],[111,-0.6593],[113,-0.5123],[114,-0.442],[115,-0.3759],[116,-0.3161],[117,-0.2659],[118,-0.2242],[119,-0.1912],[120,-0.1667],[121,-0.1502],[123,-0.1502],[124,-0.1438],[168,-0.1438],[169,-0.1606],[170,-0.2047],[171,-0.2736],[172,-0.3617],[173,-0.4641],[174,-0.5761],[176,-0.8092],[178,-1.0224],[179,-1.1309],[180,-1.2422],[181,-1.3622],[182,-1.4782],[183,-1.5783],[184,-1.6659],[185,-1.7381],[186,-1.8008],[187,-1.8302],[188,-1.9128],[189,-2.1739],[190,-2.5485],[191,-2.9906],[192,-3.4917],[193,-4.0556],[194,-4.5993],[195,-5.1084],[196,-5.3814],[197,-5.1115],[198,-4.596],[199,-4.0163],[200,-3.4158],[201,-2.8966],[202,-2.464],[203,-2.172],[204,-1.9774],[205,-1.8642],[206,-1.804],[207,-1.7663],[209,-1.7161],[210,-1.6716],[211,-1.5981],[212,-1.5205],[214,-1.3564],[217,-1.1009],[218,-1.0023],[219,-0.9083],[220,-0.8292],[221,-0.7453],[223,-0.5747],[224,-0.4939],[225,-0.4199],[226,-0.3558],[227,-0.3044],[228,-0.2687],[229,-0.2445],[230,-0.2256],[231,-0.211],[232,-0.2002],[233,-0.1923],[234,-0.1867],[235,-0.1827],[237,-0.1764],[238,-0.1728],[239,-0.1679],[240,-0.161],[241,-0.1564],[242,-0.1564],[243,-0.1533],[244,-0.1533],[245,-0.0845],[246,0.1435],[247,0.5783],[248,1.174],[249,2.0058],[250,2.7534],[251,3.0533],[252,3.111],[253,2.8815],[254,2.4111],[255,1.8208],[256,1.1828],[257,0.6104],[258,0.1017],[259,-0.2995],[260,-0.6164],[261,-0.8505],[262,-1.0224],[263,-1.1429],[264,-1.2421],[265,-1.3621],[266,-1.4781],[267,-1.5783],[268,-1.6659],[269,-1.7381],[271,-1.8604],[272,-1.9128],[273,-1.9562],[274,-1.9892],[275,-2.0099],[276,-2.0175],[278,-2.0122],[280,-1.9968],[282,-1.9721],[284,-1.9388],[287,-1.8746],[290,-1.7956],[293,-1.7069],[294,-1.6717],[295,-1.6234],[296,-1.5626],[297,-1.4844],[299,-1.3122],[300,-1.2212],[302,-1.0124],[303,-0.9213],[305,-0.7483],[307,-0.5786],[308,-0.4981],[309,-0.4243],[310,-0.3596],[311,-0.3066],[312,-0.2676],[313,-0.2503],[314,-0.241],[316,-0.2131],[317,-0.1995],[318,-0.1867],[319,-0.1749],[320,-0.1645],[321,-0.1558],[322,-0.1492],[323,-0.145],[324,-0.145]]
       }
       clavicle_l_ctrl.Scale.X {type=MovieSceneScriptingFloatChannel, default=1, const [1,325]}
       clavicle_l_ctrl.Scale.Y {type=MovieSceneScriptingFloatChannel, default=1, const [1,325]}
       clavicle_l_ctrl.Scale.Z {type=MovieSceneScriptingFloatChannel, default=1, const [1,325]}
       clavicle_r_ctrl.Location.X {type=MovieSceneScriptingFloatChannel, default=0, thinned_from=325, keys [[0,0],[324,0]]}
       clavicle_r_ctrl.Location.Y {type=MovieSceneScriptingFloatChannel, default=0, thinned_from=325, keys [[0,0],[324,0]]}
       clavicle_r_ctrl.Location.Z {type=MovieSceneScriptingFloatChannel, default=0, thinned_from=325, keys [[0,-0.0001],[324,-0.0001]]}
       clavicle_r_ctrl.Rotation.X {
        type=MovieSceneScriptingFloatChannel
        default=0
        thinned_from=325
        keys [[0,-0.0085],[1,-0.0102],[2,-0.0151],[3,-0.0225],[4,-0.032],[5,-0.0431],[6,-0.0553],[8,-0.0806],[9,-0.0928],[10,-0.1041],[11,-0.1158],[12,-0.129],[13,-0.1429],[14,-0.1562],[15,-0.168],[16,-0.1774],[17,-0.1848],[18,-0.1915],[19,-0.1974],[20,-0.2024],[21,-0.2065],[22,-0.2096],[23,-0.2115],[24,-0.2121],[26,-0.212],[29,-0.2095],[31,-0.2068],[33,-0.203],[35,-0.198],[37,-0.1918],[39,-0.1842],[41,-0.1756],[42,-0.17],[43,-0.162],[44,-0.1516],[45,-0.1382],[46,-0.1232],[47,-0.1078],[48,-0.0928],[49,-0.0789],[50,-0.067],[51,-0.0589],[52,-0.0525],[53,-0.0467],[55,-0.0362],[56,-0.0313],[57,-0.0269],[58,-0.0231],[59,-0.0198],[60,-0.0177],[65,-0.014],[73,-0.0096],[74,-0.0093],[76,-0.0093],[77,-0.0089],[85,-0.0089],[86,-0.0152],[87,-0.0228],[88,-0.0323],[89,-0.0434],[90,-0.0556],[92,-0.081],[93,-0.0934],[94,-0.1047],[95,-0.1164],[96,-0.1298],[97,-0.1436],[98,-0.1568],[99,-0.1688],[100,-0.1779],[101,-0.1855],[102,-0.1921],[103,-0.1979],[104,-0.2028],[105,-0.2067],[106,-0.2098],[107,-0.2115],[108,-0.212],[110,-0.2119],[112,-0.211],[113,-0.211],[114,-0.2093],[116,-0.2068],[118,-0.2032],[120,-0.1984],[122,-0.1922],[123,-0.1886],[124,-0.1846],[125,-0.1802],[126,-0.1764],[127,-0.17],[128,-0.1577],[129,-0.1412],[130,-0.1199],[131,-0.0973],[132,-0.0793],[133,-0.068],[134,-0.0582],[135,-0.0512],[136,-0.0463],[137,-0.0419],[138,-0.0379],[139,-0.0341],[141,-0.0274],[144,-0.0176],[145,-0.0155],[147,-0.0133],[148,-0.0123],[150,-0.011],[152,-0.0103],[157,-0.0094],[158,-0.0094],[159,-0.009],[160,-0.009],[161,-0.0088],[169,-0.0088],[170,-0.0152],[171,-0.0227],[172,-0.0323],[173,-0.0435],[174,-0.0557],[176,-0.0813],[178,-0.1049],[179,-0.1173],[180,-0.1301],[181,-0.144],[182,-0.1574],[183,-0.1689],[184,-0.1786],[185,-0.1862],[186,-0.1925],[187,-0.1982],[188,-0.203],[189,-0.2069],[190,-0.2098],[191,-0.2115],[192,-0.2121],[195,-0.2112],[197,-0.2095],[199,-0.2068],[201,-0.2031],[203,-0.1981],[205,-0.1918],[206,-0.1882],[209,-0.1882],[210,-0.17],[211,-0.17],[212,-0.1457],[213,-0.1325],[215,-0.1054],[216,-0.092],[217,-0.0793],[218,-0.067],[219,-0.0582],[222,-0.0414],[223,-0.0362],[224,-0.0313],[225,-0.0268],[226,-0.023],[227,-0.0198],[228,-0.0175],[229,-0.0158],[230,-0.0144],[231,-0.0134],[233,-0.0121],[239,-0.0103],[240,-0.0098],[241,-0.0098],[242,-0.0094],[244,-0.0094],[245,-0.0089],[252,-0.0087],[253,-0.0105],[254,-0.0154],[255,-0.0229],[256,-0.0326],[257,-0.0438],[258,-0.0561],[260,-0.0819],[261,-0.0943],[262,-0.1056],[263,-0.1175],[264,-0.1309],[265,-0.1449],[266,-0.1583],[267,-0.1702],[268,-0.1795],[269,-0.1868],[270,-0.1932],[271,-0.1987],[272,-0.2034],[273,-0.2071],[274,-0.2099],[275,-0.2115],[276,-0.2121],[279,-0.2113],[281,-0.2097],[283,-0.2072],[285,-0.2035],[287,-0.1986],[288,-0.1957],[290,-0.1887],[291,-0.1847],[292,-0.1802],[293,-0.1753],[294,-0.17],[295,-0.1613],[296,-0.1474],[297,-0.1304],[298,-0.1121],[299,-0.0944],[300,-0.0793],[301,-0.0673],[302,-0.058],[303,-0.0514],[304,-0.0464],[305,-0.0416],[306,-0.0371],[307,-0.0329],[308,-0.029],[309,-0.0255],[310,-0.0224],[311,-0.0197],[312,-0.0174],[319,-0.0108],[321,-0.0094],[323,-0.0086],[324,-0.0086]]
       }
       clavicle_r_ctrl.Rotation.Y {
        type=MovieSceneScriptingFloatChannel
        default=0
        thinned_from=325
        keys [[0,-10.4771],[1,-10.4604],[2,-10.4132],[3,-10.34],[5,-10.1334],[10,-9.4727],[11,-9.3268],[15,-8.5966],[16,-8.4481],[19,-8.0946],[22,-7.8093],[24,-7.6538],[25,-7.6175],[26,-7.6175],[33,-7.5142],[42,-7.4942],[43,-7.5043],[45,-7.5711],[47,-7.6796],[49,-7.8256],[50,-7.9311],[51,-8.0917],[57,-9.2525],[59,-9.5803],[60,-9.6929],[69,-10.2534],[73,-10.4255],[85,-10.4618],[86,-10.4119],[88,-10.2431],[94,-9.47],[95,-9.3261],[99,-8.5966],[102,-8.185],[104,-7.9438],[105,-7.8403],[106,-7.755],[107,-7.6897],[110,-7.6065],[112,-7.5728],[113,-7.5728],[114,-7.5463],[127,-7.497],[128,-7.5057],[129,-7.5312],[131,-7.6309],[132,-7.704],[134,-7.983],[144,-9.6929],[145,-9.8063],[149,-10.1147],[153,-10.3218],[157,-10.4255],[158,-10.4255],[159,-10.4487],[169,-10.4635],[170,-10.4128],[171,-10.3397],[173,-10.1332],[178,-9.4728],[180,-9.1523],[182,-8.7732],[184,-8.4481],[187,-8.0607],[189,-7.8404],[190,-7.754],[191,-7.6925],[192,-7.649],[200,-7.5208],[211,-7.4942],[213,-7.5779],[217,-7.8362],[218,-7.9525],[220,-8.2951],[225,-9.2508],[226,-9.4196],[227,-9.5688],[228,-9.6929],[230,-9.8904],[233,-10.1146],[240,-10.4255],[244,-10.446],[252,-10.4742],[253,-10.4575],[254,-10.4105],[255,-10.3374],[257,-10.1312],[262,-9.4717],[263,-9.3261],[267,-8.5969],[268,-8.4481],[271,-8.0596],[273,-7.8383],[274,-7.7515],[275,-7.6862],[276,-7.6466],[284,-7.5204],[294,-7.4943],[296,-7.5221],[298,-7.6014],[300,-7.7259],[301,-7.8308],[303,-8.1466],[305,-8.5017],[309,-9.2533],[311,-9.5684],[312,-9.6929],[319,-10.267],[321,-10.3939],[323,-10.467],[324,-10.467]]
       }
       clavicle_r_ctrl.Rotation.Z {
        type=MovieSceneScriptingFloatChannel
        default=0
        thinned_from=325
        keys [[0,-0.0793],[1,-0.0895],[2,-0.1183],[3,-0.1634],[4,-0.2222],[5,-0.2923],[6,-0.3712],[7,-0.4565],[8,-0.5456],[10,-0.7258],[11,-0.8275],[12,-0.9508],[13,-1.0872],[14,-1.2282],[15,-1.3655],[17,-1.6166],[18,-1.756],[19,-1.8995],[20,-2.0378],[21,-2.1616],[22,-2.2615],[23,-2.3283],[24,-2.3526],[26,-2.351],[29,-2.3135],[31,-2.2774],[32,-2.2554],[33,-2.2238],[34,-2.2038],[35,-2.3032],[36,-2.5315],[37,-2.8744],[38,-3.2339],[39,-3.537],[40,-3.6947],[41,-3.5996],[42,-3.4095],[43,-3.1867],[45,-2.6785],[46,-2.4162],[47,-2.1489],[48,-1.8926],[49,-1.6715],[50,-1.4611],[51,-1.2266],[52,-1.0222],[53,-0.9097],[54,-0.8316],[55,-0.7465],[56,-0.667],[57,-0.5906],[58,-0.5209],[59,-0.4563],[60,-0.4101],[61,-0.3834],[62,-0.3576],[63,-0.3326],[64,-0.3084],[65,-0.2849],[66,-0.262],[68,-0.2178],[70,-0.1751],[73,-0.1124],[74,-0.1015],[76,-0.1015],[77,-0.0889],[85,-0.0889],[86,-0.0746],[87,-0.0493],[88,0.0005],[89,0.073],[90,0.1613],[91,0.2861],[92,0.5009],[93,0.7],[94,0.7755],[95,0.7883],[96,0.6921],[97,0.4486],[98,0.0644],[99,-0.3901],[100,-0.8558],[101,-1.2917],[102,-1.6284],[103,-1.8915],[104,-2.0774],[105,-2.2038],[106,-2.2841],[107,-2.3275],[108,-2.3481],[109,-2.351],[111,-2.3388],[112,-2.3283],[113,-2.3283],[114,-2.2993],[116,-2.2604],[119,-2.1853],[122,-2.0931],[126,-1.9539],[127,-1.9099],[128,-1.8402],[129,-1.7504],[130,-1.6358],[131,-1.5129],[132,-1.4037],[133,-1.3062],[134,-1.2114],[135,-1.1235],[138,-0.8755],[140,-0.7135],[141,-0.6343],[142,-0.5567],[143,-0.4808],[144,-0.4071],[145,-0.3569],[146,-0.3246],[147,-0.29],[148,-0.2594],[149,-0.2324],[150,-0.2089],[151,-0.1884],[152,-0.1706],[153,-0.1552],[154,-0.1419],[155,-0.1302],[156,-0.1201],[157,-0.111],[158,-0.111],[159,-0.0968],[160,-0.0968],[161,-0.0877],[169,-0.0877],[170,-0.1285],[171,-0.1844],[172,-0.2562],[173,-0.3406],[174,-0.4337],[175,-0.532],[176,-0.6319],[178,-0.822],[179,-0.9208],[180,-1.0251],[181,-1.1404],[183,-1.377],[184,-1.4904],[185,-1.5992],[186,-1.6891],[187,-1.7466],[188,-1.7693],[189,-1.7398],[190,-1.5333],[191,-0.6928],[192,0.8303],[193,1.9344],[194,2.4051],[195,2.2726],[197,1.6441],[198,1.2989],[199,0.9618],[200,0.7385],[201,0.5125],[202,0.3243],[203,0.1673],[204,0.042],[205,-0.0456],[206,-0.0808],[245,-0.0808],[246,-0.184],[247,-0.6527],[248,-1.3146],[249,-2.0597],[250,-2.8799],[251,-3.609],[252,-3.8372],[253,-3.4511],[254,-2.6345],[255,-1.731],[256,-1.0611],[257,-0.7478],[258,-0.6142],[259,-0.5736],[260,-0.6328],[261,-0.7306],[262,-0.8228],[263,-0.9183],[264,-1.0257],[265,-1.1409],[267,-1.3777],[268,-1.4908],[269,-1.6115],[270,-1.7484],[272,-2.031],[273,-2.1568],[274,-2.2589],[275,-2.3275],[276,-2.3526],[278,-2.3458],[280,-2.326],[282,-2.2943],[284,-2.2517],[286,-2.1991],[289,-2.1039],[292,-1.992],[294,-1.9099],[295,-1.8509],[296,-1.7647],[297,-1.6607],[299,-1.4372],[300,-1.3364],[301,-1.2457],[302,-1.1592],[304,-0.9979],[307,-0.746],[308,-0.6645],[309,-0.5873],[310,-0.5162],[311,-0.4532],[312,-0.4],[313,-0.3551],[314,-0.3125],[315,-0.2725],[316,-0.2355],[317,-0.2017],[318,-0.1713],[319,-0.1447],[320,-0.1222],[321,-0.104],[322,-0.0906],[323,-0.0822],[324,-0.0822]]
       }
       clavicle_r_ctrl.Scale.X {type=MovieSceneScriptingFloatChannel, default=1, const [1,325]}
       clavicle_r_ctrl.Scale.Y {type=MovieSceneScriptingFloatChannel, default=1, const [1,325]}
       clavicle_r_ctrl.Scale.Z {type=MovieSceneScriptingFloatChannel, default=1, const [1,325]}
       neck_01_ctrl.Location.X {type=MovieSceneScriptingFloatChannel, default=0, const [0,325]}
       neck_01_ctrl.Location.Y {type=MovieSceneScriptingFloatChannel, default=0, const [0,325]}
       neck_01_ctrl.Location.Z {type=MovieSceneScriptingFloatChannel, default=0, const [0,325]}
       neck_01_ctrl.Rotation.X {type=MovieSceneScriptingFloatChannel, default=0, const [0,325]}
       neck_01_ctrl.Rotation.Y {type=MovieSceneScriptingFloatChannel, default=0, const [0,325]}
       neck_01_ctrl.Rotation.Z {type=MovieSceneScriptingFloatChannel, default=0, const [0,325]}
       neck_01_ctrl.Scale.X {type=MovieSceneScriptingFloatChannel, default=1, const [1,325]}
       neck_01_ctrl.Scale.Y {type=MovieSceneScriptingFloatChannel, default=1, const [1,325]}
       neck_01_ctrl.Scale.Z {type=MovieSceneScriptingFloatChannel, default=1, const [1,325]}
       neck_02_ctrl.Location.X {type=MovieSceneScriptingFloatChannel, default=0, thinned_from=325, keys [[0,0],[324,0]]}
       neck_02_ctrl.Location.Y {
        type=MovieSceneScriptingFloatChannel
        default=0
        thinned_from=325
        keys [[0,0.1981],[1,0.1983],[3,0.1999],[6,0.2044],[11,0.2137],[13,0.2166],[15,0.2185],[16,0.2188],[17,0.2187],[18,0.2181],[19,0.2169],[20,0.215],[21,0.2125],[22,0.2093],[24,0.2014],[25,0.1988],[26,0.1971],[28,0.1954],[30,0.1949],[33,0.1955],[37,0.1986],[48,0.2128],[49,0.2137],[50,0.2138],[51,0.2133],[52,0.2122],[54,0.2086],[57,0.2008],[61,0.1897],[63,0.1853],[64,0.1837],[65,0.1825],[66,0.1819],[74,0.185],[77,0.1872],[80,0.1907],[82,0.194],[90,0.2124],[93,0.2181],[95,0.2208],[97,0.2224],[99,0.2225],[101,0.2211],[102,0.2197],[103,0.2178],[104,0.2154],[105,0.2124],[106,0.2089],[107,0.2047],[108,0.201],[109,0.1985],[110,0.197],[112,0.1955],[114,0.1951],[116,0.1955],[119,0.1976],[132,0.2153],[133,0.2162],[134,0.2162],[135,0.2154],[136,0.214],[137,0.212],[139,0.2069],[144,0.1913],[146,0.1861],[147,0.1842],[148,0.1828],[149,0.182],[150,0.182],[158,0.1854],[161,0.1876],[164,0.1911],[166,0.1943],[175,0.2142],[178,0.2192],[180,0.2214],[182,0.2223],[184,0.2217],[185,0.2208],[186,0.2194],[187,0.2175],[188,0.2151],[189,0.2121],[190,0.2086],[191,0.2044],[192,0.2007],[193,0.1984],[194,0.197],[196,0.1955],[198,0.1951],[200,0.1955],[203,0.1976],[216,0.2153],[217,0.2162],[218,0.2162],[219,0.2155],[220,0.2141],[221,0.2121],[223,0.2069],[228,0.1913],[230,0.1861],[231,0.1842],[232,0.1828],[233,0.182],[234,0.182],[242,0.1854],[245,0.1876],[248,0.1911],[250,0.1944],[258,0.2123],[261,0.2176],[263,0.2201],[265,0.2215],[267,0.2216],[269,0.22],[270,0.2186],[271,0.2166],[272,0.2142],[273,0.2112],[274,0.2077],[275,0.2035],[276,0.2],[277,0.1979],[279,0.1961],[281,0.1951],[284,0.195],[287,0.1965],[300,0.2108],[301,0.2115],[302,0.2114],[303,0.2105],[304,0.209],[306,0.2047],[311,0.1902],[313,0.1852],[314,0.1833],[315,0.182],[316,0.1812],[317,0.1812],[318,0.182],[320,0.1853],[321,0.1876],[322,0.1904],[323,0.1938],[324,0.1981]]
       }
       neck_02_ctrl.Location.Z {
        type=MovieSceneScriptingFloatChannel
        default=0
        thinned_from=325
        keys [[0,-0.6721],[1,-0.6742],[2,-0.6802],[3,-0.69],[4,-0.7032],[5,-0.7194],[6,-0.7384],[7,-0.76],[8,-0.7837],[10,-0.8366],[18,-1.0767],[20,-1.1319],[22,-1.1806],[23,-1.2017],[24,-1.2181],[25,-1.2283],[26,-1.2336],[27,-1.235],[29,-1.231],[31,-1.2201],[33,-1.203],[35,-1.1805],[38,-1.1382],[42,-1.0698],[54,-0.8357],[57,-0.7826],[59,-0.7514],[61,-0.7247],[62,-0.7134],[63,-0.7034],[64,-0.695],[65,-0.6884],[66,-0.6835],[72,-0.6733],[84,-0.6724],[85,-0.675],[86,-0.6815],[87,-0.6916],[88,-0.7051],[89,-0.7216],[90,-0.7408],[91,-0.7625],[92,-0.7863],[94,-0.8393],[102,-1.079],[104,-1.1342],[106,-1.1827],[107,-1.2037],[108,-1.2198],[109,-1.2294],[110,-1.2339],[112,-1.2339],[114,-1.2259],[116,-1.2111],[118,-1.1903],[121,-1.1495],[124,-1.0995],[138,-0.8227],[140,-0.7866],[142,-0.7542],[144,-0.7263],[145,-0.7144],[146,-0.7041],[147,-0.6954],[148,-0.6885],[149,-0.6835],[155,-0.6735],[168,-0.6722],[169,-0.6748],[170,-0.6813],[171,-0.6915],[172,-0.705],[173,-0.7215],[174,-0.7408],[175,-0.7626],[176,-0.7865],[178,-0.8397],[186,-1.0801],[188,-1.1354],[190,-1.184],[191,-1.205],[192,-1.2209],[193,-1.23],[194,-1.2341],[196,-1.2339],[198,-1.2259],[200,-1.2111],[202,-1.1903],[205,-1.1494],[208,-1.0994],[222,-0.8225],[224,-0.7864],[226,-0.7539],[228,-0.7261],[229,-0.7143],[230,-0.7039],[231,-0.6953],[232,-0.6884],[233,-0.6834],[239,-0.6736],[252,-0.6725],[253,-0.6751],[254,-0.6817],[255,-0.692],[256,-0.7057],[257,-0.7224],[258,-0.7419],[259,-0.7639],[260,-0.7881],[262,-0.8418],[270,-1.0841],[272,-1.1395],[274,-1.1881],[275,-1.2091],[276,-1.2242],[277,-1.232],[278,-1.2347],[280,-1.234],[282,-1.2267],[284,-1.2133],[286,-1.1942],[289,-1.1568],[292,-1.1106],[298,-1.0013],[307,-0.8259],[310,-0.7735],[312,-0.7427],[314,-0.7164],[315,-0.7051],[316,-0.6953],[317,-0.6871],[318,-0.6805],[319,-0.6757],[320,-0.6728],[322,-0.6708],[324,-0.6721]]
       }
       neck_02_ctrl.Rotation.X {
        type=MovieSceneScriptingFloatChannel
        default=0
        thinned_from=325
        keys [[0,2.3241],[61,2.2733],[108,2.3113],[144,2.2719],[192,2.3113],[227,2.2718],[276,2.3113],[310,2.2708],[323,2.3234],[324,2.3241]]
       }
       neck_02_ctrl.Rotation.Y {type=MovieSceneScriptingFloatChannel, default=0, thinned_from=325, keys [[0,0],[324,0]]}
       neck_02_ctrl.Rotation.Z {type=MovieSceneScriptingFloatChannel, default=0, thinned_from=325, keys [[0,0],[324,0]]}
       neck_02_ctrl.Scale.X {type=MovieSceneScriptingFloatChannel, default=1, const [1,325]}
       neck_02_ctrl.Scale.Y {type=MovieSceneScriptingFloatChannel, default=1, const [1,325]}
       neck_02_ctrl.Scale.Z {type=MovieSceneScriptingFloatChannel, default=1, const [1,325]}
       head_ik_ctrl.Location.X {
        type=MovieSceneScriptingFloatChannel
        default=0
        thinned_from=325
        keys [[0,0],[28,0],[29,0.0001],[30,0.0008],[31,0.0028],[32,0.0085],[33,0.0172],[34,0.0298],[35,0.0399],[36,0.0432],[37,0.0453],[39,0.0477],[42,0.0516],[43,0.0526],[44,0.0532],[45,0.0534],[46,0.0534],[48,0.0527],[50,0.0514],[52,0.0496],[56,0.0448],[58,0.0428],[60,0.0414],[63,0.04],[66,0.0393],[83,0.0397],[85,0.0401],[86,0.0405],[87,0.0405],[88,0.0379],[89,0.0332],[90,0.0275],[91,0.0211],[92,0.0138],[93,0.0078],[94,0.0046],[95,0.002],[96,-0.0003],[97,-0.0019],[98,-0.003],[100,-0.0044],[103,-0.0056],[106,-0.0059],[110,-0.0054],[113,-0.0043],[118,-0.0017],[120,-0.001],[181,-0.001],[182,-0.0009],[186,0.0006],[188,0.0007],[189,0.0002],[190,-0.0009],[191,-0.0031],[192,-0.0081],[193,-0.0153],[194,-0.0217],[195,-0.0265],[196,-0.0304],[197,-0.0329],[198,-0.0344],[199,-0.0355],[200,-0.0362],[203,-0.0376],[205,-0.0379],[210,-0.037],[214,-0.0349],[220,-0.0328],[233,-0.0316],[244,-0.0325],[245,-0.0325],[246,-0.032],[247,-0.0309],[248,-0.029],[249,-0.0251],[250,-0.0191],[251,-0.0124],[252,-0.0067],[253,-0.0021],[254,0.002],[255,0.0055],[256,0.008],[257,0.0096],[258,0.0106],[259,0.0111],[261,0.0113],[262,0.011],[264,0.0098],[267,0.0072],[270,0.0041],[273,0.002],[275,0.0011],[279,0.0003],[324,0]]
       }
       head_ik_ctrl.Location.Y {
        type=MovieSceneScriptingFloatChannel
        default=0
        thinned_from=325
        keys [[0,0.1727],[1,0.1725],[3,0.171],[7,0.1667],[8,0.166],[9,0.166],[10,0.1668],[11,0.1683],[15,0.1764],[16,0.1776],[17,0.1779],[21,0.1777],[22,0.177],[24,0.1741],[26,0.1739],[29,0.1725],[31,0.1728],[33,0.1741],[43,0.1853],[44,0.1857],[45,0.1852],[46,0.1841],[48,0.1806],[49,0.1783],[52,0.1698],[53,0.1671],[54,0.165],[55,0.1636],[56,0.1629],[64,0.1592],[66,0.1588],[74,0.1602],[77,0.1619],[79,0.1636],[81,0.166],[83,0.1693],[85,0.173],[86,0.1739],[88,0.1742],[92,0.173],[93,0.173],[94,0.1736],[96,0.1759],[98,0.1787],[99,0.1797],[100,0.1801],[101,0.1795],[103,0.1781],[105,0.1771],[106,0.1761],[107,0.1745],[108,0.1733],[110,0.1737],[111,0.1735],[112,0.1728],[113,0.1726],[115,0.1732],[117,0.1749],[120,0.179],[127,0.191],[128,0.192],[129,0.1919],[130,0.1911],[132,0.1882],[133,0.1861],[137,0.1743],[138,0.1717],[139,0.1698],[140,0.1686],[147,0.1624],[149,0.1614],[156,0.1617],[160,0.1632],[163,0.1656],[165,0.168],[167,0.1711],[169,0.1747],[170,0.1757],[171,0.1761],[173,0.1757],[176,0.1745],[177,0.1749],[178,0.1763],[179,0.1784],[181,0.1834],[182,0.1856],[183,0.1868],[184,0.1867],[188,0.183],[189,0.1815],[190,0.1795],[191,0.1768],[192,0.1745],[193,0.1734],[195,0.1717],[196,0.1704],[197,0.1697],[198,0.1696],[199,0.1699],[201,0.1715],[204,0.1756],[211,0.1878],[212,0.1886],[213,0.1882],[214,0.1868],[216,0.1824],[217,0.1796],[220,0.1699],[221,0.1672],[222,0.1652],[223,0.1643],[231,0.1587],[233,0.158],[241,0.159],[244,0.1604],[246,0.1621],[248,0.1647],[252,0.1731],[253,0.1748],[254,0.1757],[255,0.1759],[257,0.1749],[260,0.1726],[261,0.1724],[262,0.1727],[264,0.1746],[266,0.177],[267,0.1778],[268,0.1781],[271,0.1758],[273,0.1748],[274,0.1739],[275,0.1723],[276,0.1716],[278,0.1735],[279,0.1735],[281,0.1725],[283,0.1726],[285,0.1737],[295,0.1848],[296,0.1853],[297,0.1849],[298,0.1839],[300,0.1807],[301,0.1785],[302,0.1756],[305,0.1657],[306,0.1626],[307,0.1601],[308,0.1583],[310,0.1555],[312,0.1537],[313,0.1533],[314,0.1532],[315,0.1535],[316,0.1543],[317,0.1556],[320,0.1613],[321,0.1635],[322,0.166],[323,0.169],[324,0.1727]]
       }
       head_ik_ctrl.Location.Z {
        type=MovieSceneScriptingFloatChannel
        default=0
        thinned_from=325
        keys [[0,-0.8146],[1,-0.8166],[2,-0.8225],[3,-0.8318],[4,-0.8444],[5,-0.86],[6,-0.8781],[7,-0.8986],[9,-0.9452],[12,-1.025],[18,-1.1949],[20,-1.2461],[22,-1.2905],[23,-1.3095],[24,-1.3242],[25,-1.3331],[26,-1.3376],[28,-1.3373],[30,-1.3296],[32,-1.3157],[35,-1.2844],[39,-1.2285],[57,-0.9129],[59,-0.8823],[61,-0.8559],[63,-0.835],[64,-0.8269],[65,-0.8206],[66,-0.8161],[75,-0.8083],[84,-0.8143],[85,-0.8174],[86,-0.8242],[87,-0.8344],[88,-0.8477],[89,-0.8639],[90,-0.8825],[91,-0.9032],[93,-0.95],[96,-1.0293],[102,-1.1976],[104,-1.2483],[106,-1.2925],[107,-1.3115],[108,-1.3258],[109,-1.3341],[110,-1.3379],[112,-1.3372],[114,-1.3292],[116,-1.3147],[119,-1.2826],[122,-1.2404],[140,-0.9179],[142,-0.8858],[144,-0.8581],[146,-0.8361],[147,-0.8276],[148,-0.821],[149,-0.8164],[159,-0.8087],[168,-0.8148],[169,-0.818],[170,-0.8247],[171,-0.8349],[172,-0.848],[173,-0.864],[174,-0.8825],[175,-0.9031],[177,-0.9498],[180,-1.0288],[186,-1.1975],[188,-1.2486],[190,-1.2932],[191,-1.3123],[192,-1.3266],[193,-1.3346],[194,-1.338],[196,-1.3372],[198,-1.3292],[200,-1.3147],[203,-1.2825],[206,-1.2402],[224,-0.918],[226,-0.8859],[228,-0.8582],[230,-0.8363],[231,-0.8278],[232,-0.8211],[233,-0.8165],[243,-0.8088],[252,-0.8152],[253,-0.8183],[254,-0.8252],[255,-0.8355],[256,-0.8488],[257,-0.865],[258,-0.8837],[259,-0.9046],[261,-0.9518],[264,-1.0322],[270,-1.2022],[272,-1.2533],[274,-1.2976],[275,-1.3165],[276,-1.33],[277,-1.3366],[279,-1.3386],[281,-1.3344],[283,-1.3241],[286,-1.2983],[289,-1.2623],[293,-1.2024],[309,-0.9199],[311,-0.8881],[313,-0.8604],[315,-0.8379],[316,-0.8289],[317,-0.8216],[318,-0.8161],[319,-0.8125],[321,-0.8101],[324,-0.8146]]
       }
       head_ik_ctrl.Rotation.X {
        type=MovieSceneScriptingFloatChannel
        default=0
        thinned_from=325
        keys [[0,9.8055],[1,9.7918],[3,9.6995],[4,9.6203],[6,9.4038],[8,9.1179],[10,8.7667],[16,7.5686],[18,7.1978],[22,6.357],[23,6.1559],[24,5.9715],[25,5.8103],[26,5.6784],[27,5.5822],[28,5.5746],[29,5.7201],[30,6.0299],[31,6.5745],[32,7.2905],[33,7.9844],[34,8.4741],[35,8.6753],[36,8.7517],[40,8.5977],[41,8.5802],[42,8.5965],[43,8.642],[44,8.7154],[45,8.8254],[46,8.9663],[47,9.1322],[49,9.5163],[53,10.3317],[54,10.5105],[55,10.6677],[63,11.5481],[70,12.1264],[79,12.5678],[81,12.727],[83,12.9571],[85,13.2749],[87,13.6565],[88,13.8779],[89,13.9343],[90,13.7912],[91,13.5716],[92,13.2618],[93,12.8381],[95,11.8245],[97,10.7182],[98,10.3559],[103,9.3785],[106,8.7353],[108,8.347],[109,8.185],[110,8.0531],[111,7.958],[115,7.6853],[118,7.5385],[120,7.4831],[122,7.471],[124,7.5096],[125,7.5504],[126,7.6068],[127,7.6799],[128,7.7827],[129,7.9234],[130,8.0956],[131,8.2933],[133,8.7399],[136,9.4446],[138,9.8646],[139,10.0408],[144,10.6724],[149,11.1831],[158,11.8405],[164,12.1257],[168,12.2216],[169,12.2185],[170,12.1841],[171,12.121],[172,12.0314],[174,11.7819],[176,11.4537],[178,11.0518],[180,10.6084],[181,10.4163],[182,10.2824],[183,10.2081],[184,10.2005],[185,10.2442],[186,10.501],[187,10.8943],[188,11.0503],[189,10.9156],[190,10.7213],[191,10.4888],[193,9.9488],[194,9.6465],[195,9.3194],[196,9.0248],[197,8.8041],[198,8.6608],[199,8.5566],[200,8.483],[203,8.3588],[206,8.2952],[208,8.2939],[210,8.3358],[211,8.3772],[212,8.4444],[213,8.5444],[214,8.6729],[216,8.9933],[220,9.7241],[221,9.8892],[222,10.0346],[229,10.7436],[234,11.1367],[244,11.7052],[246,11.7845],[247,11.9311],[249,12.4814],[250,12.8184],[251,13.015],[252,12.7923],[254,12.0685],[255,11.6428],[258,10.223],[259,9.7663],[260,9.3555],[261,9.0125],[262,8.7621],[268,7.5358],[270,7.1575],[274,6.3098],[275,6.1087],[276,5.9251],[277,5.7653],[278,5.636],[279,5.5435],[282,5.3415],[284,5.2303],[286,5.1466],[288,5.0976],[289,5.0884],[290,5.0907],[291,5.1053],[292,5.1331],[293,5.1751],[294,5.2322],[295,5.3054],[296,5.4049],[297,5.5372],[298,5.6974],[299,5.8805],[300,6.0816],[305,7.1838],[306,7.3889],[307,7.5775],[312,8.3442],[315,8.7504],[319,9.2181],[322,9.6563],[323,9.7587],[324,9.8041]]
       }
       head_ik_ctrl.Rotation.Y {
        type=MovieSceneScriptingFloatChannel
        default=0
        thinned_from=325
        keys [[0,-5.1409],[3,-5.1184],[6,-5.0565],[10,-4.925],[16,-4.6911],[23,-4.464],[25,-4.4213],[26,-4.4129],[28,-4.4266],[29,-4.3945],[30,-4.266],[31,-3.9987],[32,-2.95],[33,-1.3704],[34,0.2484],[35,1.4421],[36,2.267],[37,2.7359],[38,3.1534],[39,3.5059],[40,3.7804],[41,3.962],[42,4.0297],[43,4.0371],[44,4.0537],[46,4.1128],[53,4.4253],[55,4.4943],[61,4.5932],[75,4.7009],[84,4.7332],[85,4.7083],[86,4.5259],[87,3.6234],[88,2.115],[89,0.4026],[90,-1.3191],[91,-2.9119],[92,-4.0167],[93,-4.4545],[94,-4.7069],[95,-4.8321],[96,-4.8646],[97,-4.8416],[106,-4.5663],[109,-4.505],[111,-4.4916],[122,-4.5552],[127,-4.6228],[130,-4.6939],[136,-4.8802],[139,-4.9437],[169,-5.0557],[175,-4.9997],[182,-4.8663],[183,-4.8335],[184,-4.7768],[185,-4.8773],[186,-5.2448],[187,-5.7322],[188,-6.2011],[189,-6.8146],[190,-7.6254],[191,-8.3515],[192,-8.8066],[193,-9.0523],[194,-9.1734],[195,-9.2227],[196,-9.2057],[198,-9.1462],[201,-9.1329],[205,-9.2077],[209,-9.3784],[211,-9.501],[212,-9.5845],[213,-9.6954],[215,-9.9731],[219,-10.597],[221,-10.8554],[222,-10.9525],[232,-11.411],[245,-11.7143],[247,-11.8054],[249,-11.9827],[250,-11.7985],[251,-10.8662],[252,-8.491],[253,-6.4744],[254,-5.9173],[255,-5.5581],[256,-5.3158],[257,-5.1616],[258,-5.0816],[259,-5.0479],[261,-5.0217],[263,-4.9701],[275,-4.4985],[277,-4.4434],[278,-4.4266],[287,-4.4387],[294,-4.5115],[296,-4.5469],[299,-4.635],[306,-4.8902],[315,-5.0699],[317,-5.1065],[324,-5.1406]]
       }
       head_ik_ctrl.Rotation.Z {
        type=MovieSceneScriptingFloatChannel
        default=0
        thinned_from=325
        keys [[0,11.2508],[16,11.4527],[22,11.7514],[24,11.8047],[25,11.7953],[26,11.717],[27,11.5191],[28,11.1521],[29,10.5417],[30,9.6304],[31,8.095],[32,5.6777],[33,2.4898],[34,-2.1866],[35,-7.4686],[36,-10.6498],[37,-11.8022],[38,-12.7594],[39,-13.5],[40,-14.0362],[41,-14.3356],[42,-14.4807],[43,-14.5436],[46,-14.5224],[47,-14.4826],[48,-14.3937],[49,-14.2649],[51,-13.9246],[55,-13.1739],[58,-12.729],[61,-12.3796],[64,-12.1149],[68,-11.874],[73,-11.7158],[81,-11.6606],[82,-11.8095],[84,-12.4428],[85,-12.5413],[86,-12.3632],[87,-11.4375],[88,-7.7729],[89,-2.9192],[90,0.5221],[91,2.5154],[92,4.1163],[93,5.2298],[94,5.9961],[95,6.6272],[96,7.1324],[97,7.5214],[98,7.82],[100,8.3619],[101,8.5945],[102,8.795],[103,8.963],[104,9.0961],[106,9.3007],[108,9.4562],[110,9.5651],[113,9.6536],[118,9.6889],[120,9.6578],[122,9.5763],[124,9.451],[127,9.1954],[130,8.8692],[132,8.6226],[133,8.4762],[134,8.2972],[136,7.877],[138,7.4336],[139,7.2256],[140,7.0389],[143,6.546],[145,6.2526],[147,5.9937],[149,5.7748],[150,5.682],[151,5.6012],[152,5.5331],[153,5.4783],[154,5.4376],[155,5.4117],[170,5.3899],[177,5.473],[181,5.5481],[182,5.6784],[183,6.0831],[184,6.8153],[185,7.9108],[186,9.4051],[187,12.3057],[188,17.423],[189,24.6682],[190,32.6404],[191,39.3372],[192,43.5573],[193,45.58],[194,46.5177],[195,47.3357],[196,48.0211],[197,48.5621],[199,49.3082],[201,49.7519],[204,50.0099],[207,49.9767],[212,49.1625],[218,47.7643],[223,47.1363],[243,46.8475],[244,46.9827],[245,47.3287],[246,47.8663],[247,48.6083],[248,49.7381],[249,50.5604],[250,48.3161],[251,42.2438],[252,34.0503],[253,24.7639],[254,17.4513],[255,13.6724],[256,11.4787],[257,10.3649],[258,9.7416],[259,9.3286],[260,9.0842],[261,8.9669],[262,8.9344],[263,8.9444],[264,9.0028],[265,9.1287],[266,9.2991],[269,9.8789],[272,10.5626],[273,10.7742],[274,10.9537],[276,11.2521],[277,11.3757],[278,11.4724],[279,11.5344],[324,11.2509]]
       }
       head_ik_ctrl.Scale.X {type=MovieSceneScriptingFloatChannel, default=1, const [1,325]}
       head_ik_ctrl.Scale.Y {type=MovieSceneScriptingFloatChannel, default=1, const [1,325]}
       head_ik_ctrl.Scale.Z {type=MovieSceneScriptingFloatChannel, default=1, const [1,325]}
       chest_ctrl.Location.X {
        type=MovieSceneScriptingFloatChannel
        default=0
        thinned_from=325
        keys [[0,0],[29,-0.0001],[30,-0.0007],[31,-0.0022],[32,-0.0064],[33,-0.0129],[34,-0.0221],[35,-0.0295],[36,-0.0321],[37,-0.0336],[38,-0.0339],[39,-0.0335],[41,-0.0318],[49,-0.0246],[52,-0.0225],[54,-0.0216],[69,-0.0183],[76,-0.0182],[84,-0.0197],[86,-0.0201],[87,-0.0196],[88,-0.0171],[89,-0.0134],[92,-0.0013],[93,0.001],[94,0.0011],[95,0.0006],[97,-0.0009],[99,-0.0017],[104,-0.0024],[113,-0.0016],[120,-0.0004],[182,-0.0004],[187,0.0003],[189,0.0001],[190,-0.0004],[191,-0.0012],[192,-0.0031],[193,-0.0058],[194,-0.0082],[195,-0.0099],[196,-0.0115],[197,-0.0125],[199,-0.0134],[204,-0.0142],[213,-0.0136],[222,-0.0146],[239,-0.0135],[245,-0.0138],[246,-0.0137],[247,-0.0133],[248,-0.0126],[249,-0.0109],[250,-0.0084],[251,-0.0055],[252,-0.003],[254,0.0009],[255,0.0025],[256,0.0038],[257,0.0046],[258,0.0052],[260,0.0057],[262,0.0055],[271,0.0014],[276,0.0003],[324,0]]
       }
       chest_ctrl.Location.Y {
        type=MovieSceneScriptingFloatChannel
        default=0
        thinned_from=325
        keys [[0,1.4408],[1,1.4392],[3,1.4267],[5,1.404],[7,1.373],[10,1.315],[14,1.2298],[16,1.1925],[22,1.1037],[24,1.0808],[25,1.0726],[26,1.0673],[28,1.0655],[32,1.0716],[35,1.0833],[37,1.0958],[39,1.1127],[41,1.1347],[43,1.1618],[51,1.2974],[53,1.3246],[55,1.3444],[72,1.41],[84,1.4417],[85,1.4422],[86,1.4392],[87,1.433],[89,1.412],[91,1.3816],[93,1.3439],[99,1.2156],[101,1.181],[106,1.1048],[108,1.0811],[109,1.0728],[110,1.0674],[113,1.0663],[117,1.0748],[120,1.0891],[123,1.1124],[125,1.1338],[127,1.1605],[135,1.2958],[137,1.3232],[139,1.3432],[156,1.4094],[168,1.4412],[169,1.4413],[170,1.4375],[171,1.43],[172,1.4193],[174,1.3897],[176,1.3517],[182,1.2146],[183,1.1951],[188,1.1212],[190,1.0964],[192,1.0774],[193,1.0708],[194,1.0664],[198,1.0659],[201,1.0725],[204,1.0864],[206,1.1008],[208,1.12],[210,1.1447],[212,1.1755],[218,1.2885],[220,1.3196],[222,1.3418],[239,1.4069],[252,1.4411],[253,1.4415],[254,1.4384],[255,1.4321],[257,1.4112],[259,1.3809],[261,1.3434],[266,1.2355],[268,1.1971],[274,1.1049],[276,1.0812],[277,1.0728],[278,1.0674],[280,1.0655],[284,1.0716],[287,1.0834],[290,1.1036],[292,1.1225],[294,1.1465],[304,1.3027],[307,1.3398],[314,1.4004],[319,1.4297],[323,1.4404],[324,1.4405]]
       }
       chest_ctrl.Location.Z {
        type=MovieSceneScriptingFloatChannel
        default=0
        thinned_from=325
        keys [[0,0.1284],[1,0.1305],[2,0.1366],[3,0.1463],[4,0.1594],[5,0.1754],[6,0.1941],[7,0.2152],[8,0.2382],[9,0.2628],[10,0.2888],[11,0.3159],[13,0.3726],[17,0.4885],[18,0.516],[19,0.5418],[20,0.5655],[21,0.5871],[22,0.6061],[23,0.6223],[24,0.6354],[25,0.6451],[26,0.6512],[27,0.6533],[30,0.6517],[33,0.6453],[35,0.6376],[37,0.6257],[38,0.618],[39,0.6091],[40,0.5988],[41,0.5871],[42,0.5739],[43,0.5592],[44,0.5413],[45,0.5193],[46,0.494],[47,0.4661],[48,0.4363],[51,0.3428],[52,0.3126],[53,0.284],[54,0.2577],[55,0.2341],[56,0.2141],[58,0.1786],[59,0.1616],[60,0.1451],[61,0.1292],[62,0.1142],[63,0.1],[64,0.0867],[65,0.0745],[66,0.0634],[67,0.0536],[68,0.0451],[69,0.0381],[70,0.0326],[71,0.0287],[72,0.0266],[73,0.0264],[74,0.0278],[75,0.0309],[76,0.0356],[77,0.042],[78,0.0499],[79,0.0595],[80,0.0707],[81,0.0834],[82,0.0978],[83,0.1137],[84,0.1312],[85,0.1503],[86,0.1708],[87,0.1924],[88,0.215],[89,0.2383],[90,0.2623],[97,0.4375],[99,0.4859],[101,0.5318],[102,0.5532],[103,0.5727],[104,0.5905],[105,0.6063],[106,0.6201],[107,0.6317],[108,0.6409],[109,0.6477],[110,0.6519],[111,0.6533],[114,0.6517],[116,0.648],[118,0.6415],[120,0.6315],[122,0.6171],[123,0.6082],[124,0.5979],[125,0.5862],[126,0.573],[127,0.5582],[128,0.5403],[129,0.5183],[130,0.493],[131,0.4651],[132,0.4352],[135,0.3417],[136,0.3115],[137,0.283],[138,0.2566],[139,0.233],[140,0.213],[142,0.1776],[143,0.1605],[144,0.144],[145,0.1281],[146,0.1131],[147,0.0988],[148,0.0856],[149,0.0733],[150,0.0623],[151,0.0524],[152,0.044],[153,0.0369],[154,0.0314],[155,0.0276],[156,0.0254],[157,0.0252],[158,0.0267],[159,0.0298],[160,0.0347],[161,0.0412],[162,0.0494],[163,0.0591],[164,0.0705],[165,0.0834],[166,0.0979],[167,0.1139],[168,0.1314],[169,0.1504],[170,0.1708],[171,0.1923],[172,0.2147],[173,0.238],[174,0.2619],[182,0.4592],[184,0.5068],[185,0.5296],[186,0.5509],[187,0.5706],[188,0.5886],[189,0.6048],[190,0.6189],[191,0.6308],[192,0.6404],[193,0.6474],[194,0.6518],[195,0.6533],[198,0.6519],[200,0.6486],[202,0.6424],[204,0.6326],[206,0.6184],[207,0.6094],[208,0.5989],[209,0.587],[210,0.5734],[211,0.5581],[212,0.5391],[213,0.515],[214,0.4869],[215,0.4557],[216,0.4226],[218,0.354],[219,0.3205],[220,0.2887],[221,0.2597],[222,0.234],[223,0.2128],[224,0.1943],[225,0.1763],[226,0.1589],[227,0.1422],[228,0.1262],[229,0.1111],[230,0.0969],[231,0.0838],[232,0.0717],[233,0.0608],[234,0.0512],[235,0.0429],[236,0.0361],[237,0.0308],[238,0.027],[239,0.025],[240,0.0248],[241,0.026],[242,0.0287],[243,0.0327],[244,0.038],[245,0.0448],[246,0.053],[247,0.0626],[248,0.0737],[249,0.0863],[250,0.1004],[251,0.116],[252,0.133],[253,0.1516],[254,0.1717],[255,0.193],[256,0.2155],[257,0.2388],[258,0.2627],[265,0.4373],[267,0.4857],[269,0.5317],[270,0.553],[271,0.5726],[272,0.5904],[273,0.6062],[274,0.62],[275,0.6316],[276,0.6409],[277,0.6477],[278,0.6519],[279,0.6533],[282,0.6517],[284,0.648],[286,0.6415],[288,0.6315],[290,0.6171],[291,0.6082],[292,0.5979],[293,0.5862],[294,0.573],[295,0.5582],[296,0.5401],[297,0.5175],[298,0.4912],[299,0.4623],[300,0.4313],[303,0.3354],[304,0.3051],[305,0.277],[306,0.2517],[307,0.2301],[308,0.213],[309,0.1992],[310,0.1869],[311,0.1762],[312,0.1668],[313,0.1588],[314,0.1519],[315,0.1462],[316,0.1414],[317,0.1376],[318,0.1346],[319,0.1323],[320,0.1306],[321,0.1295],[322,0.1288],[323,0.1285],[324,0.1284]]
       }
       chest_ctrl.Rotation.X {
        type=MovieSceneScriptingFloatChannel
        default=0
        thinned_from=325
        keys [[0,8.9375],[1,8.9256],[2,8.8907],[3,8.8337],[4,8.7557],[5,8.6575],[6,8.5403],[7,8.405],[8,8.2526],[9,8.0812],[11,7.6913],[14,7.0741],[15,6.8825],[16,6.7076],[21,5.9704],[22,5.8355],[23,5.7144],[24,5.6121],[25,5.5331],[26,5.4822],[27,5.4642],[30,5.4793],[33,5.5324],[35,5.5945],[37,5.6833],[39,5.8033],[41,5.9579],[43,6.1493],[44,6.2664],[45,6.4049],[47,6.728],[51,7.4364],[53,7.7507],[54,7.8822],[55,7.9906],[56,8.0716],[63,8.4636],[71,8.7571],[81,8.9283],[84,8.9422],[85,8.9321],[86,8.8989],[87,8.8433],[88,8.7658],[89,8.6674],[90,8.5493],[91,8.4128],[92,8.2588],[93,8.0866],[95,7.6975],[98,7.0824],[99,6.8912],[100,6.7163],[105,5.9748],[106,5.8387],[107,5.7167],[108,5.6135],[109,5.5338],[110,5.4825],[111,5.4643],[114,5.4794],[117,5.5327],[119,5.5953],[121,5.6839],[123,5.802],[125,5.9533],[127,6.1416],[128,6.2578],[129,6.3956],[131,6.7181],[135,7.4269],[137,7.7419],[138,7.8738],[139,7.9828],[140,8.0641],[147,8.4576],[155,8.7518],[165,8.9235],[168,8.9375],[169,8.9256],[170,8.887],[171,8.8235],[172,8.7364],[173,8.6273],[174,8.4977],[175,8.349],[176,8.1827],[177,7.9946],[181,7.114],[182,6.9045],[183,6.7165],[184,6.5586],[188,6.0033],[190,5.7637],[191,5.6638],[192,5.581],[193,5.5185],[194,5.4792],[195,5.4659],[199,5.4922],[202,5.5571],[204,5.6299],[206,5.7315],[208,5.8663],[210,6.0388],[211,6.1406],[212,6.2633],[213,6.4129],[214,6.5834],[219,7.5374],[220,7.7059],[221,7.8525],[222,7.9712],[223,8.0558],[230,8.4438],[238,8.7322],[250,8.9252],[252,8.9353],[253,8.9255],[254,8.892],[255,8.8357],[256,8.7578],[257,8.6594],[258,8.5414],[259,8.4047],[260,8.2501],[262,7.888],[266,7.0783],[267,6.8889],[268,6.7155],[273,5.9753],[274,5.8392],[275,5.717],[276,5.6137],[277,5.5339],[278,5.4824],[279,5.4642],[282,5.4793],[285,5.5327],[287,5.5953],[289,5.6839],[291,5.802],[293,5.9533],[295,6.1416],[296,6.2557],[297,6.3876],[299,6.6919],[304,7.5325],[306,7.8298],[307,7.9566],[310,8.2501],[314,8.5677],[317,8.7494],[320,8.8738],[323,8.9334],[324,8.936]]
       }
       chest_ctrl.Rotation.Y {
        type=MovieSceneScriptingFloatChannel
        default=0
        thinned_from=325
        keys [[0,0],[28,0],[29,0.0013],[30,0.0107],[31,0.0361],[32,0.1062],[33,0.2135],[34,0.3682],[35,0.4929],[36,0.5373],[37,0.5648],[38,0.5751],[39,0.5741],[42,0.5593],[43,0.5554],[44,0.5536],[45,0.553],[47,0.5551],[54,0.5712],[84,0.5868],[85,0.5866],[86,0.5837],[87,0.5665],[88,0.5043],[89,0.4118],[90,0.3108],[92,0.1032],[93,0.033],[94,0.0121],[95,0.0026],[96,0.0011],[97,0.0033],[99,0.0083],[103,0.0171],[105,0.0208],[106,0.0223],[107,0.0235],[108,0.0245],[109,0.0249],[110,0.0246],[111,0.0235],[112,0.0218],[113,0.0196],[117,0.0093],[118,0.0071],[119,0.0054],[120,0.0042],[121,0.0038],[156,-0.001],[169,-0.0014],[181,0.002],[182,0.002],[183,0.0013],[186,-0.0019],[187,-0.0024],[188,-0.0023],[189,-0.0009],[190,0.0036],[191,0.0128],[192,0.0349],[193,0.0682],[194,0.0984],[195,0.121],[196,0.1395],[197,0.1512],[198,0.1574],[199,0.1612],[200,0.1631],[201,0.1634],[202,0.1624],[203,0.1605],[204,0.1574],[205,0.1529],[206,0.1473],[207,0.1408],[208,0.1337],[209,0.1262],[210,0.1179],[211,0.1088],[212,0.0989],[214,0.0786],[216,0.0587],[217,0.049],[218,0.0397],[219,0.0311],[220,0.0233],[221,0.0165],[222,0.011],[223,0.0069],[226,-0.0025],[229,-0.011],[231,-0.0162],[234,-0.0233],[236,-0.0275],[239,-0.0329],[240,-0.0341],[241,-0.036],[243,-0.0387],[244,-0.0398],[245,-0.0406],[246,-0.0407],[247,-0.0399],[248,-0.0378],[249,-0.0329],[250,-0.0251],[251,-0.0163],[252,-0.0088],[253,-0.0027],[254,0.0023],[255,0.0055],[256,0.0065],[257,0.0056],[258,0.0033],[259,0.0003],[261,-0.0065],[262,-0.0097],[263,-0.0123],[264,-0.0141],[265,-0.0153],[266,-0.0159],[267,-0.0159],[268,-0.0149],[269,-0.0134],[271,-0.01],[275,-0.0043],[276,-0.0032],[277,-0.0025],[279,-0.0015],[280,-0.0015],[281,-0.0008],[288,0],[324,0]]
       }
       chest_ctrl.Rotation.Z {
        type=MovieSceneScriptingFloatChannel
        default=0
        thinned_from=325
        keys [[0,0],[29,-0.0001],[30,0.0118],[31,0.0414],[32,0.0772],[33,0.1104],[34,0.1308],[35,0.1345],[36,0.1202],[37,0.0598],[38,-0.084],[39,-0.3512],[40,-0.7026],[41,-1.0566],[42,-1.4143],[43,-1.725],[44,-1.9382],[45,-2.1066],[46,-2.2255],[47,-2.3008],[48,-2.3461],[49,-2.3679],[50,-2.3725],[51,-2.3611],[52,-2.3269],[53,-2.2642],[54,-2.1671],[55,-2.0603],[56,-1.972],[57,-1.9012],[58,-1.8473],[59,-1.8094],[60,-1.787],[61,-1.7794],[85,-1.7744],[86,-1.8167],[87,-1.8947],[88,-1.9359],[89,-1.902],[90,-1.8002],[91,-1.6258],[92,-1.3556],[93,-1.014],[94,-0.6667],[95,-0.3056],[96,0.0624],[97,0.3511],[98,0.5473],[99,0.701],[100,0.8212],[101,0.917],[102,0.9975],[103,1.0639],[104,1.1132],[105,1.1475],[106,1.1693],[107,1.1807],[108,1.1841],[109,1.1721],[110,1.1368],[111,1.0795],[112,1.0015],[113,0.9041],[114,0.7917],[116,0.5528],[117,0.4398],[118,0.3401],[119,0.2605],[120,0.2078],[121,0.1887],[181,0.1886],[182,0.164],[183,0.1026],[185,-0.0572],[186,-0.1186],[187,-0.1432],[188,-0.1312],[189,-0.0465],[190,0.1832],[191,0.6306],[192,1.6692],[193,3.1914],[194,4.548],[195,5.5684],[196,6.4232],[197,6.9738],[198,7.2737],[199,7.4804],[200,7.611],[201,7.6828],[202,7.7131],[203,7.7193],[204,7.6923],[205,7.6199],[206,7.515],[209,7.1317],[210,6.9921],[212,6.6592],[213,6.511],[214,6.4051],[216,6.2646],[218,6.1597],[219,6.1241],[220,6.1016],[244,6.085],[245,6.0414],[246,5.9231],[247,5.6928],[248,5.313],[249,4.5631],[250,3.4361],[251,2.2159],[252,1.1864],[253,0.3769],[254,-0.3486],[255,-0.954],[256,-1.4033],[257,-1.7002],[258,-1.8884],[259,-1.9926],[260,-2.0373],[261,-2.0468],[262,-2.0125],[263,-1.9215],[264,-1.7928],[266,-1.4969],[267,-1.3355],[268,-1.1484],[269,-0.9549],[270,-0.7744],[271,-0.6262],[272,-0.5049],[273,-0.3941],[274,-0.2968],[275,-0.2159],[276,-0.1546],[277,-0.1157],[278,-0.091],[279,-0.0701],[280,-0.0701],[281,-0.0384],[282,-0.027],[283,-0.0181],[284,-0.0114],[285,-0.0066],[286,-0.0034],[287,-0.0014],[288,-0.0004],[289,-0.0001],[324,0]]
       }
       chest_ctrl.Scale.X {type=MovieSceneScriptingFloatChannel, default=1, const [1,325]}
       chest_ctrl.Scale.Y {type=MovieSceneScriptingFloatChannel, default=1, const [1,325]}
       chest_ctrl.Scale.Z {type=MovieSceneScriptingFloatChannel, default=1, const [1,325]}
       chest_tan_ctrl.Location.X {
        type=MovieSceneScriptingFloatChannel
        default=0
        thinned_from=325
        keys [[0,0],[29,-0.0001],[30,-0.0005],[31,-0.0016],[32,-0.0048],[33,-0.0098],[34,-0.017],[35,-0.0229],[36,-0.025],[37,-0.0264],[38,-0.0272],[43,-0.0294],[51,-0.0306],[86,-0.0302],[87,-0.0295],[88,-0.0266],[89,-0.0223],[90,-0.0174],[92,-0.007],[93,-0.0032],[94,-0.0017],[95,-0.0006],[96,0.0001],[99,0.0009],[112,0.0011],[188,-0.0002],[190,0.0002],[191,0.0007],[192,0.0018],[194,0.005],[196,0.007],[197,0.0076],[200,0.0084],[214,0.0078],[244,0.01],[246,0.0098],[247,0.0094],[248,0.0088],[249,0.0076],[251,0.0037],[252,0.002],[254,-0.0006],[255,-0.0015],[256,-0.0022],[258,-0.0029],[262,-0.0028],[274,-0.0003],[324,0]]
       }
       chest_tan_ctrl.Location.Y {
        type=MovieSceneScriptingFloatChannel
        default=-0.0303
        thinned_from=325
        keys [[0,0.1887],[1,0.1885],[2,0.1879],[3,0.1868],[4,0.1852],[5,0.183],[6,0.1804],[7,0.1772],[8,0.1735],[9,0.1689],[10,0.1637],[12,0.152],[14,0.1399],[15,0.1342],[16,0.1291],[19,0.1163],[21,0.1073],[22,0.1031],[23,0.0993],[24,0.096],[25,0.0934],[26,0.0918],[27,0.0912],[30,0.0916],[32,0.0924],[34,0.0937],[36,0.0956],[38,0.0981],[39,0.0997],[41,0.1035],[43,0.108],[44,0.1109],[45,0.1146],[46,0.1189],[47,0.1237],[51,0.1447],[52,0.1496],[53,0.154],[54,0.1579],[55,0.1609],[56,0.1631],[60,0.169],[67,0.1764],[84,0.1892],[85,0.1897],[86,0.1896],[87,0.189],[88,0.1877],[89,0.1858],[90,0.1832],[91,0.18],[92,0.1762],[93,0.1716],[94,0.1663],[96,0.1544],[98,0.1421],[99,0.1363],[100,0.131],[103,0.1175],[105,0.1081],[106,0.1037],[107,0.0996],[108,0.0962],[109,0.0935],[110,0.0918],[111,0.0912],[114,0.0916],[116,0.0924],[118,0.0937],[120,0.0956],[122,0.0981],[124,0.1013],[126,0.1053],[127,0.1076],[128,0.1105],[129,0.1141],[130,0.1184],[131,0.1232],[135,0.1441],[136,0.149],[137,0.1535],[138,0.1573],[139,0.1604],[140,0.1626],[144,0.1685],[151,0.176],[168,0.1889],[169,0.1893],[170,0.1891],[171,0.1882],[172,0.1867],[173,0.1846],[174,0.1817],[175,0.1783],[176,0.1741],[177,0.1691],[178,0.163],[181,0.1425],[182,0.136],[183,0.1302],[184,0.1255],[188,0.1086],[189,0.1046],[190,0.101],[191,0.0977],[192,0.095],[193,0.093],[194,0.0918],[195,0.0914],[198,0.0919],[200,0.0927],[202,0.0939],[204,0.0958],[206,0.0982],[208,0.1014],[209,0.1034],[210,0.1055],[211,0.1078],[212,0.1109],[213,0.1149],[214,0.1196],[215,0.1249],[218,0.1422],[219,0.1476],[220,0.1526],[221,0.1569],[222,0.1603],[223,0.1626],[227,0.1684],[235,0.1765],[252,0.1889],[253,0.1893],[254,0.1892],[255,0.1885],[256,0.1872],[257,0.1854],[258,0.1828],[259,0.1796],[260,0.1758],[261,0.1712],[262,0.1659],[266,0.1419],[267,0.1362],[268,0.131],[270,0.1221],[273,0.1081],[274,0.1037],[275,0.0996],[276,0.0962],[277,0.0935],[278,0.0918],[279,0.0912],[282,0.0916],[284,0.0924],[286,0.0937],[288,0.0956],[290,0.0981],[292,0.1013],[294,0.1053],[295,0.1076],[296,0.1104],[297,0.1138],[298,0.1178],[299,0.1222],[304,0.1469],[305,0.1515],[306,0.1557],[307,0.1595],[308,0.1626],[312,0.173],[315,0.1794],[318,0.1844],[320,0.1868],[322,0.1882],[324,0.1888]]
       }
       chest_tan_ctrl.Location.Z {
        type=MovieSceneScriptingFloatChannel
        default=0.265
        thinned_from=325
        keys [[0,0.2675],[1,0.2665],[2,0.2635],[3,0.2586],[4,0.2522],[5,0.2442],[6,0.2348],[7,0.2243],[8,0.2126],[9,0.2002],[10,0.1869],[11,0.1729],[12,0.1584],[16,0.0986],[17,0.0841],[18,0.0702],[19,0.057],[20,0.0447],[21,0.0335],[22,0.0236],[23,0.015],[24,0.0081],[25,0.0029],[26,-0.0004],[27,-0.0015],[28,-0.0015],[30,-0.0006],[31,0.0002],[32,0.0013],[33,0.0027],[34,0.0045],[35,0.0067],[36,0.0095],[37,0.0128],[38,0.0168],[39,0.0214],[40,0.0266],[41,0.0326],[42,0.0394],[43,0.0469],[44,0.056],[45,0.0672],[46,0.0801],[47,0.0944],[48,0.1096],[51,0.1572],[52,0.1725],[53,0.187],[54,0.2002],[55,0.212],[56,0.2219],[59,0.2478],[61,0.2636],[63,0.278],[65,0.2905],[66,0.296],[67,0.3008],[68,0.305],[69,0.3085],[70,0.3112],[71,0.3132],[72,0.3144],[73,0.3146],[74,0.3141],[75,0.3127],[76,0.3106],[77,0.3077],[78,0.3041],[79,0.2996],[80,0.2945],[81,0.2885],[82,0.2818],[83,0.2744],[84,0.2662],[85,0.2572],[86,0.2475],[87,0.2371],[88,0.2262],[90,0.2031],[92,0.1782],[97,0.1132],[98,0.1003],[99,0.0878],[100,0.0756],[101,0.0639],[102,0.0529],[103,0.0426],[104,0.0331],[105,0.0246],[106,0.0171],[107,0.0107],[108,0.0055],[109,0.0017],[110,-0.0007],[111,-0.0015],[113,-0.0011],[114,-0.0006],[115,0.0002],[116,0.0013],[117,0.0028],[118,0.0047],[119,0.007],[120,0.0099],[121,0.0133],[122,0.0172],[123,0.0218],[124,0.0271],[125,0.033],[126,0.0398],[127,0.0473],[128,0.0564],[129,0.0676],[130,0.0805],[131,0.0948],[132,0.11],[135,0.1576],[136,0.173],[137,0.1873],[138,0.2006],[139,0.2124],[140,0.2224],[143,0.2482],[145,0.2641],[147,0.2785],[149,0.291],[150,0.2964],[151,0.3013],[152,0.3055],[153,0.309],[154,0.3117],[155,0.3137],[156,0.3149],[157,0.3151],[158,0.3146],[159,0.3132],[160,0.311],[161,0.3081],[162,0.3043],[163,0.2998],[164,0.2945],[165,0.2885],[166,0.2818],[167,0.2743],[168,0.2661],[169,0.2571],[170,0.2474],[171,0.2371],[172,0.2261],[174,0.203],[176,0.1782],[182,0.1004],[183,0.0878],[184,0.0755],[185,0.0639],[186,0.0528],[187,0.0425],[188,0.0331],[189,0.0245],[190,0.017],[191,0.0106],[192,0.0055],[193,0.0017],[194,-0.0007],[195,-0.0015],[197,-0.0012],[198,-0.0007],[199,0],[200,0.001],[201,0.0024],[202,0.0042],[203,0.0065],[204,0.0093],[205,0.0126],[206,0.0166],[207,0.0212],[208,0.0265],[209,0.0326],[210,0.0395],[211,0.0473],[212,0.057],[213,0.0692],[214,0.0836],[215,0.0995],[216,0.1164],[218,0.1514],[219,0.1684],[220,0.1845],[221,0.199],[222,0.2119],[223,0.2224],[225,0.2404],[227,0.2571],[229,0.2724],[231,0.2858],[233,0.297],[234,0.3018],[235,0.3059],[236,0.3093],[237,0.312],[238,0.3139],[239,0.315],[240,0.3152],[241,0.3148],[242,0.3136],[243,0.3119],[244,0.3094],[245,0.3063],[246,0.3026],[247,0.2981],[248,0.293],[249,0.2871],[250,0.2806],[251,0.2733],[252,0.2653],[253,0.2565],[254,0.247],[255,0.2367],[256,0.2259],[258,0.2028],[260,0.178],[265,0.1132],[266,0.1004],[267,0.0878],[268,0.0757],[269,0.064],[270,0.0529],[271,0.0426],[272,0.0332],[273,0.0246],[274,0.0171],[275,0.0107],[276,0.0055],[277,0.0017],[278,-0.0007],[279,-0.0015],[280,-0.0015],[282,-0.0006],[283,0.0002],[284,0.0013],[285,0.0028],[286,0.0047],[287,0.007],[288,0.0099],[289,0.0133],[290,0.0172],[291,0.0218],[292,0.0271],[293,0.033],[294,0.0398],[295,0.0473],[296,0.0565],[297,0.0679],[298,0.0812],[299,0.0959],[300,0.1116],[303,0.1602],[304,0.1756],[305,0.1897],[306,0.2026],[307,0.2136],[308,0.2224],[309,0.2295],[310,0.2358],[311,0.2414],[312,0.2463],[313,0.2506],[314,0.2542],[316,0.2599],[318,0.2638],[320,0.2661],[323,0.2674],[324,0.2675]]
       }
       chest_tan_ctrl.Rotation.X {type=MovieSceneScriptingFloatChannel, default=0, thinned_from=325, keys [[0,4.6942],[1,0],[324,0]]}
       chest_tan_ctrl.Rotation.Y {
        type=MovieSceneScriptingFloatChannel
        default=0
        thinned_from=325
        keys [[0,0],[28,0],[29,0.0002],[30,0.001],[31,0],[93,0],[94,0.0483],[95,0.0516],[96,0.0442],[98,0.0007],[99,0],[186,0],[187,0.0034],[188,0.0074],[189,0.0124],[190,0.0197],[191,0.0258],[192,0.0499],[193,0.0393],[194,0],[255,0],[256,0.0019],[257,0.0392],[258,0.0647],[259,0.0824],[260,0.0935],[261,0.099],[262,0.1034],[263,0.1064],[264,0.1046],[265,0.0991],[266,0.091],[267,0.0842],[268,0.0779],[269,0.069],[270,0.0585],[271,0.0471],[272,0.0379],[273,0.0313],[274,0.0251],[275,0.0193],[276,0.0141],[277,0.0097],[278,0.0069],[279,0.0055],[280,0.0031],[281,0.0054],[282,0.0025],[284,0.0013],[286,0.0005],[324,0]]
       }
       chest_tan_ctrl.Rotation.Z {
        type=MovieSceneScriptingFloatChannel
        default=0
        thinned_from=325
        keys [[0,0],[29,0],[31,0.0157],[32,0.0124],[33,0.0063],[34,0],[35,0],[36,0.0004],[37,0],[40,0],[41,0.1168],[42,0.248],[43,0.4117],[44,0.589],[45,0.6971],[46,0.7915],[47,0.8635],[48,0.911],[49,0.9433],[50,0.9628],[52,0.9858],[53,0.9914],[54,0.9898],[55,0.9602],[56,0.9096],[57,0.8662],[58,0.8298],[59,0.8001],[60,0.7768],[61,0.7597],[62,0.7528],[85,0.7528],[86,0.7268],[87,0.7199],[88,0.7688],[89,0.8206],[90,0.8354],[91,0.8265],[92,0.8047],[93,0.7325],[94,0.5982],[95,0.4739],[96,0.3416],[97,0.1546],[98,0],[186,0],[187,0.0293],[188,0.0612],[189,0.1015],[190,0.1593],[191,0.2075],[192,0.4047],[193,0.3134],[194,0],[255,0],[256,0.0802],[257,0.342],[258,0.5198],[259,0.6419],[260,0.7176],[261,0.756],[262,0.7867],[263,0.8089],[264,0.7984],[265,0.7623],[266,0.7074],[267,0.6603],[268,0.6159],[269,0.5499],[270,0.4695],[271,0.3819],[272,0.3098],[273,0.2579],[274,0.208],[275,0.1613],[276,0.119],[277,0.0821],[278,0.0587],[279,0.0471],[280,0.0263],[281,0.0459],[282,0.0215],[283,0.0156],[284,0.0109],[285,0.0072],[286,0.0045],[287,0.0025],[288,0.0012],[289,0.0004],[290,0.0001],[324,0]]
       }
       chest_tan_ctrl.Scale.X {type=MovieSceneScriptingFloatChannel, default=1, const [1,325]}
       chest_tan_ctrl.Scale.Y {type=MovieSceneScriptingFloatChannel, default=1, const [1,325]}
       chest_tan_ctrl.Scale.Z {type=MovieSceneScriptingFloatChannel, default=1, const [1,325]}
      ]
     }]
   }
  ]
 }
]
display_rate {denominator=1, numerator=30}
tick_resolution {denominator=1, numerator=24000}
tracks []
